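; Blueprint function library of shared helpers for the Season Pass (Nexus) UI widgets.
; Theme helpers apply title, season text and custom logo styling from theme data.
; Navigation helpers filter categories by lock and main-set completion, then step linearly across pages and categories to a boundary reward.
; Others extract page rewards and quest, test reach-level headers, push the right-of-withdrawal dialog and toggle frontend visibility mode.
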(fn ApplyCustomThemeData
  (params (ThemeData :type Season%20Pass%20Theme%20Data%20Structure) (Switcher :type Common%20Visibility%20Switcher%20Object%20Reference) (LazyImage :type Fort%20Lazy%20Image%20Object%20Reference) (Text :type Common%20Text%20Object%20Reference))
  (bind (_titlecolor _titlematerial _titleoutline _seasoncolor _progressioncolor _overlaycolor _vaultbackgroundmaterial _giftboxvaultbackgroundmaterial _uibackgroundmaterial _upsellbackgroundmaterial _crewcategorybackgroundmaterial _customlogo _customseasontitle _customseasontitlecolor) (Utilities|Struct|BreakSeasonPassThemeData ThemeData))
  (bind _switcher (Variables|GetSwitcher))
  (bind _text (Variables|GetText))
  (if (Utilities|IsValidSoftObjectReference _customlogo)
    (CommonVisibilitySwitcher|SetActiveWidgetIndex _switcher 1)
    (LazyImage|SetBrushfromLazyDisplayAsset (Variables|GetLazyImage) _customlogo false)
    (Utilities|IsValid _text
      (:Is%20Valid
        (Variables|/Script/UMG.TextBlock|SetText _customseasontitle _text)
        (Appearance|SetColorandOpacity _text (Utilities|Struct|MakeSlateColor _customseasontitlecolor) :!class /Script/UMG.TextBlock)))
    (else
      (CommonVisibilitySwitcher|SetActiveWidgetIndex _switcher 0))))

(fn ApplyThemeToSeasonText
  (params (SeasonText :type Common%20Text%20Object%20Reference) (ThemeData :type Season%20Pass%20Theme%20Data%20Structure))
  (Variables|/Script/UMG.TextBlock|SetColorAndOpacity (Utilities|Struct|MakeSlateColor (Utilities|Struct|BreakSeasonPassThemeData ThemeData)) SeasonText))

(fn ApplyThemeToTitleText
  (params (TitleText :type Common%20Text%20Object%20Reference) (ThemeData :type Season%20Pass%20Theme%20Data%20Structure))
  (bind (_fontobject _fontmaterial _outlinesettings _typefacefontname _size _letterspacing _skewamount _bforcemonospaced _bmaterialisstencil _monospacedwidth) (Utilities|Struct|BreakSlateFontInfo (Variables|/Script/UMG.TextBlock|GetFont (Variables|GetTitleText))))
  (bind (_titlecolor _titlematerial _titleoutline) (Utilities|Struct|BreakSeasonPassThemeData ThemeData))
  (Utilities|FlowControl|Sequence
    (:then_0
      ;[0 0 0 0.5] Set Title Color
      (Variables|/Script/UMG.TextBlock|SetColorAndOpacity (Utilities|Struct|MakeSlateColor _titlecolor) (Variables|GetTitleText)))
    (:then_1
      ;[0 0 0 0.5] Set Title Material
      (if (Utilities|IsValidSoftObjectReference _titlematerial)
        (bind _returnvalue (Utilities|LoadAssetBlocking _titlematerial))
        (bind _asmaterial_instance (Utilities|Casting|CastToMaterialInstance _returnvalue))
        (Appearance|SetFontMaterial (Variables|GetTitleText) _asmaterial_instance :!class /Script/UMG.TextBlock)))
    (:then_2
      ;[0 0 0 0.5] Set Title Outline
      (Appearance|SetFont (Variables|GetTitleText) (SlateFont|MakeSlateFontInfo _fontobject _fontmaterial _titleoutline _typefacefontname _size _letterspacing _skewamount _bforcemonospaced _bmaterialisstencil _monospacedwidth :!class /Script/SlateRHIRenderer.SlateFontBlueprintLibrary) :!class /Script/UMG.TextBlock))))

(fn FilterSeasonPassCategories
  (params (ShowLocked :type Boolean :default true) (UseMainSetOnly :type Boolean) (SeasonPass :type Season%20Pass%20VM%20Object%20Reference))
  (LocalVariables|SetIsMainSetComplete (CallFunction|IsMainSetComplete SeasonPass SeasonPass))
  (for _array_element (Variables|/Script/SeasonPass.SeasonPassVM|GetCategories SeasonPass)
    ;[0.02416 0.005605 0.04092] If the main set is complete, then only show categories that are NOT in the default group.  Otherwise, only show categories that are IN the default group.
    (if (select (and (LocalVariables|GetIsMainSetComplete) (not (Variables|GetUseMainSetOnly))) (Utilities|String|NotEqualExactly(String) (Class|SeasonPassElementBaseVM|GetGroupName _array_element) (LocalVariables|GetDefaultGroup)) (Utilities|String|EqualExactly(String) (Class|SeasonPassElementBaseVM|GetGroupName _array_element) (LocalVariables|GetDefaultGroup)))
      ;[0.02416 0.005605 0.04092] If I'm not showing locked categories (only unlocked) and the target is locked, then don't add.  Otherwise add.
      (if (not (and (not (Variables|GetShowLocked)) (Variables|/Script/SeasonPass.SeasonPassElementBaseVM|GetbIsLocked _array_element)))
        (Utilities|Array|Add (LocalVariables|GetCategories) _array_element))))
  (return (LocalVariables|GetCategories) (LocalVariables|GetIsMainSetComplete)))

(fn GetAllUnlockedCategories
  (params (SeasonPassSelectionVM :type Season%20Pass%20Selection%20VM%20Object%20Reference))
  (for _array_element (Variables|/Script/SeasonPass.SeasonPassVM|GetCategories (Variables|/Script/SeasonPass.SeasonPassSelectionVM|GetSelectedPass SeasonPassSelectionVM))
    (if (not (Variables|/Script/SeasonPass.SeasonPassElementBaseVM|GetbIsLocked _array_element))
      (Utilities|Array|Add (LocalVariables|GetUnlockedCategoryVMs) _array_element)))
  (return (LocalVariables|GetUnlockedCategoryVMs)))

(fn GetFirstOrLastRewardOnSeasonPassPage
  (params (SeasonPassPageVM :type Season%20Pass%20Page%20VM%20Object%20Reference%20%28by%20ref%29) (IsFirst :type Boolean))
  (Utilities|IsValid SeasonPassPageVM
    (:Is%20Valid
      (return (Utilities|Array|Get(acopy) (Variables|/Script/SeasonPass.SeasonPassTrackVM|GetRewards (Utilities|Array|Get(acopy) (Variables|/Script/SeasonPass.SeasonPassPageVM|GetTracks SeasonPassPageVM) (select (Variables|GetIsFirst) 0 (Utilities|Array|LastIndex (Variables|/Script/SeasonPass.SeasonPassPageVM|GetTracks SeasonPassPageVM))))) (select (Variables|GetIsFirst) 0 (Utilities|Array|LastIndex (Variables|/Script/SeasonPass.SeasonPassTrackVM|GetRewards (Utilities|Array|Get(acopy) (Variables|/Script/SeasonPass.SeasonPassPageVM|GetTracks SeasonPassPageVM) (select (Variables|GetIsFirst) 0 (Utilities|Array|LastIndex (Variables|/Script/SeasonPass.SeasonPassPageVM|GetTracks SeasonPassPageVM))))))))))))

(fn GetLinearSeasonPassPageNavTarget
  (params (SeasonPassSelectionVM :type Season%20Pass%20Selection%20VM%20Object%20Reference) (IsPrevious :type Boolean) (ShowLockedCategories :type Boolean) (UseMainSetOnly :type Boolean))
  (bind _categoryindex (LocalVariables|GetCategoryIndex))
  (bind _returnvalue (+ _categoryindex 1))
  (bind _categoryindex_2 (LocalVariables|GetCategoryIndex))
  (bind _returnvalue_2 (- _categoryindex_2 1))
  (bind _isprevious (Variables|GetIsPrevious))
  (bind _returnvalue_3 (select _isprevious _returnvalue_2 _returnvalue))
  (bind _passcategories (LocalVariables|GetPassCategories))
  (Utilities|IsValid SeasonPassSelectionVM
    (:Is%20Valid
      (if (Math|Boolean|ANDBoolean (Utilities|IsValid (Variables|/Script/SeasonPass.SeasonPassSelectionVM|GetSelectedPage SeasonPassSelectionVM) :!class /Script/Engine.KismetSystemLibrary) (Utilities|IsValid (Variables|/Script/SeasonPass.SeasonPassSelectionVM|GetSelectedPass SeasonPassSelectionVM) :!class /Script/Engine.KismetSystemLibrary) (Utilities|IsValid (Variables|/Script/SeasonPass.SeasonPassSelectionVM|GetSelectedCategory SeasonPassSelectionVM) :!class /Script/Engine.KismetSystemLibrary))
        (LocalVariables|SetPassCategories (CallFunction|FilterSeasonPassCategories (Variables|GetShowLockedCategories) (Variables|GetUseMainSetOnly) (Variables|/Script/SeasonPass.SeasonPassSelectionVM|GetSelectedPass (Variables|GetSeasonPassSelectionVM)) (Variables|/Script/SeasonPass.SeasonPassSelectionVM|GetSelectedPass (Variables|GetSeasonPassSelectionVM))))
        (Utilities|FlowControl|Sequence
          (:then_0
            ;[0.02416 0.005605 0.04092] Find the current category index in the current pass
            (LocalVariables|SetCategoryIndex (Utilities|Array|FindItem (LocalVariables|GetPassCategories) (Variables|/Script/SeasonPass.SeasonPassSelectionVM|GetSelectedCategory (Variables|GetSeasonPassSelectionVM))))
            ;[0.02416 0.005605 0.04092] Find the current page index in the current category
            (LocalVariables|SetPageIndex (Utilities|Array|FindItem (Variables|/Script/SeasonPass.SeasonPassCategoryVM|GetPages (Variables|/Script/SeasonPass.SeasonPassSelectionVM|GetSelectedCategory (Variables|GetSeasonPassSelectionVM))) (Variables|/Script/SeasonPass.SeasonPassSelectionVM|GetSelectedPage (Variables|GetSeasonPassSelectionVM)))))
          (:then_1
            ;[0.02416 0.005605 0.04092] Find the target page in our list of pages - if that page does not exist, the target page is null.
            (bind _output_get (LocalVariables|SetTargetPage (select (Utilities|Array|IsValidIndex (Variables|/Script/SeasonPass.SeasonPassCategoryVM|GetPages (Variables|/Script/SeasonPass.SeasonPassSelectionVM|GetSelectedCategory (Variables|GetSeasonPassSelectionVM))) (select (Variables|GetIsPrevious) (- (LocalVariables|GetPageIndex) 1) (+ (LocalVariables|GetPageIndex) 1))) (Utilities|Array|Get(acopy) (Variables|/Script/SeasonPass.SeasonPassCategoryVM|GetPages (Variables|/Script/SeasonPass.SeasonPassSelectionVM|GetSelectedCategory (Variables|GetSeasonPassSelectionVM))) (select (Variables|GetIsPrevious) (- (LocalVariables|GetPageIndex) 1) (+ (LocalVariables|GetPageIndex) 1))) 0)))
            (bind _output_get_2 (LocalVariables|SetTargetPageFoundInCurrentCategory (Utilities|IsValid _output_get :!class /Script/Engine.KismetSystemLibrary)))
            (if _output_get_2
              ;[0.02416 0.005605 0.04092] If the target page was found in our current category, our current selected category is also the target category.
              (LocalVariables|SetTargetCategory (Variables|/Script/SeasonPass.SeasonPassSelectionVM|GetSelectedCategory (Variables|GetSeasonPassSelectionVM)))
              (else
                ;[0.02416 0.005605 0.04092] If the target page was not found in this category, find the next or previous unlocked category.
                (if (Utilities|Array|IsValidIndex _passcategories _returnvalue_3)
                  (LocalVariables|SetTargetCategory (Utilities|Array|Get(acopy) _passcategories _returnvalue_3))
                  (else
                    (LocalVariables|SetTargetCategory (Variables|/Script/SeasonPass.SeasonPassSelectionVM|GetSelectedCategory (Variables|GetSeasonPassSelectionVM))))))))
          (:then_2
            ;[0.02416 0.005605 0.04092 1 16] If no target page was initially found, we now need to locate it within the new target category.  If we are going backward, select the last page of the target category, if we're going forward, select the first page.
            (if (not (LocalVariables|GetTargetPageFoundInCurrentCategory))
              (LocalVariables|SetTargetPage (Utilities|Array|Get(acopy) (Variables|/Script/SeasonPass.SeasonPassCategoryVM|GetPages (LocalVariables|GetTargetCategory)) (select (Variables|GetIsPrevious) (Utilities|Array|LastIndex (Variables|/Script/SeasonPass.SeasonPassCategoryVM|GetPages (LocalVariables|GetTargetCategory))) 0)))))
          (:then_3
            ;[0.02416 0.005605 0.04092] Now that we have both a valid Target Category and Page, find the target page index.
            (LocalVariables|SetTargetPageIndex (Utilities|Array|FindItem (Variables|/Script/SeasonPass.SeasonPassCategoryVM|GetPages (LocalVariables|GetTargetCategory)) (LocalVariables|GetTargetPage))))
          (:then_4
            ;[0.02416 0.005605 0.04092] If we are navigating backward, return the last reward, otherwise return the first.
            (return (CallFunction|GetFirstorLastRewardonSeasonPassPage (LocalVariables|GetTargetPage) (not (Variables|GetIsPrevious)) (LocalVariables|GetTargetPage)) (LocalVariables|GetTargetCategory) (LocalVariables|GetTargetPageIndex))))))))

(fn GetRewardsAndQuest
  (params (SeasonPassPage :type Season%20Pass%20Page%20VM%20Object%20Reference))
  (Utilities|IsValid SeasonPassPage
    (:Is%20Valid
      (LocalVariables|Settracks (Variables|/Script/SeasonPass.SeasonPassPageVM|GetTracks SeasonPassPage))
      (if (> (Utilities|Array|Length (LocalVariables|Gettracks)) 1)
        (return (Variables|/Script/SeasonPass.SeasonPassTrackVM|GetRewards (Utilities|Array|Get(acopy) (LocalVariables|Gettracks))) (Utilities|Array|Get(acopy) (Variables|/Script/SeasonPass.SeasonPassTrackVM|GetRewards (Utilities|Array|Get(acopy) (LocalVariables|Gettracks) 1))))
        (else
          (if (== (Utilities|Array|Length (LocalVariables|Gettracks)) 1)
            (return (Variables|/Script/SeasonPass.SeasonPassTrackVM|GetRewards (Utilities|Array|Get(acopy) (LocalVariables|Gettracks))) 0)
            (else
              _)))))
    (:Is%20Not%20Valid
      _))
  (return (Utilities|Array|MakeArray 0) 0))

(fn IsBoundaryOnPageSelected
  (params (SeasonPassSelectionVM :type Season%20Pass%20Selection%20VM%20Object%20Reference) (IsFirst :type Boolean))
  (return (== (CallFunction|GetFirstorLastRewardonSeasonPassPage (Variables|/Script/SeasonPass.SeasonPassSelectionVM|GetSelectedPage SeasonPassSelectionVM) IsFirst (Variables|/Script/SeasonPass.SeasonPassSelectionVM|GetSelectedPage SeasonPassSelectionVM)) (Variables|/Script/SeasonPass.SeasonPassSelectionVM|GetSelectedReward SeasonPassSelectionVM))))

(fn IsMainSetComplete
  (params (SeasonPass :type Season%20Pass%20VM%20Object%20Reference))
  (for _array_element (Variables|/Script/SeasonPass.SeasonPassVM|GetCategories SeasonPass)
    (if (Utilities|String|EqualExactly(String) (Class|SeasonPassElementBaseVM|GetGroupName _array_element) "Default")
      (Math|Integer|IncrementInt (LocalVariables|GetNumberOfMainCategorySets)
        (:%20%20
          (if (Variables|/Script/SeasonPass.SeasonPassCategoryVM|GetbIsFullyClaimed _array_element)
            (Math|Integer|IncrementInt (LocalVariables|GetFullyClaimedMainCategorySets)))))))
  (return (== (LocalVariables|GetFullyClaimedMainCategorySets) (LocalVariables|GetNumberOfMainCategorySets))))

(fn OnCpcDialogRequested
  (params (Outer :type Object%20Reference) (BodyText :type Text) (OnCpcDialogDismissed :type Delegate%20%28by%20ref%29) (Target :type Widget%20Object%20Reference) (Context :type User%20Widget%20Object%20Reference))
  (bind _returnvalue (Game|ConstructObjectfromClass "/FortniteCoreUI/Frontend/RightOfWithdrawal/DialogVM_RightOfWithdrawal.DialogVM_RightOfWithdrawal_C" Outer))
  (bind _returnvalue_2 (UserInterface|CreateWidget 0 "/FortniteCoreUI/Frontend/RightOfWithdrawal/WBP_RightOfWithdrawal_Dialog.WBP_RightOfWithdrawal_Dialog_C" (Widget|GetOwningPlayer Target) _returnvalue BodyText))
  (Default|UnbindEventfromOnDialogDismissed _returnvalue_2 OnCpcDialogDismissed)
  (Default|BindEventtoOnDialogDismissed _returnvalue_2 OnCpcDialogDismissed)
  (FortUIManagement|TryPushWidgetasModalPanel Context _returnvalue_2))

(fn RewardHeaderMatchesReachLevel
  (params (RewardHeader :type Integer))
  (return (!= (Math|Integer|BitwiseAND RewardHeader (Math|Conversions|ToInteger(Byte) (Utilities|Enum|LiteralenumESeasonPassRewardHeader "ReachLevel"))) 0)))

(fn UpdateNexusFrontendMode
  (params (Target :type Widget%20Object%20Reference) (Is%20Activating :type Boolean))
  (bind _returnvalue (FortUIManagement|GetCurrentStateWidget (Widget|GetOwningPlayer Target)))
  (bind _asathena_uistate_widget_frontend (Utilities|Casting|CastToAthenaUIStateWidget_Frontend _returnvalue))
  (Class|AthenaUIStateWidgetFrontend|SetFrontendVisibilityMode _asathena_uistate_widget_frontend (select Is%20Activating "OnlyTop" "Normal")))
