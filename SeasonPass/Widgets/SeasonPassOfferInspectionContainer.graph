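# Invisible activatable widget that turns an item shop offer into a season pass selection and opens the pass purchase screen.
# On activation it looks up the pass by offer ID, parking the ID behind a timeout timer until SeasonPassesChanged if passes are loading.
# A match updates the selection viewmodel and async-pushes the purchase screen via tag alias; no match deactivates the container.
# A flag makes it deactivate itself on reactivation.
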WidgetBlueprint /SeasonPass/Widgets/SeasonPassOfferInspectionContainer
graphs [
 EventGraph {nodes [
   LaunchPurchaseScreen() {is=title K2Node_CustomEvent}
   > n1 = PushPrimaryContentSoftClassAsync!(
    execute:Exec
    SoftWidgetClass:Common Activatable Widget Soft Class Reference=GetActivatableWidgetForAlias!(WidgetAlias:Gameplay Tag Structure=(TagName="UI.Alias.PassPurchase")) -> (ReturnValue:Common Activatable Widget Soft Class Reference)
   ) -> (then:Exec, OnPrimaryContentPushed:Exec, ActivatableWidget:Common Activatable Widget Object Reference) {is=title K2Node_AsyncAction}
   SetViewModelByClass!(
    execute=@n1.OnPrimaryContentPushed
    self:MVVM View Object Reference=GetExtension!(self:User Widget Object Reference=@n1, ExtensionType:User Widget Extension Class Reference=/Script/ModelViewViewModel.MVVMView) -> (ReturnValue:MVVM View Object Reference)
    NewValue:Notify Field Value Changed Interface=Get SelectedPass!(self:Season Pass Selection VM Object Reference=Get SeasonPassSelectionVM() -> (SeasonPassSelectionVM:Season Pass Selection VM Object Reference)) -> (SelectedPass:Season Pass VM Object Reference)
   ) -> (then, ReturnValue:bool)
   Event Destruct()
   > Set PendingOfferId(PendingOfferId:String) -> (Output_Get:String)
   > Clear and Invalidate Timer by Handle(Handle=Get PendingOfferTimeoutHandle() -> (PendingOfferTimeoutHandle:Timer Handle Structure))
   Event On Activated()
   > n2 = Branch(Condition=Get IsPurchaseScreenCreated() -> (IsPurchaseScreenCreated:bool))
   > Set IsPurchaseScreenCreated(IsPurchaseScreenCreated:bool=false) -> (Output_Get:bool)
   > DeactivateWidget()
   Set IsPurchaseScreenCreated(execute=@n2.else, IsPurchaseScreenCreated=true)
   > OnShopOfferVMSet!(
    execute
    self
    ShopOfferVM:Fort Item Shop Offer VM Object Reference=Get FortItemShopOfferVM() -> (FortItemShopOfferVM:Fort Item Shop Offer VM Object Reference)
   ) -> (then)
  ]}
 OnShopOfferVMSet {nodes [
   OnShopOfferVMSet!() -> (then, ShopOfferVM:Fort Item Shop Offer VM Object Reference) {is=title K2Node_FunctionEntry}
   > n1 = Branch(Condition=Get bWaitingForPassesToLoad!(self:Season Pass System VM Object Reference=Get SeasonPassSystemVM() -> (SeasonPassSystemVM:Season Pass System VM Object Reference)) -> (bWaitingForPassesToLoad:bool))
   n2 = Get OfferId!(self:Fort Item Shop Offer VM Object Reference=Get ShopOfferVM() -> (ShopOfferVM:Fort Item Shop Offer VM Object Reference)) -> (OfferId:String)
   > Set PendingOfferId(PendingOfferId=@n2)
   > n3 = Set Timer by Event(Delegate=Create Event(), Time=1.5, bLooping=false)
   > Set PendingOfferTimeoutHandle(PendingOfferTimeoutHandle:Timer Handle Structure=@n3) -> (Output_Get:Timer Handle Structure)
   UpdateSelectionFromOfferId!(execute=@n1.else, self, OfferId:String=@n2) -> (then)
  ]}
 SeasonPassesChanged {nodes [
   SeasonPassesChanged!() -> (then, SeasonPasses:Array of Season Pass VM Object References) {is=title K2Node_FunctionEntry}
   n1 = Get PendingOfferId() -> (PendingOfferId:String)
   > Branch(Condition=NOT Boolean(A=IsEmpty(InString=@n1)))
   > n2 = Sequence()
   > UpdateSelectionFromOfferId(OfferId=@n1)
   Set PendingOfferId(execute=@n2.then_1)
  ]}
 UpdateSelectionFromOfferId {nodes [
   n1 = UpdateSelectionFromOfferId!() -> (then, OfferId:String) {is=title K2Node_FunctionEntry}
   > Set SeasonPass(SeasonPass:Season Pass VM Object Reference=FindSeasonPassByOfferId!(self:Season Pass System VM Object Reference=Get SeasonPassSystemVM(), OfferId:String=@n1) -> (ReturnValue:Season Pass VM Object Reference)) -> (Output_Get:Season Pass VM Object Reference)
   > Clear and Invalidate Timer by Handle(Handle=Get PendingOfferTimeoutHandle())
   > n2 = Is Valid(InputObject=Get SeasonPass() -> (SeasonPass:Season Pass VM Object Reference))
   n3 = Get SeasonPassSelectionVM()
   Set OfferInspection!(execute=@n2.Is Not Valid, OfferInspection:Fort Simple Offer VM Object Reference, self:Season Pass Selection VM Object Reference=@n3) -> (then, Output_Get:Fort Simple Offer VM Object Reference)
   > Set SelectedPass!(execute, SelectedPass:Season Pass VM Object Reference, self:Season Pass Selection VM Object Reference=@n3) -> (then, Output_Get:Season Pass VM Object Reference)
   > DeactivateWidget()
   n4 = Get SeasonPassSelectionVM()
   Set OfferInspection(
    execute=@n2.Is Valid
    OfferInspection=Get PassOffer!(self:Season Pass VM Object Reference=Get SeasonPass()) -> (PassOffer:Fort Simple Offer VM Object Reference)
    self=@n4
   )
   > Set SelectedPass(SelectedPass=Get SeasonPass(), self=@n4)
   > LaunchPurchaseScreen() {is=title K2Node_CallFunction}
  ]}
]
variables [
 IsPurchaseScreenCreated {type=boolean}
 PendingOfferId {type=string, default=""}
 PendingOfferTimeoutHandle {type=TimerHandle}
]
widgets {info {inheritedWidgetCount=0, parentClass=&CommonActivatableWidget, rootWidgetClass=null, widgetCount=0}, widgets []}
