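# A season pass item card widget for a purchasable reward, built on the UIKit item card base and used as a list entry.
# When its list item is set, it casts the object to a SeasonPassRewardVM, assigns it as the viewmodel and broadcasts OnVMUpdated with itself as the card.
# On construct, it hides the outline block and clears the equipped and enhanced states.
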WidgetBlueprint /SeasonPass/Widgets/SeasonPassPurchasableReward
graphs [
 EventGraph {nodes [
   n1 = Event OnListItemObjectSet()
   > n2 = Cast To SeasonPassRewardVM!(execute, Object=@n1.ListItemObject) -> (then, CastFailed, AsSeason Pass Reward VM:Season Pass Reward VM Object Reference) {is=title K2Node_DynamicCast}
   > SetSeasonPassRewardVM!(execute, self, Viewmodel:Season Pass Reward VM Object Reference=@n2) -> (then)
   > Call On VMUpdated!(
    execute:Exec
    self:Season Pass Purchasable Reward Object Reference
    RewardCard:Season Pass Purchasable Reward Object Reference=Self-Reference()
   ) -> (then:Exec) {is=title K2Node_CallDelegate}
   Event Construct()
   > SetBlockVisibility!(
    execute
    self:WBP UIKit Block Base Object Reference=Get Block_Outline() -> (Block_Outline:WBP UIKit Block Outline Object Reference)
    Visible:bool=false
   ) -> (then)
   > SetItemIsEquipped!(execute, self:WBP UIKit Item Card Base Object Reference, Is Equipped:bool=false) -> (then)
   > SetItemIsEnhanced!(execute, self:WBP UIKit Item Card Base Object Reference, Is Enhanced:bool=false) -> (then)
   Event On Hovered()
   Event On Unhovered()
  ]}
 OnVMUpdated {nodes [OnVMUpdated!() -> (then, RewardCard:Season Pass Purchasable Reward Object Reference) {is=title K2Node_FunctionEntry}]}
]
variables [OnVMUpdated {}]
widgets {
 info {
  inheritedWidgetCount=0
  parentClass=/FortUILibrary/Systems/Items/WBP_UIKit_ItemCard_Base.WBP_UIKit_ItemCard_Base_C
  rootWidgetClass=null
  widgetCount=0
 }
 widgets []
}
