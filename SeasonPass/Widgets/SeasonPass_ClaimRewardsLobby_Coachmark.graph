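# Season Pass lobby onboarding coach mark that introduces the Claim All rewards button, built on the Fortnite sidebar onboard tooltip widget.
# A single UIKit coach mark supplies the header, description and "Got it" action, with its dismiss button hidden.
# Showing the tooltip activates the coach mark, while ending it or clicking the action deactivates the widget.
# On deactivation the widget notifies that its close animation finished, then deactivates the coach mark.
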WidgetBlueprint /SeasonPass/Widgets/SeasonPass_ClaimRewardsLobby_Coachmark
graphs [EventGraph {nodes [
   "Button Other Action Clicked (WBP_UIKit_CoachMark)"() {is=title K2Node_ComponentBoundEvent}
   > n1 = HandleOtherActionClicked(self:Fort Sidebar Onboard Tooltip Widget Object Reference)
   DeactivateWidget(execute=[@n1.then, @n2.then])
   Event On Deactivated()
   > NotifyCloseAnimationFinished(self:Fort Sidebar Onboard Tooltip Widget Object Reference)
   > DeactivateWidget(self=Get WBP_UIKit_CoachMark() -> (WBP_UIKit_CoachMark:WBP UIKit Coach Mark Object Reference))
   n2 = Event OnEndTooltip() {is=title K2Node_Event}
   Event OnShowTooltip() {is=title K2Node_Event}
   > ActivateWidget(self=Get WBP_UIKit_CoachMark())
  ]}]
variables []
widgets {
 info {
  inheritedWidgetCount=2
  parentClass=/Script/FortniteUI.FortSidebarOnboardTooltipWidget
  rootWidgetClass=/FortUILibrary/UIKit/CoachMark/WBP_UIKit_CoachMark.WBP_UIKit_CoachMark_C
  widgetCount=3
 }
 widgets [
  WBP_UIKit_CoachMark {
   widgetClassPath=/FortUILibrary/UIKit/CoachMark/WBP_UIKit_CoachMark
   bIsVariable=true
   widget=WBP_UIKit_CoachMark
   properties {
    shouldShowDismissButton=false
    textDescription="This is a new way to claim all available rewards at once. From here, you can go straight to the reward track and start claiming by using the CLAIM ALL REWARDS button."
    textHeader=Introducing Claim All
    textOtherActionButton=Got it
    input Action Other {dataTable=&InputActionDataTable, rowName=GenericFaceButtonBottom}
   }
  }
  Dismiss_Button {widgetClassPath=&CommonButtonLegacy, bInherited=true, widget=null}
  OtherAction_Button {widgetClassPath=&CommonButtonLegacy, bInherited=true, widget=null}
 ]
}
