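# Expandable lobby widget that shows battle pass level progress and claimable reward and set counts, and opens the Battle Pass on click.
# View-model setters feed counts into one update that refreshes text, indicator, glow and the rewards-info callout.
# The callout plays a timed aura and particle animation, gated by visibility and a delay.
# Compact UI with touch forces an expanded state that hides XP info and expands when claimables exist.
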WidgetBlueprint /SeasonPass/Widgets/WBP_Expandable_Level
animations [
 VFX_ShowAnimation {
  binding_count=2
  length_seconds=0.4
  object=&self:VFX_ShowAnimation
  sequence {
   bindings [
    Particles {
     id=#1
     possessed_class=Image
     tracks [{
       type=MovieSceneWidgetMaterialTrack
       sections [{
         type=MovieSceneParameterSection
         end_frame=24
         start_frame=0
         channels [
          BigTrianglesOffset {type=MovieSceneScriptingFloatChannel, keys [[0,0],[15,3]]}
          Opacity {type=MovieSceneScriptingFloatChannel, keys [[0,0],[9,0],[21,1]]}
          SmallTrianglesOffset {type=MovieSceneScriptingFloatChannel, keys [[0,0],[24,1]]}
         ]
        }]
      }]
    }
    Aura {
     id=#2
     possessed_class=Image
     tracks [
      {
       type=MovieSceneFloatTrack
       display_name=Render Opacity
       sections [{type=MovieSceneFloatSection, end_frame=21, start_frame=0, channels [None {type=MovieSceneScriptingFloatChannel, keys [[0,0],[6,0],[21,1]]}]}]
      }
      {
       type=MovieScene2DTransformTrack
       sections [{
         type=MovieScene2DTransformSection
         end_frame=24
         start_frame=0
         channels [Scale.X {type=MovieSceneScriptingFloatChannel, keys [[0,3],[6,3],[24,1]]}]
        }]
      }
     ]
    }
   ]
   display_rate {numerator=60}
  }
 }
 VFX_HideAnimation {
  binding_count=2
  length_seconds=0.2
  object=&self:VFX_HideAnimation
  sequence {
   bindings [
    Aura {
     id=#3
     possessed_class=Image
     tracks [{
       type=MovieSceneFloatTrack
       display_name=Render Opacity
       sections [{type=MovieSceneFloatSection, channels [None {type=MovieSceneScriptingFloatChannel, keys [[0,1],[12,0]]}]}]
      }]
    }
    Particles {
     id=#4
     possessed_class=Image
     tracks [{
       type=MovieSceneWidgetMaterialTrack
       sections [{type=MovieSceneParameterSection, channels [Opacity {type=MovieSceneScriptingFloatChannel, keys [[0,1],[9,0]]}]}]
      }]
    }
   ]
   display_rate {numerator=60}
  }
 }
]
graphs [
 CF_GetBattlePass {nodes [
   n1 = CF_GetBattlePass!() -> (then, SeasonPassSystemVM:"Season Pass System VM Object Reference (by ref)", SeasonPasses:Array of Season Pass VM Object References) {is=title K2Node_FunctionEntry}
   > Return Node!(
    execute
    SeasonPass:Season Pass VM Object Reference=FindSeasonPass!(self:Season Pass System VM Object Reference=@n1, TemplateId:String=battlepassdata) -> (ReturnValue:Season Pass VM Object Reference)
   )
  ]}
 CF_GetRewardCount {nodes [
   CF_GetRewardCount!() -> (then, GiftBoxGroupVMs:Array of Gift Box Group VM Object References, InRewardCount:int) {is=title K2Node_FunctionEntry}
   > n1 = Branch(Condition=ContainsGiftboxesWithTags!(
     self:Gift Box System VM Object Reference=Get GiftBoxSystemVM() -> (GiftBoxSystemVM:Gift Box System VM Object Reference)
     TagQuery_TagQuery:Gameplay Tag Query Structure=(TokenStreamVersion=0,TagDictionary=((TagName="Giftbox.Display.PassHub")),QueryTokenStream=(0,1,1,1,0),UserDescription="",AutoDescription=" ANY( Giftbox.Display.PassHub )")
    ) -> (ReturnValue:bool))
   > Return Node!(execute, OutRewardCount:int=int + int(A=Get InRewardCount() -> (InRewardCount:int=0)))
   Return Node(execute=@n1.else, OutRewardCount=Get InRewardCount() -> (InRewardCount=0))
  ]}
 CollapseExtPart {nodes [
   CollapseExtPart() {is=title K2Node_FunctionEntry}
   > Parent: CollapseExtPart() {is=title K2Node_CallParentFunction}
   > SetVisibility(self=Get IndicatorWrapper() -> (IndicatorWrapper:Overlay Object Reference), InVisibility=SelfHitTestInvisible)
   > Branch(Condition=Get WasExpanded() -> (WasExpanded:bool))
   > SetGlowEffectVisibility!(execute, self, InVisibility:ESlateVisibility Enum=Hidden) -> (then)
  ]}
 EventGraph {nodes [
   n1 = OnVisibilityChanged_Event!() -> (OutputDelegate, then, InVisibility:ESlateVisibility Enum=Visible) {is=title K2Node_CustomEvent}
   > SetCanShowAnimation!(
    execute
    self:WBP Lobby Battle Pass Rewards Info Object Reference=Get WBP_Lobby_BattlePass_RewardsInfo() -> (WBP_Lobby_BattlePass_RewardsInfo:WBP Lobby Battle Pass Rewards Info Object Reference)
    CanShowAnimation:bool=IsVisible()
   ) -> (then)
   > Branch(Condition=Get IsPendingAnimation!(self:WBP Lobby Battle Pass Rewards Info Object Reference=Get WBP_Lobby_BattlePass_RewardsInfo()) -> (IsPendingAnimation:bool))
   > ShowRewardInfoVFX()
   StartVFXAnimTimer() {is=title K2Node_CustomEvent}
   > n2 = Sequence()
   Delay(
    execute=@n2.then_1
    Duration=Get DisplayLength!(self:WBP Lobby Battle Pass Rewards Info Object Reference=Get WBP_Lobby_BattlePass_RewardsInfo()) -> (DisplayLength:float)
   )
   > n3 = Sequence()
   Gate!(Enter:Exec=@n4.Finished, Open:Exec=@n3.then_0, Close:Exec=@n2.then_0, Toggle:Exec, bStartClosed:bool=true) -> (Exit:Exec) {is=title K2Node_MacroInstance}
   > SetVisibility(self=Get Overlay_Aura() -> (Overlay_Aura:Overlay Object Reference), InVisibility=Hidden)
   StopAnimation(execute=@n3.then_1, InAnimation=Get VFX_ShowAnimation() -> (VFX_ShowAnimation:Widget Animation Object Reference))
   > n4 = Play Animation with Finished event!(
    execute:Exec
    Widget:User Widget Object Reference=Self-Reference()
    InAnimation:Widget Animation Object Reference=Get VFX_HideAnimation() -> (VFX_HideAnimation:Widget Animation Object Reference)
    StartAtTime:float=0
    NumLoopsToPlay:int=1
    PlayMode:EUMGSequencePlayMode Enum
    PlaybackSpeed:float=1
   ) -> (then:Exec, Result:Widget Animation Handle Structure, Finished:Exec) {is=title K2Node_PlayAnimation2}
   HideRewardInfoVFX() {is=title K2Node_CustomEvent}
   > n5 = Branch(Condition=IsAnimationPlaying(InAnimation=Get VFX_ShowAnimation()))
   > StopAnimation(InAnimation=Get VFX_ShowAnimation())
   > n6 = Play Animation with Finished event(Widget=Self-Reference(), InAnimation=Get VFX_HideAnimation(), StartAtTime=0, NumLoopsToPlay=1, PlaybackSpeed=1)
   SetVisibility(execute=[@n6.Finished, @n5.else], self=Get Overlay_Aura(), InVisibility=Hidden)
   ShowRewardInfoVFX() {is=title K2Node_CustomEvent}
   > StopAnimation(InAnimation=Get VFX_HideAnimation())
   n7 = Get WBP_Lobby_BattlePass_RewardsInfo()
   > Branch(Condition=ShouldPlayAnimation!(self:WBP Lobby Battle Pass Rewards Info Object Reference=@n7) -> (OutShouldPlayAnim:bool))
   > Delay(Duration=Get DelayBeforePlayAnimation!(self:WBP Lobby Battle Pass Rewards Info Object Reference=@n7) -> (DelayBeforePlayAnimation:double))
   > Branch(Condition=Get CanShowAnimation!(self:WBP Lobby Battle Pass Rewards Info Object Reference=Get WBP_Lobby_BattlePass_RewardsInfo()) -> (CanShowAnimation:bool))
   > SetVisibility(self=Get Overlay_Aura(), InVisibility=SelfHitTestInvisible)
   > PlayAnimation(InAnimation=Get VFX_ShowAnimation())
   > StartVFXAnimTimer() {is=title K2Node_CallFunction}
   n8 = Event PreConstruct()
   > Parent: PreConstruct(IsDesignTime=@n8.IsDesignTime)
   > n9 = GetDynamicMaterial(self=Get RewardOverlay() -> (RewardOverlay:Image Object Reference))
   > SetScalarParameterValue(self=@n9, ParameterName=CornerRadius, Value=20)
   > SetBrushFromMaterial(
    self=Get Background!(self:WBP UIKit Highlight Base Object Reference=Get SetsAmount() -> (SetsAmount:WBP UIKit Highlight Object Reference)) -> (Background:Image Object Reference)
    Material=/SeasonPass/Widgets/Reboot/Materials/M_SeasonPass_Background_Unlock
   )
   Event Construct()
   > Assign On Visibility Changed!(execute:Exec, self:User Widget Object Reference, Delegate:"Delegate (by ref)"=@n1) -> (then:Exec) {is=title K2Node_AssignDelegate}
   n10 = Get SeasonPass_Autoclaim_LobbyCoachmark() -> (SeasonPass_Autoclaim_LobbyCoachmark:Season Pass Claim Rewards Lobby Coachmark Object Reference)
   > n11 = Branch(Condition=ShouldShowOnboarding!(self:Fort Sidebar Onboard Tooltip Widget Object Reference=@n10) -> (ReturnValue:bool))
   > n12 = ActivateWidget(self=@n10)
   SetRewardsAuraSafeZone(execute=[@n12.then, @n11.else])
   > UpdateTextContent()
   > UpdateTextVisibility()
   > n13 = Set CollapseException(CollapseException:bool=IsUsingCompactUI()) -> (Output_Get:bool)
   n14 = GetOwningPlayer()
   ForceExpandedState!(
    execute=[@n13.then, @n18.then]
    self
    shouldForceExpand:bool=AND Boolean(
     A=@n13
     B=IsUsingInputType!(InputType:ECommonInputType Enum=Touch, PlayerController:Player Controller Object Reference=@n14) -> (ReturnValue:bool)
    )
   ) -> (then)
   Event On Clicked()
   > Parent: On Clicked() {is=title K2Node_CallParentFunction}
   n15 = Get SeasonPassSystemVM() -> (SeasonPassSystemVM:Season Pass System VM Object Reference)
   > SetHasSeenPassOrHUB(self:Season Pass System VM Object Reference=@n15)
   > Fire Event Element Interacted(Element=Self-Reference())
   n16 = GetActiveProductTag!() -> (ReturnValue:Gameplay Tag Structure)
   n17 = GetCurrentCosmeticProductTag!() -> (ReturnValue:Gameplay Tag Structure)
   > NavigateToRewardExternal!(
    execute
    self:Season Pass Selection VM Object Reference=Get SeasonPassSelectionVM() -> (SeasonPassSelectionVM:Season Pass Selection VM Object Reference)
    Reward:Season Pass Reward VM Object Reference=GetFirstUnlockedReward!(self:Season Pass VM Object Reference=FindSeasonPassMatchingTags!(
      self:Season Pass System VM Object Reference=@n15
      CosmeticProductTag:"Gameplay Tag Structure (by ref)"=@n17
      PlaylistProductTag:"Gameplay Tag Structure (by ref)"=@n16
     ) -> (ReturnValue:Season Pass VM Object Reference)) -> (ReturnValue:Season Pass Reward VM Object Reference)
   ) -> (then)
   > NavigateToFeature!(execute, Feature:EFortUIFeature Enum=ShowBattlePass, bClearPrimaryContentStack:bool=true) -> (then)
   Event Destruct()
   > Unbind Event from On Visibility Changed(self:User Widget Object Reference, Delegate=@n1) {is=title K2Node_RemoveDelegate}
   n18 = Event On Input Method Changed!() -> (OutputDelegate, then, CurrentInputType:ECommonInputType Enum) {is=title K2Node_Event}
  ]}
 ExpandExtPart {nodes [
   ExpandExtPart() {is=title K2Node_FunctionEntry}
   > Parent: ExpandExtPart() {is=title K2Node_CallParentFunction}
   > SetVisibility(self=Get IndicatorWrapper(), InVisibility=Collapsed)
   > n1 = Branch(Condition=IsUsingInputType(InputType=Touch, PlayerController=GetOwningPlayer()))
   HideRewardInfo(execute=@n1.else, self:WBP Lobby Battle Pass Rewards Info Object Reference=Get WBP_Lobby_BattlePass_RewardsInfo())
   > HideRewardInfoVFX() {is=title K2Node_CallFunction}
   > Set WasExpanded(WasExpanded:bool=true) -> (Output_Get:bool)
  ]}
 ForceExpandedState {nodes [
   n1 = ForceExpandedState!() -> (then, shouldForceExpand:bool=~) {is=title K2Node_FunctionEntry}
   > n2 = Set ShouldForceExpand(ShouldForceExpand:bool=@n1) -> (Output_Get:bool)
   > SetVisibility(
    self=Get VerticalBox_LevelInfo() -> (VerticalBox_LevelInfo:Vertical Box Object Reference)
    InVisibility=Select(Option 0=SelfHitTestInvisible, Option 1=Collapsed, Index=@n2)
   )
   > UpdateSeparatorVisibilty()
   n3 = Get shouldForceExpand() -> (shouldForceExpand:bool)
   > n4 = Branch(Condition=AND Boolean(A=Get ShouldDisplayRewardsToClaim() -> (ShouldDisplayRewardsToClaim:bool), B=@n3))
   > ExpandExtPart(self:WBP Expandable Base Object Reference) {is=title K2Node_CallFunction}
   > RequestTransition!(
    execute
    self:UIKit Block Object Reference=Get Block_Outline() -> (Block_Outline:WBP UIKit Block Outline Object Reference)
    State:EUIKitBlockVisualState Enum=Unfocused
   ) -> (then)
   n5 = Branch(execute=@n4.else, Condition=@n3)
   > n6 = Set CollapseException(CollapseException=false)
   CollapseExtPart(execute=[@n5.else, @n6.then], self:WBP Expandable Base Object Reference) {is=title K2Node_CallFunction}
  ]}
 GetAttentionType {nodes [
   n1 = GetAttentionType!() -> (then, FortBangEnabled:bool=~, HasSeenPassesHUB:bool=~) {is=title K2Node_FunctionEntry}
   > n2 = Branch(Condition=@n1)
   > Return Node!(
    execute
    Result:E_UI_HighlightType Enum=Select(
     Option 0:E_UI_HighlightType Enum=NewEnumerator0
     Option 1:E_UI_HighlightType Enum=NewEnumerator5
     Index=Get HasSeenPassesHUB() -> (HasSeenPassesHUB:bool)
    ) -> (ReturnValue:E_UI_HighlightType Enum)
   )
   Return Node(execute=@n2.else, Result=NewEnumerator5)
  ]}
 ProgressXP {nodes [
   n1 = ProgressXP!() -> (then, CurrentXP:int, XPToNextLevel:int) {is=title K2Node_FunctionEntry}
   n2 = int + int(A=@n1, B=@n1.XPToNextLevel)
   > Return Node!(execute, ProgressXP:double=SafeDivide(A="To Float (Integer)"(InInt=@n1), B="To Float (Integer)"(InInt=@n2)))
  ]}
 SetGlowEffectVisibility {nodes [
   n1 = SetGlowEffectVisibility!() -> (then, InVisibility:ESlateVisibility Enum=Visible) {is=title K2Node_FunctionEntry}
   > SetVisibility(self=[Get RewardOverlay(), Get RewardGlow() -> (RewardGlow:Image Object Reference)], InVisibility=@n1)
  ]}
 SetLevelCircleProgress {nodes [
   n1 = SetLevelCircleProgress!() -> (then, XPProgress:double=~) {is=title K2Node_FunctionEntry}
   > n2 = GetDynamicMaterial(self=Get LevelProgress() -> (LevelProgress:Image Object Reference))
   > SetScalarParameterValue(self=@n2, ParameterName=Get Parameter Name() -> (Parameter Name:Name=None), Value=@n1)
  ]}
 SetRewardsAuraSafeZone {nodes [
   SetRewardsAuraSafeZone() {is=title K2Node_FunctionEntry}
   > n1 = GetSafeZonePadding!(execute) -> (
    then
    SafePadding_X:double
    SafePadding_Y:double
    SafePadding_Z:double
    SafePadding_W:double
    SafePaddingScale:Vector 2D Structure
    SpillOverPadding:Vector 4 Structure
   )
   > Branch(Condition="float > float"(A=@n1.SafePadding_Z))
   > SetPadding!(
    execute
    self:Size Box Slot Object Reference=SlotAsSizeBoxSlot!(Widget:Widget Object Reference=Get Overlay_Aura()) -> (ReturnValue:Size Box Slot Object Reference)
    InPadding_Left:float=0
    InPadding_Top:float=0
    InPadding_Right:float=-80
    InPadding_Bottom:float=0
   ) -> (then)
  ]}
 UpdateGlowEffectVisibility {nodes [
   UpdateGlowEffectVisibility() {is=title K2Node_FunctionEntry}
   n1 = NOT Boolean(A=GetHasSeenPassOrHUB!(self:Season Pass System VM Object Reference=Get SeasonPassSystemVM()) -> (ReturnValue:bool))
   n2 = OR Boolean(
    A="integer > integer"(A=Get InRewardCount() -> (InRewardCount=0), B=Get RewardCount() -> (RewardCount:int=0))
    B="integer > integer"(A=Get InSetCount() -> (InSetCount:int=0), B=Get SetCount() -> (SetCount:int=0))
   )
   > SetGlowEffectVisibility(InVisibility=Select(Option 0:ESlateVisibility Enum=Hidden, Option 1:ESlateVisibility Enum=SelfHitTestInvisible, Index=AND Boolean(A=@n2, B=@n1)) -> (ReturnValue:ESlateVisibility Enum)) {is=title K2Node_CallFunction}
  ]}
 UpdateSeparatorVisibilty {nodes [
   UpdateSeparatorVisibilty() {is=title K2Node_FunctionEntry}
   n1 = Get Has Rewards to Claim() -> (Has Rewards to Claim:bool)
   n2 = Get Has Sets to Claim() -> (Has Sets to Claim:bool)
   > SetVisibility(
    self=Get Separator() -> (Separator:Image Object Reference)
    InVisibility=Select(Option 0=Collapsed, Option 1=SelfHitTestInvisible, Index=OR Boolean(A=@n1, B=@n2))
   )
  ]}
 UpdateShowClaimableSetsAndRewards {nodes [
   UpdateShowClaimableSetsAndRewards() {is=title K2Node_FunctionEntry}
   > UpdateGlowEffectVisibility() {is=title K2Node_CallFunction}
   > n1 = Set Has Rewards to Claim(Has Rewards to Claim:bool="integer > integer"(A=Get InRewardCount() -> (InRewardCount=0))) -> (Output_Get:bool)
   > n2 = Set Has Sets to Claim(Has Sets to Claim:bool="integer > integer"(A=Get InSetCount() -> (InSetCount=0))) -> (Output_Get:bool)
   > Set ShouldDisplayRewardsToClaim(ShouldDisplayRewardsToClaim:bool=OR Boolean(A=@n2, B=@n1)) -> (Output_Get:bool)
   > Set RewardCount(RewardCount:int=Get InRewardCount() -> (InRewardCount=0)) -> (Output_Get:int=0)
   > Set SetCount(SetCount:int=Get InSetCount() -> (InSetCount=0)) -> (Output_Get:int=0)
   > UpdateTextContent()
   > UpdateTextVisibility()
   > ForceExpandedState(shouldForceExpand=Get ShouldForceExpand() -> (ShouldForceExpand:bool)) {is=title K2Node_CallFunction}
   > n3 = Branch(Condition=Get ShouldDisplayRewardsToClaim())
   > SetShowRewardsInfo!(
    execute
    self:WBP Lobby Battle Pass Rewards Info Object Reference=Get WBP_Lobby_BattlePass_RewardsInfo()
    InSetCount:int=Get SetCount() -> (SetCount=0)
    InRewardCount:int=Get RewardCount() -> (RewardCount=0)
   ) -> (then)
   > ShowRewardInfoVFX() {is=title K2Node_CallFunction}
   HideRewardInfo(execute=@n3.else, self=Get WBP_Lobby_BattlePass_RewardsInfo())
   > HideRewardInfoVFX()
  ]}
 UpdateTextContent {nodes [
   UpdateTextContent() {is=title K2Node_FunctionEntry}
   n1 = Get Has Rewards to Claim()
   n2 = Get Has Sets to Claim()
   n3 = Get RewardCount() -> (RewardCount=0)
   n4 = Get RewardCount() -> (RewardCount=0)
   > Set Text(
    Text=SelectText(
     A=Format Text!(Format="LOCTEXT(\"{Rewards}|plural(one=REWARD,other=REWARDS)\")", Rewards:int=@n3) -> (Result)
     B=Format Text(Format="LOCTEXT(\"{Rewards}|plural(one=REWARD,other=REWARDS) to claim\")", Rewards=@n4)
     bPickA=AND Boolean(A=@n1, B=@n2)
    )
    self=Get RewardAmount_Text() -> (RewardAmount_Text:Common Text Object Reference)
   )
   n5 = Get Has Rewards to Claim()
   n6 = Get Has Sets to Claim()
   n7 = Get SetCount() -> (SetCount=0)
   n8 = Get SetCount() -> (SetCount=0)
   > Set Text(
    Text=SelectText(
     A=Format Text!(Format="LOCTEXT(\"{Sets}|plural(one=SET,other=SETS)\")", Sets:int=@n7) -> (Result)
     B=Format Text(Format="LOCTEXT(\"{Sets}|plural(one=SET,other=SETS) to unlock\")", Sets=@n8)
     bPickA=AND Boolean(A=@n5, B=@n6)
    )
    self=Get SetsAmount_Text() -> (SetsAmount_Text:Common Text Object Reference)
   )
   > SetText!(
    execute
    self:WBP UIKit Highlight Base Object Reference=Get RewardAmount() -> (RewardAmount:WBP UIKit Highlight Object Reference)
    Text:Text="To Text (Integer)"(Value=Get RewardCount() -> (RewardCount=0))
   ) -> (then)
   > SetText(self=Get SetsAmount(), Text="To Text (Integer)"(Value=Get SetCount() -> (SetCount=0)))
  ]}
 UpdateTextVisibility {nodes [
   UpdateTextVisibility() {is=title K2Node_FunctionEntry}
   > SetVisibility(
    self=Get WBP_UIKit_StatusIndicator() -> (WBP_UIKit_StatusIndicator:WBP UIKit Status Indicator Object Reference)
    InVisibility=Select(Option 0=Hidden, Option 1=SelfHitTestInvisible, Index=Get ShouldDisplayRewardsToClaim())
   )
   > SetVisibility(
    self=Get Sets() -> (Sets:Horizontal Box Object Reference)
    InVisibility=Select(Option 0=Collapsed, Option 1=SelfHitTestInvisible, Index=Get Has Sets to Claim())
   )
  ]}
 VM_GetBangVM {nodes [
   n1 = VM_GetBangVM!() -> (then, PlayerVM:Fort Local Player VM Object Reference) {is=title K2Node_FunctionEntry}
   > n2 = Is Valid(InputObject=@n1)
   Return Node!(execute=@n2.Is Not Valid, Result:Fort Local Player Bang VM Object Reference)
   > n3 = GetBangVM!(execute, self:Fort Local Player VM Object Reference=@n1, BangType:EFortBangType Enum=BattlePassTab) -> (then, ReturnValue:Fort Local Player Bang VM Object Reference)
   > Return Node(Result=@n3)
  ]}
 VM_SetClaimableRewards {nodes [
   n1 = VM_SetClaimableRewards!() -> (then, ClaimableRewards:int) {is=title K2Node_FunctionEntry}
   > Set InRewardCount(InRewardCount:int=@n1) -> (Output_Get:int=0)
   > UpdateShowClaimableSetsAndRewards() {is=title K2Node_CallFunction}
  ]}
 VM_SetClaimableSets {nodes [
   n1 = VM_SetClaimableSets!() -> (then, NumUnlockedSets:int) {is=title K2Node_FunctionEntry}
   > Set InSetCount(InSetCount:int=@n1) -> (Output_Get:int=0)
   > UpdateShowClaimableSetsAndRewards()
  ]}
]
variables [
 Has Rewards to Claim {type=boolean}
 Has Sets to Claim {type=boolean}
 InRewardCount {type=integer}
 InSetCount {type=integer}
 Parameter Name {type=string, default=Progress}
 RewardCount {type=integer}
 SetCount {type=integer}
 ShouldDisplayRewardsToClaim {type=boolean}
 ShouldForceExpand {type=boolean}
 WasExpanded {type=boolean}
]
widgets {
 info {
  inheritedWidgetCount=0
  namedSlotCount=4
  parentClass=/Game/UI/Frontend/Lobby/Habanero/WBP_Expandable_Base.WBP_Expandable_Base_C
  rootWidgetClass=null
  widgetCount=38
 }
 widgets [
  GridPanel_LevelBox {widgetClassPath=&GridPanel, widget=GridPanel_LevelBox, properties {columnFill [1], rowFill [1]}}
  FortVisualAttachment_1 {
   widgetClassPath=&FortVisualAttachment
   parent=GridPanel_LevelBox
   widget=FortVisualAttachment_1
   properties {contentAnchor=<0.96 1.1>}
   slot {class=GridSlot, properties {horizontalAlignment=HAlign_Right, verticalAlignment=VAlign_Top}}
  }
  SeasonPass_Autoclaim_LobbyCoachmark {
   widgetClassPath=&/SeasonPass_ClaimRewardsLobby_Coachmark
   bIsVariable=true
   parent=FortVisualAttachment_1
   widget=SeasonPass_Autoclaim_LobbyCoachmark
   properties {bSupportsActivationFocus=false, visibility=Collapsed}
   slot {class=SizeBoxSlot}
  }
  FortVisualAttachment_0 {
   widgetClassPath=&FortVisualAttachment
   parent=GridPanel_LevelBox
   widget=FortVisualAttachment_0
   properties {bOverride_HeightOverride=true, bOverride_WidthOverride=true, heightOverride=144, widthOverride=144, contentAnchor=<0.5 0.5>}
   slot {class=GridSlot, properties {horizontalAlignment=HAlign_Center, verticalAlignment=VAlign_Center}}
  }
  LevelProgress {
   widgetClassPath=&Image
   bIsVariable=true
   parent=FortVisualAttachment_0
   widget=LevelProgress
   properties {
    visibility=SelfHitTestInvisible
    brush {
     bIsDynamicallyLoaded=false
     drawAs=Image
     imageType=FullColor
     mirroring=NoMirror
     resourceName=None
     resourceObject=/Game/UI/Frontend/Lobby/Habanero/Habanero_Materials/MaterialInstances/ProgressBars/MI_UI_Habanero_ProgressBar_PassLevel
     tiling=NoTile
     imageSize=<32 32>
     margin=<ltrb 0 0 0 0>
     outlineSettings {
      bUseBrushTransparency=false
      roundingType=HalfHeightRadius
      width=0
      color {colorUseRule=UseColor_Specified, specifiedColor=<rgba 0 0 0 0>}
      cornerRadii=<0 0 0 0>
     }
     tintColor {colorUseRule=UseColor_Specified, specifiedColor=<rgba 1 1 1 1>}
     uVRegion {bIsValid=false, max=<0 0>, min=<0 0>}
    }
   }
   slot {class=SizeBoxSlot}
  }
  VerticalBox_Level {
   widgetClassPath=&VerticalBox
   parent=GridPanel_LevelBox
   widget=VerticalBox_Level
   slot {class=GridSlot, properties {horizontalAlignment=HAlign_Center, layer=1, verticalAlignment=VAlign_Center}}
  }
  Text_2 {
   widgetClassPath=&CommonTextBlock
   parent=VerticalBox_Level
   widget=Text_2
   properties {
    bIsScrollingEnabled=false
    text=lvl
    textTransformPolicy=ToUpper
    colorAndOpacity {colorUseRule=UseColor_Specified, specifiedColor=<rgba 0.0048 0.004 0.0116 0.5>}
    font {
     bForceMonospaced=false
     bMaterialIsStencil=false
     fontMaterial=null
     fontObject=&HeadingNow
     letterSpacing=0
     monospacedWidth=1
     size=15.75
     skewAmount=0
     typefaceFontName=75 Medium
     outlineSettings {
      bApplyOutlineToDropShadows=false
      bMiteredCorners=false
      bSeparateFillAlpha=false
      outlineMaterial=null
      outlineSize=0
      outlineColor=<rgba 0 0 0 1>
     }
    }
    shadowOffset=<0 0>
    shapedTextOptions {bOverride_TextFlowDirection=true, bOverride_TextShapingMethod=false, textFlowDirection=Culture, textShapingMethod=Auto}
   }
   slot {class=VerticalBoxSlot, properties {horizontalAlignment=HAlign_Center, verticalAlignment=VAlign_Top, padding=<ltrb 0 0 0 -6>}}
  }
  Text_Level {
   widgetClassPath=&CommonTextBlock
   bIsVariable=true
   parent=VerticalBox_Level
   widget=Text_Level
   properties {
    bIsScrollingEnabled=false
    lineHeightPercentage=0.9
    mobileFontSizeMultiplier=1.25
    text=999
    colorAndOpacity {colorUseRule=UseColor_Specified, specifiedColor=<rgba 0.0048 0.004 0.0116 1>}
    font {
     bForceMonospaced=false
     bMaterialIsStencil=false
     fontMaterial=null
     fontObject=&HeadingNow
     letterSpacing=10
     monospacedWidth=1
     size=18
     skewAmount=0
     typefaceFontName=86 Bold
     outlineSettings {
      bApplyOutlineToDropShadows=false
      bMiteredCorners=false
      bSeparateFillAlpha=false
      outlineMaterial=null
      outlineSize=0
      outlineColor=<rgba 0 0 0 1>
     }
    }
    shadowOffset=<0 0>
    shapedTextOptions {bOverride_TextFlowDirection=true, bOverride_TextShapingMethod=false, textFlowDirection=Culture, textShapingMethod=Auto}
   }
   slot {class=VerticalBoxSlot, properties {horizontalAlignment=HAlign_Center, verticalAlignment=VAlign_Bottom, padding=<ltrb 0 -4 0 0>}}
  }
  IndicatorWrapper {
   widgetClassPath=&Overlay
   bIsVariable=true
   parent=GridPanel_LevelBox
   widget=IndicatorWrapper
   slot {class=GridSlot, properties {horizontalAlignment=HAlign_Right, verticalAlignment=VAlign_Top, padding=<ltrb 0 -3 -3 0>}}
  }
  WBP_UIKit_StatusIndicator {
   widgetClassPath=/FortUILibrary/UIKit/StatusIndicator/WBP_UIKit_StatusIndicator
   bIsVariable=true
   parent=IndicatorWrapper
   widget=WBP_UIKit_StatusIndicator
   properties {type=Attention, configuration=Dot, overrideStatusIndicatorVisibility=true, visibility=Collapsed}
   slot {class=OverlaySlot}
  }
  GridPanel_73 {widgetClassPath=&GridPanel, widget=GridPanel_73, properties {columnFill [1], rowFill [1]}}
  HorizontalBox_49 {
   widgetClassPath=&HorizontalBox
   parent=GridPanel_73
   widget=HorizontalBox_49
   slot {class=GridSlot, properties {padding=<ltrb 8 0 8 0>}}
  }
  VerticalBox_1 {
   widgetClassPath=&VerticalBox
   parent=HorizontalBox_49
   widget=VerticalBox_1
   slot {class=HorizontalBoxSlot, properties {verticalAlignment=VAlign_Center, size {sizeRule=Fill, value=1}}}
  }
  Rewards {widgetClassPath=&HorizontalBox, bIsVariable=true, parent=VerticalBox_1, widget=Rewards, slot {class=VerticalBoxSlot}}
  SizeBox_0 {
   widgetClassPath=&SizeBox
   parent=Rewards
   widget=SizeBox_0
   properties {bOverride_HeightOverride=true, bOverride_MinDesiredWidth=true, heightOverride=32, minDesiredWidth=32}
   slot {class=HorizontalBoxSlot, properties {horizontalAlignment=HAlign_Center, verticalAlignment=VAlign_Center, padding=<ltrb 0 4 0 4>}}
  }
  RewardAmount {
   widgetClassPath=/FortUILibrary/UIKit/Highlight/WBP_UIKit_Highlight
   bIsVariable=true
   parent=SizeBox_0
   widget=RewardAmount
   properties {highlightText=XX, highlightType=Attention}
   slot {class=SizeBoxSlot}
  }
  RewardAmount_Text {
   widgetClassPath=&CommonTextBlock
   bIsVariable=true
   parent=Rewards
   widget=RewardAmount_Text
   properties {
    autoWrapText=true
    bIsScrollingEnabled=false
    mobileFontSizeMultiplier=1.25
    text=Rewards to claim
    textTransformPolicy=ToUpper
    wrapTextAt=150
    colorAndOpacity {colorUseRule=UseColor_Specified, specifiedColor=<rgba 0.0048 0.004 0.0116 1>}
    font {
     bForceMonospaced=false
     bMaterialIsStencil=false
     fontMaterial=null
     fontObject=&HeadingNow
     letterSpacing=0
     monospacedWidth=1
     size=22.5
     skewAmount=0
     typefaceFontName=75 Medium Italic
     outlineSettings {
      bApplyOutlineToDropShadows=false
      bMiteredCorners=false
      bSeparateFillAlpha=false
      outlineMaterial=null
      outlineSize=0
      outlineColor=<rgba 0 0 0 1>
     }
    }
    shadowOffset=<0 0>
    shapedTextOptions {bOverride_TextFlowDirection=true, bOverride_TextShapingMethod=false, textFlowDirection=Culture, textShapingMethod=Auto}
   }
   slot {class=HorizontalBoxSlot, properties {horizontalAlignment=HAlign_Center, verticalAlignment=VAlign_Center, padding=<ltrb 8 0 0 0>}}
  }
  Sets {widgetClassPath=&HorizontalBox, bIsVariable=true, parent=VerticalBox_1, widget=Sets, slot {class=VerticalBoxSlot}}
  SizeBox {
   widgetClassPath=&SizeBox
   parent=Sets
   widget=SizeBox
   properties {bOverride_HeightOverride=true, bOverride_MinDesiredWidth=true, heightOverride=32, minDesiredWidth=32}
   slot {class=HorizontalBoxSlot, properties {horizontalAlignment=HAlign_Center, verticalAlignment=VAlign_Center, padding=<ltrb 0 4 0 4>}}
  }
  SetsAmount {
   widgetClassPath=&WBP_UIKit_Highlight
   bIsVariable=true
   parent=SizeBox
   widget=SetsAmount
   properties {highlightText=XX, highlightType=Urgent}
   slot {class=SizeBoxSlot}
  }
  SetsAmount_Text {
   widgetClassPath=&CommonTextBlock
   bIsVariable=true
   parent=Sets
   widget=SetsAmount_Text
   properties {
    autoWrapText=true
    bIsScrollingEnabled=false
    mobileFontSizeMultiplier=1.25
    text=Sets to unlock
    textTransformPolicy=ToUpper
    wrapTextAt=150
    colorAndOpacity {colorUseRule=UseColor_Specified, specifiedColor=<rgba 0.0048 0.004 0.0116 1>}
    font {
     bForceMonospaced=false
     bMaterialIsStencil=false
     fontMaterial=null
     fontObject=&HeadingNow
     letterSpacing=0
     monospacedWidth=1
     size=22.5
     skewAmount=0
     typefaceFontName=75 Medium Italic
     outlineSettings {
      bApplyOutlineToDropShadows=false
      bMiteredCorners=false
      bSeparateFillAlpha=false
      outlineMaterial=null
      outlineSize=0
      outlineColor=<rgba 0 0 0 1>
     }
    }
    shadowOffset=<0 0>
    shapedTextOptions {bOverride_TextFlowDirection=true, bOverride_TextShapingMethod=false, textFlowDirection=Culture, textShapingMethod=Auto}
   }
   slot {class=HorizontalBoxSlot, properties {horizontalAlignment=HAlign_Center, verticalAlignment=VAlign_Center, padding=<ltrb 8 0 0 0>}}
  }
  Separator {
   widgetClassPath=&Image
   bIsVariable=true
   parent=HorizontalBox_49
   widget=Separator
   properties {
    renderOpacity=0.25
    visibility=SelfHitTestInvisible
    brush {
     bIsDynamicallyLoaded=false
     drawAs=Image
     imageType=NoImage
     mirroring=NoMirror
     resourceName=None
     resourceObject=null
     tiling=NoTile
     imageSize=<2 1>
     margin=<ltrb 0 0 0 0>
     outlineSettings {
      bUseBrushTransparency=false
      roundingType=HalfHeightRadius
      width=0
      color {colorUseRule=UseColor_Specified, specifiedColor=<rgba 0 0 0 0>}
      cornerRadii=<0 0 0 0>
     }
     tintColor {colorUseRule=UseColor_Specified, specifiedColor=<rgba 1 1 1 1>}
     uVRegion {bIsValid=false, max=<0 0>, min=<0 0>}
    }
   }
   slot {class=HorizontalBoxSlot, properties {horizontalAlignment=HAlign_Center, padding=<ltrb 16 8 16 8>}}
  }
  VerticalBox_LevelInfo {
   widgetClassPath=&VerticalBox
   bIsVariable=true
   parent=HorizontalBox_49
   widget=VerticalBox_LevelInfo
   slot {class=HorizontalBoxSlot, properties {verticalAlignment=VAlign_Center}}
  }
  HorizontalBox_51 {
   widgetClassPath=&HorizontalBox
   parent=VerticalBox_LevelInfo
   widget=HorizontalBox_51
   slot {class=VerticalBoxSlot, properties {verticalAlignment=VAlign_Center, size {sizeRule=Fill, value=1}}}
  }
  Text_XpToNextLevel {
   widgetClassPath=&CommonTextBlock
   bIsVariable=true
   parent=HorizontalBox_51
   widget=Text_XpToNextLevel
   properties {
    bIsScrollingEnabled=false
    mobileFontSizeMultiplier=1.25
    text=XXXXX
    textTransformPolicy=ToUpper
    colorAndOpacity {colorUseRule=UseColor_Specified, specifiedColor=<rgba 0.0048 0.004 0.0116 1>}
    font {
     bForceMonospaced=false
     bMaterialIsStencil=false
     fontMaterial=null
     fontObject=&HeadingNow
     letterSpacing=0
     monospacedWidth=1
     size=22.5
     skewAmount=0
     typefaceFontName=75 Medium Italic
     outlineSettings {
      bApplyOutlineToDropShadows=false
      bMiteredCorners=false
      bSeparateFillAlpha=false
      outlineMaterial=null
      outlineSize=0
      outlineColor=<rgba 0 0 0 1>
     }
    }
    shadowOffset=<0 0>
    shapedTextOptions {bOverride_TextFlowDirection=true, bOverride_TextShapingMethod=false, textFlowDirection=Culture, textShapingMethod=Auto}
   }
   slot {class=HorizontalBoxSlot, properties {padding=<ltrb 0 0 8 0>}}
  }
  Text_Level_3 {
   widgetClassPath=&CommonTextBlock
   parent=HorizontalBox_51
   widget=Text_Level_3
   properties {
    bIsScrollingEnabled=false
    mobileFontSizeMultiplier=1.25
    text=XP
    textTransformPolicy=ToUpper
    colorAndOpacity {colorUseRule=UseColor_Specified, specifiedColor=<rgba 0.0048 0.004 0.0116 1>}
    font {
     bForceMonospaced=false
     bMaterialIsStencil=false
     fontMaterial=null
     fontObject=&HeadingNow
     letterSpacing=0
     monospacedWidth=1
     size=22.5
     skewAmount=0
     typefaceFontName=75 Medium Italic
     outlineSettings {
      bApplyOutlineToDropShadows=false
      bMiteredCorners=false
      bSeparateFillAlpha=false
      outlineMaterial=null
      outlineSize=0
      outlineColor=<rgba 0 0 0 1>
     }
    }
    shadowOffset=<0 0>
    shapedTextOptions {bOverride_TextFlowDirection=true, bOverride_TextShapingMethod=false, textFlowDirection=Culture, textShapingMethod=Auto}
   }
   slot {class=HorizontalBoxSlot}
  }
  Overlay_0 {
   widgetClassPath=&Overlay
   parent=VerticalBox_LevelInfo
   widget=Overlay_0
   slot {class=VerticalBoxSlot, properties {verticalAlignment=VAlign_Center, size {sizeRule=Fill, value=1}}}
  }
  Text_Level_2 {
   widgetClassPath=&CommonTextBlock
   parent=Overlay_0
   widget=Text_Level_2
   properties {
    bIsScrollingEnabled=false
    mobileFontSizeMultiplier=1.25
    text=to next LVL
    textTransformPolicy=ToUpper
    colorAndOpacity {colorUseRule=UseColor_Specified, specifiedColor=<rgba 0.0666 0.0666 0.0666 1>}
    font {
     bForceMonospaced=false
     bMaterialIsStencil=false
     fontMaterial=null
     fontObject=&HeadingNow
     letterSpacing=0
     monospacedWidth=1
     size=18
     skewAmount=0
     typefaceFontName=75 Medium
     outlineSettings {
      bApplyOutlineToDropShadows=false
      bMiteredCorners=false
      bSeparateFillAlpha=false
      outlineMaterial=null
      outlineSize=0
      outlineColor=<rgba 0 0 0 1>
     }
    }
    shadowOffset=<0 0>
    shapedTextOptions {bOverride_TextFlowDirection=true, bOverride_TextShapingMethod=false, textFlowDirection=Culture, textShapingMethod=Auto}
   }
   slot {class=OverlaySlot}
  }
  Overlay_20 {widgetClassPath=&Overlay, widget=Overlay_20}
  RewardOverlay {
   widgetClassPath=&Image
   bIsVariable=true
   parent=Overlay_20
   widget=RewardOverlay
   properties {
    visibility=Hidden
    brush {
     bIsDynamicallyLoaded=false
     drawAs=Image
     imageType=FullColor
     mirroring=NoMirror
     resourceName=None
     resourceObject=/FortUILibrary/MaterialInstances/Effects/MI_UI_RewardState
     tiling=NoTile
     imageSize=<32 32>
     margin=<ltrb 0 0 0 0>
     outlineSettings {
      bUseBrushTransparency=false
      roundingType=HalfHeightRadius
      width=0
      color {colorUseRule=UseColor_Specified, specifiedColor=<rgba 0 0 0 0>}
      cornerRadii=<0 0 0 0>
     }
     tintColor {colorUseRule=UseColor_Specified, specifiedColor=<rgba 1 1 1 1>}
     uVRegion {bIsValid=false, max=<0 0>, min=<0 0>}
    }
   }
   slot {class=OverlaySlot, properties {horizontalAlignment=HAlign_Fill, verticalAlignment=VAlign_Fill}}
  }
  Overlay_59 {widgetClassPath=&Overlay, widget=Overlay_59}
  RewardGlow {
   widgetClassPath=&Image
   bIsVariable=true
   parent=Overlay_59
   widget=RewardGlow
   properties {
    visibility=Hidden
    brush {
     bIsDynamicallyLoaded=false
     drawAs=Image
     imageType=FullColor
     mirroring=NoMirror
     resourceName=None
     resourceObject=/FortUILibrary/MaterialInstances/Effects/MI_UI_RewardBacking
     tiling=NoTile
     imageSize=<32 32>
     margin=<ltrb 0 0 0 0>
     outlineSettings {
      bUseBrushTransparency=false
      roundingType=HalfHeightRadius
      width=0
      color {colorUseRule=UseColor_Specified, specifiedColor=<rgba 0 0 0 0>}
      cornerRadii=<0 0 0 0>
     }
     tintColor {colorUseRule=UseColor_Specified, specifiedColor=<rgba 1 1 1 1>}
     uVRegion {bIsValid=false, max=<0 0>, min=<0 0>}
    }
   }
   slot {class=OverlaySlot, properties {horizontalAlignment=HAlign_Fill, verticalAlignment=VAlign_Fill, padding=<ltrb -128 -64 -128 -64>}}
  }
  FortVisualAttachment_41 {
   widgetClassPath=&FortVisualAttachment
   parent=Overlay_59
   widget=FortVisualAttachment_41
   properties {contentAnchor=<1 0.5>}
   slot {class=OverlaySlot, properties {verticalAlignment=VAlign_Center}}
  }
  WBP_Lobby_BattlePass_RewardsInfo {
   widgetClassPath=&/WBP_Lobby_BattlePass_RewardsInfo
   bIsVariable=true
   parent=FortVisualAttachment_41
   widget=WBP_Lobby_BattlePass_RewardsInfo
   slot {class=SizeBoxSlot, properties {horizontalAlignment=HAlign_Right, verticalAlignment=VAlign_Center, padding=<ltrb 0 0 40 0>}}
  }
  Attachement_VFX {
   widgetClassPath=&CommonVisualAttachment
   parent=Overlay_59
   widget=Attachement_VFX
   properties {
    bOverride_HeightOverride=true
    bOverride_WidthOverride=true
    heightOverride=120
    widthOverride=520
    contentAnchor=<1 0.5>
    renderTransform {angle=0, scale=<1 1>, shear=<0 0>, translation=<216 0>}
   }
   slot {class=OverlaySlot, properties {horizontalAlignment=HAlign_Right, verticalAlignment=VAlign_Center, padding=<ltrb -200 0 136 0>}}
  }
  Overlay_Aura {
   widgetClassPath=&Overlay
   bIsVariable=true
   parent=Attachement_VFX
   widget=Overlay_Aura
   properties {visibility=Hidden}
   slot {class=SizeBoxSlot}
  }
  Aura {
   widgetClassPath=&Image
   parent=Overlay_Aura
   widget=Aura
   properties {
    visibility=SelfHitTestInvisible
    brush {
     bIsDynamicallyLoaded=false
     drawAs=Image
     imageType=FullColor
     mirroring=NoMirror
     resourceName=None
     resourceObject=/FortniteCoreUI/UI/Frontend/Lobby/Habanero/Materials/MI_UI_Lobby_RewardFlare
     tiling=NoTile
     imageSize=<332 32>
     margin=<ltrb 0 0 0 0>
     outlineSettings {
      bUseBrushTransparency=false
      roundingType=HalfHeightRadius
      width=0
      color {colorUseRule=UseColor_Specified, specifiedColor=<rgba 0 0 0 0>}
      cornerRadii=<0 0 0 0>
     }
     tintColor {colorUseRule=UseColor_Specified, specifiedColor=<rgba 1 1 1 1>}
     uVRegion {bIsValid=false, max=<0 0>, min=<0 0>}
    }
    renderTransformPivot=<1 0.5>
   }
   slot {class=OverlaySlot, properties {horizontalAlignment=HAlign_Right, verticalAlignment=VAlign_Fill}}
  }
  Particles {
   widgetClassPath=&Image
   parent=Overlay_Aura
   widget=Particles
   properties {
    visibility=SelfHitTestInvisible
    brush {
     bIsDynamicallyLoaded=false
     drawAs=Image
     imageType=FullColor
     mirroring=NoMirror
     resourceName=None
     resourceObject=/FortniteCoreUI/UI/Frontend/Lobby/Habanero/MaterialInstances/MI_UI_Lobby_Level_Particles
     tiling=NoTile
     imageSize=<32 32>
     margin=<ltrb 0 0 0 0>
     outlineSettings {
      bUseBrushTransparency=false
      roundingType=HalfHeightRadius
      width=0
      color {colorUseRule=UseColor_Specified, specifiedColor=<rgba 0 0 0 0>}
      cornerRadii=<0 0 0 0>
     }
     tintColor {colorUseRule=UseColor_Specified, specifiedColor=<rgba 1 1 1 1>}
     uVRegion {bIsValid=false, max=<0 0>, min=<0 0>}
    }
   }
   slot {class=OverlaySlot, properties {horizontalAlignment=HAlign_Fill, verticalAlignment=VAlign_Fill}}
  }
 ]
}
