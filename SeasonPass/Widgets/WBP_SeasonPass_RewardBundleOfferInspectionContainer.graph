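# Widgetless activatable container that opens the Buy Levels screen for a season pass level bundle offer from the item shop.
# On activation, it finds the SeasonPassVM by the shop offer ID and sets it, deactivating if lookup fails.
# VM_SeasonPassSet asynchronously pushes WBP_RewardTrack_BuyLevels, passing the viewmodel and the bundle's level count.
# When that screen deactivates, the container unbinds, clears its SeasonPassVM and deactivates itself.
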WidgetBlueprint /SeasonPass/Widgets/WBP_SeasonPass_RewardBundleOfferInspectionContainer
graphs [EventGraph {nodes [
   n1 = VM_SeasonPassSet!() -> (OutputDelegate, then, SeasonPass:Season Pass VM Object Reference) {is=title K2Node_CustomEvent}
   > n2 = PushPrimaryContentClassAsync!(
    execute:Exec
    WidgetClass:Common Activatable Widget Class Reference=/SeasonPass/Widgets/Reboot/WBP_RewardTrack_BuyLevels.WBP_RewardTrack_BuyLevels_C
   ) -> (then:Exec, OnPrimaryContentPushed:Exec, ActivatableWidget:Common Activatable Widget Object Reference) {is=title K2Node_AsyncAction}
   n3 = Cast To WBP_RewardTrack_BuyLevels!(execute=@n2.OnPrimaryContentPushed, Object=@n2) -> (then, CastFailed, AsWBP Reward Track Buy Levels:WBP Reward Track Buy Levels Object Reference) {is=title K2Node_DynamicCast}
   > SetSeasonPassVM!(execute, self:WBP Reward Track Buy Levels Object Reference=@n3, Viewmodel:Season Pass VM Object Reference=@n1.SeasonPass) -> (then)
   > n4 = SetLevelsToBuy!(
    execute
    self:WBP Reward Track Buy Levels Object Reference=@n3
    Amount:int=GetLevelsToGrant!(self:Season Pass Bundle Offer VM Object Reference=Get LevelBundleOffer!(self:Season Pass VM Object Reference=@n1.SeasonPass) -> (LevelBundleOffer:Season Pass Bundle Offer VM Object Reference)) -> (ReturnValue:int=0)
    IsAdditive:bool=false
   ) -> (then)
   n5 = OnModalDeactivated() {is=title K2Node_CustomEvent}
   Assign On Widget Deactivated!(execute:Exec=@n4.then, self:Common Activatable Widget Object Reference=@n3, Delegate:"Delegate (by ref)"=@n5) -> (then:Exec) {is=title K2Node_AssignDelegate}
   DeactivateWidget(execute=@n3.CastFailed)
   Event On Activated()
   n6 = Get FortItemShopOfferVM() -> (FortItemShopOfferVM:Fort Item Shop Offer VM Object Reference)
   n7 = Get SeasonPassSystemVM() -> (SeasonPassSystemVM:Season Pass System VM Object Reference)
   > n8 = Branch(Condition=AND Boolean(A=IsValid(Object=@n6), B=IsValid(Object=@n7)))
   n9 = FindSeasonPassByLevelBundleOfferId!(
    self:Season Pass System VM Object Reference=Get SeasonPassSystemVM()
    LevelBundleOfferId:String=Get OfferId!(self:Fort Item Shop Offer VM Object Reference=Get FortItemShopOfferVM()) -> (OfferId:String)
   ) -> (ReturnValue:Season Pass VM Object Reference)
   > n10 = Is Valid(InputObject=@n9)
   DeactivateWidget(execute=@n10.Is Not Valid)
   SetSeasonPassVM(execute=@n10.Is Valid, self:Self Object Reference, Viewmodel=@n9)
   DeactivateWidget(execute=@n8.else)
   DeactivateWidget(execute=@n11.then)
   n11 = SetSeasonPassVM(execute=@n12.then)
   n12 = Unbind Event from On Widget Deactivated(execute=@n5.then, self:Common Activatable Widget Object Reference=@n3, Delegate=@n5) {is=title K2Node_RemoveDelegate}
  ]}]
variables []
widgets {info {inheritedWidgetCount=0, parentClass=&CommonActivatableWidget, rootWidgetClass=null, widgetCount=0}, widgets []}
