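# Battle Pass variant of the UIKit item card, adding a collapsed owned checkmark and removing the inherited equipped outline.
# List items that cast to FortItemVM initialize the base card, and focusing the card broadcasts OnCardFocused with itself.
# OnNewCardSelected syncs focus to whether this card is selected, and the selected card skips the parent unhover handling.
# Leaving reward viewing resets animation and focus state.
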WidgetBlueprint /FortUILibrary/UIKit/Items/WBP_UIKit_ItemCard_BattlePass
graphs [
 EventGraph {nodes [
   n1 = OnNewCardSelected!() -> (OutputDelegate, then, SelectedCard:WBP UIKit Item Card Battle Pass Object Reference) {is=title K2Node_CustomEvent}
   > n2 = Set IsCurrentlySelectedCard(IsCurrentlySelectedCard:bool="Equal (Object)"(A=@n1.SelectedCard, B=Self-Reference())) -> (Output_Get:bool)
   > Branch(Condition="Not Equal (Boolean)"(A=@n2, B=Get IsButtonFocused() -> (IsButtonFocused:bool)) -> (ReturnValue=~))
   > SetFocusState!(execute, self:WBP UIKit Item Card Base Object Reference, IsFocused:bool=Get IsCurrentlySelectedCard() -> (IsCurrentlySelectedCard:bool)) -> (then)
   n3 = OnViewingRewardStateChanged!() -> (OutputDelegate, then, IsViewingReward:bool=~) {is=title K2Node_CustomEvent}
   > n4 = Branch(Condition=@n3.IsViewingReward)
   ResetAnimationState(execute=@n4.else, self:WBP UIKit Item Card Base Object Reference)
   > Set IsButtonFocused(IsButtonFocused:bool=false) -> (Output_Get:bool)
   n5 = Event PreConstruct()
   > Parent: PreConstruct(IsDesignTime=@n5.IsDesignTime)
   > RemoveFromParent(self=Get EquippedState_Outline() -> (EquippedState_Outline:WBP UIKit Equipped State Object Reference))
   n6 = Event OnListItemObjectSet()
   > Cast To FortItemVM!(execute, Object=@n6.ListItemObject) -> (then, CastFailed, AsFort Item VM:Fort Item VM Object Reference) {is=title K2Node_DynamicCast}
   > InitializeItemCard(self:WBP UIKit Item Card Base Object Reference)
   Event On Focused() {is=title K2Node_Event}
   > Parent: On Focused() {is=title K2Node_CallParentFunction}
   > Call On Card Focused!(execute:Exec, self:WBP UIKit Item Card Battle Pass Object Reference, Card:WBP UIKit Item Card Battle Pass Object Reference=Self-Reference()) -> (then:Exec) {is=title K2Node_CallDelegate}
   Event On Unhovered()
   > n7 = Branch(Condition=Get IsCurrentlySelectedCard())
   Parent: On Unhovered(execute=@n7.else) {is=title K2Node_CallParentFunction}
  ]}
 OnCardFocused {nodes [OnCardFocused!() -> (then, Card:WBP UIKit Item Card Battle Pass Object Reference) {is=title K2Node_FunctionEntry}]}
 Set Item from Item Definition {nodes [Set Item from Item Definition!() -> (then, ItemDef:Fort Item Definition Object Reference) {is=title K2Node_FunctionEntry}]}
 SetAthenaItemEntry {nodes [
   n1 = SetAthenaItemEntry!() -> (then, ItemEntry:Athena Season Item Entry Base Object Reference) {is=title K2Node_FunctionEntry}
   > Set ItemEntryBase(ItemEntryBase:Athena Season Item Entry Base Object Reference=@n1) -> (Output_Get:Athena Season Item Entry Base Object Reference)
  ]}
]
variables [
 IsCurrentlySelectedCard {type=boolean}
 IsOwned {type=boolean}
 ItemEntryBase {type=AthenaSeasonItemEntryBase, default=null}
 OnCardFocused {}
 PartiallyOwnsVariants {type=boolean}
 Variants {type="array<McpVariantChannelInfo>", default []}
]
widgets {
 info {
  inheritedWidgetCount=0
  namedSlotCount=1
  parentClass=/FortUILibrary/Systems/Items/WBP_UIKit_ItemCard_Base.WBP_UIKit_ItemCard_Base_C
  rootWidgetClass=null
  widgetCount=2
 }
 widgets [
  GridPanel_Owned {widgetClassPath=&GridPanel, widget=GridPanel_Owned, properties {columnFill [1], rowFill [1]}}
  StatusIndicator_Owned {
   widgetClassPath=/FortUILibrary/UIKit/StatusIndicator/WBP_UIKit_StatusIndicator
   bIsVariable=true
   parent=GridPanel_Owned
   widget=StatusIndicator_Owned
   properties {icon=/FortUILibrary/Textures/MaterialInstances/MI_UIKit_Icon_Check, overrideStatusIndicatorVisibility=true, visibility=Collapsed}
   slot {class=GridSlot, properties {horizontalAlignment=HAlign_Right, verticalAlignment=VAlign_Top, padding=<ltrb 0 8 8 0>}}
  }
 ]
}
