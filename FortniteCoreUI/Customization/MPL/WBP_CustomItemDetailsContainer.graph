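# A container widget that hosts category-specific custom item details in a single named slot, apparently for the Fortnite locker.
# CreateCustomItemDetails clears the slot, reads the soft widget class from the item category's locker config, and creates it asynchronously if valid, suspending input meanwhile.
# On completion it adds the widget, broadcasts OnCustomWidgetSet, then injects the item details view model by class.
# RemoveCustomItemDetails clears the slot.
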WidgetBlueprint /FortniteCoreUI/Customization/MPL/WBP_CustomItemDetailsContainer
graphs [
 EventGraph {nodes [
   n1 = CreateCustomItemDetails!() -> (OutputDelegate, then, Item Details VM:Fort Item Details VM Object Reference) {is=title K2Node_CustomEvent}
   > ClearChildren(self=Get NamedSlot_CustomDetails() -> (NamedSlot_CustomDetails:Named Slot Object Reference))
   n2 = Get CustomItemDetails!(self:Fort Locker Category Config Object Reference=GetItemCategory!(self:Fort Item VM Object Reference=Get Item!(self:Fort Item Details VM Object Reference=@n1.Item Details VM) -> (Item:Fort Item VM Object Reference)) -> (ReturnValue:Fort Locker Category Config Object Reference)) -> (CustomItemDetails:User Widget Soft Class Reference)
   > Branch(Condition=IsValidSoftClassReference!(SoftClassReference:"Object Soft Class Reference (by ref)"=@n2) -> (ReturnValue:bool))
   > n3 = CreateWidgetAsync!(
    execute:Exec
    UserWidgetSoftClass:User Widget Soft Class Reference=@n2
    OwningPlayer:Player Controller Object Reference=GetOwningPlayer()
    bSuspendInputUntilComplete:bool=true
   ) -> (then:Exec, OnComplete:Exec, OnFailure:Exec, UserWidget:User Widget Object Reference) {is=title K2Node_AsyncAction}
   AddChild(execute=@n3.OnComplete, self=Get NamedSlot_CustomDetails(), Content=@n3)
   > Call On Custom Widget Set!(execute:Exec, self:WBP Custom Item Details Container Object Reference, Widget:User Widget Object Reference=@n3) -> (then:Exec) {is=title K2Node_CallDelegate}
   > SetViewModelByClass!(execute, Widget:User Widget Object Reference=@n3, ViewModel:Notify Field Value Changed Interface=@n1.Item Details VM) -> (then)
  ]}
 OnCustomWidgetSet {nodes [OnCustomWidgetSet!() -> (then, Widget:User Widget Object Reference) {is=title K2Node_FunctionEntry}]}
 RemoveCustomItemDetails {nodes [
   RemoveCustomItemDetails() {is=title K2Node_FunctionEntry}
   > ClearChildren(self=Get NamedSlot_CustomDetails())
  ]}
]
variables [OnCustomWidgetSet {}]
widgets {
 info {inheritedWidgetCount=0, parentClass=&UserWidget, rootWidgetClass=&NamedSlot, widgetCount=1}
 widgets [NamedSlot_CustomDetails {widgetClassPath=&NamedSlot, bIsVariable=true, widget=NamedSlot_CustomDetails}]
}
