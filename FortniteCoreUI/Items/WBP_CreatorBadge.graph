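# Creator badge widget that shows a single lazy-loaded image, for valid badge assets only when the item's current product is Product.BR.
# SetBadge checks both conditions, then loads and shows the image, caches its dynamic material and plays Intro, or collapses the image otherwise.
# The Intro animation keys IntroLerp, whose SetIntroLerp setter feeds the material's "Reveal Amount" parameter to animate the reveal.
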WidgetBlueprint /FortniteCoreUI/Items/WBP_CreatorBadge
animations [Intro {
  binding_count=1
  length_seconds=2
  object=&self:Intro
  sequence {bindings [WBP_CreatorBadge_C_27 {
     id=#1
     possessed_class=WBP_CreatorBadge_C
     tracks [{
       type=MovieSceneDoubleTrack
       display_name=Intro Lerp
       sections [{type=MovieSceneDoubleSection, end_frame=40, start_frame=0, channels [None {type=MovieSceneScriptingDoubleChannel, keys [[0,0],[33,1]]}]}]
      }]
    }]}
 }]
graphs [
 EventGraph {nodes [
   Event PreConstruct()
   Event Construct()
   Event Tick()
  ]}
 SetBadge {nodes [
   n1 = SetBadge!() -> (then, LazyObject:"Soft Object Reference (by ref)") {is=title K2Node_FunctionEntry}
   n2 = GetCurrentProduct!(self:Fort Item VM Object Reference=Get Item!(self:Fort Item Details VM Object Reference=Get FortItemDetailsVM() -> (FortItemDetailsVM:Fort Item Details VM Object Reference)) -> (Item:Fort Item VM Object Reference)) -> (ReturnValue:Gameplay Tag Structure)
   > n3 = Branch(Condition=AND Boolean(A=IsValidSoftObjectReference(SoftObjectReference=@n1), B="Equal (GameplayTag)"(A=@n2, B=(TagName="Product.BR"))))
   n4 = Get FortLazyImage_30() -> (FortLazyImage_30:Fort Lazy Image Object Reference)
   > SetBrushFromLazyDisplayAsset!(execute, self:Common Lazy Image Object Reference=@n4, LazyObject:"Soft Object Reference (by ref)"=@n1, bMatchTextureSize:bool=false) -> (then)
   > SetVisibility(self=@n4, InVisibility=SelfHitTestInvisible)
   > n5 = GetDynamicMaterial(self=Get FortLazyImage_30())
   > Set MaterialInstance(MaterialInstance:Material Instance Dynamic Object Reference=@n5) -> (Output_Get:Material Instance Dynamic Object Reference)
   > PlayAnimation(InAnimation=Get Intro())
   SetVisibility(execute=@n3.else, self=@n4, InVisibility=Collapsed)
  ]}
 SetIntroLerp {nodes [
   n1 = SetIntroLerp!() -> (then, Alpha:double=~) {is=title K2Node_FunctionEntry}
   > SetScalarParameterValue(self=Get MaterialInstance() -> (MaterialInstance:Material Instance Dynamic Object Reference), ParameterName=Reveal Amount, Value=@n1)
  ]}
]
variables [
 IntroLerp {type=number}
 MaterialInstance {type=MaterialInstanceDynamic}
]
widgets {
 info {inheritedWidgetCount=0, parentClass=&CommonUserWidget, rootWidgetClass=&FortLazyImage, widgetCount=1}
 widgets [FortLazyImage_30 {
   widgetClassPath=&FortLazyImage
   bIsVariable=true
   widget=FortLazyImage_30
   properties {
    visibility=Collapsed
    brush {
     bIsDynamicallyLoaded=false
     drawAs=Image
     imageType=NoImage
     mirroring=NoMirror
     resourceName=None
     resourceObject=null
     tiling=NoTile
     imageSize=<300 100>
     margin=<ltrb 0 0 0 0>
     outlineSettings {
      bUseBrushTransparency=false
      roundingType=HalfHeightRadius
      width=0
      color {colorUseRule=UseColor_Specified, specifiedColor=<rgba 0 0 0 0>}
      cornerRadii=<0 0 0 0>
     }
     tintColor {colorUseRule=UseColor_Specified, specifiedColor=<rgba 1 1 1 1>}
     uVRegion {bIsValid=false, max=<0 0>, min=<0 0>}
    }
   }
  }]
}
