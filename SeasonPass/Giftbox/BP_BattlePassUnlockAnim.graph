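# Battle Pass unlock presentation widget, a subclass of AthenaUpgradedPreMessage, showing an animated pass icon, the pass name and "Unlocked!".
# SetGiftBox looks up the season pass from the gift box item's seasonTemplateId, and SetPass applies its name and lazily loaded icon material.
# Activation plays a fanfare and UnlockSequence, which animates transforms, tints and material parameters.
# The CallGiftbox sequencer event ends the presentation and broadcasts OnUnlockAnimationFinished.
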WidgetBlueprint /SeasonPass/Giftbox/BP_BattlePassUnlockAnim
animations [UnlockSequence {
  binding_count=7
  length_seconds=4.5
  object=&self:UnlockSequence
  sequence {
   bindings [
    Image_Backing {
     id=#1
     possessed_class=Image
     tracks [
      {
       type=MovieSceneWidgetMaterialTrack
       sections [{
         type=MovieSceneParameterSection
         channels [
          Triangles_Amount {type=MovieSceneScriptingFloatChannel, keys [[33,0],[78,0],[138,2.5],[213,1],[270,0]]}
          Density {type=MovieSceneScriptingFloatChannel, thinned_from=5, keys [[33,0.666],[78,1],[237,1],[270,0.8]]}
          TransitionBackground {type=MovieSceneScriptingFloatChannel, keys [[33,0],[93,0],[138,1],[213,1]]}
          TrianglesMask_Radius {type=MovieSceneScriptingFloatChannel, thinned_from=6, keys [[33,2],[78,2],[108,4],[237,4],[270,8]]}
          Hotspot_Intensity {type=MovieSceneScriptingFloatChannel, keys [[33,1],[63,1],[108,3],[213,1],[237,1],[270,0.9]]}
         ]
        }]
      }
      {
       type=MovieSceneFloatTrack
       display_name=Render Opacity
       sections [{
         type=MovieSceneFloatSection
         channels [None {type=MovieSceneScriptingFloatChannel, thinned_from=5, keys [[12,0],[33,1],[251.716,1],[264.767,0]]}]
        }]
      }
     ]
    }
    AnimBoundBattlepassText {
     id=#2
     possessed_class=Border
     tracks [{
       type=MovieSceneColorTrack
       display_name=Content Color and Opacity
       sections [{
         type=MovieSceneColorSection
         channels [
          Color.R {type=MovieSceneScriptingFloatChannel, const [1,2]}
          Color.G {type=MovieSceneScriptingFloatChannel, keys [[0,1],[221,1],[246,0.625]]}
          Color.B {type=MovieSceneScriptingFloatChannel, keys [[0,1],[221,1],[246,0.125]]}
          Color.A {type=MovieSceneScriptingFloatChannel, const [1,1]}
         ]
        }]
      }]
    }
    AnimBoundUnlockText {
     id=#3
     possessed_class=Border
     tracks [{
       type=MovieSceneColorTrack
       display_name=Content Color and Opacity
       sections [{
         type=MovieSceneColorSection
         channels [
          Color.R {type=MovieSceneScriptingFloatChannel, const [1,2]}
          Color.G {type=MovieSceneScriptingFloatChannel, keys [[0,1],[217,1],[246,0.8]]}
          Color.B {type=MovieSceneScriptingFloatChannel, keys [[0,1],[217,1],[246,0.25]]}
          Color.A {type=MovieSceneScriptingFloatChannel, const [1,1]}
         ]
        }]
      }]
    }
    Text_BattlePass {
     id=#4
     display_name="Backing (5)"
     possessed_class=CommonTextBlock
     tracks [
      {
       type=MovieScene2DTransformTrack
       sections [{
         type=MovieScene2DTransformSection
         channels [
          Translation.X {type=MovieSceneScriptingFloatChannel}
          Translation.Y {type=MovieSceneScriptingFloatChannel, keys [[105,64],[132,0],[221,0],[254,144]]}
          Angle {type=MovieSceneScriptingFloatChannel}
          Scale.X {type=MovieSceneScriptingFloatChannel, keys [[221,1],[254,3]]}
          Scale.Y {type=MovieSceneScriptingFloatChannel, keys [[221,1],[254,3]]}
          Shear.X {type=MovieSceneScriptingFloatChannel}
          Shear.Y {type=MovieSceneScriptingFloatChannel}
         ]
        }]
      }
      {
       type=MovieSceneFloatTrack
       display_name=Render Opacity
       sections [{type=MovieSceneFloatSection, channels [None {type=MovieSceneScriptingFloatChannel, keys [[105,0],[153,1],[240,1],[250,0]]}]}]
      }
     ]
    }
    Unlocked {
     id=#5
     possessed_class=CommonTextBlock
     tracks [
      {
       type=MovieScene2DTransformTrack
       sections [{
         type=MovieScene2DTransformSection
         channels [
          Translation.X {type=MovieSceneScriptingFloatChannel}
          Translation.Y {type=MovieSceneScriptingFloatChannel, keys [[217,0],[246,160]]}
          Angle {type=MovieSceneScriptingFloatChannel}
          Scale.X {type=MovieSceneScriptingFloatChannel, keys [[123,4],[138,0.95],[144,1.025],[153,1],[217,1],[246,4]]}
          Scale.Y {type=MovieSceneScriptingFloatChannel, keys [[123,2],[138,0.975],[144,1.025],[153,1],[217,1],[246,4]]}
          Shear.X {type=MovieSceneScriptingFloatChannel}
          Shear.Y {type=MovieSceneScriptingFloatChannel}
         ]
        }]
      }
      {
       type=MovieSceneFloatTrack
       display_name=Render Opacity
       sections [{type=MovieSceneFloatSection, channels [None {type=MovieSceneScriptingFloatChannel, thinned_from=5, keys [[124,0],[138,1],[237,1],[246,0]]}]}]
      }
     ]
    }
    IconBox {
     id=#6
     possessed_class=SizeBox
     tracks [
      {
       type=MovieSceneFloatTrack
       display_name=Render Opacity
       sections [{type=MovieSceneFloatSection, channels [None {type=MovieSceneScriptingFloatChannel, thinned_from=5, keys [[33,0],[36,1],[252,1],[254,0]]}]}]
      }
      {
       type=MovieScene2DTransformTrack
       sections [{
         type=MovieScene2DTransformSection
         channels [
          Translation.X {type=MovieSceneScriptingFloatChannel}
          Translation.Y {type=MovieSceneScriptingFloatChannel, keys [[98,0],[128,-128],[228,-128],[254,-80]]}
          Angle {type=MovieSceneScriptingFloatChannel}
          Scale.X {type=MovieSceneScriptingFloatChannel, keys [[33,0],[63,1],[228,1],[254,3]]}
          Scale.Y {type=MovieSceneScriptingFloatChannel, keys [[33,0],[63,1],[228,1],[254,3]]}
          Shear.X {type=MovieSceneScriptingFloatChannel}
          Shear.Y {type=MovieSceneScriptingFloatChannel}
         ]
        }]
      }
     ]
    }
    Image_Icon {
     id=#7
     possessed_class=Image
     tracks [{
       type=MovieSceneWidgetMaterialTrack
       sections [{
         type=MovieSceneParameterSection
         channels [
          ShineAngle {type=MovieSceneScriptingFloatChannel, const [0,2]}
          ShineBend {type=MovieSceneScriptingFloatChannel, keys [[48,-4],[63,-4],[198,6]]}
          ShineSwipe {type=MovieSceneScriptingFloatChannel, keys [[48,-0.3],[63,-0.3],[198,0.3]]}
          DividerLine_Thiccness {type=MovieSceneScriptingFloatChannel, keys [[48,-3.8894],[63,-3],[114,0],[213,-4]]}
          WaveDistortAmount {type=MovieSceneScriptingFloatChannel, keys [[48,0],[78,0],[114,0.2],[198,0]]}
          NoiseIntensity_Center {type=MovieSceneScriptingFloatChannel, keys [[48,0],[114,6],[228,0]]}
          NoiseIntensity_Edges {type=MovieSceneScriptingFloatChannel, keys [[48,0],[78,3],[228,0]]}
          SoftNoise_Intensity {type=MovieSceneScriptingFloatChannel, keys [[48,0],[114,1],[228,0]]}
          NoiseDistortion_Amount {type=MovieSceneScriptingFloatChannel, keys [[48,0],[114,0.1],[213,0]]}
          DotsIntensity {type=MovieSceneScriptingFloatChannel, keys [[48,0],[63,0],[114,4],[213,0]]}
          PureNoiseTint_Amount {type=MovieSceneScriptingFloatChannel, keys [[48,0],[114,6],[228,0]]}
          BeamGlow {type=MovieSceneScriptingFloatChannel, keys [[48,0],[93,0.9867],[168,0]]}
          BeamThiccness {type=MovieSceneScriptingFloatChannel, keys [[48,0],[88,1],[168,-1]]}
          NoiseContrast_Edges {type=MovieSceneScriptingFloatChannel, keys [[48,1.325],[114,1],[228,2]]}
          SheenSwipe {type=MovieSceneScriptingFloatChannel, keys [[138,0.75],[258,-0.75]]}
          SheenAngle {type=MovieSceneScriptingFloatChannel, const [0.1,1]}
          UsePureTint {type=MovieSceneScriptingFloatChannel, keys [[33,1],[48,0],[228,0],[254,1]]}
          PureTintColorR {type=MovieSceneScriptingFloatChannel, keys [[40,0.25],[228,1]]}
          PureTintColorG {type=MovieSceneScriptingFloatChannel, keys [[40,0.8125],[228,0.496]]}
          PureTintColorB {type=MovieSceneScriptingFloatChannel, keys [[40,1],[228,0.055]]}
          PureTintColorA {type=MovieSceneScriptingFloatChannel, const [1,1]}
         ]
        }]
      }]
    }
   ]
   display_rate {numerator=60}
   tracks [{
     type=MovieSceneEventTrack
     sections [{
       type=MovieSceneEventTriggerSection
       channels [None {type=MovieSceneScriptingEventChannel, const ["<Struct 'MovieSceneEvent' (0x46916f156e0) {}>", 1]}]
      }]
    }]
  }
 }]
graphs [
 EventGraph {nodes [
   PlayIntro() {is=title K2Node_CustomEvent}
   > PlayAnimation(InAnimation=Get UnlockSequence() -> (UnlockSequence:Widget Animation Object Reference))
   > PlaySound2D(Sound=Get SoundOnPlayIntro() -> (SoundOnPlayIntro:Sound Base Object Reference))
   Event On Activated()
   > PlayIntro() {is=title K2Node_CallFunction}
  ]}
 OnUnlockAnimationFinished {nodes [OnUnlockAnimationFinished() {is=title K2Node_FunctionEntry}]}
 Sequencer Events {nodes [
   CallGiftbox() {is=title K2Node_CustomEvent}
   > OnPresentationFinished(self:Athena Upgraded Pre Message Object Reference)
   > Call On Unlock Animation Finished!(execute:Exec, self:BP Battle Pass Unlock Anim Object Reference) -> (then:Exec) {is=title K2Node_CallDelegate}
  ]}
 SetGiftBox {nodes [
   n1 = SetGiftBox!() -> (then, GiftBox:Gift Box VM Object Reference) {is=title K2Node_FunctionEntry}
   > Is Valid(InputObject=@n1)
   n2 = Get GiftBoxItem!(self:Gift Box VM Object Reference=@n1) -> (GiftBoxItem:Fort Gift Box Item Object Reference)
   > SetPass!(
    execute
    self
    Pass:Season Pass VM Object Reference=FindSeasonPass!(
     self:Season Pass System VM Object Reference=Get SeasonPassSystemVM() -> (SeasonPassSystemVM:Season Pass System VM Object Reference)
     TemplateId:String=TryGetParamString!(self:Fort Gift Box Item Object Reference=@n2, ParamName:String=seasonTemplateId) -> (OutValue:String, ReturnValue:bool)
    ) -> (ReturnValue:Season Pass VM Object Reference)
   ) -> (then)
  ]}
 SetPass {nodes [
   n1 = SetPass!() -> (then, Pass:Season Pass VM Object Reference) {is=title K2Node_FunctionEntry}
   > Is Valid(InputObject=@n1)
   > Set Text(
    Text=GetDisplayName!(self:Season Pass VM Object Reference=@n1) -> (ReturnValue:Text)
    self=Get Text_BattlePass() -> (Text_BattlePass:Common Text Object Reference)
   )
   > SetBrushFromLazyDisplayAsset!(
    execute
    self:Common Lazy Image Object Reference=Get Image_Icon() -> (Image_Icon:Fort Lazy Image Object Reference)
    LazyObject:"Soft Object Reference (by ref)"=GetPurchasePreMessageMaterial!(self:Season Pass VM Object Reference=@n1) -> (ReturnValue:Soft Object Reference)
    bMatchTextureSize:bool=false
   ) -> (then)
  ]}
]
variables [
 OnUnlockAnimationFinished {}
 SoundOnPlayIntro {type=SoundBase, default=/Game/Sounds/Fort_UI_Sounds/FrontEnd/BattlePass/BattlePass_Button_Confirm_Buy_Fanfare_Cue}
]
widgets {
 info {inheritedWidgetCount=0, parentClass=/Script/FortniteUI.AthenaUpgradedPreMessage, rootWidgetClass=&Overlay, widgetCount=13}
 widgets [
  Overlay_1 {widgetClassPath=&Overlay, widget=Overlay_1}
  Overlay {
   widgetClassPath=&Overlay
   parent=Overlay_1
   widget=Overlay
   properties {clipping=ClipToBounds}
   slot {class=OverlaySlot, properties {horizontalAlignment=HAlign_Fill, verticalAlignment=VAlign_Fill}}
  }
  Image_Backing {
   widgetClassPath=&Image
   bIsVariable=true
   parent=Overlay
   widget=Image_Backing
   properties {brush {
     bIsDynamicallyLoaded=false
     drawAs=Image
     imageType=NoImage
     mirroring=NoMirror
     resourceName=None
     resourceObject=&/M_UI_BattlePassPurchase_Screen_Transition_Inst
     tiling=NoTile
     imageSize=<32 32>
     margin=<ltrb 0 0 0 0>
     outlineSettings {
      bUseBrushTransparency=false
      roundingType=HalfHeightRadius
      width=0
      color {colorUseRule=UseColor_Specified, specifiedColor=<rgba 0 0 0 0>}
      cornerRadii=<0 0 0 0>
     }
     tintColor {colorUseRule=UseColor_Specified, specifiedColor=<rgba 1 1 1 1>}
     uVRegion {bIsValid=false, max=<0 0>, min=<0 0>}
    }}
   slot {class=OverlaySlot, properties {horizontalAlignment=HAlign_Fill, verticalAlignment=VAlign_Fill}}
  }
  IconBox {
   widgetClassPath=&SizeBox
   parent=Overlay
   widget=IconBox
   properties {bOverride_HeightOverride=true, bOverride_WidthOverride=true, heightOverride=512, widthOverride=512}
   slot {class=OverlaySlot, properties {horizontalAlignment=HAlign_Center, verticalAlignment=VAlign_Center}}
  }
  AnimBoundIconTint {
   widgetClassPath=&Border
   parent=IconBox
   widget=AnimBoundIconTint
   properties {
    background {
     bIsDynamicallyLoaded=false
     drawAs=NoDrawType
     imageType=NoImage
     mirroring=NoMirror
     resourceName=None
     resourceObject=null
     tiling=NoTile
     imageSize=<32 32>
     margin=<ltrb 0 0 0 0>
     outlineSettings {
      bUseBrushTransparency=false
      roundingType=HalfHeightRadius
      width=0
      color {colorUseRule=UseColor_Specified, specifiedColor=<rgba 0 0 0 0>}
      cornerRadii=<0 0 0 0>
     }
     tintColor {colorUseRule=UseColor_Specified, specifiedColor=<rgba 1 1 1 1>}
     uVRegion {bIsValid=false, max=<0 0>, min=<0 0>}
    }
    padding=<ltrb 0 0 0 0>
   }
   slot {class=SizeBoxSlot}
  }
  Image_Icon {
   widgetClassPath=&FortLazyImage
   bIsVariable=true
   parent=AnimBoundIconTint
   widget=Image_Icon
   properties {brush {
     bIsDynamicallyLoaded=false
     drawAs=Image
     imageType=FullColor
     mirroring=NoMirror
     resourceName=None
     resourceObject=&/M_UI_BattlePassPurchase_Animated_Transition_Inst
     tiling=NoTile
     imageSize=<32 32>
     margin=<ltrb 0 0 0 0>
     outlineSettings {
      bUseBrushTransparency=false
      roundingType=HalfHeightRadius
      width=0
      color {colorUseRule=UseColor_Specified, specifiedColor=<rgba 0 0 0 0>}
      cornerRadii=<0 0 0 0>
     }
     tintColor {colorUseRule=UseColor_Specified, specifiedColor=<rgba 1 1 1 1>}
     uVRegion {bIsValid=false, max=<0 0>, min=<0 0>}
    }}
   slot {class=BorderSlot, properties {padding=<ltrb 0 0 0 0>}}
  }
  VerticalBox {
   widgetClassPath=&VerticalBox
   parent=Overlay
   widget=VerticalBox
   slot {class=OverlaySlot, properties {horizontalAlignment=HAlign_Center, verticalAlignment=VAlign_Fill}}
  }
  Spacer_93 {
   widgetClassPath=&Spacer
   parent=VerticalBox
   widget=Spacer_93
   slot {class=VerticalBoxSlot, properties {size {sizeRule=Fill, value=1}}}
  }
  VerticalBox_2 {
   widgetClassPath=&VerticalBox
   parent=VerticalBox
   widget=VerticalBox_2
   slot {class=VerticalBoxSlot, properties {padding=<ltrb 0 256 0 0>, size {sizeRule=Fill, value=1}}}
  }
  AnimBoundBattlepassText {
   widgetClassPath=&Border
   parent=VerticalBox_2
   widget=AnimBoundBattlepassText
   properties {
    background {
     bIsDynamicallyLoaded=false
     drawAs=NoDrawType
     imageType=NoImage
     mirroring=NoMirror
     resourceName=None
     resourceObject=null
     tiling=NoTile
     imageSize=<32 32>
     margin=<ltrb 0 0 0 0>
     outlineSettings {
      bUseBrushTransparency=false
      roundingType=HalfHeightRadius
      width=0
      color {colorUseRule=UseColor_Specified, specifiedColor=<rgba 0 0 0 0>}
      cornerRadii=<0 0 0 0>
     }
     tintColor {colorUseRule=UseColor_Specified, specifiedColor=<rgba 1 1 1 1>}
     uVRegion {bIsValid=false, max=<0 0>, min=<0 0>}
    }
    padding=<ltrb 0 0 0 0>
   }
   slot {class=VerticalBoxSlot, properties {horizontalAlignment=HAlign_Center}}
  }
  Text_BattlePass {
   widgetClassPath=&CommonTextBlock
   bIsVariable=true
   parent=AnimBoundBattlepassText
   widget=Text_BattlePass
   properties {
    justification=Center
    text=Battle Pass
    textTransformPolicy=ToUpper
    colorAndOpacity {colorUseRule=UseColor_Specified, specifiedColor=<rgba 1 0.95 0.4 1>}
    font {
     bForceMonospaced=false
     bMaterialIsStencil=false
     fontMaterial=null
     fontObject=&Burbank
     letterSpacing=0
     monospacedWidth=1
     size=40
     skewAmount=0
     typefaceFontName=BurbankRegular-Black
     outlineSettings {
      bApplyOutlineToDropShadows=false
      bMiteredCorners=false
      bSeparateFillAlpha=false
      outlineMaterial=null
      outlineSize=2
      outlineColor=<rgba 0.0908 0.0185 0.0033 0>
     }
    }
    shadowOffset=<0 0>
   }
   slot {class=BorderSlot, properties {padding=<ltrb 0 0 0 0>}}
  }
  AnimBoundUnlockText {
   widgetClassPath=&Border
   parent=VerticalBox_2
   widget=AnimBoundUnlockText
   properties {
    background {
     bIsDynamicallyLoaded=false
     drawAs=NoDrawType
     imageType=NoImage
     mirroring=NoMirror
     resourceName=None
     resourceObject=null
     tiling=NoTile
     imageSize=<32 32>
     margin=<ltrb 0 0 0 0>
     outlineSettings {
      bUseBrushTransparency=false
      roundingType=HalfHeightRadius
      width=0
      color {colorUseRule=UseColor_Specified, specifiedColor=<rgba 0 0 0 0>}
      cornerRadii=<0 0 0 0>
     }
     tintColor {colorUseRule=UseColor_Specified, specifiedColor=<rgba 1 1 1 1>}
     uVRegion {bIsValid=false, max=<0 0>, min=<0 0>}
    }
    padding=<ltrb 0 0 0 0>
   }
   slot {class=VerticalBoxSlot, properties {horizontalAlignment=HAlign_Center, verticalAlignment=VAlign_Center, padding=<ltrb 0 -16 0 0>}}
  }
  Unlocked {
   widgetClassPath=&CommonTextBlock
   parent=AnimBoundUnlockText
   widget=Unlocked
   properties {
    justification=Center
    text="Unlocked!"
    textTransformPolicy=ToUpper
    font {
     bForceMonospaced=false
     bMaterialIsStencil=false
     fontMaterial=null
     fontObject=&Burbank
     letterSpacing=0
     monospacedWidth=1
     size=88
     skewAmount=0
     typefaceFontName=BurbankRegular-Black
     outlineSettings {
      bApplyOutlineToDropShadows=false
      bMiteredCorners=false
      bSeparateFillAlpha=false
      outlineMaterial=null
      outlineSize=0
      outlineColor=<rgba 0 0 0 1>
     }
    }
    margin=<ltrb 16 0 16 0>
    shadowOffset=<0 0>
   }
   slot {class=BorderSlot, properties {padding=<ltrb 0 0 0 0>}}
  }
 ]
}
