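# Translucent UI material that serves as a placeholder for a tray in the season pass reward track.
# It blends a dark backplate color toward a grid color parameter, weighted by the grid color's alpha and a mask.
# The Grid Mask reroute takes the smooth output of a radial gradient function.
# A battle pass tray mask function is also included.
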Material /SeasonPass/Materials/RewardTrack/M_RewardTrack_TrayPlaceholder
graph {
 properties {bAutomaticallySetUsageInEditor=false, blendMode=BLEND_Translucent, materialDomain=MD_UI, usedShadingModels=Default Lit}
 nodes [
  n1 = VectorParameter() {group=Pattern, parameterName=Grid Color, defaultValue=<rgba 0.1217 0 1 1>}
  LinearInterpolate(
   A=VectorParameter() {group=Backplate, parameterName=Backplate Color, defaultValue=<rgba 0.0055 0.0021 0.0521 1>}.RGB
   B=@n1.RGB
   Alpha=Multiply(A=NamedRerouteUsage(), B=@n1.A)
  )
  MaterialFunctionCall() {materialFunction=/SeasonPass/MaterialFunctions/MF_UI_BattlePassTrayMask}
  NamedRerouteDeclaration(None=MaterialFunctionCall() {materialFunction=&MF_UI_RadialGradient}.Smooth) {name=Grid Mask, nodeColor=<rgba 1 0 0.2578 1>}
 ]
}
