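# Season Pass tag widget that shows the bonus cross-product item in a multi-item reward, such as LEGO style, Sparks instrument or Rocket League.
# SetSeasonReward picks an item outside the pass product and BR, falls back to the second, and collapses for single-item rewards.
# Gameplay-tag switches drive UpdateItemIcon, which async-loads the preview or uses a rocket icon, and UpdateTexts, which sets product-specific labels and owned/includes state.
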WidgetBlueprint /SeasonPass/Widgets/BonusInfoMiniTagNexus
graphs [
 BuildSparksText {nodes [
   BuildSparksText!() -> (then, IsInstrument:bool=~, RewardName:Text) {is=title K2Node_FunctionEntry}
   > n1 = Branch(Condition=Get IsInstrument() -> (IsInstrument:bool))
   n2 = Get RewardName() -> (RewardName:Text)
   > Return Node!(
    execute
    SparksText:Text=Format Text!(Format="LOCTEXT(\"{0} INSTRUMENT, PICKAXE AND BACKBLING\")", 0:Text=TextToUpper!(InText:"Text (by ref)"=@n2) -> (ReturnValue:Text)) -> (Result)
   )
   n3 = Get RewardName()
   Return Node(execute=@n1.else, SparksText=Format Text(Format="LOCTEXT(\"{0} JAM TRACK\")", 0=TextToUpper(InText=@n3)))
  ]}
 EventGraph {nodes [
   n1 = RequestLoadIcon!() -> (OutputDelegate, then, SoftIconAsset:Texture 2D Soft Object Reference) {is=title K2Node_CustomEvent}
   > n2 = Load From Soft Texture!(execute:Exec, Texture:Texture 2D Soft Object Reference=@n1.SoftIconAsset) -> (then:Exec, Completed:Exec, Result:Texture 2D Object Reference) {is=title MVVMK2Node_LoadSoftTexture}
   n3 = Get MiniTagIcon() -> (MiniTagIcon:Mini Tag Icon Object Reference)
   Set LegoIcon!(execute=@n2.Completed, LegoIcon:Texture 2D Object Reference=@n2, self:Mini Tag Icon Object Reference=@n3) -> (then, Output_Get:Texture 2D Object Reference)
   > Refresh(self:Mini Tag Icon Object Reference=@n3)
   Event Construct()
   > SetMobileStyle() {is=title K2Node_CallFunction}
  ]}
 SetMobileStyle {nodes [
   SetMobileStyle() {is=title K2Node_FunctionEntry}
   > Branch(Condition=IsMobileGame())
   > SetPadding!(
    execute
    self:Overlay Slot Object Reference=SlotAsOverlaySlot()
    InPadding_Left:float=0
    InPadding_Top:float=24
    InPadding_Right:float=0
    InPadding_Bottom:float=0
   ) -> (then)
  ]}
 SetSeasonReward {nodes [
   n1 = SetSeasonReward!() -> (then, SeasonPassReward:Season Pass Reward VM Object Reference) {is=title K2Node_FunctionEntry}
   > n2 = Set SeasonPassRewardVM(SeasonPassRewardVM:Season Pass Reward VM Object Reference=@n1) -> (Output_Get:Season Pass Reward VM Object Reference)
   > n3 = Branch(Condition=IsValid(Object=@n2))
   > Set OtherProductItemVM(OtherProductItemVM:Fort Item VM Object Reference) -> (Output_Get:Fort Item VM Object Reference)
   n4 = Get ItemVMs!(self:Season Pass Reward VM Object Reference=Get SeasonPassRewardVM() -> (SeasonPassRewardVM:Season Pass Reward VM Object Reference)) -> (ItemVMs:Array of Fort Item VM Object References)
   n5 = "integer > integer"(A=Length(TargetArray:Array of Fort Item VM Object References=@n4), B=1)
   > SetVisibility(InVisibility=Select(Option 0=Collapsed, Option 1=Visible, Index=@n5))
   > Branch(Condition=@n5)
   > n6 = For Each Loop with Break!(Exec:Exec, Array:Array of Fort Item VM Object References=@n4, Break:Exec=@n12.then) -> (LoopBody:Exec, Array Element:Fort Item VM Object Reference, Array Index:int, Completed:Exec)
   Branch(execute=@n6.Completed, Condition="Equal (Object)"(A=Get OtherProductItemVM() -> (OtherProductItemVM:Fort Item VM Object Reference)))
   > n7 = Set OtherProductItemVM(OtherProductItemVM="Get (a copy)"(Array:Array of Fort Item VM Object References=@n4, Dimension 1=1) -> (Output:Fort Item VM Object Reference))
   > UpdateItemIcon!(execute, self, ItemVM:Fort Item VM Object Reference=@n7) -> (then)
   > UpdateTexts()
   n8 = GetCurrentProduct!(self:Fort Item VM Object Reference=@n6) -> (ReturnValue:Gameplay Tag Structure)
   n9 = Get CurrentProductTag() -> (CurrentProductTag:Gameplay Tag Structure)
   n10 = Branch(
    execute=@n6.LoopBody
    Condition=OR Boolean(
     A=MatchesTag!(TagOne:Gameplay Tag Structure=@n8, TagTwo:Gameplay Tag Structure=@n9, bExactMatch:bool=false) -> (ReturnValue:bool)
     B=MatchesTag(TagOne=@n8, TagTwo=(TagName="Product.BR"), bExactMatch=false)
    )
   )
   n11 = Set OtherProductItemVM(execute=@n10.else, OtherProductItemVM=@n6)
   > UpdateItemIcon(ItemVM=@n11)
   > n12 = UpdateTexts()
   SetVisibility(execute=@n3.else, InVisibility=Collapsed)
  ]}
 SetSelectedSeasonPass {nodes [
   n1 = SetSelectedSeasonPass!() -> (then, SeasonPass:Season Pass VM Object Reference) {is=title K2Node_FunctionEntry}
   > Set CurrentProductTag(CurrentProductTag:Gameplay Tag Structure=GetPassProduct!(self:Season Pass VM Object Reference=@n1) -> (ReturnValue:"Gameplay Tag Structure (by ref)")) -> (Output_Get:Gameplay Tag Structure)
   > SetSeasonReward!(execute, self, SeasonPassReward:Season Pass Reward VM Object Reference) -> (then) {is=title K2Node_CallFunction}
  ]}
 UpdateItemIcon {nodes [
   n1 = UpdateItemIcon!() -> (then, ItemVM:Fort Item VM Object Reference) {is=title K2Node_FunctionEntry}
   > SetVisibility(self=Get MiniTagIcon(), InVisibility=HitTestInvisible)
   > n2 = Switch on Gameplay Tag!(execute:Exec, Selection:Gameplay Tag Structure=GetCurrentProduct(self=@n1)) -> (Default:Exec, Product.BR:Exec, Product.Juno:Exec, Product.DelMar:Exec, Product.Sparks:Exec) {
    is=title GameplayTagsK2Node_SwitchGameplayTag
    pinTags [{tagName=Product.BR}, {tagName=Product.Juno}, {tagName=Product.DelMar}, {tagName=Product.Sparks}]
   }
   > SetVisibility(self=Get MiniTagIcon(), InVisibility=Collapsed)
   n3 = Get ItemVM() -> (ItemVM:Fort Item VM Object Reference)
   RequestLoadIcon!(
    execute=[@n2.Product.BR, @n2.Product.Juno, @n2.Product.Sparks]
    self
    SoftIconAsset:Texture 2D Soft Object Reference=Get SmallPreviewImage!(self:Fort Item Details VM Object Reference=Get ItemDetails!(self:Fort Item VM Object Reference=@n3) -> (ItemDetails:Fort Item Details VM Object Reference)) -> (SmallPreviewImage:Texture 2D Soft Object Reference)
   ) -> (then) {is=title K2Node_CallFunction}
   n4 = Get MiniTagIcon()
   Set LegoIcon(execute=@n2.Product.DelMar, LegoIcon=/Game/UI/Frontend/Misc/BonusInfoMiniTag/Textures/T_UI_S28_RocketIcon, self=@n4)
   > Refresh(self=@n4)
  ]}
 UpdateTexts {nodes [
   UpdateTexts() {is=title K2Node_FunctionEntry}
   > n1 = Switch on Gameplay Tag!(execute, Selection=GetCurrentProduct(self=Get OtherProductItemVM())) -> (Default, Product.DelMar) {pinTags [{tagName=Product.DelMar}]}
   > SetVisibility(self=Get ItemName() -> (ItemName:Common Text Object Reference), InVisibility=HitTestInvisible)
   n2 = Get AlreadyOwned() -> (AlreadyOwned:Common Text Object Reference)
   > Set Text(Text="LOCTEXT(\"ALREADY OWNED\")", self=@n2)
   n3 = Get Includes() -> (Includes:Common Text Object Reference)
   > n4 = Set Text(Text="LOCTEXT(\"INCLUDES:\")", self=@n3)
   n5 = Get bIsRewardOwned!(self:Season Pass Reward VM Object Reference=Get SeasonPassRewardVM()) -> (bIsRewardOwned:bool)
   SetVisibility(execute=[@n4.then, @n15.then], self=Get AlreadyOwned(), InVisibility=Select(Option 0=Collapsed, Option 1=HitTestInvisible, Index=@n5))
   > SetVisibility(self=Get Includes(), InVisibility=Select(Option 0=SelfHitTestInvisible, Option 1=Collapsed, Index=@n5))
   n6 = Get OtherProductItemVM()
   > Set RewardName(RewardName:Text=Get ItemName!(self:Item Definition Base Object Reference=Get ItemDefinition!(self:Fort Item VM Object Reference=@n6) -> (ItemDefinition:Fort Item Definition Object Reference)) -> (ItemName:Text)) -> (Output_Get:Text)
   > SetVisibility(self=Get ItemName!() -> (ItemName:Common Text Object Reference), InVisibility=SelfHitTestInvisible)
   > n7 = Switch on Gameplay Tag!(execute, Selection=GetCurrentProduct(self=Get OtherProductItemVM())) -> (Default, Product.BR:Exec, Product.Juno:Exec, Product.DelMar, Product.Sparks:Exec) {pinTags [{tagName=Product.BR}, {tagName=Product.Juno}, {tagName=Product.DelMar}, {tagName=Product.Sparks}]}
   SetVisibility(execute=@n7.Product.DelMar, self=Get ItemName(), InVisibility=Collapsed)
   n8 = Get RewardName()
   Set Text(execute=@n7.Product.Juno, Text=Format Text(Format="LOCTEXT(\"{0} LEGO® STYLE\")", 0=TextToUpper(InText=@n8)), self=Get ItemName())
   n9 = Sequence(execute=@n7.Product.Sparks)
   n10 = Get OtherProductItemVM()
   > n11 = Switch on EFortItemType!(
    execute
    Selection:EFortItemType Enum=GetItemType!(ItemDefinition:Item Definition Base Object Reference=Get ItemDefinition(self=@n10)) -> (ReturnValue:EFortItemType Enum=WorldItem)
   ) -> (
    WorldItem:Exec
    Ammo:Exec
    Badge:Exec
    BackpackPickup:Exec
    BuildingPiece:Exec
    CharacterPart:Exec
    Consumable:Exec
    Deco:Exec
    EditTool:Exec
    Ingredient:Exec
    ItemCache:Exec
    Food:Exec
    Gadget:Exec
    AthenaGadget:Exec
    HomebaseGadget:Exec
    SpyTechPerk:Exec
    HeroAbility:Exec
    MissionItem:Exec
    Trap:Exec
    MultiItem:Exec
    Weapon:Exec
    WeaponMelee:Exec
    WeaponRanged:Exec
    WeaponHarvest:Exec
    WeaponCreativePhone:Exec
    WeaponMod:Exec
    WorldResource:Exec
    CreativeUserPrefab:Exec
    CreativePlayset:Exec
    Vehicle:Exec
    Npc:Exec
    PlayerAugment:Exec
    AccountItem:Exec
    AccountResource:Exec
    CollectedResource:Exec
    Alteration:Exec
    CardPack:Exec
    Currency:Exec
    Hero:Exec
    Schematic:Exec
    Worker:Exec
    TeamPerk:Exec
    PlayerTech:Exec
    Token:Exec
    DailyRewardScheduleToken:Exec
    CodeToken:Exec
    Stat:Exec
    Buff:Exec
    BuffCredit:Exec
    Quest:Exec
    Accolades:Exec
    FriendChest:Exec
    MedalsPunchCard:Exec
    RepeatableDailiesCard:Exec
    ChallengeBundle:Exec
    ChallengeBundleSchedule:Exec
    ChallengeBundleCompletionToken:Exec
    GameplayModifier:Exec
    Outpost:Exec
    HomebaseNode:Exec
    Defender:Exec
    ConversionControl:Exec
    DeployableBaseCloudSave:Exec
    ConsumableAccountItem:Exec
    Quota:Exec
    Expedition:Exec
    ExpeditionV2:Exec
    HomebaseBannerIcon:Exec
    HomebaseBannerColor:Exec
    AthenaSkyDiveContrail:Exec
    PersonalVehicle:Exec
    AthenaGlider:Exec
    AthenaPickaxe:Exec
    AthenaBackpack:Exec
    CosmeticShoes:Exec
    CosmeticMimosa:Exec
    CosmeticMimosaA:Exec
    CosmeticMimosaB:Exec
    CosmeticMimosaC:Exec
    AthenaCharacter:Exec
    AthenaDance:Exec
    AthenaLoadingScreen:Exec
    AthenaBattleBus:Exec
    AthenaVehicleCosmetic:Exec
    AthenaItemWrap:Exec
    AthenaMusicPack:Exec
    AthenaPetCosmetic:Exec
    AthenaSeasonTreasure:Exec
    AthenaSeason:Exec
    AthenaRewardGraph:Exec
    AthenaExtResource:Exec
    EventDescription:Exec
    AthenaEventToken:Exec
    EventPurchaseTracker:Exec
    CosmeticVariantToken:Exec
    CampaignHeroLoadout:Exec
    Playset:Exec
    PrerollData:Exec
    CreativePlot:Exec
    PlayerSurveyToken:Exec
    CosmeticLocker:Exec
    BannerToken:Exec
    RestedXpBoosterToken:Exec
    RewardEventGraphPurchaseToken:Exec
    EventDependentItem:Exec
    ItemAccessToken:Exec
    STWAccoladeReward:Exec
    Campsite:Exec
    VictoryCrown:Exec
    RealitySapling:Exec
    Apparel:Exec
    ApparelLayout:Exec
    PlayerAugmentsPersistence:Exec
    SparksAura:Exec
    SparksGuitar:Exec
    SparksBass:Exec
    SparksKeyboard:Exec
    SparksMicrophone:Exec
    SparksDrums:Exec
    SparksSpotlightAnim:Exec
    SparksSong:Exec
    JunoBuildingSet:Exec
    JunoBuildingProp:Exec
    MagpieEntitlementReward:Exec
    DeferredItemGrant:Exec
    XpBooster:Exec
    SpecialItem:Exec
    Emote:Exec
    Stack:Exec
    CollectionBookPage:Exec
    BGAConsumableWrapper:Exec
    GiftBox:Exec
    GiftBoxUnlock:Exec
    PlaysetProp:Exec
    RegCosmeticDef:Exec
    Profile:Exec
    Max_None:Exec
   ) {is=title K2Node_SwitchEnum}
   Set SparksInstrument(execute=[@n11.SparksBass, @n11.SparksDrums, @n11.SparksGuitar, @n11.SparksKeyboard, @n11.SparksMicrophone], SparksInstrument:bool=true) -> (Output_Get:bool)
   Set Text(
    execute=@n9.then_1
    Text=BuildSparksText!(IsInstrument:bool=Get SparksInstrument() -> (SparksInstrument:bool), RewardName:Text=Get RewardName()) -> (SparksText:Text) {is=title K2Node_CallFunction}
    self=Get ItemName()
   )
   n12 = Get RewardName()
   n13 = Get OtherProductItemVM()
   n14 = "Equal (Enum)"(A:EFortItemType Enum=GetItemType(ItemDefinition=Get ItemDefinition(self=@n13)) -> (ReturnValue=WorldItem), B:EFortItemType Enum=SparksSong)
   Set Text(
    execute=@n7.Product.BR
    Text=Format Text(
     Format=Select(Option 0:Text="LOCTEXT(\"{0} FORTNITE STYLE\")", Option 1:Text="LOCTEXT(\"{0} LOBBY MUSIC\")", Index=@n14) -> (ReturnValue:Text)
     0=TextToUpper(InText=@n12)
    )
    self=Get ItemName()
   )
   SetVisibility(execute=@n1.Product.DelMar, self=Get ItemName(), InVisibility=Collapsed)
   > Set Text(Text="LOCTEXT(\"Item now available in Rocket League\")", self=@n2)
   > n15 = Set Text(Text="LOCTEXT(\"Unlocking this item will also make it available in Rocket League\")", self=@n3)
  ]}
]
variables [
 CurrentProductTag {type=GameplayTag, default {tagName=None}}
 Icon2D {type=Texture2D}
 TextColor {type=LinearColor, default=<rgba 1 1 1 1>}
]
widgets {
 info {inheritedWidgetCount=0, parentClass=&CommonActivatableWidget, rootWidgetClass=&Overlay, widgetCount=7}
 widgets [
  Overlay {widgetClassPath=&Overlay, bIsVariable=true, widget=Overlay}
  TagContainer {
   widgetClassPath=&Overlay
   parent=Overlay
   widget=TagContainer
   slot {class=OverlaySlot, properties {horizontalAlignment=HAlign_Fill, padding=<ltrb 0 24 0 0>}}
  }
  MiniTagIcon {
   widgetClassPath=/Game/UI/Frontend/Misc/BonusInfoMiniTag/MiniTagIcon
   bIsVariable=true
   parent=TagContainer
   widget=MiniTagIcon
   properties {iconHeight=0, iconScale=1, useBackgroundGradient=true}
   slot {class=OverlaySlot}
  }
  VB_Texts {
   widgetClassPath=&VerticalBox
   parent=TagContainer
   widget=VB_Texts
   slot {class=OverlaySlot, properties {horizontalAlignment=HAlign_Fill, verticalAlignment=VAlign_Center, padding=<ltrb 90 0 0 0>}}
  }
  Includes {
   widgetClassPath=&CommonTextBlock
   bIsVariable=true
   parent=VB_Texts
   widget=Includes
   properties {
    bAutoCollapseWithEmptyText=true
    text=INCLUDES:
    wrapTextAt=550
    colorAndOpacity {colorUseRule=UseColor_Specified, specifiedColor=<rgba 1 1 1 0.6>}
    font {
     bForceMonospaced=false
     bMaterialIsStencil=false
     fontMaterial=null
     fontObject=&HeadingNow
     letterSpacing=0
     monospacedWidth=1
     size=13.5
     skewAmount=0
     typefaceFontName=76 Bold Italic
     outlineSettings {
      bApplyOutlineToDropShadows=false
      bMiteredCorners=false
      bSeparateFillAlpha=false
      outlineMaterial=null
      outlineSize=0
      outlineColor=<rgba 0 0 0 1>
     }
    }
    renderTransform {angle=0, scale=<1 1>, shear=<-5 0>, translation=<0 0>}
    shadowOffset=<0 0>
   }
   slot {class=VerticalBoxSlot}
  }
  ItemName {
   widgetClassPath=&CommonTextBlock
   bIsVariable=true
   parent=VB_Texts
   widget=ItemName
   properties {
    bAutoCollapseWithEmptyText=true
    bIsScrollingEnabled=false
    scrollStyle=&TextScrollStyle-Base_C
    text=" LEGO® STYLE "
    wrapTextAt=350
    font {
     bForceMonospaced=false
     bMaterialIsStencil=false
     fontMaterial=null
     fontObject=&HeadingNow
     letterSpacing=0
     monospacedWidth=1
     size=13.5
     skewAmount=0
     typefaceFontName=76 Bold Italic
     outlineSettings {
      bApplyOutlineToDropShadows=false
      bMiteredCorners=false
      bSeparateFillAlpha=false
      outlineMaterial=null
      outlineSize=0
      outlineColor=<rgba 0 0 0 1>
     }
    }
    renderTransform {angle=0, scale=<1 1>, shear=<-5 0>, translation=<0 0>}
    shadowOffset=<0 0>
   }
   slot {class=VerticalBoxSlot}
  }
  AlreadyOwned {
   widgetClassPath=&CommonTextBlock
   bIsVariable=true
   parent=VB_Texts
   widget=AlreadyOwned
   properties {
    bAutoCollapseWithEmptyText=true
    text=ALREADY OWNED
    wrapTextAt=550
    font {
     bForceMonospaced=false
     bMaterialIsStencil=false
     fontMaterial=null
     fontObject=&HeadingNow
     letterSpacing=0
     monospacedWidth=1
     size=13.5
     skewAmount=0
     typefaceFontName=76 Bold Italic
     outlineSettings {
      bApplyOutlineToDropShadows=false
      bMiteredCorners=false
      bSeparateFillAlpha=false
      outlineMaterial=null
      outlineSize=0
      outlineColor=<rgba 0 0 0 1>
     }
    }
    renderTransform {angle=0, scale=<1 1>, shear=<-5 0>, translation=<0 0>}
    shadowOffset=<0 0>
   }
   slot {class=VerticalBoxSlot}
  }
 ]
}
