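# Season pass reward track tile, an item card button that shows one reward and lets the player select and claim it.
# Focus selects the reward in the selection viewmodel, and a click purchases it when hold is required and the reward is selected.
# Hold progress goes to that viewmodel, with helpers mapping it, including claim-all membership, onto dynamic material parameters.
# Conversion functions derive visibility, while category theme data supplies colors and game mode branding.
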WidgetBlueprint /SeasonPass/Widgets/Reboot/WBP_RewardTrack_ItemButton
animations [
 Focus {
  binding_count=1
  length_seconds=0.1
  object=&self:Focus
  sequence {bindings [SizeBox {
     id=#1
     possessed_class=SizeBox
     tracks [{
       type=MovieSceneFloatTrack
       display_name=Height Override
       sections [{type=MovieSceneFloatSection, end_frame=2, start_frame=0, channels [None {type=MovieSceneScriptingFloatChannel, keys [[0,160],[2,180]]}]}]
      }]
    }]}
 }
 ClaimRewardFlourish {
  binding_count=6
  length_seconds=1.5
  object=&self:ClaimRewardFlourish
  sequence {bindings [
    LazyImage_Item {
     id=#2
     possessed_class=FortLazyImage
     tracks [{
       type=MovieSceneWidgetMaterialTrack
       sections [{
         type=MovieSceneParameterSection
         end_frame=30
         start_frame=0
         channels [ColorWash {type=MovieSceneScriptingFloatChannel, keys [[1,0],[3,0.8],[12,0]]}]
        }]
      }]
    }
    CommonVisualAttachment_0 {
     id=#3
     possessed_class=CommonVisualAttachment
     tracks [{
       type=MovieSceneByteTrack
       display_name=Visibility
       sections [{
         type=MovieSceneByteSection
         end_frame=30
         start_frame=0
         channels [None {type=MovieSceneScriptingByteChannel, keys [[0,1],[1,4],[29,4],[30,1]]}]
        }]
      }]
    }
    Background {
     id=#4
     possessed_class=Image
     tracks [{
       type=MovieSceneWidgetMaterialTrack
       sections [{
         type=MovieSceneParameterSection
         end_frame=30
         start_frame=0
         channels [ColorWash {type=MovieSceneScriptingFloatChannel, keys [[0,0],[3,0.2],[10,0]]}]
        }]
      }]
    }
    Img_RewardExplosion {
     id=#5
     possessed_class=Image
     tracks [
      {
       type=MovieSceneWidgetMaterialTrack
       sections [{
         type=MovieSceneParameterSection
         end_frame=30
         start_frame=0
         channels [
          MaskRadius {type=MovieSceneScriptingFloatChannel, keys [[0,0],[13,0.5],[29.9993,0]]}
          TriScale {type=MovieSceneScriptingFloatChannel, keys [[0,0],[5,1],[29.9997,1],[30,1.7]]}
          MaskDesity {type=MovieSceneScriptingFloatChannel, const [1,1]}
         ]
        }]
      }
      {
       type=MovieScene2DTransformTrack
       sections [{
         type=MovieScene2DTransformSection
         end_frame=30
         start_frame=0
         channels [
          Translation.X {type=MovieSceneScriptingFloatChannel, const [0,2]}
          Translation.Y {type=MovieSceneScriptingFloatChannel, keys [[0,70],[5,-86.2701],[29.9993,92.3108],[30,7.1553]]}
          Angle {type=MovieSceneScriptingFloatChannel, const [0,1]}
          Scale.X {type=MovieSceneScriptingFloatChannel, keys [[0,0],[11,1],[29.9993,1.3]]}
          Scale.Y {type=MovieSceneScriptingFloatChannel, keys [[0,0],[11,1],[29.9993,1.3]]}
          Shear.X {type=MovieSceneScriptingFloatChannel, const [0,1]}
          Shear.Y {type=MovieSceneScriptingFloatChannel, const [0,1]}
         ]
        }]
      }
     ]
    }
    WBP_RewardTrack_ItemButton_C_135 {
     id=#6
     possessed_class=WBP_RewardTrack_ItemButton_C
     tracks [{
       type=MovieScene2DTransformTrack
       sections [{
         type=MovieScene2DTransformSection
         end_frame=30
         start_frame=0
         channels [
          Translation.X {type=MovieSceneScriptingFloatChannel, const [0,2]}
          Translation.Y {type=MovieSceneScriptingFloatChannel, keys [[0,0],[3,0],[5,-6.1728],[7,10.3563],[9,-4.3947],[11,0],[30,0]]}
          Angle {type=MovieSceneScriptingFloatChannel, const [0,1]}
          Scale.X {type=MovieSceneScriptingFloatChannel, const [1,1]}
          Scale.Y {type=MovieSceneScriptingFloatChannel, const [1,1]}
          Shear.X {type=MovieSceneScriptingFloatChannel, const [0,1]}
          Shear.Y {type=MovieSceneScriptingFloatChannel, const [0,1]}
         ]
        }]
      }]
    }
    Img_RewardEffect {
     id=#7
     possessed_class=Image
     tracks [{
       type=MovieSceneFloatTrack
       display_name=Render Opacity
       sections [{type=MovieSceneFloatSection, end_frame=30, start_frame=0, channels [None {type=MovieSceneScriptingFloatChannel, keys [[0,0],[8,1],[18,0]]}]}]
      }]
    }
   ]}
 }
]
graphs [
 CF_GetCustomizableCharacterIconVisibility {nodes [
   n1 = CF_GetCustomizableCharacterIconVisibility!() -> (then, SeasonPassRewardVM:Season Pass Element Base VM Object Reference) {is=title K2Node_FunctionEntry}
   n2 = GetFirstExtensionVMByClass!(
    self:Season Pass Element Base VM Object Reference=@n1
    ExtensionClass:Season Pass Element Base VM Class Reference=/Script/SeasonPass.SeasonPassCustomizableItemRewardVM
   ) -> (ReturnValue:Season Pass Element Base VM Object Reference)
   > Return Node!(
    execute
    Visibility:ESlateVisibility Enum=Select(
     Option 0=Collapsed
     Option 1=SelfHitTestInvisible
     Index=Cast To SeasonPassCustomizableItemRewardVM!(Object=@n2) -> (AsSeason Pass Customizable Item Reward VM:Season Pass Customizable Item Reward VM Object Reference, bSuccess:bool=~) {is=title K2Node_DynamicCast}.bSuccess
    )
   )
  ]}
 CF_GetIsAthenaCharacterOrSidekick {nodes [
   n1 = CF_GetIsAthenaCharacterOrSidekick!() -> (then, Reward:Season Pass Reward VM Object Reference) {is=title K2Node_FunctionEntry}
   > n2 = Is Valid(InputObject=@n1)
   Return Node!(execute=@n2.Is Not Valid, IsCharacterOrSidekick:bool=false)
   n3 = GetFirstExtensionVMByClass(self=@n1, ExtensionClass=&SeasonPassCustomizableItemRewardVM)
   n4 = GetItemType!(ItemDefinition:Item Definition Base Object Reference=Get ItemDefinition!(self:Fort Item VM Object Reference=Get Item VM!(self, Reward:Season Pass Reward VM Object Reference=@n1) -> (Result:Fort Item VM Object Reference) {is=title K2Node_CallFunction}) -> (ItemDefinition:Fort Item Definition Object Reference)) -> (ReturnValue:EFortItemType Enum=WorldItem)
   > Return Node(IsCharacterOrSidekick=OR Boolean!(
     A="Equal (Enum)"(A:EFortItemType Enum=@n4, B:EFortItemType Enum=AthenaCharacter)
     B="Equal (Enum)"(A=@n4, B=CosmeticMimosa)
     C:bool=Cast To SeasonPassCustomizableItemRewardVM(Object=@n3) -> (bSuccess=~).bSuccess
    ) -> (ReturnValue))
  ]}
 CF_GetIsInClaimAllArray {nodes [
   n1 = CF_GetIsInClaimAllArray!() -> (then, IsClaimAll:bool=~, SelectedPage:Season Pass Element Base VM Object Reference) {is=title K2Node_FunctionEntry}
   > n2 = Is Valid(InputObject=@n1.SelectedPage)
   Return Node!(execute=@n2.Is Not Valid, IsInClaimAllArray:bool=false)
   > n3 = Set PageRewards(PageRewards:Array of Season Pass Reward VM Object References=GetPurchaseableRewards!(self:Season Pass Element Base VM Object Reference=@n1.SelectedPage) -> (ReturnValue:Array of Season Pass Reward VM Object References)) -> (Output_Get:Array of Season Pass Reward VM Object References)
   > n4 = For Each Loop with Break!(Exec:Exec, Array:Array of Season Pass Reward VM Object References=@n3, Break:Exec) -> (LoopBody:Exec, Array Element:Season Pass Reward VM Object Reference, Array Index:int, Completed:Exec)
   Return Node(execute=@n4.Completed, IsInClaimAllArray=false)
   > Branch(Condition="Equal (Object)"(A=@n4, B=Get SeasonPassRewardVM() -> (SeasonPassRewardVM:Season Pass Reward VM Object Reference)))
   > Return Node(IsInClaimAllArray=true)
  ]}
 CF_GetOwnedStatusIndicatorVisibility {nodes [
   n1 = CF_GetOwnedStatusIndicatorVisibility!() -> (then, IsRewardOwned:bool=~, GameplayTags:Gameplay Tag Container Structure) {is=title K2Node_FunctionEntry}
   n2 = NOT Boolean(A=HasTag!(
     TagContainer:"Gameplay Tag Container Structure (by ref)"=@n1.GameplayTags
     Tag:Gameplay Tag Structure=(TagName="UI.SeasonPass.Presentation.Hide.Claimed")
     bExactMatch:bool=false
    ) -> (ReturnValue:bool))
   > Return Node!(execute, visibility:ESlateVisibility Enum=Select(Option 0=Collapsed, Option 1=Visible, Index=AND Boolean(A=@n1, B=@n2)))
  ]}
 CF_GetProgressBarVisibility {nodes [
   n1 = CF_GetProgressBarVisibility!() -> (then, DoesRequireProgress:bool=~, IsQuestDriven:bool=~, IsSetLocked:bool=~) {is=title K2Node_FunctionEntry}
   n2 = NOT Boolean(A=@n1.IsQuestDriven)
   n3 = NOT Boolean(A=@n1.IsSetLocked)
   > Return Node!(
    execute
    Visibility:ESlateVisibility Enum=Select(Option 0=Collapsed, Option 1=SelfHitTestInvisible, Index=AND Boolean!(A=@n1, B=@n2, C:bool=@n3) -> (ReturnValue))
   )
  ]}
 CF_GetThemeData {nodes [
   n1 = CF_GetThemeData!() -> (then, SeasonPassRewardVM:Season Pass Reward VM Object Reference) {is=title K2Node_FunctionEntry}
   > Return Node!(
    execute
    ThemeData:Season Pass Category Theme Data Structure=GetThemeData!(self:Season Pass Category VM Object Reference=GetOwnerCategory!(self:Season Pass Reward VM Object Reference=@n1) -> (ReturnValue:Season Pass Category VM Object Reference)) -> (ReturnValue:"Season Pass Category Theme Data Structure (by ref)")
   )
  ]}
 EventGraph {nodes [
   HandleInputMethodChanged!() -> (OutputDelegate, then, NewInputMethod:ECommonInputType Enum) {is=title K2Node_CustomEvent}
   > n1 = Get!(execute:Exec) -> (then:Exec, else:Exec, SeasonPassRewardVM:Season Pass Reward VM Object Reference)
   > n2 = Get!(execute) -> (then, else, SeasonPassSelectionVM:Season Pass Selection VM Object Reference)
   > Get!(execute, self:Season Pass Selection VM Object Reference=@n2) -> (then, else, SelectedPage:Season Pass Page VM Object Reference)
   n3 = Get SeasonPassVM() -> (SeasonPassVM:Season Pass VM Object Reference)
   n4 = Get SelectedPage!(self:Season Pass Selection VM Object Reference=Get SeasonPassSelectionVM() -> (SeasonPassSelectionVM:Season Pass Selection VM Object Reference)) -> (SelectedPage:Season Pass Page VM Object Reference)
   > SetRequiresHold!(
    execute
    self:Common Button Base Object Reference
    bInRequiresHold:bool=GetCanBeHeld!(
     self
     IsRewardLocked:bool=Get bIsLocked!(self:Season Pass Element Base VM Object Reference=@n1) -> (bIsLocked:bool)
     IsRewardOwned:bool=Get bIsRewardOwned!(self:Season Pass Reward VM Object Reference=@n1) -> (bIsRewardOwned:bool)
     IsPageLocked:bool=Get bIsLocked(self=@n4)
     IsQuestDriven:bool=Get bIsQuestDriven!(self:Season Pass Reward VM Object Reference=@n1) -> (bIsQuestDriven:bool)
     PurchasableRewards:int=GetNumPurchaseableRewards!(self:Season Pass Element Base VM Object Reference=@n3) -> (ReturnValue:int=0)
    ) -> (CanBeHeld:bool)
   ) -> (then)
   Event On Focused() {is=title K2Node_Event}
   > Set SelectedReward!(
    execute
    SelectedReward:Season Pass Reward VM Object Reference=Get SeasonPassRewardVM()
    self:Season Pass Selection VM Object Reference=Get SeasonPassSelectionVM()
   ) -> (then, Output_Get:Season Pass Reward VM Object Reference)
   > Fire Event Element Interacted(
    Element=Self-Reference()
    AdditionalInfo=Get Analytics Info Item Name!(
     MakeFlowAnalyticsAdditionalInfo:"Flow Analytics Additional Info Structure (by ref)"=Make Flow Analytics Additional Info!() -> (FlowAnalyticsAdditionalInfo:Flow Analytics Additional Info Structure) {is=title K2Node_MakeStruct}
     Key:String=SelectReward
     FortItemVM:Fort Item VM Object Reference=Get CurrentItemVM() -> (CurrentItemVM:Fort Item VM Object Reference)
    ) -> (ItemName:Flow Analytics Additional Info Structure) {is=title K2Node_MacroInstance}
   )
   Event On Clicked()
   > Set OfferPurchaseHoldProgress!(execute, OfferPurchaseHoldProgress:float=0, self:Season Pass Selection VM Object Reference=Get SeasonPassSelectionVM()) -> (then, Output_Get:float)
   > Fire Event Element Interacted(
    Element=Self-Reference()
    AdditionalInfo=Get Analytics Info Item Name(MakeFlowAnalyticsAdditionalInfo=Make Flow Analytics Additional Info(), Key=ClaimReward, FortItemVM=Get CurrentItemVM())
   )
   > Parent: On Clicked() {is=title K2Node_CallParentFunction}
   > n5 = Branch(Condition=Get IsModeSpecificReward() -> (IsModeSpecificReward:bool))
   Branch(execute=[@n6.then_0, @n5.then], Condition=IsInputType!(InputType:ECommonInputType Enum=Touch, IsNot:bool=false) -> (Result:bool))
   > Set SelectedReward(SelectedReward=Get SeasonPassRewardVM(), self=Get SeasonPassSelectionVM())
   n6 = Sequence(execute=@n5.else)
   n7 = Get SeasonPassRewardVM()
   n8 = Get SelectedReward!(self:Season Pass Selection VM Object Reference=Get SeasonPassSelectionVM()) -> (SelectedReward:Season Pass Reward VM Object Reference)
   Branch(
    execute=@n6.then_1
    Condition=AND Boolean!(A=GetRequiresHold!(self:Common Button Base Object Reference) -> (ReturnValue:bool), B="Equal (Object)"(A=@n8, B=@n7)) -> (ReturnValue)
   )
   > n9 = PurchaseItem!(
    execute:Exec
    Item:Season Pass Reward VM Object Reference=Get SeasonPassRewardVM()
    SeasonPass:Season Pass VM Object Reference=Get SeasonPassVM()
    bTryPurchasePrerequisites:bool=false
   ) -> (then:Exec, OnOperationCompleted:Exec, Success:bool=~) {is=title K2Node_AsyncAction}
   > SetIsEnabled(bInIsEnabled=false)
   SetIsEnabled(execute=@n9.OnOperationCompleted, bInIsEnabled=true)
   Event Construct()
   n10 = Get CommonInputSubsystem()
   > HandleInputMethodChanged!(execute, self, NewInputMethod:ECommonInputType Enum=GetCurrentInputType(self=@n10)) -> (then) {is=title K2Node_CallFunction}
   n11 = Create Event()
   > Bind Event to On Input Method Changed(self=@n10, Delegate=@n11)
   > SetFortBrushSize!(
    execute
    self:Fort Lazy Image Object Reference=Get LazyImage_Item() -> (LazyImage_Item:Fort Lazy Image Object Reference)
    FortBrushSize:EFortBrushSize Enum=Large
   ) -> (then)
   n12 = Event PreConstruct()
   > Parent: PreConstruct(IsDesignTime=@n12.IsDesignTime)
   > SetIsItemCustomizable!(execute, self:WBP UIKit Item Card Base Object Reference, IsCustomizable:bool=true) -> (then)
   > SetIsItemCustomizableCharacter!(execute, self:WBP UIKit Item Card Base Object Reference, IsCustomizableCharacter:bool=true) -> (then)
   > SetAsyncBackground!(
    execute
    self:WBP UIKit Item Card Base Object Reference
    MaterialInterfaceValue:Material Interface Soft Object Reference=/SeasonPass/Widgets/Reboot/Materials/M_UI_BPReboot_RewardTrack_TileBackground
   ) -> (then)
   > Branch(Condition=@n12.IsDesignTime)
   > UpdateProgressBarColors!(execute, self, ShouldUseExpMaterial:bool=Get Preview_UseAlterProgressBar() -> (Preview_UseAlterProgressBar:bool)) -> (then)
   > SetItemIsEquipped!(execute, self:WBP UIKit Item Card Base Object Reference, Is Equipped:bool=false) -> (then)
   Event On Unhovered()
   n13 = Get SelectedReward(self=Get SeasonPassSelectionVM())
   n14 = Get SeasonPassRewardVM()
   > Branch(Condition=AND Boolean(
     A=Get IsButtonFocused() -> (IsButtonFocused:bool)
     B="Not Equal (Object)"(A:Object Reference=@n13, B:Object Reference=@n14) -> (ReturnValue:bool=~) {is=title K2Node_PromotableOperator}
    ))
   > Parent: On Unhovered() {is=title K2Node_CallParentFunction}
   Event Destruct()
   > Parent: Destruct()
   > Unbind Event from On Input Method Changed(self=@n10, Delegate=@n11)
   n15 = Event OnActionProgress!() -> (OutputDelegate, then, HeldPercent:float) {is=title K2Node_Event}
   > Parent: OnActionProgress!(execute, HeldPercent:float=@n15.HeldPercent) -> (then) {is=title K2Node_CallParentFunction}
   n16 = Get SeasonPassSelectionVM()
   > Branch(Condition="Equal (Object)"(A=Get SelectedReward(self=@n16), B=Get SeasonPassRewardVM()))
   > Set OfferPurchaseHoldProgress(OfferPurchaseHoldProgress=@n15.HeldPercent, self=Get SeasonPassSelectionVM())
   Event OnInitialized()
   > Parent: OnInitialized() {is=title K2Node_CallParentFunction}
   > n17 = GetDynamicMaterial(self=Get Img_RewardEffect() -> (Img_RewardEffect:Image Object Reference))
   > Set DynamicRewardMaterial(DynamicRewardMaterial:Material Instance Dynamic Object Reference=@n17) -> (Output_Get:Material Instance Dynamic Object Reference)
   > SetItemMaterial!(
    execute
    self:WBP UIKit Item Card Base Object Reference
    Material Instance:"Material Instance Object Reference (by ref)"=SelectMaterialInstance!(
     A:Material Instance Object Reference=/SeasonPass/Materials/RewardTrack/MI_UI_Nexus_RewardTrack_ItemPreview_Quest
     B:Material Instance Object Reference=/SeasonPass/Materials/RewardTrack/MI_UI_Nexus_RewardTrack_ItemPreview_Base
     bSelectA:bool=Get IsModeSpecificReward()
    ) -> (ReturnValue:Material Instance Object Reference)
   ) -> (then)
   Reroute Node()
   > n18 = CreateWidgetAsync!(
    execute:Exec
    UserWidgetSoftClass:User Widget Soft Class Reference=GetActivatableWidgetForAlias!(WidgetAlias:Gameplay Tag Structure=(TagName="UI.SeasonPass.QuestInfoModal")) -> (ReturnValue:Common Activatable Widget Soft Class Reference)
    OwningPlayer:Player Controller Object Reference
    bSuspendInputUntilComplete:bool=true
   ) -> (then:Exec, OnComplete:Exec, OnFailure:Exec, UserWidget:User Widget Object Reference) {is=title K2Node_AsyncAction}
   SetViewModelByClass!(execute=@n18.OnComplete, Widget:User Widget Object Reference=@n18, ViewModel:Notify Field Value Changed Interface=Get SeasonPassRewardVM()) -> (then)
   > TryPushWidgetAsModalPanel!(execute, PushWidget:User Widget Object Reference=@n18) -> (then)
  ]}
 Get Item VM {nodes [
   n1 = Get Item VM!() -> (then, Reward:Season Pass Reward VM Object Reference) {is=title K2Node_FunctionEntry}
   > Is Valid(InputObject=@n1)
   > Return Node!(
    execute
    Result:Fort Item VM Object Reference="Get (a copy)"(Array:Array of Fort Item VM Object References=Get ItemVMs!(self:Season Pass Reward VM Object Reference=@n1) -> (ItemVMs:Array of Fort Item VM Object References)) -> (Output:Fort Item VM Object Reference)
   )
  ]}
 GetBackgroundMaterial {nodes [
   n1 = GetBackgroundMaterial!() -> (then, InIsFree:bool=~, InOwnsPass:bool=~, InThemeData:Season Pass Category Theme Data Structure) {is=title K2Node_FunctionEntry}
   n2 = NOT Boolean(A=Get IsModeSpecificReward())
   n3 = NOT Boolean(A=@n1.InOwnsPass)
   > n4 = CreateDynamicMaterialInstance!(
    execute
    Parent=Select(
     Option 0:Material Interface Object Reference=Get BackgroundMaterial() -> (BackgroundMaterial:Material Object Reference)
     Option 1:Material Interface Object Reference=Get BackgroundMaterial_Free() -> (BackgroundMaterial_Free:Material Instance Object Reference)
     Index=AND Boolean!(A=@n1, B=@n3, C:bool=@n2) -> (ReturnValue)
    ) -> (ReturnValue:Material Interface Object Reference)
    OptionalName=None
    CreationFlags=None
   ) -> (then, ReturnValue)
   > Set BackgroundMID(BackgroundMID:Material Instance Dynamic Object Reference=@n4) -> (Output_Get:Material Instance Dynamic Object Reference)
   n5 = TryLookupColor!(
    Table:"Fort Color Reference Table Structure (by ref)"=Get!() -> (
     InThemeData_ImageReferenceTable:Fort Image Reference Table Structure
     InThemeData_ColorReferenceTable:Fort Color Reference Table Structure
     InThemeData_TextReferenceTable:Fort Text Reference Table Structure
    ).InThemeData_ColorReferenceTable
    Name:String=GamemodeColor
   ) -> (OutColor:Linear Color Structure, ReturnValue:bool)
   > n6 = Branch(Condition=AND Boolean!(A=Get IsModeSpecificReward(), B=@n5.ReturnValue) -> (ReturnValue))
   > SetVectorParameterValue(self=Get BackgroundMID() -> (BackgroundMID:Material Instance Dynamic Object Reference), ParameterName=CharacterColor, Value=@n5)
   > n7 = GetDynamicMaterial(self=Get RewardLogoBackground() -> (RewardLogoBackground:Image Object Reference))
   > n8 = SetVectorParameterValue(self=@n7, ParameterName=Color, Value=@n5)
   Return Node!(
    execute=[@n14.then, @n10.then, @n8.then]
    MaterialInterface:Material Interface Soft Object Reference=Make Soft Reference!(Input:Material Instance Dynamic Object Reference=Get BackgroundMID()) -> (Output:Material Instance Dynamic Soft Object Reference)
   )
   n9 = Get InOwnsPass() -> (InOwnsPass:bool)
   n10 = Branch(execute=@n6.else, Condition=AND Boolean(A=Get InIsFree() -> (InIsFree:bool), B=NOT Boolean(A=@n9)))
   n11 = TryLookupColor(Table=Get().InThemeData_ColorReferenceTable, Name=CharacterColor)
   n12 = Select(Option 0:Linear Color Structure=(R=0.0802,G=1,B=0,A=1), Option 1:Linear Color Structure=@n11, Index=@n11.ReturnValue) -> (ReturnValue:Linear Color Structure)
   SetVectorParameterValue(execute=@n10.else, self=Get BackgroundMID(), ParameterName=CharacterColor, Value=@n12)
   > n13 = GetDynamicMaterial(self=Get RewardLogoBackground())
   > n14 = SetVectorParameterValue(self=@n13, ParameterName=Color, Value=@n12)
  ]}
 GetCanBeHeld {nodes [
   n1 = GetCanBeHeld!() -> (then, IsRewardLocked:bool=~, IsRewardOwned:bool=~, IsPageLocked:bool=~, IsQuestDriven:bool=~, PurchasableRewards:int) {is=title K2Node_FunctionEntry}
   n2 = OR Boolean!(A=@n1, B=@n1.IsRewardOwned, C=@n1.IsPageLocked, D:bool=@n1.IsQuestDriven, E:bool="Equal (Integer)"(A=@n1.PurchasableRewards)) -> (ReturnValue)
   > Set Visibility(
    Visibility=Select(Option 0:ESlateVisibility Enum=Hidden, Option 1:ESlateVisibility Enum=SelfHitTestInvisible, Index=NOT Boolean(A=@n2)) -> (ReturnValue:ESlateVisibility Enum)
    self=Get Img_RewardEffect()
   )
   n3 = GetConfigDataBoolWithDefault!(EntryName:String=bPassesTapNoHold, bDefaultValue:bool=false) -> (ReturnValue:bool)
   > Return Node!(execute, CanBeHeld:bool=NOT Boolean(A=OR Boolean!(A=@n2, B=@n3) -> (ReturnValue)))
  ]}
 GetProgressVisibility {nodes [
   GetProgressVisibility() {is=title K2Node_FunctionEntry}
   > Return Node!(execute, NewParam:ESlateVisibility Enum=Visible)
  ]}
 GetProgressVisuals {nodes [
   n1 = GetProgressVisuals!() -> (then, Progress:double=~, SelectedReward:Season Pass Reward VM Object Reference, IsClaimAll:bool=~) {is=title K2Node_FunctionEntry}
   n2 = "Equal (Object)"(A=Get SelectedReward!() -> (SelectedReward), B=Get SeasonPassRewardVM())
   n3 = Get Progress() -> (Progress:double)
   n4 = Get Progress()
   n5 = Get InClaimAllArray() -> (InClaimAllArray:bool)
   > Return Node(NewParam:double=SelectFloat(A=SelectFloat(A=@n4, B=0, bPickA=@n5), B=SelectFloat(A=@n3, B=0, bPickA=@n2), bPickA=@n1.IsClaimAll))
  ]}
 HandleProgressVisuals {nodes [
   n1 = Get Background() -> (Background:Image Object Reference)
   n2 = HandleProgressVisuals!() -> (then, Progress:double=~) {is=title K2Node_FunctionEntry}
   n3 = SelectFloat(
    A=MapRangeClamped(Value=@n2, InRangeA=Get HoldProgressPercentThreshold() -> (HoldProgressPercentThreshold:double), InRangeB=1)
    B=0
    bPickA="float > float"(A=@n2, B=Get HoldProgressPercentThreshold())
   )
   n4 = SetScalarParameterValue(
    self=GetDynamicMaterial!(Brush:"Slate Brush Structure (by ref)"=Get Brush!(self:Image Object Reference=@n1) -> (Brush:Slate Brush Structure)) -> (ReturnValue)
    ParameterName=Progress
    Value=@n3
   )
   n5 = GetDynamicMaterial(Brush=Get Brush(self=Get LazyImage_Item()))
   n6 = GetDynamicMaterial(Brush=Get Brush(self=Get Character() -> (Character:Fort Lazy Image Object Reference)))
   SetScalarParameterValue(execute=[@n4.then, @n2.then], self=[@n5, @n6], ParameterName=HoldEffect, Value=@n3)
   > SetScalarParameterValue(self=[@n5, @n6], ParameterName=HoldAnimation, Value=@n3)
   > SetScalarParameterValue(self=Get DynamicRewardMaterial() -> (DynamicRewardMaterial:Material Instance Dynamic Object Reference), ParameterName=HoldRatio, Value=@n3)
  ]}
 OnSeasonPassRewardSet {nodes [
   n1 = OnSeasonPassRewardSet!() -> (then, SeasonPassRerwardVM:Season Pass Reward VM Object Reference) {is=title K2Node_FunctionEntry}
   > n2 = Cast To SeasonPassCustomizableItemRewardVM!(execute:Exec, Object=GetFirstExtensionVMByClass(self=@n1, ExtensionClass=&SeasonPassCustomizableItemRewardVM)) -> (then:Exec, CastFailed:Exec, AsSeason Pass Customizable Item Reward VM)
   > SetCustomizableRewardLogo!(execute, self, CustomizableRewardText:Text=Get Label!(self:Season Pass Customizable Item Reward VM Object Reference=@n2) -> (Label:Text)) -> (then)
   SetCustomizableRewardLogo(execute=@n2.CastFailed)
  ]}
 OwnerShipChanged {nodes [
   n1 = OwnerShipChanged!() -> (then, IsOwned:bool=~) {is=title K2Node_FunctionEntry}
   > n2 = Branch(Condition=@n1)
   > Set Status Indicator Visibility!(execute, self:WBP UIKit Item Card Base Object Reference, Indicator:E_UIKit_ItemCardStatusIndicator Enum=NewEnumerator5, Visibility:bool=true) -> (then) {is=title K2Node_CallFunction}
   Set Status Indicator Visibility(execute=@n2.else, Indicator=NewEnumerator5, Visibility=false)
  ]}
 RewardVM_Updated {nodes [
   RewardVM_Updated!() -> (then, RewardVM:Season Pass Reward VM Object Reference) {is=title K2Node_FunctionEntry}
   > VM_SelectedRewardUpdated!(
    execute
    self
    SelectedReward:Season Pass Reward VM Object Reference=Get SelectedReward!(self:Season Pass Selection VM Object Reference=Get SeasonPassSelectionVM()) -> (SelectedReward)
   ) -> (then)
  ]}
 SetCustomizableCharacterIconVisibility {nodes [
   SetCustomizableCharacterIconVisibility!() -> (then, Visibility:ESlateVisibility Enum=Visible) {is=title K2Node_FunctionEntry}
   > n1 = FindCachedStatusIndicator!(execute, self:WBP UIKit Item Card Base Object Reference, IndicatorType:E_UIKit_ItemCardStatusIndicator Enum=NewEnumerator13) -> (then, Status Indicator:WBP UIKit Status Indicator Base Object Reference, Valid:bool)
   > Branch(Condition=@n1.Valid)
   > SetVisibility(self=@n1, InVisibility=Get Visibility() -> (Visibility:ESlateVisibility Enum=Visible))
  ]}
 SetCustomizableRewardLogo {nodes [
   n1 = SetCustomizableRewardLogo!() -> (then, CustomizableRewardText:Text) {is=title K2Node_FunctionEntry}
   > SetVisibility(
    self=Get RewardLogo() -> (RewardLogo:Overlay Object Reference)
    InVisibility=Select(Option 0=SelfHitTestInvisible, Option 1=Collapsed, Index=TextIsEmpty(InText=@n1))
   )
   > "SetText (Text)"(self=Get RewardLogo_Text() -> (RewardLogo_Text:Common Text Object Reference), InText=@n1)
  ]}
 SetCustomizeIconVisibility {nodes [
   SetCustomizeIconVisibility!() -> (then, ShouldBeVisible:bool=~) {is=title K2Node_FunctionEntry}
   > n1 = FindCachedStatusIndicator(IndicatorType=NewEnumerator12)
   > Branch(Condition=@n1.Valid)
   > UpdateSize!(execute, self:WBP UIKit Status Indicator Base Object Reference=@n1, IconSize:E_UI_StatusIndicator_Size Enum=NewEnumerator2) -> (then)
   > SetVisibility(self=@n1, InVisibility=Select(Option 0=Collapsed, Option 1=SelfHitTestInvisible, Index=Get ShouldBeVisible() -> (ShouldBeVisible:bool)))
  ]}
 SetFocusState {nodes [
   n1 = SetFocusState!() -> (then, IsFocused:bool) {is=title K2Node_FunctionEntry}
   > n2 = Branch(Condition="Equal (Boolean)"(A:bool=@n1, B:bool=Get IsButtonFocused()) -> (ReturnValue:bool=~) {is=title K2Node_PromotableOperator})
   > n3 = Set SkipFocusAnimation(SkipFocusAnimation:bool=true) -> (Output_Get:bool)
   Set IsButtonFocused(execute=[@n2.else, @n3.then], IsButtonFocused:bool=Get IsFocused() -> (IsFocused:bool)) -> (Output_Get:bool)
   n4 = Get Anim_Hovered_Base() -> (Anim_Hovered_Base:Widget Animation Object Reference)
   > StopAnimation(InAnimation=@n4)
   > PlayAnimation(
    InAnimation=@n4
    StartAtTime=SelectFloat(A=GetEndTime(self=@n4), B=0, bPickA=Get SkipFocusAnimation() -> (SkipFocusAnimation:bool))
    PlayMode=Select(Option 0:EUMGSequencePlayMode Enum=Reverse, Option 1:EUMGSequencePlayMode Enum, Index=Get IsFocused()) -> (ReturnValue:EUMGSequencePlayMode Enum)
   )
   n5 = Get Focus() -> (Focus:Widget Animation Object Reference)
   > PlayAnimation(
    InAnimation=@n5
    StartAtTime=SelectFloat(A=GetEndTime(self=@n5), B=0, bPickA=Get SkipFocusAnimation())
    PlayMode=Select(Option 0=Reverse, Index=Get IsFocused())
   )
   > RequestTransitionOnBlocks!(
    execute
    self:UIKit Modular Button Object Reference
    State:EUIKitBlockVisualState Enum=Select(Option 0:EUIKitBlockVisualState Enum=Unfocused, Option 1:EUIKitBlockVisualState Enum=Focused, Index=Get IsFocused()) -> (ReturnValue:EUIKitBlockVisualState Enum)
   ) -> (then)
  ]}
 SetItemImage {nodes [
   n1 = SetItemImage!() -> (then, Item Image:Soft Object Reference) {is=title K2Node_FunctionEntry}
   > Parent: SetItemImage!(execute, Item Image:Soft Object Reference=@n1) -> (then) {is=title K2Node_CallParentFunction}
   n2 = SlotAsGridSlot(Widget=Get LazyImage_Item())
   n3 = OR Boolean(A=NOT Boolean(A=IsAssetNull!(Asset:Soft Object Reference=@n1) -> (ReturnValue:bool)), B=InEditor!() -> (ReturnValue:bool))
   > SetHorizontalAlignment(
    self:Grid Slot Object Reference=@n2
    InHorizontalAlignment=Select(Option 0:EHorizontalAlignment Enum=HAlign_Center, Option 1:EHorizontalAlignment Enum=HAlign_Fill, Index=@n3) -> (ReturnValue:EHorizontalAlignment Enum)
   )
   > SetVerticalAlignment(
    self:Grid Slot Object Reference=@n2
    InVerticalAlignment=Select(Option 0:EVerticalAlignment Enum=VAlign_Center, Option 1:EVerticalAlignment Enum=VAlign_Fill, Index=@n3) -> (ReturnValue:EVerticalAlignment Enum)
   )
  ]}
 SetItemImageReference {nodes [
   n1 = SetItemImageReference!() -> (then, Item Image:Fort Image Reference Structure) {is=title K2Node_FunctionEntry}
   > Parent: SetItemImageReference!(execute, Item Image:Fort Image Reference Structure=@n1) -> (then) {is=title K2Node_CallParentFunction}
   n2 = Break Fort Image Reference!(FortImageReference:"Fort Image Reference Structure (by ref)"=@n1) -> (Image:Soft Object Reference, ImageUrl:String, MaterialTextureParamName:Name) {is=title K2Node_BreakStruct}
   n3 = InEditor()
   n4 = NOT Boolean(A=IsAssetNull(Asset=@n2))
   > n5 = Set members in Fort Image Reference!(
    execute:Exec
    StructRef:"Fort Image Reference Structure (by ref)"=Copy!(Item:Fort Image Reference Structure=@n1) -> (ReturnValue:Fort Image Reference Structure) {is=title K2Node_Copy}
    Image:Soft Object Reference=Select(
     Option 0:"Soft Object Reference (by ref)"=Get Fallback Image() -> (Fallback Image:Soft Object Reference)
     Option 1:"Soft Object Reference (by ref)"=@n2
     Index=OR Boolean(A=@n4, B=@n3)
    ) -> (ReturnValue:"Soft Object Reference (by ref)")
   ) -> (then:Exec, StructOut:"Fort Image Reference Structure (by ref)") {is=title K2Node_SetFieldsInStruct}
   n6 = Get Character()
   > SetMaterialTextureParamName!(execute, self:Common Lazy Image Object Reference=@n6, TextureParamName:Name=Item) -> (then)
   > SetBrushFromImageReference!(
    execute
    self:Fort Lazy Image Object Reference=@n6
    ImageReference:"Fort Image Reference Structure (by ref)"=@n5
    bMatchSize:bool=false
    DownloadPriority:ECMSDownloadPriority Enum=Normal
   ) -> (then)
  ]}
 SetItemMaterial {nodes [
   n1 = SetItemMaterial!() -> (then, Material Instance:"Material Instance Object Reference (by ref)") {is=title K2Node_FunctionEntry}
   > Parent: SetItemMaterial!(execute, Material Instance:"Material Instance Object Reference (by ref)"=@n1) -> (then) {is=title K2Node_CallParentFunction}
   n2 = Get Character()
   > SetBrushFromMaterial(self=@n2, Material=Get Material Instance() -> (Material Instance:"Material Instance Object Reference (by ref)"))
   > n3 = Branch(Condition=Get IsAthenaCharacter() -> (IsAthenaCharacter:bool))
   > n4 = GetDynamicMaterial!(execute:Exec, self:Image Object Reference=@n2) -> (then:Exec, ReturnValue)
   > n5 = Set ItemDynamicMaterial(ItemDynamicMaterial:Material Instance Dynamic Object Reference=@n4) -> (Output_Get:Material Instance Dynamic Object Reference)
   Return Node!(execute=[@n7.then, @n5.then])
   n6 = GetDynamicMaterial(execute=@n3.else, self=Get LazyImage_Item())
   > n7 = Set ItemDynamicMaterial(ItemDynamicMaterial=@n6)
  ]}
 SetLockedState {nodes [
   SetLockedState!() -> (then, IsLocked:bool=~) {is=title K2Node_FunctionEntry}
   > SetVisibility(
    self=Get StatusIndicator_Locked() -> (StatusIndicator_Locked:WBP UIKit Status Indicator Object Reference)
    InVisibility=Select(Option 0:ESlateVisibility Enum=Collapsed, Option 1:ESlateVisibility Enum=Visible, Index=Get IsLocked() -> (IsLocked:bool)) -> (ReturnValue:ESlateVisibility Enum)
   )
  ]}
 SetRewardLogoText {nodes [
   n1 = SetRewardLogoText!() -> (then, Reward Text:Text) {is=title K2Node_FunctionEntry}
   > SetRewardLogoVisibility!(execute, self, IsVisible:bool=NOT Boolean(A=TextIsEmpty(InText=@n1))) -> (then)
   > "SetText (Text)"(self=Get RewardLogo_Text(), InText=@n1)
  ]}
 SetRewardLogoVisibility {nodes [
   n1 = SetRewardLogoVisibility!() -> (then, IsVisible:bool=~) {is=title K2Node_FunctionEntry}
   > SetVisibility(self=Get RewardLogo(), InVisibility=Select(Option 0=Collapsed, Option 1=SelfHitTestInvisible, Index=@n1))
  ]}
 SetUsingCharacterImage {nodes [
   SetUsingCharacterImage!() -> (then, UsingCharacterImage:bool=~) {is=title K2Node_FunctionEntry}
   n1 = Get UsingCharacterImage() -> (UsingCharacterImage:bool)
   > SetVisibility(self=Get Character(), InVisibility=Select(Option 0=Hidden, Option 1=SelfHitTestInvisible, Index=@n1))
   > SetVisibility(self=Get LazyImage_Item(), InVisibility=Select(Option 0=SelfHitTestInvisible, Option 1=Hidden, Index=@n1))
   > SetGridLayer!(
    execute
    self:WBP UIKit Item Card Base Object Reference
    WIdget:Widget Object Reference=Get EquippedState_Outline() -> (EquippedState_Outline:WBP UIKit Equipped State Object Reference)
    Layer:int=Select(Option 0:int=11, Option 1:int=9, Index=@n1) -> (ReturnValue:int)
   ) -> (then)
  ]}
 TryGetGamemodeImage {nodes [
   n1 = TryGetGamemodeImage!() -> (then, ImageReferenceTable:"Fort Image Reference Table Structure (by ref)") {is=title K2Node_FunctionEntry}
   n2 = TryLookupImage!(Table:"Fort Image Reference Table Structure (by ref)"=@n1, Name:String=GamemodeLogo) -> (OutImageReference:Fort Image Reference Structure, ReturnValue:bool)
   > n3 = Branch(Condition=@n2.ReturnValue)
   > Return Node!(execute, OutImageReference:Fort Image Reference Structure=@n2, ReturnValue:bool=true)
   n4 = Get SeasonPassSelectionVM()
   n5 = Find Item!(
    TargetArray:Array of Season Pass Page VM Object References=Get Pages!(self:Season Pass Category VM Object Reference=Get SelectedCategory!(self:Season Pass Selection VM Object Reference=@n4) -> (SelectedCategory:Season Pass Category VM Object Reference)) -> (Pages:Array of Season Pass Page VM Object References)
    ItemToFind:"Season Pass Page VM Object Reference (by ref)"=Get SelectedPage(self=@n4)
   ) -> (ReturnValue)
   n6 = TryLookupImage(
    Table=Get ImageReferenceTable() -> (ImageReferenceTable:"Fort Image Reference Table Structure (by ref)")
    Name=Append(A=GamemodeLogo, B="To String (Integer)"(InInt=@n5))
   )
   Return Node(execute=@n3.else, OutImageReference=@n6, ReturnValue=@n6.ReturnValue)
  ]}
 UpdateProgressBarColors {nodes [
   n1 = UpdateProgressBarColors!() -> (then, ShouldUseExpMaterial:bool=~) {is=title K2Node_FunctionEntry}
   > SetActiveWidget(
    self=Get ProgressBar_Switcher() -> (ProgressBar_Switcher:Common Visibility Switcher Object Reference)
    Widget=Select(
     Option 0:Widget Object Reference=Get ProgressBar() -> (ProgressBar:WBP UIKit Progress Bar Object Reference)
     Option 1:Widget Object Reference=Get ProgressBar_Alter() -> (ProgressBar_Alter:WBP UIKit Progress Bar Object Reference)
     Index=@n1
    ) -> (ReturnValue:Widget Object Reference)
   )
  ]}
 UpdateSelectedState {nodes [
   n1 = UpdateSelectedState!() -> (then, ShouldSelect:bool=~) {is=title K2Node_FunctionEntry}
   > n2 = Branch(Condition=@n1)
   > Parent: On Selected() {is=title K2Node_CallParentFunction}
   Branch(execute=@n2.else, Condition=Get IsButtonFocused())
   > Parent: On Deselected() {is=title K2Node_CallParentFunction}
  ]}
 VM_HandleThemeData {nodes [
   n1 = VM_HandleThemeData!() -> (then, Theme:Season Pass Category Theme Data Structure) {is=title K2Node_FunctionEntry}
   > n2 = Sequence!(execute) -> (then_0, then_1, then_2:Exec)
   n3 = Break Season Pass Category Theme Data!(SeasonPassCategoryThemeData:"Season Pass Category Theme Data Structure (by ref)"=@n1) -> (
    ImageReferenceTable:Fort Image Reference Table Structure
    ColorReferenceTable:Fort Color Reference Table Structure
    TextReferenceTable:Fort Text Reference Table Structure
   ) {is=title K2Node_BreakStruct}
   n4 = TryLookupText!(Table:"Fort Text Reference Table Structure (by ref)"=@n3.TextReferenceTable, Name:String=GamemodeTextSmall) -> (OutTextReference:Text, ReturnValue:bool)
   n5 = Branch(execute=@n2.then_2, Condition=AND Boolean(A=Get IsModeSpecificReward(), B=@n4.ReturnValue))
   n6 = Get TextModeName() -> (TextModeName:Overlay Object Reference)
   > SetVisibility(self=@n6, InVisibility=HitTestInvisible)
   > "SetText (Text)"(self=Get Text_GameModeTextSmall() -> (Text_GameModeTextSmall:Common Text Object Reference), InText=@n4)
   > "SetText (Text)"(
    self=Get Text_GameModeTextLarge() -> (Text_GameModeTextLarge:Common Text Object Reference)
    InText=TryLookupText(Table=@n3.TextReferenceTable, Name=GamemodeTextLarge)
   )
   SetVisibility(execute=@n5.else, self=@n6, InVisibility=Collapsed)
   n7 = TryLookupColor(Table=Break Season Pass Category Theme Data(SeasonPassCategoryThemeData=@n1).ColorReferenceTable, Name=ProgressionColor)
   SetVectorParameterValue(
    execute=@n2.then_0
    self=Get ProgressBarMID!(self:WBP UIKit Progress Bar Base Object Reference=Get ProgressBar()) -> (ProgressBarMID:Material Instance Dynamic Object Reference)
    ParameterName=Primary Color
    Value=Select(Option 0:Linear Color Structure=(R=0.0802,G=1,B=0,A=1), Option 1:Linear Color Structure=@n7, Index=@n7.ReturnValue) -> (ReturnValue:Linear Color Structure)
   )
   n8 = TryGetGamemodeImage!(ImageReferenceTable:"Fort Image Reference Table Structure (by ref)"=Break Season Pass Category Theme Data(SeasonPassCategoryThemeData=@n1)) -> (OutImageReference:Fort Image Reference Structure, ReturnValue:bool) {is=title K2Node_CallFunction}
   n9 = Branch(execute=@n2.then_1, Condition=AND Boolean(A=Get IsModeSpecificReward(), B=@n8.ReturnValue))
   n10 = Get Image_GamemodeLogo() -> (Image_GamemodeLogo:Fort Lazy Image Object Reference)
   > Set Visibility(Visibility=SelfHitTestInvisible, self=@n10)
   > SetMaterialTextureParamName(self=Get Image_GamemodeLogo(), TextureParamName=LogoTexture)
   > SetBrushFromImageReference(self=Get Image_GamemodeLogo(), ImageReference=@n8, bMatchSize=false, DownloadPriority=Normal)
   Set Visibility(execute=@n9.else, Visibility=Collapsed, self=@n10)
  ]}
 VM_SelectedRewardUpdated {nodes [
   n1 = VM_SelectedRewardUpdated!() -> (then, SelectedReward:Season Pass Reward VM Object Reference) {is=title K2Node_FunctionEntry}
   n2 = Get SeasonPassRewardVM()
   > UpdateSelectedState!(execute, self, ShouldSelect:bool=AND Boolean(A="Equal (Object)"(A=@n1, B=@n2), B=IsValid(Object=@n2))) -> (then) {is=title K2Node_CallFunction}
  ]}
 VM_SetBarsProgress {nodes [
   n1 = VM_SetBarsProgress!() -> (then, Progress:double=~) {is=title K2Node_FunctionEntry}
   > SetProgress!(execute, self:WBP UIKit Progress Bar Base Object Reference=[Get ProgressBar(), Get ProgressBar_Alter()], Progress:double=@n1) -> (then)
  ]}
 ViewmodelsUpdated {nodes [
   n1 = ViewmodelsUpdated!() -> (then, MyRewardVM:Season Pass Selection VM Object Reference, SelectedReward:Season Pass Selection VM Object Reference) {is=title K2Node_FunctionEntry}
   > Return Node!(execute, AreTheSameReward:bool="Equal (Object)"(A=@n1, B=@n1.SelectedReward))
  ]}
]
variables [
 BackgroundMaterial {type=Material, default=&M_UI_BPReboot_RewardTrack_TileBackground, class=~, graph=~}
 BackgroundMaterial_Free {type=MaterialInstance, default=/SeasonPass/Widgets/Reboot/Materials/MI_UI_BPReboot_RewardTrack_TileBackground_Free}
 CardIndex {type=integer}
 DynamicRewardMaterial {type=MaterialInstanceDynamic}
 HideEnhanced {type=boolean, default=true}
 HoldProgressPercentThreshold {type=number, default=0.05}
 InClaimAllArray {type=boolean}
 IsAthenaCharacter {type=boolean}
 IsModeSpecificReward {type=boolean}
 Preview_UseAlterProgressBar {type=boolean, default=true}
 WidgetSeasonPassCategory {type=SeasonPassCategory_C, default=null}
]
widgets {
 info {
  inheritedWidgetCount=0
  namedSlotCount=2
  parentClass=/FortUILibrary/Systems/Items/WBP_UIKit_ItemCard_Base.WBP_UIKit_ItemCard_Base_C
  rootWidgetClass=null
  widgetCount=25
 }
 widgets [
  Status {widgetClassPath=&Overlay, widget=Status}
  Character {
   widgetClassPath=&FortLazyImage
   bIsVariable=true
   parent=Status
   widget=Character
   properties {
    materialTextureParamName=RewardTexture
    brush {
     bIsDynamicallyLoaded=false
     drawAs=Box
     imageType=FullColor
     mirroring=NoMirror
     resourceName=None
     resourceObject=&MI_UI_Nexus_RewardTrack_ItemPreview_Base
     tiling=NoTile
     imageSize=<128 128>
     margin=<ltrb 0 0 0 0>
     outlineSettings {
      bUseBrushTransparency=false
      roundingType=HalfHeightRadius
      width=0
      color {colorUseRule=UseColor_Specified, specifiedColor=<rgba 0 0 0 0>}
      cornerRadii=<0 0 0 0>
     }
     tintColor {colorUseRule=UseColor_Specified, specifiedColor=<rgba 1 1 1 1>}
     uVRegion {bIsValid=false, max=<0 0>, min=<0 0>}
    }
    loadingBackgroundBrush {
     bIsDynamicallyLoaded=false
     drawAs=Box
     imageType=NoImage
     mirroring=NoMirror
     resourceName=None
     resourceObject=null
     tiling=NoTile
     imageSize=<32 32>
     margin=<ltrb 0 0 0 0>
     outlineSettings {
      bUseBrushTransparency=false
      roundingType=HalfHeightRadius
      width=0
      color {colorUseRule=UseColor_Specified, specifiedColor=<rgba 0 0 0 0>}
      cornerRadii=<0 0 0 0>
     }
     tintColor {colorUseRule=UseColor_Specified, specifiedColor=<rgba 1 1 1 1>}
     uVRegion {bIsValid=false, max=<0 0>, min=<0 0>}
    }
    loadingThrobberBrush {
     bIsDynamicallyLoaded=false
     drawAs=Box
     imageType=NoImage
     mirroring=NoMirror
     resourceName=None
     resourceObject=null
     tiling=NoTile
     imageSize=<32 32>
     margin=<ltrb 0 0 0 0>
     outlineSettings {
      bUseBrushTransparency=false
      roundingType=HalfHeightRadius
      width=0
      color {colorUseRule=UseColor_Specified, specifiedColor=<rgba 0 0 0 0>}
      cornerRadii=<0 0 0 0>
     }
     tintColor {colorUseRule=UseColor_Specified, specifiedColor=<rgba 1 1 1 1>}
     uVRegion {bIsValid=false, max=<0 0>, min=<0 0>}
    }
   }
   slot {class=OverlaySlot, properties {horizontalAlignment=HAlign_Fill, verticalAlignment=VAlign_Fill, padding=<ltrb 0 -50 0 0>}}
  }
  StatusIndicator_Owned {
   widgetClassPath=/FortUILibrary/UIKit/StatusIndicator/WBP_UIKit_StatusIndicator
   bIsVariable=true
   parent=Status
   widget=StatusIndicator_Owned
   properties {icon=/FortUILibrary/Textures/MaterialInstances/MI_UIKit_Icon_Check, iconSize=Compact, visibility=Collapsed}
   slot {class=OverlaySlot, properties {padding=<ltrb 8 8 8 8>}}
  }
  StatusIndicator_Locked {
   widgetClassPath=&WBP_UIKit_StatusIndicator
   bIsVariable=true
   parent=Status
   widget=StatusIndicator_Locked
   properties {type=Utility, icon=/FortUILibrary/Textures/MaterialInstances/MI_UIKit_Icon_Locked, iconSize=Compact, visibility=Collapsed}
   slot {class=OverlaySlot, properties {padding=<ltrb 8 8 8 8>}}
  }
  StatusIndicator_ComingSoon {
   widgetClassPath=&WBP_UIKit_StatusIndicator
   bIsVariable=true
   parent=Status
   widget=StatusIndicator_ComingSoon
   properties {type=Attention, icon=/FortUILibrary/Textures/Icons/T_UI_Icon_UIKit_UpdatePending, iconSize=Compact, visibility=Collapsed}
   slot {class=OverlaySlot, properties {padding=<ltrb 8 8 8 8>}}
  }
  VerticalBox_0 {
   widgetClassPath=&VerticalBox
   parent=Status
   widget=VerticalBox_0
   slot {class=OverlaySlot, properties {horizontalAlignment=HAlign_Fill, verticalAlignment=VAlign_Bottom, padding=<ltrb 8 8 8 8>}}
  }
  StatusIndicator_Included {
   widgetClassPath=&CommonTextBlock
   bIsVariable=true
   parent=VerticalBox_0
   widget=StatusIndicator_Included
   properties {
    mobileFontSizeMultiplier=1.15
    scrollStyle=&TextScrollStyle-Base_C
    text=INCLUDED
    visibility=Collapsed
    font {
     bForceMonospaced=false
     bMaterialIsStencil=false
     fontMaterial=null
     fontObject=&HeadingNow
     letterSpacing=0
     monospacedWidth=1
     size=12
     skewAmount=0
     typefaceFontName=75 Medium
     outlineSettings {
      bApplyOutlineToDropShadows=false
      bMiteredCorners=false
      bSeparateFillAlpha=false
      outlineMaterial=null
      outlineSize=0
      outlineColor=<rgba 0 0 0 1>
     }
    }
    shadowOffset=<0 0>
   }
   slot {class=VerticalBoxSlot, properties {verticalAlignment=VAlign_Bottom, size {sizeRule=Fill, value=1}}}
  }
  ProgressBar_Switcher {
   widgetClassPath=&CommonVisibilitySwitcher
   bIsVariable=true
   parent=VerticalBox_0
   widget=ProgressBar_Switcher
   slot {class=VerticalBoxSlot, properties {padding=<ltrb 0 8 0 0>}}
  }
  ProgressBar {
   widgetClassPath=/FortUILibrary/UIKit/ProgressBar/WBP_UIKit_ProgressBar
   bIsVariable=true
   parent=ProgressBar_Switcher
   widget=ProgressBar
   properties {
    height=16
    linearMaterial=/SeasonPass/Materials/ProgressBar/MI_UI_NexusProgerssBar
    outline_PX=6
    progressValue=0.5488
    useCircular=false
    useOutline=true
   }
   slot {class=CommonVisibilitySwitcherSlot}
  }
  ProgressBar_Alter {
   widgetClassPath=&WBP_UIKit_ProgressBar
   bIsVariable=true
   parent=ProgressBar_Switcher
   widget=ProgressBar_Alter
   properties {
    height=16
    linearMaterial=/SeasonPass/Materials/ProgressBar/MI_UI_NexusProgerssBar_Exp
    outline_PX=6
    progressValue=0.5488
    useCircular=false
    useOutline=true
   }
   slot {class=CommonVisibilitySwitcherSlot}
  }
  Image_GamemodeLogo {
   widgetClassPath=&FortLazyImage
   bIsVariable=true
   parent=Status
   widget=Image_GamemodeLogo
   properties {
    visibility=Collapsed
    brush {
     bIsDynamicallyLoaded=false
     drawAs=Image
     imageType=FullColor
     mirroring=NoMirror
     resourceName=None
     resourceObject=/SeasonPass/Widgets/Reboot/Materials/M_UI_BPReboot_RewardTrack_GameModeOverlay
     tiling=NoTile
     imageSize=<32 32>
     margin=<ltrb 0 0 0 0>
     outlineSettings {
      bUseBrushTransparency=false
      roundingType=HalfHeightRadius
      width=0
      color {colorUseRule=UseColor_Specified, specifiedColor=<rgba 0 0 0 0>}
      cornerRadii=<0 0 0 0>
     }
     tintColor {colorUseRule=UseColor_Specified, specifiedColor=<rgba 1 1 1 1>}
     uVRegion {bIsValid=false, max=<0 0>, min=<0 0>}
    }
   }
   slot {class=OverlaySlot, properties {horizontalAlignment=HAlign_Fill, verticalAlignment=VAlign_Fill}}
  }
  TextModeName {
   widgetClassPath=&Overlay
   bIsVariable=true
   parent=Status
   widget=TextModeName
   slot {class=OverlaySlot, properties {horizontalAlignment=HAlign_Fill, verticalAlignment=VAlign_Bottom, padding=<ltrb 20 0 20 20>}}
  }
  DarkenLayer {
   widgetClassPath=&CommonVisualAttachment
   bIsVariable=true
   parent=TextModeName
   widget=DarkenLayer
   properties {contentAnchor=<0.264 0.3209>}
   slot {class=OverlaySlot}
  }
  Darken {
   widgetClassPath=&Image
   bIsVariable=true
   parent=DarkenLayer
   widget=Darken
   properties {
    visibility=SelfHitTestInvisible
    brush {
     bIsDynamicallyLoaded=false
     drawAs=Image
     imageType=FullColor
     mirroring=NoMirror
     resourceName=None
     resourceObject=/FortUILibrary/UIKit/Buttons/Materials/M_UIKit_RadialGlow
     tiling=NoTile
     imageSize=<400 250>
     margin=<ltrb 0 0 0 0>
     outlineSettings {
      bUseBrushTransparency=false
      roundingType=HalfHeightRadius
      width=0
      color {colorUseRule=UseColor_Specified, specifiedColor=<rgba 0 0 0 0>}
      cornerRadii=<0 0 0 0>
     }
     tintColor {colorUseRule=UseColor_Specified, specifiedColor=<rgba 1 1 1 1>}
     uVRegion {bIsValid=false, max=<0 0>, min=<0 0>}
    }
    colorAndOpacity=<rgba 0 0 0 0.4>
   }
   slot {class=SizeBoxSlot, properties {horizontalAlignment=HAlign_Center, verticalAlignment=VAlign_Center}}
  }
  VerticalBox_1 {widgetClassPath=&VerticalBox, parent=TextModeName, widget=VerticalBox_1, slot {class=OverlaySlot}}
  Text_GameModeTextSmall {
   widgetClassPath=&CommonTextBlock
   bIsVariable=true
   parent=VerticalBox_1
   widget=Text_GameModeTextSmall
   properties {
    autoWrapText=true
    bIsScrollingEnabled=false
    lineHeightPercentage=0.7
    text=Any Epic
    textTransformPolicy=ToUpper
    font {
     bForceMonospaced=false
     bMaterialIsStencil=false
     fontMaterial=null
     fontObject=&HeadingNow
     letterSpacing=0
     monospacedWidth=1
     size=12
     skewAmount=0
     typefaceFontName=76 Bold
     outlineSettings {
      bApplyOutlineToDropShadows=false
      bMiteredCorners=false
      bSeparateFillAlpha=false
      outlineMaterial=null
      outlineSize=0
      outlineColor=<rgba 0 0 0 1>
     }
    }
    shadowOffset=<0 0>
   }
   slot {class=VerticalBoxSlot}
  }
  Text_GameModeTextLarge {
   widgetClassPath=&CommonTextBlock
   bIsVariable=true
   parent=VerticalBox_1
   widget=Text_GameModeTextLarge
   properties {
    autoWrapText=true
    bIsScrollingEnabled=false
    lineHeightPercentage=0.7
    text=Battle Royale
    textTransformPolicy=ToUpper
    font {
     bForceMonospaced=false
     bMaterialIsStencil=false
     fontMaterial=null
     fontObject=&HeadingNow
     letterSpacing=0
     monospacedWidth=1
     size=18
     skewAmount=0
     typefaceFontName=86 Bold Italic
     outlineSettings {
      bApplyOutlineToDropShadows=false
      bMiteredCorners=false
      bSeparateFillAlpha=false
      outlineMaterial=null
      outlineSize=0
      outlineColor=<rgba 0 0 0 1>
     }
    }
    shadowOffset=<0 0>
   }
   slot {class=VerticalBoxSlot}
  }
  RewardLogo {
   widgetClassPath=&Overlay
   bIsVariable=true
   parent=Status
   widget=RewardLogo
   properties {visibility=Collapsed}
   slot {class=OverlaySlot, properties {horizontalAlignment=HAlign_Fill, verticalAlignment=VAlign_Fill}}
  }
  RewardLogoBackground {
   widgetClassPath=&Image
   bIsVariable=true
   parent=RewardLogo
   widget=RewardLogoBackground
   properties {
    visibility=SelfHitTestInvisible
    brush {
     bIsDynamicallyLoaded=false
     drawAs=Image
     imageType=FullColor
     mirroring=NoMirror
     resourceName=None
     resourceObject=/SeasonPass/Widgets/Reboot/Materials/M_UI_BP_Triangle
     tiling=NoTile
     imageSize=<32 32>
     margin=<ltrb 0 0 0 0>
     outlineSettings {
      bUseBrushTransparency=false
      roundingType=HalfHeightRadius
      width=0
      color {colorUseRule=UseColor_Specified, specifiedColor=<rgba 0 0 0 0>}
      cornerRadii=<0 0 0 0>
     }
     tintColor {colorUseRule=UseColor_Specified, specifiedColor=<rgba 1 1 1 1>}
     uVRegion {bIsValid=false, max=<0 0>, min=<0 0>}
    }
   }
   slot {class=OverlaySlot, properties {horizontalAlignment=HAlign_Fill, verticalAlignment=VAlign_Fill}}
  }
  SizeBox_1 {
   widgetClassPath=&SizeBox
   parent=RewardLogo
   widget=SizeBox_1
   properties {bOverride_HeightOverride=true, heightOverride=64}
   slot {class=OverlaySlot, properties {verticalAlignment=VAlign_Center, padding=<ltrb 16 26 0 0>}}
  }
  RewardLogo_Text {
   widgetClassPath=&CommonTextBlock
   bIsVariable=true
   parent=SizeBox_1
   widget=RewardLogo_Text
   properties {
    autoWrapText=true
    bAutoCollapseWithEmptyText=true
    flowDirectionPreference=Culture
    mobileFontSizeMultiplier=1.15
    scrollOrientation=Orient_Vertical
    scrollStyle=&TextScrollStyle-Base_C
    text="[PH] Logo Text Logo"
    textTransformPolicy=ToUpper
    font {
     bForceMonospaced=false
     bMaterialIsStencil=false
     fontMaterial=null
     fontObject=&HeadingNow
     letterSpacing=0
     monospacedWidth=1
     size=15.75
     skewAmount=0
     typefaceFontName=86 Bold
     outlineSettings {
      bApplyOutlineToDropShadows=false
      bMiteredCorners=false
      bSeparateFillAlpha=false
      outlineMaterial=null
      outlineSize=0
      outlineColor=<rgba 0 0 0 1>
     }
    }
    shadowOffset=<0 0>
   }
   slot {class=SizeBoxSlot, properties {verticalAlignment=VAlign_Bottom, padding=<ltrb 0 0 0 8>}}
  }
  Overlay_134 {widgetClassPath=&Overlay, widget=Overlay_134}
  Img_RewardEffect {
   widgetClassPath=&Image
   bIsVariable=true
   parent=Overlay_134
   widget=Img_RewardEffect
   properties {
    visibility=Hidden
    brush {
     bIsDynamicallyLoaded=false
     drawAs=Image
     imageType=FullColor
     mirroring=NoMirror
     resourceName=None
     resourceObject=/SeasonPass/Materials/MI_UI_Pass_RewardState
     tiling=NoTile
     imageSize=<32 32>
     margin=<ltrb 0 0 0 0>
     outlineSettings {
      bUseBrushTransparency=false
      roundingType=HalfHeightRadius
      width=0
      color {colorUseRule=UseColor_Specified, specifiedColor=<rgba 0 0 0 0>}
      cornerRadii=<0 0 0 0>
     }
     tintColor {colorUseRule=UseColor_Specified, specifiedColor=<rgba 1 1 1 1>}
     uVRegion {bIsValid=false, max=<0 0>, min=<0 0>}
    }
   }
   slot {class=OverlaySlot, properties {horizontalAlignment=HAlign_Fill, verticalAlignment=VAlign_Fill}}
  }
  CommonVisualAttachment_0 {
   widgetClassPath=&CommonVisualAttachment
   parent=Overlay_134
   widget=CommonVisualAttachment_0
   properties {visibility=Collapsed, contentAnchor=<0.5 0.5>}
   slot {class=OverlaySlot, properties {horizontalAlignment=HAlign_Center, verticalAlignment=VAlign_Center}}
  }
  Img_RewardExplosion {
   widgetClassPath=&Image
   bIsVariable=true
   parent=CommonVisualAttachment_0
   widget=Img_RewardExplosion
   properties {
    visibility=SelfHitTestInvisible
    brush {
     bIsDynamicallyLoaded=false
     drawAs=Image
     imageType=FullColor
     mirroring=NoMirror
     resourceName=None
     resourceObject=/FortUILibrary/Materials/Effects/M_UI_Tri_Explode_FX
     tiling=NoTile
     imageSize=<300 300>
     margin=<ltrb 0 0 0 0>
     outlineSettings {
      bUseBrushTransparency=false
      roundingType=HalfHeightRadius
      width=0
      color {colorUseRule=UseColor_Specified, specifiedColor=<rgba 0 0 0 0>}
      cornerRadii=<0 0 0 0>
     }
     tintColor {colorUseRule=UseColor_Specified, specifiedColor=<rgba 1 1 1 1>}
     uVRegion {bIsValid=false, max=<0 0>, min=<0 0>}
    }
   }
   slot {class=SizeBoxSlot}
  }
 ]
}
