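# Season Pass widget that shows an "X of Y" counter for the previewed item within a multi-item reward.
# UpdateRewardCount broadcasts the reward VM's item count and, for valid VMs, collapses the widget unless there are several items.
# UpdateCurrentRewardIndex matches the previewed item's definition against the reward's items and broadcasts CurrentItemNumber from that index.
# CF_GetCounterText formats both numbers into the counter text.
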WidgetBlueprint /SeasonPass/Widgets/RewardCounter
graphs [
 CF_GetCounterText {nodes [
   n1 = CF_GetCounterText!() -> (then, CurrentItemNumber:int, RewardCount:int) {is=title K2Node_FunctionEntry}
   > Return Node!(
    execute
    Result:Text=Format Text!(Format="LOCTEXT(\"{CurrentReward} of {RewardCount}\")", CurrentReward:int=@n1, RewardCount:int=@n1.RewardCount) -> (Result)
   )
  ]}
 EventGraph {nodes [
   Event PreConstruct()
   Event Construct()
   Event Tick()
  ]}
 UpdateCurrentRewardIndex {nodes [
   UpdateCurrentRewardIndex!() -> (then, PreviewedFortItemVM:Fort Item VM Object Reference) {is=title K2Node_FunctionEntry}
   n1 = Get SeasonPassRewardVM() -> (SeasonPassRewardVM:Season Pass Reward VM Object Reference)
   > Is Valid(InputObject=@n1)
   > n2 = For Each Loop with Break!(
    Exec:Exec
    Array:Array of Fort Item VM Object References=Get ItemVMs!(self:Season Pass Reward VM Object Reference=@n1) -> (ItemVMs:Array of Fort Item VM Object References)
    Break:Exec
   ) -> (LoopBody:Exec, Array Element:Fort Item VM Object Reference, Array Index:int, Completed:Exec)
   n3 = Get PreviewedFortItemVM() -> (PreviewedFortItemVM:Fort Item VM Object Reference)
   > Branch(Condition="Equal (Object)"(
     A=Get ItemDefinition!(self:Fort Item VM Object Reference=@n2) -> (ItemDefinition:Fort Item Definition Object Reference)
     B=Get ItemDefinition(self=@n3)
    ))
   > Set with Broadcast CurrentItemNumber!(execute, CurrentItemNumber:int=int + int(A=@n2.Array Index)) -> (then, Output_Get:int=0)
  ]}
 UpdateRewardCount {nodes [
   n1 = UpdateRewardCount!() -> (then, RewardVM:Season Pass Reward VM Object Reference) {is=title K2Node_FunctionEntry}
   > n2 = Is Valid(InputObject=@n1)
   Set with Broadcast RewardsCount!(execute=@n2.Is Not Valid, RewardsCount:int=0) -> (then, Output_Get:int=0)
   n3 = Length(TargetArray:Array of Fort Item VM Object References=Get ItemVMs(self=@n1))
   Set with Broadcast RewardsCount(execute=@n2.Is Valid, RewardsCount=@n3) -> (Output_Get=0)
   > SetVisibility(InVisibility=Select(Option 0=Collapsed, Option 1=SelfHitTestInvisible, Index="integer > integer"(A=@n3, B=1)))
  ]}
]
variables [
 CurrentItemNumber {type=integer, default=1}
 RewardsCount {type=integer}
]
widgets {
 info {inheritedWidgetCount=0, parentClass=&UserWidget, rootWidgetClass=&Overlay, widgetCount=3}
 widgets [
  VariantCounterInfo {widgetClassPath=&Overlay, bIsVariable=true, widget=VariantCounterInfo, properties {visibility=HitTestInvisible}}
  WBP_UIKit_Backplate {
   widgetClassPath=&WBP_UIKit_Backplate
   parent=VariantCounterInfo
   widget=WBP_UIKit_Backplate
   properties {brightness=Dark}
   slot {class=OverlaySlot, properties {horizontalAlignment=HAlign_Fill, verticalAlignment=VAlign_Fill}}
  }
  Text_VariantCycleInfo {
   widgetClassPath=&CommonTextBlock
   bIsVariable=true
   parent=VariantCounterInfo
   widget=Text_VariantCycleInfo
   properties {
    mobileFontSizeMultiplier=1.15
    text="## of ##"
    font {
     bForceMonospaced=false
     bMaterialIsStencil=false
     fontMaterial=null
     fontObject=&HeadingNow
     letterSpacing=0
     monospacedWidth=1
     size=15.75
     skewAmount=0
     typefaceFontName=74 Regular
     outlineSettings {
      bApplyOutlineToDropShadows=false
      bMiteredCorners=false
      bSeparateFillAlpha=false
      outlineMaterial=null
      outlineSize=0
      outlineColor=<rgba 0 0 0 1>
     }
    }
   }
   slot {class=OverlaySlot, properties {padding=<ltrb 16 8 16 8>}}
  }
 ]
}
