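# Season pass claim-page reward card, a child of the UIKit item card base that serves as a selectable list entry.
# The list item becomes its reward view model; selecting it writes to the shared selection view model and logs an analytics interaction with the reward name.
# The parent's hover look doubles as the selected state, set on item assignment; unhover passes through only when focused and unselected.
# Construct clears equipped/enhanced flags and applies the claim-page backplate.
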WidgetBlueprint /SeasonPass/Widgets/SeasonPassClaimPagesReward
graphs [
 EventGraph {nodes [
   n1 = Event OnListItemObjectSet()
   > n2 = Cast To SeasonPassRewardVM!(execute, Object=@n1.ListItemObject) -> (then, CastFailed, AsSeason Pass Reward VM:Season Pass Reward VM Object Reference) {is=title K2Node_DynamicCast}
   > SetSeasonPassRewardVM!(execute, self, Viewmodel:Season Pass Reward VM Object Reference=@n2) -> (then)
   > UpdateSelectedState!(
    execute
    self
    SelectedRewardVM:Season Pass Reward VM Object Reference=Get SelectedReward!(self:Season Pass Selection VM Object Reference=Get SeasonPassSelectionVM() -> (SeasonPassSelectionVM:Season Pass Selection VM Object Reference)) -> (SelectedReward:Season Pass Reward VM Object Reference)
   ) -> (then)
   Event On Selected()
   > n3 = Set SelectedReward!(
    execute
    SelectedReward:Season Pass Reward VM Object Reference=Get SeasonPassRewardVM() -> (SeasonPassRewardVM:Season Pass Reward VM Object Reference)
    self:Season Pass Selection VM Object Reference=Get SeasonPassSelectionVM()
   ) -> (then, Output_Get:Season Pass Reward VM Object Reference)
   > Fire Event Element Interacted(
    Element=Self-Reference()
    AdditionalInfo="Add String (Non-Localized)"!(
     Info:"Flow Analytics Additional Info Structure (by ref)"=Make Flow Analytics Additional Info!() -> (FlowAnalyticsAdditionalInfo:Flow Analytics Additional Info Structure) {is=title K2Node_MakeStruct}
     Key:String=RewardName
     Value:String=GetDisplayName(Object=@n3)
    ) -> (ReturnValue:Flow Analytics Additional Info Structure)
   )
   Event On Unhovered()
   n4 = Get SeasonPassRewardVM()
   n5 = Get SelectedReward(self=Get SeasonPassSelectionVM())
   > Branch(Condition=AND Boolean(
     A=Get IsButtonFocused() -> (IsButtonFocused:bool)
     B="Not Equal (Object)"(A:Object Reference=@n5, B:Object Reference=@n4) -> (ReturnValue:bool=~) {is=title K2Node_PromotableOperator}
    ))
   > Parent: On Unhovered() {is=title K2Node_CallParentFunction}
   Event Construct()
   > SetItemIsEquipped!(execute, self:WBP UIKit Item Card Base Object Reference, Is Equipped:bool=false) -> (then)
   > SetItemIsEnhanced!(execute, self:WBP UIKit Item Card Base Object Reference, Is Enhanced:bool=false) -> (then)
   > SetBrushFromMaterial(self=Get Background() -> (Background:Image Object Reference), Material=/SeasonPass/Materials/Backplates/MI_UIKit_ClaimPagesItemBackplate)
  ]}
 UpdateSelectedState {nodes [
   n1 = UpdateSelectedState!() -> (then, SelectedRewardVM:Season Pass Reward VM Object Reference) {is=title K2Node_FunctionEntry}
   > n2 = Branch(Condition="Equal (Object)"(A=@n1, B=Get SeasonPassRewardVM()))
   > Parent: On Hovered() {is=title K2Node_CallParentFunction}
   Parent: On Unhovered(execute=@n2.else)
  ]}
]
variables []
widgets {
 info {
  inheritedWidgetCount=0
  parentClass=/FortUILibrary/Systems/Items/WBP_UIKit_ItemCard_Base.WBP_UIKit_ItemCard_Base_C
  rootWidgetClass=null
  widgetCount=0
 }
 widgets []
}
