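# Season pass widget for claiming every unowned reward from the first page with an unowned reward through a player-chosen page.
# Plus/minus buttons move the end page up to the last unlockable one, and UpdateItemCards rebuilds the tile list, cost and status.
# With enough purchasable rewards, Claim Pages calls PurchaseItems on the list; otherwise Buy Rewards opens a modal preset from the shortfall.
# It plays an opening animation and applies mobile-specific styling.
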WidgetBlueprint /SeasonPass/Widgets/SeasonPass_ClaimPages
animations [Opening {
  binding_count=1
  length_seconds=1
  object=&self:Opening
  sequence {bindings [SizeBox_0 {
     id=#1
     possessed_class=SizeBox
     tracks [
      {
       type=MovieSceneFloatTrack
       display_name=Render Opacity
       sections [{type=MovieSceneFloatSection, end_frame=20, start_frame=0, channels [None {type=MovieSceneScriptingFloatChannel, keys [[0,0],[20,1]]}]}]
      }
      {
       type=MovieScene2DTransformTrack
       sections [{
         type=MovieScene2DTransformSection
         end_frame=20
         start_frame=0
         channels [
          Translation.X {type=MovieSceneScriptingFloatChannel, keys [[0,100],[20,0]]}
          Translation.Y {type=MovieSceneScriptingFloatChannel, const [0,2]}
          Angle {type=MovieSceneScriptingFloatChannel}
          Scale.X {type=MovieSceneScriptingFloatChannel}
          Scale.Y {type=MovieSceneScriptingFloatChannel}
          Shear.X {type=MovieSceneScriptingFloatChannel}
          Shear.Y {type=MovieSceneScriptingFloatChannel}
         ]
        }]
      }
     ]
    }]}
 }]
graphs [
 EventGraph {nodes [
   "On Clicked (Button_Subtract_1)"!() -> (OutputDelegate, then, Button:Common Button Base Object Reference) {is=title K2Node_ComponentBoundEvent}
   n1 = GetLastUnlockablePage!() -> (LastUnlockablePage:int=0)
   n2 = Get PageIndex() -> (PageIndex:int=0)
   n3 = Get PageInitial() -> (PageInitial:int=0)
   > SetPageIndexAndUpdate!(execute, self, NewIndex:int="Clamp (Integer)"(Value=int - int(A=@n2), Min=@n3, Max=@n1) -> (ReturnValue=0)) -> (then)
   "On Clicked (Button_Add_1)"!() -> (OutputDelegate, then, Button:Common Button Base Object Reference) {is=title K2Node_ComponentBoundEvent}
   > SetPageIndexAndUpdate(NewIndex="Clamp (Integer)"(Value=int + int(A=@n2), Min=@n3, Max=@n1) -> (ReturnValue=0))
   n4 = "On Clicked (Button_BuyRewards)"!() -> (OutputDelegate, then, Button:Common Button Base Object Reference) {is=title K2Node_ComponentBoundEvent}
   > Fire Event Element Interacted(Element=@n4.Button, AdditionalInfo=GetAnalyticsInfo!() -> (ReturnValue:Flow Analytics Additional Info Structure))
   > n5 = Create Widget!(
    execute:Exec
    Class:User Widget Class Reference=&/SeasonPass_BuyRewards.SeasonPass_BuyRewards_C
    OwningPlayer:Player Controller Object Reference
    SeasonPassVM:Season Pass VM Object Reference=Get SeasonPassVM() -> (SeasonPassVM:Season Pass VM Object Reference)
    SeasonPassCategoryVM:Season Pass Category VM Object Reference
   ) -> (then:Exec, ReturnValue:Season Pass Buy Rewards Object Reference)
   > TryPushWidgetAsModalPanel!(execute, PushWidget:User Widget Object Reference=@n5) -> (then)
   > n6 = SetRewardsToBuy!(
    execute
    self:Season Pass Buy Rewards Object Reference=@n5
    Amount:int=int - int(A=Get NeedToClaim() -> (NeedToClaim:int=0), B=Get RewardsToBuy!(self:Season Pass Buy Rewards Object Reference=@n5) -> (RewardsToBuy:int=0))
    IsAdditive:bool=false
   ) -> (then)
   n7 = Refresh On Deactivating Buy Rewards() {is=title K2Node_CustomEvent}
   n8 = Assign On Widget Deactivated!(execute:Exec=@n6.then, self:Common Activatable Widget Object Reference=@n5, Delegate:"Delegate (by ref)"=@n7) -> (then:Exec) {is=title K2Node_AssignDelegate}
   n9 = On Deactivating Buy Rewards Changing Tab() {is=title K2Node_CustomEvent}
   Bind Event to On Going to Buy VBucks Page(execute=@n8.then, self:Season Pass Buy Rewards Object Reference=@n5, Delegate=@n9) {is=title K2Node_AddDelegate}
   "On Clicked (Button_ClaimPages)"!() -> (OutputDelegate, then, Button:Common Button Base Object Reference) {is=title K2Node_ComponentBoundEvent}
   > n10 = PurchaseItems!(
    execute:Exec
    Items:Array of Season Pass Reward VM Object References=Get ItemList() -> (ItemList:Array of Season Pass Reward VM Object References)
    SeasonPass:Season Pass VM Object Reference=Get SeasonPassVM()
    bTryPurchasePrerequisites:bool=false
   ) -> (then:Exec, OnOperationCompleted:Exec, Success:bool=~) {is=title K2Node_AsyncAction}
   Call Finish Claim Event Dispatcher!(execute:Exec=@n10.OnOperationCompleted, self:Season Pass Claim Pages Object Reference) -> (then:Exec) {is=title K2Node_CallDelegate}
   n11 = Initialize() {is=title K2Node_CustomEvent}
   n12 = Set PageInitial(execute=[@n20.then, @n11.then], PageInitial:int=GetFirstPurchaseablePage!() -> (Page:int=0)) -> (Output_Get:int=0)
   > n13 = Set PageIndex(PageIndex:int=@n12) -> (Output_Get:int=0)
   > SetTotalText!(execute, self, ActualPage:int=@n13) -> (then)
   > SetIsEnabled(bInIsEnabled=true)
   > SetupVisuals()
   > n14 = Branch(Condition=NeedsButtonOnPassCategory!() -> (Value:bool))
   n15 = Get VisibilitySwitcherIsClaimed() -> (VisibilitySwitcherIsClaimed:Common Visibility Switcher Object Reference)
   > SetActiveWidgetIndex(self:Common Visibility Switcher Object Reference=@n15, Index=0)
   n16 = Get ItemTileView() -> (ItemTileView:Fort Item Tile View Object Reference)
   > ClearListItems(self=@n16)
   > Clear Selection(self:List View Object Reference=@n16)
   SetActiveWidgetIndex(execute=@n14.else, self=@n15, Index=1)
   n17 = SetPageIndexAndUpdate!() -> (OutputDelegate, then, NewIndex:int) {is=title K2Node_CustomEvent}
   > n18 = Set PageIndex(PageIndex="Clamp (Integer)"(Value=@n17.NewIndex, Min=Get PageInitial() -> (PageInitial=0), Max=GetLastUnlockablePage() -> (LastUnlockablePage=0)) -> (ReturnValue=0)) -> (Output_Get=0)
   > SetTotalText(ActualPage=@n18)
   > UpdateItemCards!(execute, self, ItemCards:Array of Season Pass Reward VM Object References) -> (then)
   n19 = Get ItemTileView()
   > SetSelectedIndex!(execute, self:List View Object Reference=@n19, Index:int=0) -> (then)
   > NavigateToIndex!(execute, self:List View Object Reference=@n19, Index:int=0) -> (then)
   > Set SelectedReward!(
    execute
    SelectedReward:Season Pass Reward VM Object Reference="Get (a copy)"(Array:Array of Season Pass Reward VM Object References=Get ItemList()) -> (Output:Season Pass Reward VM Object Reference)
    self:Season Pass Selection VM Object Reference=Get SeasonPassSelectionVM() -> (SeasonPassSelectionVM:Season Pass Selection VM Object Reference)
   ) -> (then, Output_Get:Season Pass Reward VM Object Reference)
   Event Construct()
   > SetMobileBackToTop()
   Event OnInitialized()
   > SetMobileStyle() {is=title K2Node_CallFunction}
   > n20 = PlayAnimation(InAnimation=Get Opening() -> (Opening:Widget Animation Object Reference))
   Call Quick Close Claim Pages Dispatcher!(execute:Exec=@n9.then, self:Season Pass Claim Pages Object Reference) -> (then:Exec) {is=title K2Node_CallDelegate}
   UpdateItemCards(execute=@n7.then)
  ]}
 FinishClaimEventDispatcher {nodes [FinishClaimEventDispatcher() {is=title K2Node_FunctionEntry}]}
 GetAnalyticsInfo {nodes [
   GetAnalyticsInfo() {is=title K2Node_FunctionEntry}
   n1 = GetDisplayName!(self:Season Pass VM Object Reference=Get SeasonPassVM()) -> (ReturnValue:Text)
   > Return Node!(
    execute
    ReturnValue:Flow Analytics Additional Info Structure="Add String (Non-Localized)"!(
     Info:"Flow Analytics Additional Info Structure (by ref)"=Make Flow Analytics Additional Info!() -> (FlowAnalyticsAdditionalInfo:Flow Analytics Additional Info Structure) {is=title K2Node_MakeStruct}
     Key:String=PassName
     Value:String=Get Source String from Text!(Text:Text=@n1) -> (ReturnValue:String) {is=title K2Node_CallFunction}
    ) -> (ReturnValue:Flow Analytics Additional Info Structure)
   )
  ]}
 GetFirstPurchaseablePage {nodes [
   GetFirstPurchaseablePage() {is=title K2Node_FunctionEntry}
   n1 = Get SeasonPassCategoryVM() -> (SeasonPassCategoryVM:Season Pass Category VM Object Reference)
   > n2 = Is Valid(InputObject=@n1)
   Return Node!(execute=[@n3.Completed, @n2.Is Not Valid], Page:int=-1)
   > n3 = For Each Loop with Break!(
    Exec:Exec
    Array:Array of Season Pass Page VM Object References=Get Pages!(self:Season Pass Category VM Object Reference=@n1) -> (Pages:Array of Season Pass Page VM Object References)
    Break:Exec
   ) -> (LoopBody:Exec, Array Element:Season Pass Page VM Object Reference, Array Index:int, Completed:Exec)
   > n4 = For Each Loop with Break(Array:Array of Season Pass Track VM Object References=Get Tracks!(self:Season Pass Page VM Object Reference=@n3) -> (Tracks:Array of Season Pass Track VM Object References)) -> (Array Element:Season Pass Track VM Object Reference)
   > n5 = For Each Loop with Break(Array:Array of Season Pass Reward VM Object References=Get Rewards!(self:Season Pass Track VM Object Reference=@n4) -> (Rewards:Array of Season Pass Reward VM Object References)) -> (Array Element:Season Pass Reward VM Object Reference)
   > n6 = Branch(Condition=Get bIsRewardOwned!(self:Season Pass Reward VM Object Reference=@n5) -> (bIsRewardOwned:bool))
   Return Node(execute=@n6.else, Page=@n3.Array Index)
  ]}
 GetLastUnlockablePage {nodes [
   GetLastUnlockablePage() {is=title K2Node_FunctionEntry}
   n1 = Get SeasonPassCategoryVM()
   > n2 = Is Valid(InputObject=@n1)
   Return Node!(execute=@n2.Is Not Valid, LastUnlockablePage:int=-1)
   > n3 = For Each Loop with Break(Array:Array of Season Pass Page VM Object References=Get Pages(self=@n1)) -> (Array Element:Season Pass Page VM Object Reference)
   n4 = Get Pages(self=@n1)
   Return Node(execute=@n3.Completed, LastUnlockablePage=int - int(A=Length(TargetArray:Array of Season Pass Page VM Object References=@n4)))
   > Branch(Condition=Get bIsTimeLocked!(self:Season Pass Element Base VM Object Reference=@n3) -> (bIsTimeLocked:bool))
   > Return Node(LastUnlockablePage=int - int(A=@n3.Array Index))
  ]}
 NeedsButtonOnPassCategory {nodes [
   NeedsButtonOnPassCategory() {is=title K2Node_FunctionEntry}
   n1 = Get SeasonPassCategoryVM()
   > n2 = Is Valid(InputObject=@n1)
   Return Node!(execute=@n2.Is Not Valid, Value:bool=false)
   > n3 = For Each Loop with Break(Array=Get Pages(self=@n1))
   Return Node(execute=@n3.Completed, Value=false)
   > n4 = For Each Loop with Break(Array:Array of Season Pass Track VM Object References=Get Tracks(self=@n3)) -> (Array Element:Season Pass Track VM Object Reference)
   > n5 = For Each Loop with Break(Array:Array of Season Pass Reward VM Object References=Get Rewards(self=@n4)) -> (Array Element:Season Pass Reward VM Object Reference)
   > n6 = Branch(Condition=Get bIsRewardOwned(self=@n5))
   n7 = Branch(
    execute=@n6.else
    Condition=OR Boolean(
     A=Get bIsQuestDriven!(self:Season Pass Reward VM Object Reference=@n5) -> (bIsQuestDriven:bool)
     B=IsTimeLocked!(self:Season Pass Element Base VM Object Reference=@n5) -> (ReturnValue:bool)
    )
   )
   > Return Node(Value=false)
   Return Node(execute=@n7.else, Value=true)
  ]}
 QuickCloseClaimPagesDispatcher {nodes [QuickCloseClaimPagesDispatcher() {is=title K2Node_FunctionEntry}]}
 SetArrowsVisibility {nodes [
   SetArrowsVisibility() {is=title K2Node_FunctionEntry}
   > SetIsEnabled(
    self=Get Button_Subtract_1() -> (Button_Subtract_1:WBP UIKit Button Regular Object Reference)
    bInIsEnabled="Not Equal (Integer)"(A=Get PageIndex() -> (PageIndex=0), B=Get PageInitial() -> (PageInitial=0))
   )
   > SetIsEnabled(
    self=Get Button_Add_1() -> (Button_Add_1:WBP UIKit Button Regular Object Reference)
    bInIsEnabled="Not Equal (Integer)"(A=GetLastUnlockablePage() -> (LastUnlockablePage=0) {is=title K2Node_CallFunction}, B=Get PageIndex() -> (PageIndex=0))
   )
  ]}
 SetBlackBackplateCorner {nodes [
   SetBlackBackplateCorner() {is=title K2Node_FunctionEntry}
   > n1 = GetDynamicMaterial(self=Get Backplate!(self:WBP UIKit Backplate Base Object Reference=Get Backplate_Black() -> (Backplate_Black:WBP UIKit Backplate Object Reference)) -> (Backplate:Image Object Reference))
   > SetVectorParameterValue(
    self=@n1
    ParameterName=Corner Radius
    Value=Make Linear Color!(R:float=0, G:float=20, B:float=0, A:float=20) -> (LinearColor:Linear Color Structure) {is=title K2Node_MakeStruct}
   )
  ]}
 SetMobileBackToTop {nodes [
   SetMobileBackToTop() {is=title K2Node_FunctionEntry}
   n1 = Get WBP_MPLocker_BackToTopButton() -> (WBP_MPLocker_BackToTopButton:WBP MPLocker Back To Top Button Object Reference)
   > SetPadding!(
    execute
    self:Size Box Slot Object Reference=SlotAsSizeBoxSlot!(Widget:Widget Object Reference=@n1) -> (ReturnValue:Size Box Slot Object Reference)
    InPadding_Left:float=4
    InPadding_Top:float=0
    InPadding_Right:float=0
    InPadding_Bottom:float=24
   ) -> (then)
   n2 = Get WBP_UIKit_Button_BackToTop_Touch!(self:WBP MPLocker Back To Top Button Object Reference=@n1) -> (WBP_UIKit_Button_BackToTop_Touch:WBP UIKit Button Quiet Object Reference)
   > SetVisibility(self=Get Texts!(self:WBP UIKit Button CTA Base Object Reference=@n2) -> (Texts:Vertical Box Object Reference), InVisibility=Collapsed)
   > SetIconMaterial!(
    execute
    self:WBP UIKit Button CTA Base Object Reference=@n2
    IconMaterial:Material Instance Object Reference=/FortUILibrary/UIKit/Buttons/MaterialInstances/GenericButtons/MI_GenericIconDeepBlue
   ) -> (then)
   > SetIcon!(execute, self:WBP UIKit Button CTA Base Object Reference=@n2, Icon:Object Reference=/Game/UI/CodeReferenced/T_UI_DirectionArrow) -> (then)
   > Set SmallIconSize!(execute, SmallIconSize:Vector 2D Structure=(X=32,Y=32), self:WBP UIKit Button CTA Base Object Reference=@n2) -> (then, Output_Get:Vector 2D Structure)
   > Set MobileSmallHeight!(execute, MobileSmallHeight:int=120, self:WBP UIKit Button CTA Base Object Reference=@n2) -> (then, Output_Get:int=0)
   > Set!(
    execute
    IconPadding_Left:float=-16
    IconPadding_Top:float=4
    IconPadding_Right:float=-16
    IconPadding_Bottom:float=16
    self:WBP UIKit Button CTA Base Object Reference=@n2
   ) -> (then, Output_Get:Margin Structure)
  ]}
 SetMobileStyle {nodes [
   SetMobileStyle() {is=title K2Node_FunctionEntry}
   > Branch(Condition=IsMobileGame())
   > SetMobileBackToTop() {is=title K2Node_CallFunction}
   n1 = Get Button_Subtract_1()
   > SetPadding(
    self:Grid Slot Object Reference=SlotAsGridSlot(Widget=Get Texts(self=@n1))
    InPadding_Left=0
    InPadding_Top=0
    InPadding_Right=0
    InPadding_Bottom=4
   )
   n2 = Get Button_Add_1()
   > SetPadding(self=SlotAsGridSlot(Widget=Get Texts(self=@n2)), InPadding_Left=0, InPadding_Top=0, InPadding_Right=0, InPadding_Bottom=0)
   > Set VerticalAlignment!(
    execute
    VerticalAlignment:EVerticalAlignment Enum=VAlign_Center
    self:Horizontal Box Slot Object Reference=SlotAsHorizontalBoxSlot!(Widget:Widget Object Reference=Get VerticalBox_Header() -> (VerticalBox_Header:Vertical Box Object Reference)) -> (ReturnValue:Horizontal Box Slot Object Reference)
   ) -> (then, Output_Get:EVerticalAlignment Enum=VAlign_Fill)
  ]}
 SetPlusMinusFontSize {nodes [
   SetPlusMinusFontSize() {is=title K2Node_FunctionEntry}
   > n1 = Sequence()
   n2 = Get TextBlock!(self:WBP UIKit Block Text Object Reference=Get Block_Text!(self:WBP UIKit Button CTA Base Object Reference=Get Button_Add_1()) -> (Block_Text:WBP UIKit Block Text Object Reference)) -> (TextBlock:Common Text Object Reference)
   n3 = Break Slate Font Info!(SlateFontInfo:"Slate Font Info Structure (by ref)"=Get Font!(self:Text Object Reference=@n2) -> (Font:Slate Font Info Structure)) -> (
    FontObject:Object Reference
    FontMaterial:Object Reference
    OutlineSettings:Font Outline Settings Structure
    TypefaceFontName:Name
    Size:float=~
    LetterSpacing:int
    SkewAmount:float=~
    bForceMonospaced:bool=~
    bMaterialIsStencil:bool=~
    MonospacedWidth:float=~
   ) {is=title K2Node_BreakStruct}
   > SetFont!(
    execute
    self=@n2
    InFontInfo_FontObject:Object Reference=@n3
    InFontInfo_FontMaterial:Object Reference=@n3.FontMaterial
    InFontInfo_OutlineSettings:Font Outline Settings Structure=@n3.OutlineSettings
    InFontInfo_TypefaceFontName:Name=@n3.TypefaceFontName
    InFontInfo_Size:float=40
    InFontInfo_LetterSpacing:int=@n3.LetterSpacing
    InFontInfo_SkewAmount:float=@n3.SkewAmount
    InFontInfo_bForceMonospaced:bool=@n3.bForceMonospaced
    InFontInfo_bMaterialIsStencil:bool=@n3.bMaterialIsStencil
    InFontInfo_MonospacedWidth:float=@n3.MonospacedWidth
   ) -> (then)
   n4 = Get TextBlock(self=Get Block_Text(self=Get Button_Subtract_1()))
   n5 = Break Slate Font Info(SlateFontInfo=Get Font(self=@n4)) -> (Size=~, SkewAmount=~, bForceMonospaced=~, bMaterialIsStencil=~, MonospacedWidth=~)
   SetFont(
    execute=@n1.then_1
    self=@n4
    InFontInfo_FontObject=@n5
    InFontInfo_FontMaterial=@n5.FontMaterial
    InFontInfo_OutlineSettings=@n5.OutlineSettings
    InFontInfo_TypefaceFontName=@n5.TypefaceFontName
    InFontInfo_Size=40
    InFontInfo_LetterSpacing=@n5.LetterSpacing
    InFontInfo_SkewAmount=@n5.SkewAmount
    InFontInfo_bForceMonospaced=@n5.bForceMonospaced
    InFontInfo_bMaterialIsStencil=@n5.bMaterialIsStencil
    InFontInfo_MonospacedWidth=@n5.MonospacedWidth
   )
  ]}
 SetTotalText {nodes [
   n1 = SetTotalText!() -> (then, ActualPage:int) {is=title K2Node_FunctionEntry}
   n2 = Get PageInitial() -> (PageInitial=0)
   > n3 = Branch(Condition="Equal (Integer)"(A=@n1, B=@n2))
   n4 = int + int(A=Get ActualPage() -> (ActualPage:int=0))
   > Set Text(
    Text=Format Text!(Format="LOCTEXT(\"Page {page}\")", page:int=@n4) -> (Result)
    self=Get Text_Rewards() -> (Text_Rewards:Common Text Object Reference)
   )
   Set Text(
    execute=@n3.else
    Text=Format Text!(Format="LOCTEXT(\"Page {initialpage} - {page}\")", initialpage:int=int + int(A=@n2), page=@n4) -> (Result)
    self=Get Text_Rewards()
   )
  ]}
 SetupVisuals {nodes [
   SetupVisuals() {is=title K2Node_FunctionEntry}
   > SetBlackBackplateCorner() {is=title K2Node_CallFunction}
   > SetPlusMinusFontSize() {is=title K2Node_CallFunction}
  ]}
 UpdateItemCards {nodes [
   UpdateItemCards!() -> (then, ItemCards:Array of Season Pass Reward VM Object References) {is=title K2Node_FunctionEntry}
   > Clear(TargetArray:Array of Season Pass Reward VM Object References=Get ItemList())
   > n1 = For Each Loop with Break(Array:Array of Season Pass Page VM Object References=Get Pages(self=Get SeasonPassCategoryVM()), Break=@n11.then) -> (Array Element:Season Pass Page VM Object Reference)
   n2 = Sequence(execute=@n1.Completed)
   n3 = GetNumPurchaseableRewards!(self:Season Pass Element Base VM Object Reference=Get SeasonPassCategoryVM()) -> (ReturnValue:int=0)
   n4 = Get CountBuy() -> (CountBuy:int=0)
   > n5 = Branch(Condition="integer >= integer"(A=@n3, B=@n4))
   n6 = Get Status_ToClaim() -> (Status_ToClaim:WBP UIKit Highlight Object Reference)
   > SetText!(
    execute
    self:WBP UIKit Highlight Base Object Reference=@n6
    Text:Text=Format Text!(Format="LOCTEXT(\"{rewards} TO CLAIM\")", rewards:int=@n3) -> (Result)
   ) -> (then)
   > SetType!(execute, self:WBP UIKit Highlight Base Object Reference=@n6, Type:E_UI_HighlightType Enum=NewEnumerator0) -> (then)
   n7 = Get HB_Buttons() -> (HB_Buttons:Common Visibility Switcher Object Reference)
   > SetActiveWidgetIndex(self=@n7, Index=0)
   n8 = Get FortPreloadAssetsVM() -> (FortPreloadAssetsVM:Fort Preload Assets VM Object Reference)
   > SetAssetsToPreload!(
    execute
    self:Fort Preload Assets VM Object Reference=@n8
    AssetsToPreload:Array of Soft Object References=Make Array!() -> (Array:Array of Soft Object References)
    bInstallSoftReferences:bool=false
   ) -> (then)
   n9 = int - int(A=@n4, B=@n3)
   Set NeedToClaim(execute=@n5.else, NeedToClaim:int=@n9) -> (Output_Get:int=0)
   > SetText(self=@n6, Text=Format Text(Format="LOCTEXT(\"NEED {rewards} MORE TO CLAIM PAGES\")", rewards=@n9))
   > SetType(self=@n6, Type=NewEnumerator3)
   > SetActiveWidgetIndex(self=@n7, Index=1)
   > SetAssetsToPreload(self=@n8, AssetsToPreload=Get PreloadAssets() -> (PreloadAssets:Array of Soft Object References), bInstallSoftReferences=false)
   n10 = Get CountBuy() -> (CountBuy=0)
   Set Text(
    execute=@n2.then_1
    Text=Format Text!(Format="LOCTEXT(\"{cost} rewards\")", cost:Text="To Text (Integer)"(Value=@n10)) -> (Result)
    self=Get Text_TotalRewardsValue() -> (Text_TotalRewardsValue:Common Text Object Reference)
   )
   > SetTotalText(ActualPage=Get PageIndex() -> (PageIndex=0)) {is=title K2Node_CallFunction}
   > Set List Items!(execute, self:List View Object Reference=Get ItemTileView(), InListItems:Array of Object References=Get ItemList()) -> (then) {is=title K2Node_CallFunction}
   > SetArrowsVisibility() {is=title K2Node_CallFunction}
   Branch(execute=@n1.LoopBody, Condition="integer >= integer"(A=@n1.Array Index, B=Get PageInitial() -> (PageInitial=0)))
   > n11 = Branch(Condition="integer > integer"(A=@n1.Array Index, B=Get PageIndex() -> (PageIndex=0)))
   n12 = For Each Loop with Break(Exec=@n11.else, Array:Array of Season Pass Track VM Object References=Get Tracks(self=@n1)) -> (Array Element:Season Pass Track VM Object Reference)
   > n13 = For Each Loop with Break(Array:Array of Season Pass Reward VM Object References=Get Rewards(self=@n12)) -> (Array Element:Season Pass Reward VM Object Reference)
   > n14 = Branch(Condition=Get bIsRewardOwned(self=@n13))
   Set CountBuy(execute=@n14.else, CountBuy:int=int + int(A=1, B=Get CountBuy() -> (CountBuy=0))) -> (Output_Get:int=0)
   > Add(TargetArray:Array of Season Pass Reward VM Object References=Get ItemList(), NewItem:"Season Pass Reward VM Object Reference (by ref)"=@n13) {is=title K2Node_CallArrayFunction}
   > Add(
    TargetArray:Array of Soft Object References=Get PreloadAssets()
    NewItem:"Soft Object Reference (by ref)"=GetButtonIcon!(self:Season Pass Reward VM Object Reference=@n13) -> (ReturnValue:Soft Object Reference)
   )
  ]}
 VM_PreviewItemChange {nodes [
   n1 = VM_PreviewItemChange!() -> (then, Preview Item VM:Fort Item VM Object Reference) {is=title K2Node_FunctionEntry}
   > Is Valid(InputObject=@n1)
   > Set FortItemDetailsVM!(
    execute
    FortItemDetailsVM:Fort Item Details VM Object Reference=Get ItemDetails!(self:Fort Item VM Object Reference=@n1) -> (ItemDetails:Fort Item Details VM Object Reference)
    self:WBP UIKit Description With Subtitle Object Reference=Get WBP_UIKit_DescriptionWithSubtitle() -> (WBP_UIKit_DescriptionWithSubtitle:WBP UIKit Description With Subtitle Object Reference)
   ) -> (then, Output_Get:Fort Item Details VM Object Reference)
  ]}
]
variables [
 FinishClaimEventDispatcher {}
 ItemList {type="array<SeasonPassRewardVM>", default []}
 NeedToClaim {type=integer}
 PageIndex {type=integer}
 PageInitial {type=integer}
 QuickCloseClaimPagesDispatcher {}
]
widgets {
 info {inheritedWidgetCount=0, parentClass=&CommonActivatableWidget, rootWidgetClass=&SizeBox, widgetCount=32}
 widgets [
  SizeBox_0 {widgetClassPath=&SizeBox, widget=SizeBox_0, properties {bOverride_WidthOverride=true, widthOverride=760}}
  VisibilitySwitcherIsClaimed {
   widgetClassPath=&CommonVisibilitySwitcher
   bIsVariable=true
   parent=SizeBox_0
   widget=VisibilitySwitcherIsClaimed
   slot {class=SizeBoxSlot, properties {horizontalAlignment=HAlign_Center}}
  }
  Overlay_1 {
   widgetClassPath=&Overlay
   parent=VisibilitySwitcherIsClaimed
   widget=Overlay_1
   slot {class=CommonVisibilitySwitcherSlot, properties {horizontalAlignment=HAlign_Center}}
  }
  Backplate_Dark {
   widgetClassPath=&WBP_UIKit_Backplate
   bIsVariable=true
   parent=Overlay_1
   widget=Backplate_Dark
   properties {brightness=Dark, cornerRadius=Interactive Component}
   slot {class=OverlaySlot, properties {horizontalAlignment=HAlign_Fill, verticalAlignment=VAlign_Fill}}
  }
  VB_All {
   widgetClassPath=&VerticalBox
   parent=Overlay_1
   widget=VB_All
   slot {class=OverlaySlot, properties {verticalAlignment=VAlign_Fill}}
  }
  SizeBox_88 {
   widgetClassPath=&SizeBox
   parent=VB_All
   widget=SizeBox_88
   properties {bOverride_WidthOverride=true, widthOverride=760}
   slot {class=VerticalBoxSlot, properties {padding=<ltrb 0 0 0 16>, size {sizeRule=Fill, value=1}}}
  }
  Overlay_4 {widgetClassPath=&Overlay, parent=SizeBox_88, widget=Overlay_4, slot {class=SizeBoxSlot}}
  VerticalBox_71 {
   widgetClassPath=&VerticalBox
   parent=Overlay_4
   widget=VerticalBox_71
   slot {class=OverlaySlot, properties {horizontalAlignment=HAlign_Fill, verticalAlignment=VAlign_Fill, padding=<ltrb 0 16 0 0>}}
  }
  Controllers {
   widgetClassPath=&HorizontalBox
   parent=VerticalBox_71
   widget=Controllers
   slot {class=VerticalBoxSlot, properties {padding=<ltrb 0 0 0 16>}}
  }
  Button_Subtract_1 {
   widgetClassPath=&WBP_UIKit_Button_Regular
   bIsVariable=true
   parent=Controllers
   widget=Button_Subtract_1
   properties {
    bIsFocusable=false
    minHeight=48
    minWidth=112
    mobileMinHeight=90
    mobileMinWidth=90
    overrideDefaultMobileFontSizeMultiplier=true
    text=-
    visibility=Visible
    triggeringInputAction {dataTable=&InputActionDataTable, rowName=GenericLeftShoulder}
   }
   slot {class=HorizontalBoxSlot, properties {horizontalAlignment=HAlign_Left, padding=<ltrb 16 0 0 0>}}
  }
  VerticalBox_Header {
   widgetClassPath=&VerticalBox
   bIsVariable=true
   parent=Controllers
   widget=VerticalBox_Header
   properties {bOverride_Cursor=true}
   slot {class=HorizontalBoxSlot, properties {verticalAlignment=VAlign_Top, padding=<ltrb 16 0 16 0>, size {sizeRule=Fill, value=1}}}
  }
  Text_Header {
   widgetClassPath=&CommonTextBlock
   bIsVariable=true
   parent=VerticalBox_Header
   widget=Text_Header
   properties {
    autoWrapText=true
    justification=Center
    text=Claim all rewards from
    textTransformPolicy=ToUpper
    colorAndOpacity {colorUseRule=UseColor_Specified, specifiedColor=<rgba 1 1 1 0.6>}
    font {
     bForceMonospaced=false
     bMaterialIsStencil=false
     fontMaterial=null
     fontObject=&HeadingNow
     letterSpacing=0
     monospacedWidth=1
     size=25.5
     skewAmount=0
     typefaceFontName=86 Bold Italic
     outlineSettings {
      bApplyOutlineToDropShadows=false
      bMiteredCorners=false
      bSeparateFillAlpha=false
      outlineMaterial=null
      outlineSize=0
      outlineColor=<rgba 0 0 0 1>
     }
    }
    shadowOffset=<0 0>
   }
   slot {class=VerticalBoxSlot}
  }
  Text_Rewards {
   widgetClassPath=&CommonTextBlock
   bIsVariable=true
   parent=VerticalBox_Header
   widget=Text_Rewards
   properties {
    autoWrapText=true
    justification=Center
    text=Page 1-2
    textTransformPolicy=ToUpper
    font {
     bForceMonospaced=false
     bMaterialIsStencil=false
     fontMaterial=null
     fontObject=&HeadingNow
     letterSpacing=0
     monospacedWidth=1
     size=25.5
     skewAmount=0
     typefaceFontName=86 Bold Italic
     outlineSettings {
      bApplyOutlineToDropShadows=false
      bMiteredCorners=false
      bSeparateFillAlpha=false
      outlineMaterial=null
      outlineSize=0
      outlineColor=<rgba 0 0 0 1>
     }
    }
    shadowOffset=<0 0>
   }
   slot {class=VerticalBoxSlot}
  }
  Button_Add_1 {
   widgetClassPath=&WBP_UIKit_Button_Regular
   bIsVariable=true
   parent=Controllers
   widget=Button_Add_1
   properties {
    bIsFocusable=false
    iconBeforeText=false
    minHeight=48
    minWidth=112
    mobileMinHeight=90
    mobileMinWidth=90
    overrideDefaultMobileFontSizeMultiplier=true
    overrideSecondaryTextSize=true
    overrideSecondaryTextSizeMultiplier=2
    text=+
    visibility=Visible
    triggeringInputAction {dataTable=&InputActionDataTable, rowName=GenericRightShoulder}
   }
   slot {class=HorizontalBoxSlot, properties {horizontalAlignment=HAlign_Right, padding=<ltrb 0 0 16 0>}}
  }
  Overlay_0 {
   widgetClassPath=&Overlay
   parent=VerticalBox_71
   widget=Overlay_0
   properties {clipping=ClipToBounds}
   slot {class=VerticalBoxSlot, properties {padding=<ltrb -16 0 -8 0>, size {sizeRule=Fill, value=1}}}
  }
  ItemTileView {
   widgetClassPath=/Script/FortniteUI.FortItemTileView
   bIsVariable=true
   parent=Overlay_0
   widget=ItemTileView
   properties {
    bEntrySizeIncludesEntrySpacing=false
    clipping=Inherit
    entryHeight=160
    entryWidgetClass=&/SeasonPassClaimPagesReward.SeasonPassClaimPagesReward_C
    horizontalEntrySpacing=16
    navigation=&self:WidgetTree.ItemTileView.WidgetNavigation_1
    numDesignerPreviewEntries=12
    scrollIntoViewAlignment=TopOrLeft
    tileAlignment=LeftAligned
    verticalEntrySpacing=16
    scrollBarStyle {
     thickness=10
     draggedThumbImage {
      bIsDynamicallyLoaded=false
      drawAs=Box
      imageType=FullColor
      mirroring=NoMirror
      resourceName=../../../Engine/Content/Slate/Common/Scrollbar_Thumb.png
      resourceObject=/FortUILibrary/Foundation/Art_ToolKit/Tools/M_UI_Circle_White
      tiling=NoTile
      imageSize=<10 10>
      margin=<ltrb 0.5 0.5 0.5 0.5>
      outlineSettings {
       bUseBrushTransparency=false
       roundingType=HalfHeightRadius
       width=0
       color {colorUseRule=UseColor_Specified, specifiedColor=<rgba 0 0 0 0>}
       cornerRadii=<0 0 0 0>
      }
      tintColor {colorUseRule=UseColor_Specified, specifiedColor=<rgba 1 1 1 1>}
      uVRegion {bIsValid=false, max=<0 0>, min=<0 0>}
     }
     horizontalBackgroundImage {
      bIsDynamicallyLoaded=false
      drawAs=NoDrawType
      imageType=NoImage
      mirroring=NoMirror
      resourceName=None
      resourceObject=null
      tiling=NoTile
      imageSize=<0 0>
      margin=<ltrb 0 0 0 0>
      outlineSettings {
       bUseBrushTransparency=false
       roundingType=HalfHeightRadius
       width=0
       color {colorUseRule=UseColor_Specified, specifiedColor=<rgba 0 0 0 0>}
       cornerRadii=<0 0 0 0>
      }
      tintColor {colorUseRule=UseColor_Specified, specifiedColor=<rgba 1 1 1 1>}
      uVRegion {bIsValid=false, max=<0 0>, min=<0 0>}
     }
     horizontalBottomSlotImage {
      bIsDynamicallyLoaded=false
      drawAs=Image
      imageType=FullColor
      mirroring=NoMirror
      resourceName=../../../Engine/Content/Slate/Common/Scrollbar_Background_Horizontal.png
      resourceObject=null
      tiling=NoTile
      imageSize=<8 8>
      margin=<ltrb 0 0 0 0>
      outlineSettings {
       bUseBrushTransparency=false
       roundingType=HalfHeightRadius
       width=0
       color {colorUseRule=UseColor_Specified, specifiedColor=<rgba 0 0 0 0>}
       cornerRadii=<0 0 0 0>
      }
      tintColor {colorUseRule=UseColor_Specified, specifiedColor=<rgba 1 1 1 1>}
      uVRegion {bIsValid=false, max=<0 0>, min=<0 0>}
     }
     horizontalTopSlotImage {
      bIsDynamicallyLoaded=false
      drawAs=Image
      imageType=FullColor
      mirroring=NoMirror
      resourceName=../../../Engine/Content/Slate/Common/Scrollbar_Background_Horizontal.png
      resourceObject=null
      tiling=NoTile
      imageSize=<8 8>
      margin=<ltrb 0 0 0 0>
      outlineSettings {
       bUseBrushTransparency=false
       roundingType=HalfHeightRadius
       width=0
       color {colorUseRule=UseColor_Specified, specifiedColor=<rgba 0 0 0 0>}
       cornerRadii=<0 0 0 0>
      }
      tintColor {colorUseRule=UseColor_Specified, specifiedColor=<rgba 1 1 1 1>}
      uVRegion {bIsValid=false, max=<0 0>, min=<0 0>}
     }
     hoveredThumbImage {
      bIsDynamicallyLoaded=false
      drawAs=Box
      imageType=FullColor
      mirroring=NoMirror
      resourceName=../../../Engine/Content/Slate/Common/Scrollbar_Thumb.png
      resourceObject=&M_UI_Circle_White
      tiling=NoTile
      imageSize=<10 10>
      margin=<ltrb 0.5 0.5 0.5 0.5>
      outlineSettings {
       bUseBrushTransparency=false
       roundingType=HalfHeightRadius
       width=0
       color {colorUseRule=UseColor_Specified, specifiedColor=<rgba 0 0 0 0>}
       cornerRadii=<0 0 0 0>
      }
      tintColor {colorUseRule=UseColor_Specified, specifiedColor=<rgba 1 1 1 1>}
      uVRegion {bIsValid=false, max=<0 0>, min=<0 0>}
     }
     normalThumbImage {
      bIsDynamicallyLoaded=false
      drawAs=Box
      imageType=FullColor
      mirroring=NoMirror
      resourceName=../../../Engine/Content/Slate/Common/Scrollbar_Thumb.png
      resourceObject=&M_UI_Circle_White
      tiling=NoTile
      imageSize=<10 10>
      margin=<ltrb 0.5 0.5 0.5 0.5>
      outlineSettings {
       bUseBrushTransparency=false
       roundingType=HalfHeightRadius
       width=0
       color {colorUseRule=UseColor_Specified, specifiedColor=<rgba 0 0 0 0>}
       cornerRadii=<0 0 0 0>
      }
      tintColor {colorUseRule=UseColor_Specified, specifiedColor=<rgba 1 1 1 1>}
      uVRegion {bIsValid=false, max=<0 0>, min=<0 0>}
     }
     verticalBackgroundImage {
      bIsDynamicallyLoaded=false
      drawAs=Box
      imageType=NoImage
      mirroring=NoMirror
      resourceName=None
      resourceObject=&M_UI_Circle_White
      tiling=NoTile
      imageSize=<10 10>
      margin=<ltrb 0.5 0.5 0.5 0.5>
      outlineSettings {
       bUseBrushTransparency=false
       roundingType=HalfHeightRadius
       width=0
       color {colorUseRule=UseColor_Specified, specifiedColor=<rgba 0 0 0 0>}
       cornerRadii=<0 0 0 0>
      }
      tintColor {colorUseRule=UseColor_Specified, specifiedColor=<rgba 1 1 1 0.2>}
      uVRegion {bIsValid=false, max=<0 0>, min=<0 0>}
     }
     verticalBottomSlotImage {
      bIsDynamicallyLoaded=false
      drawAs=NoDrawType
      imageType=FullColor
      mirroring=NoMirror
      resourceName=../../../Engine/Content/Slate/Common/Scrollbar_Background_Vertical.png
      resourceObject=null
      tiling=NoTile
      imageSize=<8 8>
      margin=<ltrb 0 0 0 0>
      outlineSettings {
       bUseBrushTransparency=false
       roundingType=HalfHeightRadius
       width=0
       color {colorUseRule=UseColor_Specified, specifiedColor=<rgba 0 0 0 0>}
       cornerRadii=<0 0 0 0>
      }
      tintColor {colorUseRule=UseColor_Specified, specifiedColor=<rgba 1 1 1 1>}
      uVRegion {bIsValid=false, max=<0 0>, min=<0 0>}
     }
     verticalTopSlotImage {
      bIsDynamicallyLoaded=false
      drawAs=NoDrawType
      imageType=FullColor
      mirroring=NoMirror
      resourceName=../../../Engine/Content/Slate/Common/Scrollbar_Background_Vertical.png
      resourceObject=null
      tiling=NoTile
      imageSize=<8 8>
      margin=<ltrb 0 0 0 0>
      outlineSettings {
       bUseBrushTransparency=false
       roundingType=HalfHeightRadius
       width=0
       color {colorUseRule=UseColor_Specified, specifiedColor=<rgba 0 0 0 0>}
       cornerRadii=<0 0 0 0>
      }
      tintColor {colorUseRule=UseColor_Specified, specifiedColor=<rgba 1 1 1 1>}
      uVRegion {bIsValid=false, max=<0 0>, min=<0 0>}
     }
    }
    widgetStyle {backgroundBrush {
      bIsDynamicallyLoaded=false
      drawAs=Box
      imageType=NoImage
      mirroring=NoMirror
      resourceName=None
      resourceObject=&M_UI_Circle_White
      tiling=NoTile
      imageSize=<10 10>
      margin=<ltrb 0.5 0.5 0.5 0.5>
      outlineSettings {
       bUseBrushTransparency=false
       roundingType=HalfHeightRadius
       width=0
       color {colorUseRule=UseColor_Specified, specifiedColor=<rgba 0 0 0 0>}
       cornerRadii=<0 0 0 0>
      }
      tintColor {colorUseRule=UseColor_Specified, specifiedColor=<rgba 1 1 1 0.2>}
      uVRegion {bIsValid=false, max=<0 0>, min=<0 0>}
     }}
   }
   slot {class=OverlaySlot, properties {horizontalAlignment=HAlign_Fill, verticalAlignment=VAlign_Fill, padding=<ltrb 24 16 16 16>}}
  }
  FortVisualAttachment_1 {
   widgetClassPath=&FortVisualAttachment
   parent=Overlay_4
   widget=FortVisualAttachment_1
   properties {contentAnchor=<0 1>}
   slot {class=OverlaySlot, properties {horizontalAlignment=HAlign_Right, verticalAlignment=VAlign_Bottom}}
  }
  WBP_MPLocker_BackToTopButton {
   widgetClassPath=/Locker/MPL/WBP_MPLocker_BackToTopButton
   bIsVariable=true
   parent=FortVisualAttachment_1
   widget=WBP_MPLocker_BackToTopButton
   properties {targetList=&self:WidgetTree.ItemTileView}
   slot {class=SizeBoxSlot, properties {horizontalAlignment=HAlign_Center, padding=<ltrb 8 0 0 20>}}
  }
  Overlay_2 {widgetClassPath=&Overlay, parent=VB_All, widget=Overlay_2, slot {class=VerticalBoxSlot}}
  Backplate_Black {
   widgetClassPath=&WBP_UIKit_Backplate
   bIsVariable=true
   parent=Overlay_2
   widget=Backplate_Black
   properties {backplateMaterial=/SeasonPass/Materials/Backplates/MI_UIKit_ClaimPagesBackplate, brightness=Obscure}
   slot {class=OverlaySlot, properties {horizontalAlignment=HAlign_Fill, verticalAlignment=VAlign_Fill}}
  }
  VerticalBox_1 {
   widgetClassPath=&VerticalBox
   parent=Overlay_2
   widget=VerticalBox_1
   slot {class=OverlaySlot, properties {horizontalAlignment=HAlign_Fill, padding=<ltrb 16 0 16 16>}}
  }
  Status_ToClaim {
   widgetClassPath=/FortUILibrary/UIKit/Highlight/WBP_UIKit_Highlight
   bIsVariable=true
   parent=VerticalBox_1
   widget=Status_ToClaim
   properties {highlightText=2 to Claim, highlightType=Attention}
   slot {class=VerticalBoxSlot, properties {horizontalAlignment=HAlign_Center, verticalAlignment=VAlign_Center, padding=<ltrb 0 16 0 0>}}
  }
  HorizontalBox_1 {
   widgetClassPath=&HorizontalBox
   parent=VerticalBox_1
   widget=HorizontalBox_1
   slot {class=VerticalBoxSlot, properties {horizontalAlignment=HAlign_Center, padding=<ltrb 0 24 0 0>}}
  }
  Text_TotalRewardsLabel {
   widgetClassPath=&CommonTextBlock
   bIsVariable=true
   parent=HorizontalBox_1
   widget=Text_TotalRewardsLabel
   properties {
    justification=Center
    scrollStyle=&TextScrollStyle-Base_C
    text=Total cost:
    textTransformPolicy=ToUpper
    colorAndOpacity {colorUseRule=UseColor_Specified, specifiedColor=<rgba 1 1 1 0.6>}
    font {
     bForceMonospaced=false
     bMaterialIsStencil=false
     fontMaterial=null
     fontObject=&HeadingNow
     letterSpacing=0
     monospacedWidth=1
     size=25.5
     skewAmount=0
     typefaceFontName=86 Bold Italic
     outlineSettings {
      bApplyOutlineToDropShadows=false
      bMiteredCorners=false
      bSeparateFillAlpha=false
      outlineMaterial=null
      outlineSize=0
      outlineColor=<rgba 0 0 0 1>
     }
    }
    shadowOffset=<0 0>
   }
   slot {class=HorizontalBoxSlot, properties {padding=<ltrb 0 0 8 0>}}
  }
  Text_TotalRewardsValue {
   widgetClassPath=&CommonTextBlock
   bIsVariable=true
   parent=HorizontalBox_1
   widget=Text_TotalRewardsValue
   properties {
    justification=Center
    scrollStyle=&TextScrollStyle-Base_C
    text=18 rewards
    textTransformPolicy=ToUpper
    font {
     bForceMonospaced=false
     bMaterialIsStencil=false
     fontMaterial=null
     fontObject=&HeadingNow
     letterSpacing=0
     monospacedWidth=1
     size=25.5
     skewAmount=0
     typefaceFontName=86 Bold Italic
     outlineSettings {
      bApplyOutlineToDropShadows=false
      bMiteredCorners=false
      bSeparateFillAlpha=false
      outlineMaterial=null
      outlineSize=0
      outlineColor=<rgba 0 0 0 1>
     }
    }
    shadowOffset=<0 0>
   }
   slot {class=HorizontalBoxSlot}
  }
  HB_Buttons {
   widgetClassPath=&CommonVisibilitySwitcher
   bIsVariable=true
   parent=VerticalBox_1
   widget=HB_Buttons
   properties {bOverride_Cursor=true}
   slot {class=VerticalBoxSlot, properties {padding=<ltrb 0 12 0 0>}}
  }
  Button_ClaimPages {
   widgetClassPath=/FortUILibrary/UIKit/Buttons/WBP_UIKit_Button_Loud
   bIsVariable=true
   parent=HB_Buttons
   widget=Button_ClaimPages
   properties {
    bIsFocusable=false
    bRequiresHold=true
    holdData=/FortUILibrary/Input/PurchaseInteraction_HoldData.PurchaseInteraction_HoldData_C
    navigation=&self:WidgetTree.Button_ClaimPages.WidgetNavigation_1
    text=Claim Pages
    visibility=Visible
    triggeringInputAction {dataTable=&InputActionDataTable, rowName=Confirm}
   }
   slot {class=CommonVisibilitySwitcherSlot}
  }
  Button_BuyRewards {
   widgetClassPath=&WBP_UIKit_Button_Loud
   bIsVariable=true
   parent=HB_Buttons
   widget=Button_BuyRewards
   properties {
    bHideInputActionWithKeyboard=true
    bIsFocusable=false
    navigation=&self:WidgetTree.Button_BuyRewards.WidgetNavigation_1
    text=Buy Rewards
    visibility=Visible
    triggeringInputAction {dataTable=&InputActionDataTable, rowName=Confirm}
   }
   slot {class=CommonVisibilitySwitcherSlot}
  }
  FortVisualAttachment_0 {
   widgetClassPath=&FortVisualAttachment
   parent=VB_All
   widget=FortVisualAttachment_0
   properties {contentAnchor=<0 1>}
   slot {class=VerticalBoxSlot, properties {horizontalAlignment=HAlign_Right, verticalAlignment=VAlign_Bottom}}
  }
  SizeBox_1 {
   widgetClassPath=&SizeBox
   parent=FortVisualAttachment_0
   widget=SizeBox_1
   properties {bOverride_WidthOverride=true, widthOverride=800}
   slot {class=SizeBoxSlot, properties {padding=<ltrb 56 0 0 0>}}
  }
  WBP_UIKit_DescriptionWithSubtitle {
   widgetClassPath=/FortUILibrary/UIKit/Items/WBP_UIKit_DescriptionWithSubtitle
   bIsVariable=true
   parent=SizeBox_1
   widget=WBP_UIKit_DescriptionWithSubtitle
   properties {hideSubtitle=true}
   slot {class=SizeBoxSlot, properties {padding=<ltrb 0 0 0 16>}}
  }
  Text_AllRewardsClaimed {
   widgetClassPath=&CommonTextBlock
   bIsVariable=true
   parent=VisibilitySwitcherIsClaimed
   widget=Text_AllRewardsClaimed
   properties {
    autoWrapText=true
    justification=Center
    text=No rewards left to claim
    textTransformPolicy=ToUpper
    font {
     bForceMonospaced=false
     bMaterialIsStencil=false
     fontMaterial=null
     fontObject=&HeadingNow
     letterSpacing=0
     monospacedWidth=1
     size=25.5
     skewAmount=0
     typefaceFontName=86 Bold Italic
     outlineSettings {
      bApplyOutlineToDropShadows=false
      bMiteredCorners=false
      bSeparateFillAlpha=false
      outlineMaterial=null
      outlineSize=0
      outlineColor=<rgba 0 0 0 1>
     }
    }
    shadowOffset=<0 0>
   }
   slot {class=CommonVisibilitySwitcherSlot, properties {verticalAlignment=VAlign_Center}}
  }
 ]
}
