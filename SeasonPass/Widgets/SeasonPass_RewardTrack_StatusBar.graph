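# Status bar widget for the Season Pass reward track, showing a state badge and description text.
# Helper functions pick the badge by priority (level-granted, pass completed, page completed, locked or claimable count), testing the ReachLevel header bit.
# Level-granted rewards use the selected reward's view-model lock text; setters write page status text only while time-locked.
# PreConstruct applies compact sizing and hides the backplate and description when abridged.
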WidgetBlueprint /SeasonPass/Widgets/SeasonPass_RewardTrack_StatusBar
graphs [
 DisplayIsAbridged {nodes [
   DisplayIsAbridged() {is=title K2Node_FunctionEntry}
   > SetVisibility(
    self=[
     Get RespPanel_StatusArea() -> (RespPanel_StatusArea:Fort Responsive Panel Object Reference)
     Get WBP_UIKit_Backplate() -> (WBP_UIKit_Backplate:WBP UIKit Backplate Object Reference)
    ]
    InVisibility=Select(Option 0=SelfHitTestInvisible, Option 1=Collapsed, Index=Get IsAbridged() -> (IsAbridged:bool))
   )
  ]}
 EventGraph {nodes [
   Event PreConstruct()
   > SetCompactUIStyle() {is=title K2Node_CallFunction}
   > DisplayIsAbridged() {is=title K2Node_CallFunction}
  ]}
 HandleStatusText {nodes [
   n1 = HandleStatusText!() -> (then, PageText:Text, Rewardheader:int) {is=title K2Node_FunctionEntry}
   n2 = "To Integer (Byte)"!(InByte:Byte=Literal enum ESeasonPassRewardHeader!(Enum:ESeasonPassRewardHeader Enum=ReachLevel) -> (ReturnValue:ESeasonPassRewardHeader Enum) {is=title K2Node_EnumLiteral}) -> (ReturnValue:int=0)
   > n3 = Branch(Condition="Not Equal (Integer)"(A=Bitwise AND(A:int=@n1.Rewardheader, B:int=@n2) -> (ReturnValue:int=0) {is=title K2Node_CommutativeAssociativeBinaryOperator}))
   n4 = Get SeasonPassSelectionVM() -> (SeasonPassSelectionVM:Season Pass Selection VM Object Reference)
   > Return Node!(
    execute
    StatusText:Text=Get LockStatusText!(self:Season Pass Element Base VM Object Reference=Get SelectedReward!(self:Season Pass Selection VM Object Reference=@n4) -> (SelectedReward:Season Pass Reward VM Object Reference)) -> (LockStatusText:Text)
   )
   Return Node(execute=@n3.else, StatusText=Get PageText() -> (PageText:Text))
  ]}
 HandleTagActiveWidget {nodes [
   n1 = HandleTagActiveWidget!() -> (then, SelectedPageIsLocked:bool=~, IsPageCompleted:bool=~, IsPassCompleted:bool=~, RewardHeader:int) {is=title K2Node_FunctionEntry}
   n2 = "To Integer (Byte)"(InByte=Literal enum ESeasonPassRewardHeader(Enum=ReachLevel)) -> (ReturnValue=0)
   > n3 = Branch(Condition="Not Equal (Integer)"(A=Bitwise AND(A=@n1.RewardHeader, B=@n2) -> (ReturnValue=0)))
   > Return Node!(execute, Index:Widget Object Reference=Get GrantedByLevel() -> (GrantedByLevel:Overlay Object Reference))
   n4 = Branch(execute=@n3.else, Condition=Get IsPassCompleted() -> (IsPassCompleted:bool))
   > Return Node(Index=Get PassCompleted() -> (PassCompleted:Overlay Object Reference))
   n5 = Branch(execute=@n4.else, Condition=Get IsPageCompleted() -> (IsPageCompleted:bool))
   > Return Node(Index=Get PageCompleted() -> (PageCompleted:Overlay Object Reference))
   Return Node(
    execute=@n5.else
    Index=Select(
     Option 0:Widget Object Reference=Get RewardAmount() -> (RewardAmount:Overlay Object Reference)
     Option 1:Widget Object Reference=Get Locked() -> (Locked:Overlay Object Reference)
     Index=Get SelectedPageIsLocked() -> (SelectedPageIsLocked:bool)
    ) -> (ReturnValue:Widget Object Reference)
   )
  ]}
 HandleTagVisibility {nodes [
   HandleTagVisibility!() -> (then, SelectedPageIsLocked:bool=~, IsPageCompleted:bool=~, IsPassCompleted:bool=~, NumOfPurchaseableRewards:int, PassRewardHeader:int) {is=title K2Node_FunctionEntry}
   n1 = Literal enum ESeasonPassRewardHeader(Enum=ReachLevel)
   n2 = "Not Equal (Integer)"(A=Bitwise AND(A=Get PassRewardHeader() -> (PassRewardHeader:int=0), B="To Integer (Byte)"(InByte=@n1) -> (ReturnValue=0)) -> (ReturnValue=0))
   n3 = "integer > integer"(A=Get NumOfPurchaseableRewards() -> (NumOfPurchaseableRewards:int=0))
   n4 = Get IsPassCompleted()
   n5 = Get IsPageCompleted()
   n6 = Get SelectedPageIsLocked()
   > Return Node(Index:ESlateVisibility Enum=Select(
     Option 0:ESlateVisibility Enum=Collapsed
     Option 1:ESlateVisibility Enum=SelfHitTestInvisible
     Index=OR Boolean!(A=@n3, B=@n6, C:bool=@n5, D:bool=@n4, E:bool=@n2) -> (ReturnValue)
    ) -> (ReturnValue:ESlateVisibility Enum))
  ]}
 SetCanBePurchasedRewardCount {nodes [
   n1 = SetCanBePurchasedRewardCount!() -> (then, CanBeClaimed:int) {is=title K2Node_FunctionEntry}
   > "SetText (Text)"(
    self=Get Text_Highlight!(self:WBP UIKit Highlight Base Object Reference=Get RewardAmount_Label() -> (RewardAmount_Label:WBP UIKit Highlight Object Reference)) -> (Text_Highlight:Common Text Object Reference)
    InText=Format Text!(Format="LOCTEXT(\"{NumRewards} to claim\")", NumRewards:int=@n1) -> (Result)
   )
   > Set with Broadcast NoRewardsAvailable!(execute, NoRewardsAvailable:bool="Equal (Integer)"(A=@n1)) -> (then, Output_Get:bool)
   > UpdatePageLocked!(execute, self, Locked:bool=Get LockedPage() -> (LockedPage:bool)) -> (then)
  ]}
 SetCompactUIStyle {nodes [
   SetCompactUIStyle() {is=title K2Node_FunctionEntry}
   > Branch(Condition=IsUsingCompactUI())
   > SetSize!(execute, self:Spacer Object Reference=Get Spacer() -> (Spacer:Spacer Object Reference), InSize:Vector 2D Structure=(X=800,Y=40)) -> (then)
  ]}
 UpdatePageLocked {nodes [
   n1 = UpdatePageLocked!() -> (then, Locked:bool=~) {is=title K2Node_FunctionEntry}
   > Set LockedPage(LockedPage:bool=@n1) -> (Output_Get:bool)
   > UpdateStatusTextWidget()
  ]}
 UpdatePageStatusText {nodes [
   n1 = UpdatePageStatusText!() -> (then, Status:Text) {is=title K2Node_FunctionEntry}
   > Set PageStatusText(PageStatusText:Text=@n1) -> (Output_Get:Text)
   > UpdateStatusTextWidget()
  ]}
 UpdatePageTimeLocked {nodes [
   n1 = UpdatePageTimeLocked!() -> (then, Locked:bool=~) {is=title K2Node_FunctionEntry}
   > Set TimeLockedPage(TimeLockedPage:bool=@n1) -> (Output_Get:bool)
   > UpdateStatusTextWidget()
  ]}
 UpdateStatusTextWidget {nodes [
   UpdateStatusTextWidget() {is=title K2Node_FunctionEntry}
   > Branch(Condition=Get TimeLockedPage() -> (TimeLockedPage:bool))
   > "SetText (Text)"(
    self=Get Text_StatusDescription() -> (Text_StatusDescription:Common Text Object Reference)
    InText=Get PageStatusText() -> (PageStatusText:Text)
   )
  ]}
]
variables [
 IsAbridged {type=boolean}
 IsGrantedBylevel {type=boolean}
 LockedPage {type=boolean}
 NoRewardsAvailable {type=boolean}
 PageStatusText {type=string, default=""}
 TimeLockedPage {type=boolean}
]
widgets {
 info {inheritedWidgetCount=0, parentClass=&UserWidget, rootWidgetClass=&Overlay, widgetCount=25}
 widgets [
  Overlay_0 {widgetClassPath=&Overlay, widget=Overlay_0}
  WBP_UIKit_Backplate {
   widgetClassPath=&WBP_UIKit_Backplate
   bIsVariable=true
   parent=Overlay_0
   widget=WBP_UIKit_Backplate
   properties {brightness=Dark, cornerRadius=Circle}
   slot {class=OverlaySlot, properties {horizontalAlignment=HAlign_Fill, verticalAlignment=VAlign_Fill}}
  }
  HorizontalBox_56 {
   widgetClassPath=&HorizontalBox
   parent=Overlay_0
   widget=HorizontalBox_56
   slot {class=OverlaySlot, properties {horizontalAlignment=HAlign_Fill, verticalAlignment=VAlign_Fill, padding=<ltrb 8 8 8 8>}}
  }
  LabelStatusSwitcher {
   widgetClassPath=&CommonVisibilitySwitcher
   bIsVariable=true
   parent=HorizontalBox_56
   widget=LabelStatusSwitcher
   slot {class=HorizontalBoxSlot, properties {horizontalAlignment=HAlign_Left}}
  }
  RewardAmount {
   widgetClassPath=&Overlay
   bIsVariable=true
   parent=LabelStatusSwitcher
   widget=RewardAmount
   slot {class=CommonVisibilitySwitcherSlot, properties {verticalAlignment=VAlign_Center}}
  }
  RewardAmount_Label {
   widgetClassPath=/FortUILibrary/UIKit/Highlight/WBP_UIKit_Highlight
   bIsVariable=true
   parent=RewardAmount
   widget=RewardAmount_Label
   properties {highlightText=XX to claim, highlightType=Attention}
   slot {class=OverlaySlot}
  }
  Locked {
   widgetClassPath=&Overlay
   bIsVariable=true
   parent=LabelStatusSwitcher
   widget=Locked
   slot {class=CommonVisibilitySwitcherSlot, properties {horizontalAlignment=HAlign_Left, verticalAlignment=VAlign_Center}}
  }
  LockedBackground {
   widgetClassPath=&Image
   bIsVariable=true
   parent=Locked
   widget=LockedBackground
   properties {brush {
     bIsDynamicallyLoaded=false
     drawAs=Box
     imageType=NoImage
     mirroring=NoMirror
     resourceName=None
     resourceObject=/FortUILibrary/Systems/Badge/Materials/M_UIKit_Badge_Background
     tiling=NoTile
     imageSize=<32 32>
     margin=<ltrb 0.5 0.5 0.5 0.5>
     outlineSettings {
      bUseBrushTransparency=false
      roundingType=HalfHeightRadius
      width=0
      color {colorUseRule=UseColor_Specified, specifiedColor=<rgba 0 0 0 0>}
      cornerRadii=<0 0 0 0>
     }
     tintColor {colorUseRule=UseColor_Specified, specifiedColor=<rgba 0.0296 0.0296 0.0356 1>}
     uVRegion {bIsValid=false, max=<0 0>, min=<0 0>}
    }}
   slot {class=OverlaySlot, properties {horizontalAlignment=HAlign_Fill, verticalAlignment=VAlign_Fill}}
  }
  LockedPage_HB {
   widgetClassPath=&HorizontalBox
   parent=Locked
   widget=LockedPage_HB
   slot {class=OverlaySlot, properties {horizontalAlignment=HAlign_Center, verticalAlignment=VAlign_Center, padding=<ltrb 12 0 12 0>}}
  }
  LockedIcon {
   widgetClassPath=&Image
   bIsVariable=true
   parent=LockedPage_HB
   widget=LockedIcon
   properties {brush {
     bIsDynamicallyLoaded=false
     drawAs=Image
     imageType=FullColor
     mirroring=NoMirror
     resourceName=None
     resourceObject=/FortUILibrary/Textures/MaterialInstances/MI_UIKit_Icon_Locked
     tiling=NoTile
     imageSize=<24 24>
     margin=<ltrb 0 0 0 0>
     outlineSettings {
      bUseBrushTransparency=false
      roundingType=HalfHeightRadius
      width=0
      color {colorUseRule=UseColor_Specified, specifiedColor=<rgba 0 0 0 0>}
      cornerRadii=<0 0 0 0>
     }
     tintColor {colorUseRule=UseColor_Specified, specifiedColor=<rgba 0.5776 0.5776 0.5776 1>}
     uVRegion {bIsValid=false, max=<0 0>, min=<0 0>}
    }}
   slot {class=HorizontalBoxSlot, properties {horizontalAlignment=HAlign_Left, verticalAlignment=VAlign_Center, padding=<ltrb 0 0 8 0>}}
  }
  Text_LockedLabel {
   widgetClassPath=&CommonTextBlock
   parent=LockedPage_HB
   widget=Text_LockedLabel
   properties {
    bIsScrollingEnabled=false
    mobileFontSizeMultiplier=1.25
    text=Locked Page
    textTransformPolicy=ToUpper
    colorAndOpacity {colorUseRule=UseColor_Specified, specifiedColor=<rgba 0.5776 0.5776 0.5776 1>}
    font {
     bForceMonospaced=false
     bMaterialIsStencil=false
     fontMaterial=null
     fontObject=&HeadingNow
     letterSpacing=10
     monospacedWidth=1
     size=15.75
     skewAmount=0
     typefaceFontName=86 Bold Italic
     outlineSettings {
      bApplyOutlineToDropShadows=false
      bMiteredCorners=false
      bSeparateFillAlpha=false
      outlineMaterial=null
      outlineSize=0
      outlineColor=<rgba 0 0 0 1>
     }
    }
    shadowOffset=<0 0>
    shapedTextOptions {bOverride_TextFlowDirection=true, bOverride_TextShapingMethod=false, textFlowDirection=Culture, textShapingMethod=Auto}
   }
   slot {class=HorizontalBoxSlot}
  }
  PageCompleted {
   widgetClassPath=&Overlay
   bIsVariable=true
   parent=LabelStatusSwitcher
   widget=PageCompleted
   slot {class=CommonVisibilitySwitcherSlot, properties {horizontalAlignment=HAlign_Left, verticalAlignment=VAlign_Center}}
  }
  PageCompletedBackground {
   widgetClassPath=&Image
   bIsVariable=true
   parent=PageCompleted
   widget=PageCompletedBackground
   properties {brush {
     bIsDynamicallyLoaded=false
     drawAs=Box
     imageType=NoImage
     mirroring=NoMirror
     resourceName=None
     resourceObject=&M_UIKit_Badge_Background
     tiling=NoTile
     imageSize=<32 32>
     margin=<ltrb 0.5 0.5 0.5 0.5>
     outlineSettings {
      bUseBrushTransparency=false
      roundingType=HalfHeightRadius
      width=0
      color {colorUseRule=UseColor_Specified, specifiedColor=<rgba 0 0 0 0>}
      cornerRadii=<0 0 0 0>
     }
     tintColor {colorUseRule=UseColor_Specified, specifiedColor=<rgba 0.0395 1 0.2307 1>}
     uVRegion {bIsValid=false, max=<0 0>, min=<0 0>}
    }}
   slot {class=OverlaySlot, properties {horizontalAlignment=HAlign_Fill, verticalAlignment=VAlign_Fill}}
  }
  Text_PageCompleted {
   widgetClassPath=&CommonTextBlock
   parent=PageCompleted
   widget=Text_PageCompleted
   properties {
    bIsScrollingEnabled=false
    mobileFontSizeMultiplier=1.25
    text="Page Completed!"
    textTransformPolicy=ToUpper
    colorAndOpacity {colorUseRule=UseColor_Specified, specifiedColor=<rgba 0 0.0331 0.0296 1>}
    font {
     bForceMonospaced=false
     bMaterialIsStencil=false
     fontMaterial=null
     fontObject=&HeadingNow
     letterSpacing=10
     monospacedWidth=1
     size=15.75
     skewAmount=0
     typefaceFontName=86 Bold Italic
     outlineSettings {
      bApplyOutlineToDropShadows=false
      bMiteredCorners=false
      bSeparateFillAlpha=false
      outlineMaterial=null
      outlineSize=0
      outlineColor=<rgba 0 0 0 1>
     }
    }
    shadowOffset=<0 0>
    shapedTextOptions {bOverride_TextFlowDirection=true, bOverride_TextShapingMethod=false, textFlowDirection=Culture, textShapingMethod=Auto}
   }
   slot {class=OverlaySlot, properties {horizontalAlignment=HAlign_Fill, verticalAlignment=VAlign_Center, padding=<ltrb 12 0 12 0>}}
  }
  PassCompleted {
   widgetClassPath=&Overlay
   bIsVariable=true
   parent=LabelStatusSwitcher
   widget=PassCompleted
   slot {class=CommonVisibilitySwitcherSlot, properties {horizontalAlignment=HAlign_Left, verticalAlignment=VAlign_Center}}
  }
  PassCompletedBackground {
   widgetClassPath=&Image
   bIsVariable=true
   parent=PassCompleted
   widget=PassCompletedBackground
   properties {brush {
     bIsDynamicallyLoaded=false
     drawAs=Box
     imageType=NoImage
     mirroring=NoMirror
     resourceName=None
     resourceObject=&M_UIKit_Badge_Background
     tiling=NoTile
     imageSize=<32 32>
     margin=<ltrb 0.5 0.5 0.5 0.5>
     outlineSettings {
      bUseBrushTransparency=false
      roundingType=HalfHeightRadius
      width=0
      color {colorUseRule=UseColor_Specified, specifiedColor=<rgba 0 0 0 0>}
      cornerRadii=<0 0 0 0>
     }
     tintColor {colorUseRule=UseColor_Specified, specifiedColor=<rgba 0.0395 1 0.2307 1>}
     uVRegion {bIsValid=false, max=<0 0>, min=<0 0>}
    }}
   slot {class=OverlaySlot, properties {horizontalAlignment=HAlign_Fill, verticalAlignment=VAlign_Fill}}
  }
  Text_PassCompleted {
   widgetClassPath=&CommonTextBlock
   parent=PassCompleted
   widget=Text_PassCompleted
   properties {
    bIsScrollingEnabled=false
    mobileFontSizeMultiplier=1.25
    text="Pass Completed!"
    textTransformPolicy=ToUpper
    colorAndOpacity {colorUseRule=UseColor_Specified, specifiedColor=<rgba 0 0.0331 0.0296 1>}
    font {
     bForceMonospaced=false
     bMaterialIsStencil=false
     fontMaterial=null
     fontObject=&HeadingNow
     letterSpacing=10
     monospacedWidth=1
     size=15.75
     skewAmount=0
     typefaceFontName=86 Bold Italic
     outlineSettings {
      bApplyOutlineToDropShadows=false
      bMiteredCorners=false
      bSeparateFillAlpha=false
      outlineMaterial=null
      outlineSize=0
      outlineColor=<rgba 0 0 0 1>
     }
    }
    shadowOffset=<0 0>
    shapedTextOptions {bOverride_TextFlowDirection=true, bOverride_TextShapingMethod=false, textFlowDirection=Culture, textShapingMethod=Auto}
   }
   slot {class=OverlaySlot, properties {horizontalAlignment=HAlign_Fill, verticalAlignment=VAlign_Center, padding=<ltrb 12 0 12 0>}}
  }
  GrantedByLevel {
   widgetClassPath=&Overlay
   bIsVariable=true
   parent=LabelStatusSwitcher
   widget=GrantedByLevel
   slot {class=CommonVisibilitySwitcherSlot, properties {horizontalAlignment=HAlign_Left, verticalAlignment=VAlign_Center}}
  }
  Image_GrantedByLevel {
   widgetClassPath=&Image
   bIsVariable=true
   parent=GrantedByLevel
   widget=Image_GrantedByLevel
   properties {
    brush {
     bIsDynamicallyLoaded=false
     drawAs=Box
     imageType=NoImage
     mirroring=NoMirror
     resourceName=None
     resourceObject=&M_UIKit_Badge_Background
     tiling=NoTile
     imageSize=<32 32>
     margin=<ltrb 0.5 0.5 0.5 0.5>
     outlineSettings {
      bUseBrushTransparency=false
      roundingType=HalfHeightRadius
      width=0
      color {colorUseRule=UseColor_Specified, specifiedColor=<rgba 0 0 0 0>}
      cornerRadii=<0 0 0 0>
     }
     tintColor {colorUseRule=UseColor_Specified, specifiedColor=<rgba 1 0.063 0.3372 1>}
     uVRegion {bIsValid=false, max=<0 0>, min=<0 0>}
    }
    colorAndOpacity=<rgba 0.9877 0.9735 1 1>
   }
   slot {class=OverlaySlot, properties {horizontalAlignment=HAlign_Fill, verticalAlignment=VAlign_Fill}}
  }
  GrantedByuLevel_HB_ {
   widgetClassPath=&HorizontalBox
   parent=GrantedByLevel
   widget=GrantedByuLevel_HB_
   slot {class=OverlaySlot, properties {horizontalAlignment=HAlign_Center, verticalAlignment=VAlign_Center, padding=<ltrb 12 0 12 0>}}
  }
  LockedIcon_1 {
   widgetClassPath=&Image
   bIsVariable=true
   parent=GrantedByuLevel_HB_
   widget=LockedIcon_1
   properties {brush {
     bIsDynamicallyLoaded=false
     drawAs=Image
     imageType=FullColor
     mirroring=NoMirror
     resourceName=None
     resourceObject=&MI_UIKit_Icon_Locked
     tiling=NoTile
     imageSize=<24 24>
     margin=<ltrb 0 0 0 0>
     outlineSettings {
      bUseBrushTransparency=false
      roundingType=HalfHeightRadius
      width=0
      color {colorUseRule=UseColor_Specified, specifiedColor=<rgba 0 0 0 0>}
      cornerRadii=<0 0 0 0>
     }
     tintColor {colorUseRule=UseColor_Specified, specifiedColor=<rgba 1 1 1 1>}
     uVRegion {bIsValid=false, max=<0 0>, min=<0 0>}
    }}
   slot {class=HorizontalBoxSlot, properties {horizontalAlignment=HAlign_Left, verticalAlignment=VAlign_Center, padding=<ltrb 0 0 8 0>}}
  }
  Text_GrantedByLevel {
   widgetClassPath=&CommonTextBlock
   bIsVariable=true
   parent=GrantedByuLevel_HB_
   widget=Text_GrantedByLevel
   properties {
    bIsScrollingEnabled=false
    mobileFontSizeMultiplier=1.25
    text="Level {0}"
    textTransformPolicy=ToUpper
    font {
     bForceMonospaced=false
     bMaterialIsStencil=false
     fontMaterial=null
     fontObject=&HeadingNow
     letterSpacing=10
     monospacedWidth=1
     size=15.75
     skewAmount=0
     typefaceFontName=86 Bold Italic
     outlineSettings {
      bApplyOutlineToDropShadows=false
      bMiteredCorners=false
      bSeparateFillAlpha=false
      outlineMaterial=null
      outlineSize=0
      outlineColor=<rgba 0 0 0 1>
     }
    }
    shadowOffset=<0 0>
    shapedTextOptions {bOverride_TextFlowDirection=true, bOverride_TextShapingMethod=false, textFlowDirection=Culture, textShapingMethod=Auto}
   }
   slot {class=HorizontalBoxSlot}
  }
  RespPanel_StatusArea {
   widgetClassPath=/Script/FortniteUI.FortResponsivePanel
   bIsVariable=true
   parent=HorizontalBox_56
   widget=RespPanel_StatusArea
   slot {class=HorizontalBoxSlot, properties {verticalAlignment=VAlign_Center, padding=<ltrb 16 0 0 0>, size {sizeRule=Fill, value=1}}}
  }
  Text_StatusDescription {
   widgetClassPath=&CommonTextBlock
   bIsVariable=true
   parent=RespPanel_StatusArea
   widget=Text_StatusDescription
   properties {
    autoWrapText=true
    bIsScrollingEnabled=false
    mobileFontSizeMultiplier=1.15
    text=You have rewards available to claim
    textTransformPolicy=ToUpper
    colorAndOpacity {colorUseRule=UseColor_Specified, specifiedColor=<rgba 1 1 1 0.6>}
    font {
     bForceMonospaced=false
     bMaterialIsStencil=false
     fontMaterial=null
     fontObject=&HeadingNow
     letterSpacing=10
     monospacedWidth=1
     size=14.25
     skewAmount=0
     typefaceFontName=86 Bold
     outlineSettings {
      bApplyOutlineToDropShadows=false
      bMiteredCorners=false
      bSeparateFillAlpha=false
      outlineMaterial=null
      outlineSize=0
      outlineColor=<rgba 0 0 0 1>
     }
    }
    shadowOffset=<0 0>
    shapedTextOptions {bOverride_TextFlowDirection=true, bOverride_TextShapingMethod=false, textFlowDirection=Culture, textShapingMethod=Auto}
   }
   slot {class=FortResponsivePanelSlot}
  }
  Spacer {
   widgetClassPath=&Spacer
   bIsVariable=true
   parent=Overlay_0
   widget=Spacer
   properties {size=<800 40>}
   slot {class=OverlaySlot, properties {horizontalAlignment=HAlign_Fill, verticalAlignment=VAlign_Fill}}
  }
 ]
}
