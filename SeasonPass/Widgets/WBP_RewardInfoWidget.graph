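# Season pass widget that shows an animated "Rewards Available!" banner with glow, text flares and a "to claim" count.
# Show_VIZ and Hide_VIZ, guarded by IsCompact, play their animations or snap to the final frame by playing the last 0.01 seconds.
# SetCanBePurchasedRewardCount updates the label and, for a positive count with QueueIntroAnimation set, runs the intro with sound once.
# The banner starts transparent on Construct, and StopAudio fades out the stored reward sound.
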WidgetBlueprint /SeasonPass/Widgets/WBP_RewardInfoWidget
animations [
 Anim_Show {
  binding_count=7
  length_seconds=0.4
  object=&self:Anim_Show
  sequence {bindings [
    Glow {
     id=#1
     possessed_class=Image
     tracks [{
       type=MovieSceneFloatTrack
       display_name=Render Opacity
       sections [{type=MovieSceneFloatSection, end_frame=6, start_frame=0, channels [None {type=MovieSceneScriptingFloatChannel, keys [[0,0],[6,1]]}]}]
      }]
    }
    Txt_Rewards {
     id=#2
     possessed_class=CommonTextBlock
     tracks [
      {
       type=MovieScene2DTransformTrack
       sections [{
         type=MovieScene2DTransformSection
         end_frame=8
         start_frame=0
         channels [
          Translation.X {type=MovieSceneScriptingFloatChannel, keys [[0,-360],[8,0]]}
          Translation.Y {type=MovieSceneScriptingFloatChannel, const [0,2]}
          Scale.X {type=MovieSceneScriptingFloatChannel, keys [[0,1.1],[4,1.1],[8,1]]}
          Scale.Y {type=MovieSceneScriptingFloatChannel, keys [[0,1.1],[4,1.1],[8,1]]}
         ]
        }]
      }
      {
       type=MovieSceneFloatTrack
       display_name=Render Opacity
       sections [{type=MovieSceneFloatSection, end_frame=8, start_frame=0, channels [None {type=MovieSceneScriptingFloatChannel, keys [[0,0],[4,0],[8,1]]}]}]
      }
     ]
    }
    Txt_Available {
     id=#3
     possessed_class=CommonTextBlock
     tracks [
      {
       type=MovieScene2DTransformTrack
       sections [{
         type=MovieScene2DTransformSection
         end_frame=8
         start_frame=0
         channels [
          Translation.X {type=MovieSceneScriptingFloatChannel, keys [[0,-360],[8,0]]}
          Translation.Y {type=MovieSceneScriptingFloatChannel, const [0,2]}
          Scale.X {type=MovieSceneScriptingFloatChannel, keys [[0,1.1],[4,1.1],[8,1]]}
          Scale.Y {type=MovieSceneScriptingFloatChannel, keys [[0,1.1],[4,1.1],[8,1]]}
         ]
        }]
      }
      {
       type=MovieSceneFloatTrack
       display_name=Render Opacity
       sections [{type=MovieSceneFloatSection, end_frame=8, start_frame=0, channels [None {type=MovieSceneScriptingFloatChannel, keys [[0,0],[2,0],[8,1]]}]}]
      }
     ]
    }
    Available_Flare {
     id=#4
     possessed_class=Image
     tracks [
      {
       type=MovieSceneFloatTrack
       display_name=Render Opacity
       sections [{type=MovieSceneFloatSection, end_frame=6, start_frame=0, channels [None {type=MovieSceneScriptingFloatChannel, keys [[0,0],[5,0],[6,1]]}]}]
      }
      {
       type=MovieSceneWidgetMaterialTrack
       sections [{
         type=MovieSceneParameterSection
         end_frame=8
         start_frame=0
         channels [FlareOpacity {type=MovieSceneScriptingFloatChannel, keys [[0,1],[6,1],[8,0.25]]}]
        }]
      }
      {
       type=MovieScene2DTransformTrack
       sections [{
         type=MovieScene2DTransformSection
         end_frame=8
         start_frame=0
         channels [
          Translation.X {type=MovieSceneScriptingFloatChannel, keys [[0,0],[4,-180],[8,0]]}
          Translation.Y {type=MovieSceneScriptingFloatChannel, const [0,2]}
         ]
        }]
      }
     ]
    }
    Rewards_Flare {
     id=#5
     possessed_class=Image
     tracks [
      {
       type=MovieSceneFloatTrack
       display_name=Render Opacity
       sections [{type=MovieSceneFloatSection, end_frame=8, start_frame=0, channels [None {type=MovieSceneScriptingFloatChannel, keys [[0,0],[6,0],[8,1]]}]}]
      }
      {
       type=MovieSceneWidgetMaterialTrack
       sections [{
         type=MovieSceneParameterSection
         end_frame=8
         start_frame=0
         channels [FlareOpacity {type=MovieSceneScriptingFloatChannel, keys [[0,1],[7,1],[8,0.25]]}]
        }]
      }
      {
       type=MovieScene2DTransformTrack
       sections [{
         type=MovieScene2DTransformSection
         end_frame=8
         start_frame=0
         channels [
          Translation.X {type=MovieSceneScriptingFloatChannel, keys [[0,0],[4,180],[8,0]]}
          Translation.Y {type=MovieSceneScriptingFloatChannel, const [0,2]}
         ]
        }]
      }
     ]
    }
    RewardAmount_Label {
     id=#6
     possessed_class=WBP_UIKit_Highlight_C
     tracks [
      {
       type=MovieSceneFloatTrack
       display_name=Render Opacity
       sections [{type=MovieSceneFloatSection, end_frame=8, start_frame=0, channels [None {type=MovieSceneScriptingFloatChannel, keys [[0,0],[5.2,0],[8,1]]}]}]
      }
      {
       type=MovieScene2DTransformTrack
       sections [{
         type=MovieScene2DTransformSection
         end_frame=8
         start_frame=0
         channels [
          Translation.X {type=MovieSceneScriptingFloatChannel, const [0,2]}
          Translation.Y {type=MovieSceneScriptingFloatChannel, const [0,2]}
          Scale.X {type=MovieSceneScriptingFloatChannel, keys [[0,1.2],[4,1.2],[8,1]]}
          Scale.Y {type=MovieSceneScriptingFloatChannel, keys [[0,1.2],[4,1.2],[8,1]]}
         ]
        }]
      }
     ]
    }
    Root {
     id=#7
     possessed_class=Overlay
     tracks [{
       type=MovieSceneFloatTrack
       display_name=Render Opacity
       sections [{type=MovieSceneFloatSection, end_frame=8, start_frame=0, channels [None {type=MovieSceneScriptingFloatChannel, keys [[0,0],[1,1],[8,1]]}]}]
      }]
    }
   ]}
 }
 Anim_Hide {
  binding_count=7
  length_seconds=0.4
  object=&self:Anim_Hide
  sequence {
   bindings [
    Root {
     id=#8
     possessed_class=Overlay
     tracks [{
       type=MovieSceneFloatTrack
       display_name=Render Opacity
       sections [{type=MovieSceneFloatSection, end_frame=12, start_frame=0, channels [None {type=MovieSceneScriptingFloatChannel, const [1,2]}]}]
      }]
    }
    Available_Flare {
     id=#9
     possessed_class=Image
     tracks [
      {
       type=MovieSceneFloatTrack
       display_name=Render Opacity
       sections [{type=MovieSceneFloatSection, end_frame=12, start_frame=0, channels [None {type=MovieSceneScriptingFloatChannel, keys [[3,1],[5,0],[12,0]]}]}]
      }
      {
       type=MovieSceneWidgetMaterialTrack
       sections [{
         type=MovieSceneParameterSection
         end_frame=12
         start_frame=0
         channels [FlareOpacity {type=MovieSceneScriptingFloatChannel, keys [[0,0.25],[3,1],[12,1]]}]
        }]
      }
      {
       type=MovieScene2DTransformTrack
       sections [{
         type=MovieScene2DTransformSection
         end_frame=12
         start_frame=0
         channels [
          Translation.X {type=MovieSceneScriptingFloatChannel, keys [[0,0],[6,-180],[12,0]]}
          Translation.Y {type=MovieSceneScriptingFloatChannel, const [0,2]}
         ]
        }]
      }
     ]
    }
    Rewards_Flare {
     id=#10
     possessed_class=Image
     tracks [
      {
       type=MovieSceneFloatTrack
       display_name=Render Opacity
       sections [{type=MovieSceneFloatSection, end_frame=12, start_frame=0, channels [None {type=MovieSceneScriptingFloatChannel, keys [[0,1],[3,0],[12,0]]}]}]
      }
      {
       type=MovieSceneWidgetMaterialTrack
       sections [{
         type=MovieSceneParameterSection
         end_frame=12
         start_frame=0
         channels [FlareOpacity {type=MovieSceneScriptingFloatChannel, keys [[0,0.25],[2,1],[12,1]]}]
        }]
      }
      {
       type=MovieScene2DTransformTrack
       sections [{
         type=MovieScene2DTransformSection
         end_frame=12
         start_frame=0
         channels [
          Translation.X {type=MovieSceneScriptingFloatChannel, keys [[0,0],[6,180],[12,0]]}
          Translation.Y {type=MovieSceneScriptingFloatChannel, const [0,2]}
         ]
        }]
      }
     ]
    }
    RewardAmount_Label {
     id=#11
     possessed_class=WBP_UIKit_Highlight_C
     tracks [{
       type=MovieSceneFloatTrack
       display_name=Render Opacity
       sections [{type=MovieSceneFloatSection, end_frame=12, start_frame=0, channels [None {type=MovieSceneScriptingFloatChannel, const [1,2]}]}]
      }]
    }
    Txt_Rewards {
     id=#12
     possessed_class=CommonTextBlock
     tracks [
      {
       type=MovieScene2DTransformTrack
       sections [{
         type=MovieScene2DTransformSection
         end_frame=12
         start_frame=0
         channels [
          Translation.X {type=MovieSceneScriptingFloatChannel, keys [[0,0],[12,-360]]}
          Translation.Y {type=MovieSceneScriptingFloatChannel, const [0,2]}
          Scale.X {type=MovieSceneScriptingFloatChannel, keys [[0,1],[6,1.1],[12,1.1]]}
          Scale.Y {type=MovieSceneScriptingFloatChannel, keys [[0,1],[6,1.1],[12,1.1]]}
         ]
        }]
      }
      {
       type=MovieSceneFloatTrack
       display_name=Render Opacity
       sections [{type=MovieSceneFloatSection, end_frame=12, start_frame=0, channels [None {type=MovieSceneScriptingFloatChannel, keys [[0,1],[6,0],[12,0]]}]}]
      }
     ]
    }
    Txt_Available {
     id=#13
     possessed_class=CommonTextBlock
     tracks [
      {
       type=MovieScene2DTransformTrack
       sections [{
         type=MovieScene2DTransformSection
         end_frame=12
         start_frame=0
         channels [
          Translation.X {type=MovieSceneScriptingFloatChannel, keys [[0,0],[12,-360]]}
          Translation.Y {type=MovieSceneScriptingFloatChannel, const [0,2]}
          Scale.X {type=MovieSceneScriptingFloatChannel, keys [[0,1],[6,1.1],[12,1.1]]}
          Scale.Y {type=MovieSceneScriptingFloatChannel, keys [[0,1],[6,1.1],[12,1.1]]}
         ]
        }]
      }
      {
       type=MovieSceneFloatTrack
       display_name=Render Opacity
       sections [{type=MovieSceneFloatSection, end_frame=12, start_frame=0, channels [None {type=MovieSceneScriptingFloatChannel, keys [[0,1],[9,0],[12,0]]}]}]
      }
     ]
    }
    Glow {
     id=#14
     possessed_class=Image
     tracks [{
       type=MovieSceneFloatTrack
       display_name=Render Opacity
       sections [{type=MovieSceneFloatSection, end_frame=12, start_frame=0, channels [None {type=MovieSceneScriptingFloatChannel, keys [[3,1],[12,0]]}]}]
      }]
    }
   ]
   display_rate {numerator=30}
  }
 }
]
graphs [
 EventGraph {nodes [
   n1 = Hide_VIZ!() -> (OutputDelegate, then, NeedsToAnimate:bool=~) {is=title K2Node_CustomEvent}
   > Branch(Condition=NOT Boolean(A=Get IsCompact() -> (IsCompact:bool)))
   > n2 = Branch(Condition=@n1.NeedsToAnimate)
   > PlayAnimation(InAnimation=Get Anim_Hide() -> (Anim_Hide:Widget Animation Object Reference))
   > Set IsCompact(IsCompact:bool=true) -> (Output_Get:bool)
   n3 = Get Anim_Hide()
   n4 = GetEndTime(self=@n3)
   PlayAnimationTimeRange!(
    execute=@n2.else
    self:User Widget Object Reference
    InAnimation:Widget Animation Object Reference=@n3
    StartAtTime:float=float - float(A=@n4, B=0.01)
    EndAtTime:float=@n4
    NumLoopsToPlay:int=1
    PlayMode:EUMGSequencePlayMode Enum
    PlaybackSpeed:float=1
    bRestoreState:bool=false
   ) -> (then, ReturnValue:Widget Animation Handle Structure)
   > Set IsCompact(IsCompact=true)
   n5 = Show_VIZ!() -> (OutputDelegate, then, NeedsToAnimate:bool=~, PlayAudio:bool=~) {is=title K2Node_CustomEvent}
   > Branch(Condition=Get IsCompact())
   > n6 = Branch(Condition=@n5.NeedsToAnimate)
   > PlayAnimation(InAnimation=Get Anim_Show() -> (Anim_Show:Widget Animation Object Reference))
   > n7 = Branch(Condition=@n5.PlayAudio)
   > n8 = SpawnSound2D!(
    execute
    Sound:Sound Base Object Reference=/Game/Sounds/Fort_UI_Sounds/FrontEnd/MSS_Fort_UI_PostMatchScreen_RewardsAvailable
    VolumeMultiplier:float=1
    PitchMultiplier:float=1
    StartTime:float=0
    ConcurrencySettings:Sound Concurrency Object Reference
    bPersistAcrossLevelTransition:bool=false
    bAutoDestroy:bool=true
   ) -> (then, ReturnValue:Audio Component Object Reference)
   > n9 = Set RewardSoundComponent(RewardSoundComponent:Audio Component Object Reference=@n8) -> (Output_Get:Audio Component Object Reference)
   Set IsCompact(execute=[@n9.then, @n7.else], IsCompact=false)
   n10 = Get Anim_Show()
   n11 = GetEndTime(self=@n10)
   PlayAnimationTimeRange(
    execute=@n6.else
    InAnimation=@n10
    StartAtTime=float - float(A=@n11, B=0.01)
    EndAtTime=@n11
    NumLoopsToPlay=1
    PlaybackSpeed=1
    bRestoreState=false
   )
   > Set IsCompact(IsCompact=false)
   Event Construct()
   > SetRenderOpacity(self=Get Root() -> (Root:Overlay Object Reference), InOpacity=0)
  ]}
 SetCanBePurchasedRewardCount {nodes [
   n1 = SetCanBePurchasedRewardCount!() -> (then, CanBeClaimed:int) {is=title K2Node_FunctionEntry}
   > "SetText (Text)"(
    self=Get Text_Highlight!(self:WBP UIKit Highlight Base Object Reference=Get RewardAmount_Label() -> (RewardAmount_Label:WBP UIKit Highlight Object Reference)) -> (Text_Highlight:Common Text Object Reference)
    InText=Format Text!(Format="LOCTEXT(\"{NumRewards} to claim\")", NumRewards:int=@n1) -> (Result)
   )
   > Branch(Condition=AND Boolean(A="integer > integer"(A=@n1), B=Get QueueIntroAnimation() -> (QueueIntroAnimation:bool)))
   > Show_VIZ!(execute, self, NeedsToAnimate:bool=true, PlayAudio:bool=true) -> (then) {is=title K2Node_CallFunction}
   > Set QueueIntroAnimation(QueueIntroAnimation:bool=false) -> (Output_Get:bool)
  ]}
 StopAudio {nodes [
   StopAudio() {is=title K2Node_FunctionEntry}
   n1 = Get RewardSoundComponent() -> (RewardSoundComponent:Audio Component Object Reference)
   > Is Valid(InputObject=@n1)
   > FadeOut!(
    execute
    self:Audio Component Object Reference=@n1
    FadeOutDuration:float=0.5
    FadeVolumeLevel:float=0
    FadeCurve:EAudioFaderCurve Enum=Linear
   ) -> (then)
  ]}
]
variables [
 IsCompact {type=boolean, default=true}
 IsGrantedBylevel {type=boolean}
 LockedPage {type=boolean}
 NoRewardsAvailable {type=boolean}
 PageStatusText {type=string, default=""}
 QueueIntroAnimation {type=boolean}
 RewardSoundComponent {type=AudioComponent, default=null}
 TimeLockedPage {type=boolean}
]
widgets {
 info {inheritedWidgetCount=0, parentClass=&UserWidget, rootWidgetClass=&Overlay, widgetCount=10}
 widgets [
  Root {widgetClassPath=&Overlay, bIsVariable=true, widget=Root, properties {visibility=HitTestInvisible}}
  Glow {
   widgetClassPath=&Image
   parent=Root
   widget=Glow
   properties {
    visibility=SelfHitTestInvisible
    brush {
     bIsDynamicallyLoaded=false
     drawAs=Image
     imageType=FullColor
     mirroring=NoMirror
     resourceName=None
     resourceObject=/FortniteCoreUI/UI/Frontend/Lobby/Habanero/MaterialInstances/MI_UI_Lobby_RewardGlow
     tiling=NoTile
     imageSize=<32 32>
     margin=<ltrb 0 0 0 0>
     outlineSettings {
      bUseBrushTransparency=false
      roundingType=HalfHeightRadius
      width=0
      color {colorUseRule=UseColor_Specified, specifiedColor=<rgba 0 0 0 0>}
      cornerRadii=<0 0 0 0>
     }
     tintColor {colorUseRule=UseColor_Specified, specifiedColor=<rgba 1 1 1 1>}
     uVRegion {bIsValid=false, max=<0 0>, min=<0 0>}
    }
   }
   slot {class=OverlaySlot, properties {horizontalAlignment=HAlign_Fill, verticalAlignment=VAlign_Fill, padding=<ltrb -96 0 -96 0>}}
  }
  VB_RewardsInfo {
   widgetClassPath=&VerticalBox
   bIsVariable=true
   parent=Root
   widget=VB_RewardsInfo
   slot {class=OverlaySlot, properties {horizontalAlignment=HAlign_Center, verticalAlignment=VAlign_Center}}
  }
  Overlay_1 {
   widgetClassPath=&Overlay
   parent=VB_RewardsInfo
   widget=Overlay_1
   slot {class=VerticalBoxSlot, properties {horizontalAlignment=HAlign_Center, verticalAlignment=VAlign_Bottom}}
  }
  Rewards_Flare {
   widgetClassPath=&Image
   parent=Overlay_1
   widget=Rewards_Flare
   properties {
    visibility=SelfHitTestInvisible
    brush {
     bIsDynamicallyLoaded=false
     drawAs=Image
     imageType=FullColor
     mirroring=NoMirror
     resourceName=None
     resourceObject=/FortniteCoreUI/UI/Frontend/Lobby/Habanero/MaterialInstances/MI_UI_Lobby_Level_TextFlare
     tiling=NoTile
     imageSize=<32 2>
     margin=<ltrb 0 0 0 0>
     outlineSettings {
      bUseBrushTransparency=false
      roundingType=HalfHeightRadius
      width=0
      color {colorUseRule=UseColor_Specified, specifiedColor=<rgba 0 0 0 0>}
      cornerRadii=<0 0 0 0>
     }
     tintColor {colorUseRule=UseColor_Specified, specifiedColor=<rgba 1 1 1 1>}
     uVRegion {bIsValid=false, max=<0 0>, min=<0 0>}
    }
   }
   slot {class=OverlaySlot, properties {horizontalAlignment=HAlign_Fill, verticalAlignment=VAlign_Fill, padding=<ltrb -192 4 -192 4>}}
  }
  Txt_Rewards {
   widgetClassPath=&CommonTextBlock
   parent=Overlay_1
   widget=Txt_Rewards
   properties {
    applyLineHeightToBottomLine=false
    justification=Center
    renderOpacity=0.7
    text=Rewards
    textTransformPolicy=ToUpper
    font {
     bForceMonospaced=false
     bMaterialIsStencil=false
     fontMaterial=null
     fontObject=&HeadingNow
     letterSpacing=0
     monospacedWidth=1
     size=15
     skewAmount=0
     typefaceFontName=86 Bold Italic
     outlineSettings {
      bApplyOutlineToDropShadows=false
      bMiteredCorners=false
      bSeparateFillAlpha=false
      outlineMaterial=null
      outlineSize=0
      outlineColor=<rgba 0 0 0 1>
     }
    }
    shadowOffset=<0 0>
   }
   slot {class=OverlaySlot, properties {horizontalAlignment=HAlign_Center, verticalAlignment=VAlign_Center}}
  }
  Overlay_2 {
   widgetClassPath=&Overlay
   parent=VB_RewardsInfo
   widget=Overlay_2
   slot {class=VerticalBoxSlot, properties {horizontalAlignment=HAlign_Center, verticalAlignment=VAlign_Bottom}}
  }
  Available_Flare {
   widgetClassPath=&Image
   parent=Overlay_2
   widget=Available_Flare
   properties {
    visibility=SelfHitTestInvisible
    brush {
     bIsDynamicallyLoaded=false
     drawAs=Image
     imageType=FullColor
     mirroring=NoMirror
     resourceName=None
     resourceObject=&MI_UI_Lobby_Level_TextFlare
     tiling=NoTile
     imageSize=<32 2>
     margin=<ltrb 0 0 0 0>
     outlineSettings {
      bUseBrushTransparency=false
      roundingType=HalfHeightRadius
      width=0
      color {colorUseRule=UseColor_Specified, specifiedColor=<rgba 0 0 0 0>}
      cornerRadii=<0 0 0 0>
     }
     tintColor {colorUseRule=UseColor_Specified, specifiedColor=<rgba 1 1 1 1>}
     uVRegion {bIsValid=false, max=<0 0>, min=<0 0>}
    }
   }
   slot {class=OverlaySlot, properties {horizontalAlignment=HAlign_Fill, verticalAlignment=VAlign_Fill, padding=<ltrb -256 10 -256 10>}}
  }
  Txt_Available {
   widgetClassPath=&CommonTextBlock
   parent=Overlay_2
   widget=Txt_Available
   properties {
    justification=Center
    text="Available!"
    textTransformPolicy=ToUpper
    font {
     bForceMonospaced=false
     bMaterialIsStencil=false
     fontMaterial=null
     fontObject=&HeadingNow
     letterSpacing=-10
     monospacedWidth=1
     size=25.5
     skewAmount=0
     typefaceFontName=86 Bold Italic
     outlineSettings {
      bApplyOutlineToDropShadows=false
      bMiteredCorners=false
      bSeparateFillAlpha=false
      outlineMaterial=null
      outlineSize=0
      outlineColor=<rgba 0 0 0 1>
     }
    }
    shadowOffset=<0 0>
   }
   slot {class=OverlaySlot, properties {horizontalAlignment=HAlign_Center, verticalAlignment=VAlign_Center}}
  }
  RewardAmount_Label {
   widgetClassPath=/FortUILibrary/UIKit/Highlight/WBP_UIKit_Highlight
   bIsVariable=true
   parent=VB_RewardsInfo
   widget=RewardAmount_Label
   properties {highlightText=XX to claim, highlightType=Attention}
   slot {class=VerticalBoxSlot, properties {horizontalAlignment=HAlign_Center, verticalAlignment=VAlign_Bottom}}
  }
 ]
}
