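# Translucent UI material that draws the blue reward background for the Season Pass buy-pass upsell preview, judging by its name and path.
# Procedural trapezoid SDFs on offset, slightly rotated UVs are unioned, antialiased and maxed with a box fill to form a dimmed base shape.
# Color blends two blues along a rotated, power-curved UV gradient, with a second scaled gradient added as shading.
# Named reroutes organize UV, shape and color stages; no textures are sampled.
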Material /SeasonPass/Widgets/Materials/M_UI_Upsell_BuyPassPreview_Reward_Background
graph {
 properties {bAutomaticallySetUsageInEditor=false, blendMode=BLEND_Translucent, materialDomain=MD_UI, usedShadingModels=Default Lit}
 nodes [
  n1 = NamedRerouteUsage()
  n2 = NamedRerouteUsage()
  n3 = AppendVector(A=Constant(), B=Constant() {r=0.2})
  n4 = Constant() {r=1.8}
  n5 = Constant() {r=1.8}
  n6 = Power(Base=MaterialFunctionCall(UVs=Add(A=@n2, B=@n3), Texture Scale=AppendVector(A=@n4, B=@n5)) {materialFunction=&ScaleUVsByCenter}.UVs) {constExponent=7.5}
  Add(A=Multiply(A=LinearInterpolate(B=@n1, Alpha=@n6)) {constB=0.65}, B=NamedRerouteUsage())
  n7 = MaterialFunctionCall(DistanceField=MaterialFunctionCall!(DistanceFieldA=NamedRerouteUsage(), DistanceFieldB=NamedRerouteUsage(), BlendAmount) {materialFunction=/FortUILibrary/MaterialFunctions/Shapes/Modifiers/MF_UI_SDF_Union}.Exact) {materialFunction=&MF_UI_AntiAliasing}
  n8 = MaterialFunctionCall(UVs=Add(A=TextureCoordinate(), B=Constant2Vector() {g=-0.4}), Size=Constant2Vector() {g=0.9, r=1}) {materialFunction=&MF_UI_Box}
  NamedRerouteDeclaration(None=Multiply(A=Max(A=@n7.Inner, B=@n8.Fill), B=Constant() {r=0.7})) {name=Base Shape, nodeColor=<rgba 0 0 0 1>}
  NamedRerouteDeclaration(None=MaterialFunctionCall(UVs=NamedRerouteUsage(), Rotation Center=Constant2Vector() {g=0.5, r=0.5}, "Rotation Angle (0-1)"=Constant() {r=-0.009}) {materialFunction=&CustomRotator}.Rotated Values) {name=Rotation Core, nodeColor=<rgba 1 0 0.6261 1>}
  n9 = NamedRerouteUsage()
  n10 = Constant() {r=0.4}
  n11 = ComponentMask(None=MaterialFunctionCall(
    UVs=MaterialFunctionCall(UVs=@n9, "Rotation Angle (0-1)"=@n10) {materialFunction=&CustomRotator}.Rotated Values
    Texture Scale=Constant() {r=1.5}
   ) {materialFunction=&ScaleUVsByCenter}.UVs) {g=true}
  NamedRerouteDeclaration(None=LinearInterpolate(A=NamedRerouteUsage(), B=NamedRerouteUsage(), Alpha=Power(Base=@n11) {constExponent=3.5})) {name=ColorBase, nodeColor=<rgba 1 0 0.4453 1>}
  n12 = NamedRerouteUsage()
  n13 = AppendVector(A=Constant() {r=-0.5}, B=Constant() {r=0.12})
  NamedRerouteDeclaration(None=MaterialFunctionCall!(
    UVs=Add(A=@n12, B=@n13)
    TopWidth=Constant() {r=1.35}
    BottomWidth=Constant() {r=0.4}
    Height=Constant() {r=0.65}
    OutlineThickness
    OutlineAlignment
    GlowMin
    GlowMax
   ) {materialFunction=/FortUILibrary/MaterialFunctions/Shapes/MF_UI_IsoscelesTrapezoid}.SDF) {name=Right_Cut, nodeColor=<rgba 0 0.8594 1 1>}
  n14 = NamedRerouteUsage()
  n15 = AppendVector(A=Constant(), B=Constant() {r=0.005})
  n16 = Constant() {r=1.5}
  NamedRerouteDeclaration(None=MaterialFunctionCall(UVs=Add(A=@n14, B=@n15), TopWidth=@n16, BottomWidth=@n16, Height=Constant() {r=0.814}) {materialFunction=&MF_UI_IsoscelesTrapezoid}.SDF) {name=Center_Shape, nodeColor=<rgba 0.5 0 1 1>}
  NamedRerouteDeclaration(None=Constant3Vector() {constant=<rgba 0.0273 0.0865 0.5089 1>}) {name=Color1, nodeColor=<rgba 1 0.2812 0 1>}
  NamedRerouteDeclaration(None=Constant3Vector() {constant=<rgba 0.0844 0.227 1 1>}) {name=Color2, nodeColor=<rgba 0 0.0391 1 1>}
  NamedRerouteDeclaration(None=Add(A=TextureCoordinate(), B=Constant2Vector() {g=0.05})) {name=UV, nodeColor=<rgba 0 0 0 1>}
  NamedRerouteUsage()
 ]
}
