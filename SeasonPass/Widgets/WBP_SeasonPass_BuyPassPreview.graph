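# Season pass upsell widget, apparently for post-match, that previews a reward, shows a claimable count and offers a buy button.
# GetPreviewInfo reads element theme tables for title, tint and image, async-loading the image into a dynamic material.
# Show_VIZ/Hide_VIZ toggle IsCompact, animating or jumping to the end; Show can play a sound, while Anim_Hide has no bindings.
# Clicking fires analytics, pushes the PassPurchase screen and passes it SeasonPassVM.
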WidgetBlueprint /SeasonPass/Widgets/WBP_SeasonPass_BuyPassPreview
animations [
 Anim_Show {
  binding_count=4
  length_seconds=0.4
  object=&self:Anim_Show
  sequence {bindings [
    Highlight_RewardAmount {
     id=#1
     possessed_class=WBP_UIKit_Highlight_C
     tracks [
      {
       type=MovieSceneFloatTrack
       display_name=Render Opacity
       sections [{type=MovieSceneFloatSection, end_frame=8, start_frame=0, channels [None {type=MovieSceneScriptingFloatChannel, keys [[0,0],[5.2,0],[8,1]]}]}]
      }
      {
       type=MovieScene2DTransformTrack
       sections [{
         type=MovieScene2DTransformSection
         end_frame=8
         start_frame=0
         channels [
          Translation.X {type=MovieSceneScriptingFloatChannel, const [0,2]}
          Translation.Y {type=MovieSceneScriptingFloatChannel, const [0,2]}
          Scale.X {type=MovieSceneScriptingFloatChannel, keys [[0,1.2],[4,1.2],[8,1]]}
          Scale.Y {type=MovieSceneScriptingFloatChannel, keys [[0,1.2],[4,1.2],[8,1]]}
         ]
        }]
      }
     ]
    }
    VerticalBox_PreviewText {
     id=#2
     possessed_class=VerticalBox
     tracks [
      {
       type=MovieScene2DTransformTrack
       sections [{
         type=MovieScene2DTransformSection
         end_frame=8
         start_frame=0
         channels [
          Translation.X {type=MovieSceneScriptingFloatChannel, keys [[0,-360],[8,0]]}
          Translation.Y {type=MovieSceneScriptingFloatChannel, const [0,2]}
          Scale.X {type=MovieSceneScriptingFloatChannel, keys [[0,1.1],[4,1.1],[8,1]]}
          Scale.Y {type=MovieSceneScriptingFloatChannel, keys [[0,1.1],[4,1.1],[8,1]]}
         ]
        }]
      }
      {
       type=MovieSceneFloatTrack
       display_name=Render Opacity
       sections [{type=MovieSceneFloatSection, end_frame=8, start_frame=0, channels [None {type=MovieSceneScriptingFloatChannel, keys [[0,0],[4,0],[8,1]]}]}]
      }
     ]
    }
    Image_RewardPreview {
     id=#3
     possessed_class=FortLazyImage
     tracks [
      {
       type=MovieSceneFloatTrack
       display_name=Render Opacity
       sections [{type=MovieSceneFloatSection, end_frame=8, start_frame=0, channels [None {type=MovieSceneScriptingFloatChannel, keys [[0,0],[5.2,0],[8,1]]}]}]
      }
      {
       type=MovieScene2DTransformTrack
       sections [{
         type=MovieScene2DTransformSection
         end_frame=8
         start_frame=0
         channels [
          Translation.X {type=MovieSceneScriptingFloatChannel, const [0,2]}
          Translation.Y {type=MovieSceneScriptingFloatChannel, const [0,2]}
          Scale.X {type=MovieSceneScriptingFloatChannel, keys [[0,1.2],[4,1.2],[8,1]]}
          Scale.Y {type=MovieSceneScriptingFloatChannel, keys [[0,1.2],[4,1.2],[8,1]]}
         ]
        }]
      }
     ]
    }
    Overlay_Root {
     id=#4
     possessed_class=Overlay
     tracks [{
       type=MovieSceneFloatTrack
       display_name=Render Opacity
       sections [{type=MovieSceneFloatSection, end_frame=8, start_frame=0, channels [None {type=MovieSceneScriptingFloatChannel, keys [[0,0],[1,1],[8,1]]}]}]
      }]
    }
   ]}
 }
 Anim_Hide {binding_count=0, length_seconds=5, object=&self:Anim_Hide, sequence {bindings []}}
]
graphs [
 EventGraph {nodes [
   n1 = "On Clicked (LoudButton_BuySeasonPass)"!() -> (OutputDelegate, then, Button:Common Button Base Object Reference) {is=title K2Node_ComponentBoundEvent}
   > Fire Event Element Interacted(Element=@n1.Button, AdditionalInfo=GetAnalyticsInfo!() -> (ReturnValue:Flow Analytics Additional Info Structure))
   > n2 = PushPrimaryContentSoftClassAsync!(
    execute:Exec
    SoftWidgetClass:Common Activatable Widget Soft Class Reference=GetActivatableWidgetForAlias!(WidgetAlias:Gameplay Tag Structure=(TagName="UI.Alias.PassPurchase")) -> (ReturnValue:Common Activatable Widget Soft Class Reference)
   ) -> (then:Exec, OnPrimaryContentPushed:Exec, ActivatableWidget:Common Activatable Widget Object Reference) {is=title K2Node_AsyncAction}
   SetViewModelByClass!(
    execute=@n2.OnPrimaryContentPushed
    self:MVVM View Object Reference=GetExtension!(self:User Widget Object Reference, ExtensionType:User Widget Extension Class Reference=/Script/ModelViewViewModel.MVVMView) -> (ReturnValue:MVVM View Object Reference)
    NewValue:Notify Field Value Changed Interface=Get SeasonPassVM() -> (SeasonPassVM:Season Pass VM Object Reference)
   ) -> (then, ReturnValue:bool)
   n3 = Hide_VIZ!() -> (OutputDelegate, then, NeedsToAnimate:bool=~) {is=title K2Node_CustomEvent}
   > Branch(Condition=NOT Boolean(A=Get IsCompact() -> (IsCompact:bool)))
   > n4 = Branch(Condition=@n3.NeedsToAnimate)
   > PlayAnimation(InAnimation=Get Anim_Hide() -> (Anim_Hide:Widget Animation Object Reference))
   > Set IsCompact(IsCompact:bool=true) -> (Output_Get:bool)
   n5 = Get Anim_Hide()
   n6 = GetEndTime(self=@n5)
   PlayAnimationTimeRange!(
    execute=@n4.else
    self:User Widget Object Reference
    InAnimation:Widget Animation Object Reference=@n5
    StartAtTime:float=float - float(A=@n6, B=0.01)
    EndAtTime:float=@n6
    NumLoopsToPlay:int=1
    PlayMode:EUMGSequencePlayMode Enum
    PlaybackSpeed:float=1
    bRestoreState:bool=false
   ) -> (then, ReturnValue:Widget Animation Handle Structure)
   > Set IsCompact(IsCompact=true)
   n7 = Show_VIZ!() -> (OutputDelegate, then, NeedsToAnimate:bool=~, PlayAudio:bool=~) {is=title K2Node_CustomEvent}
   > Branch(Condition=Get IsCompact())
   > n8 = Branch(Condition=@n7.NeedsToAnimate)
   > PlayAnimation(InAnimation=Get Anim_Show() -> (Anim_Show:Widget Animation Object Reference))
   > n9 = Branch(Condition=@n7.PlayAudio)
   > n10 = SpawnSound2D!(
    execute
    Sound:Sound Base Object Reference=/Game/Sounds/Fort_UI_Sounds/FrontEnd/MSS_Fort_UI_PostMatchScreen_RewardsAvailable
    VolumeMultiplier:float=1
    PitchMultiplier:float=1
    StartTime:float=0
    ConcurrencySettings:Sound Concurrency Object Reference
    bPersistAcrossLevelTransition:bool=false
    bAutoDestroy:bool=true
   ) -> (then, ReturnValue:Audio Component Object Reference)
   > n11 = Set RewardSoundComponent(RewardSoundComponent:Audio Component Object Reference=@n10) -> (Output_Get:Audio Component Object Reference)
   Set IsCompact(execute=[@n11.then, @n9.else], IsCompact=false)
   n12 = Get Anim_Show()
   n13 = GetEndTime(self=@n12)
   PlayAnimationTimeRange(
    execute=@n8.else
    InAnimation=@n12
    StartAtTime=float - float(A=@n13, B=0.01)
    EndAtTime=@n13
    NumLoopsToPlay=1
    PlaybackSpeed=1
    bRestoreState=false
   )
   > Set IsCompact(IsCompact=false)
   n14 = LoadRewardImage!() -> (OutputDelegate, then, InRewardImage:"Soft Object Reference (by ref)") {is=title K2Node_CustomEvent}
   > Branch(Condition=IsValidSoftObjectReference(SoftObjectReference=@n14.InRewardImage))
   > n15 = Async Load Asset(Asset=@n14.InRewardImage)
   n16 = Cast To Texture2D(execute=@n15.Completed, Object=@n15)
   n17 = Get Image_RewardPreview() -> (Image_RewardPreview:Fort Lazy Image Object Reference)
   > n18 = GetDynamicMaterial(self=@n17)
   > SetTextureParameterValue(self=@n18, ParameterName=Character Texture, Value=@n16)
   n19 = GetDynamicMaterial(execute=@n16.CastFailed, self=@n17)
   > SetTextureParameterValue(self=@n19, ParameterName=Character Texture, Value=/FortUILibrary/Input/PeripheralGuitar/Source/Empty1x1)
   Event Construct()
   > Hide_VIZ!(execute, self, NeedsToAnimate:bool=false) -> (then) {is=title K2Node_CallFunction}
  ]}
 GetAnalyticsInfo {nodes [
   GetAnalyticsInfo() {is=title K2Node_FunctionEntry}
   n1 = GetDisplayName!(self:Season Pass VM Object Reference=Get SeasonPassVM()) -> (ReturnValue:Text)
   > Return Node!(
    execute
    ReturnValue:Flow Analytics Additional Info Structure="Add String (Non-Localized)"!(
     Info:"Flow Analytics Additional Info Structure (by ref)"=Make Flow Analytics Additional Info!() -> (FlowAnalyticsAdditionalInfo:Flow Analytics Additional Info Structure) {is=title K2Node_MakeStruct}
     Key:String=PassName
     Value:String=Get Source String from Text!(Text:Text=@n1) -> (ReturnValue:String) {is=title K2Node_CallFunction}
    ) -> (ReturnValue:Flow Analytics Additional Info Structure)
   )
  ]}
 GetPreviewInfo {nodes [
   n1 = GetPreviewInfo!() -> (then, Target:Season Pass Element Base VM Object Reference) {is=title K2Node_FunctionEntry}
   > n2 = Sequence!(execute) -> (then_0, then_1, then_2:Exec, then_3:Exec)
   n3 = Break Season Pass Element Theme Data!(SeasonPassElementThemeData:"Season Pass Element Theme Data Structure (by ref)"=GetElementThemeData!(self:Season Pass Element Base VM Object Reference=@n1) -> (ReturnValue:"Season Pass Element Theme Data Structure (by ref)")) -> (
    ImageReferenceTable:Fort Image Reference Table Structure
    ColorReferenceTable:Fort Color Reference Table Structure
    TextReferenceTable:Fort Text Reference Table Structure
   ) {is=title K2Node_BreakStruct}
   n4 = TryLookupImage!(Table:"Fort Image Reference Table Structure (by ref)"=@n3, Name:String=PostMatchBPUpsell-Image) -> (OutImageReference:Fort Image Reference Structure, ReturnValue:bool)
   > Branch(Condition=@n4.ReturnValue)
   > LoadRewardImage!(
    execute
    self
    InRewardImage:"Soft Object Reference (by ref)"=Break Fort Image Reference!(FortImageReference:"Fort Image Reference Structure (by ref)"=@n4) -> (Image:Soft Object Reference, ImageUrl:String, MaterialTextureParamName:Name) {is=title K2Node_BreakStruct}
   ) -> (then) {is=title K2Node_CallFunction}
   n5 = TryLookupColor!(Table:"Fort Color Reference Table Structure (by ref)"=@n3.ColorReferenceTable, Name:String=PostMatchBPUpsell-Color) -> (OutColor:Linear Color Structure, ReturnValue:bool)
   Branch(execute=@n2.then_1, Condition=@n5.ReturnValue)
   > n6 = GetDynamicMaterial(self=Get Image_RewardPreview())
   > SetVectorParameterValue(self=@n6, ParameterName=Base Color, Value=@n5)
   n7 = TryLookupText!(Table:"Fort Text Reference Table Structure (by ref)"=@n3.TextReferenceTable, Name:String=PostMatchBPUpsell-Title) -> (OutTextReference:Text, ReturnValue:bool)
   Branch(execute=@n2.then_2, Condition=@n7.ReturnValue)
   > "SetText (Text)"(self=Get Text_Title() -> (Text_Title:Common Text Object Reference), InText=@n7)
  ]}
 SetCanBePurchasedRewardCount {nodes [
   n1 = SetCanBePurchasedRewardCount!() -> (then, CanBeClaimed:int) {is=title K2Node_FunctionEntry}
   > "SetText (Text)"(
    self=Get Text_Highlight!(self:WBP UIKit Highlight Base Object Reference=Get Highlight_RewardAmount() -> (Highlight_RewardAmount:WBP UIKit Highlight Object Reference)) -> (Text_Highlight:Common Text Object Reference)
    InText=Format Text!(Format="LOCTEXT(\"{NumRewards} to claim\")", NumRewards:int=@n1) -> (Result)
   )
  ]}
]
variables [
 IsCompact {type=boolean}
 QueueIntroAnimation {type=boolean}
 RewardSoundComponent {type=AudioComponent, default=null}
]
widgets {
 info {inheritedWidgetCount=0, parentClass=&UserWidget, rootWidgetClass=&SizeBox, widgetCount=11}
 widgets [
  SizeBox_Root {
   widgetClassPath=&SizeBox
   bIsVariable=true
   widget=SizeBox_Root
   properties {
    bOverride_HeightOverride=true
    bOverride_WidthOverride=true
    heightOverride=310
    maxAspectRatio=1.1864
    minAspectRatio=1.0769
    widthOverride=500
   }
  }
  Overlay_Root {
   widgetClassPath=&Overlay
   bIsVariable=true
   parent=SizeBox_Root
   widget=Overlay_Root
   properties {bOverride_Cursor=true}
   slot {class=SizeBoxSlot}
  }
  Image_Backplate {
   widgetClassPath=&Image
   bIsVariable=true
   parent=Overlay_Root
   widget=Image_Backplate
   properties {
    visibility=SelfHitTestInvisible
    brush {
     bIsDynamicallyLoaded=false
     drawAs=Image
     imageType=FullColor
     mirroring=NoMirror
     resourceName=None
     resourceObject=/SeasonPass/Widgets/Materials/M_UI_Upsell_BuyPassPreview_Reward_Background
     tiling=NoTile
     imageSize=<32 32>
     margin=<ltrb 0 0 0 0>
     outlineSettings {
      bUseBrushTransparency=false
      roundingType=HalfHeightRadius
      width=0
      color {colorUseRule=UseColor_Specified, specifiedColor=<rgba 0 0 0 0>}
      cornerRadii=<0 0 0 0>
     }
     tintColor {colorUseRule=UseColor_Specified, specifiedColor=<rgba 1 1 1 1>}
     uVRegion {bIsValid=false, max=<0 0>, min=<0 0>}
    }
   }
   slot {class=OverlaySlot, properties {horizontalAlignment=HAlign_Fill, verticalAlignment=VAlign_Fill}}
  }
  VerticalBox_MainContent {
   widgetClassPath=&VerticalBox
   parent=Overlay_Root
   widget=VerticalBox_MainContent
   slot {class=OverlaySlot, properties {horizontalAlignment=HAlign_Fill, verticalAlignment=VAlign_Fill, padding=<ltrb 35 25 35 30>}}
  }
  VerticalBox_RewardAmount {
   widgetClassPath=&VerticalBox
   parent=VerticalBox_MainContent
   widget=VerticalBox_RewardAmount
   properties {bOverride_Cursor=true}
   slot {class=VerticalBoxSlot, properties {padding=<ltrb -4 0 0 0>}}
  }
  Highlight_RewardAmount {
   widgetClassPath=/FortUILibrary/UIKit/Highlight/WBP_UIKit_Highlight
   bIsVariable=true
   parent=VerticalBox_RewardAmount
   widget=Highlight_RewardAmount
   properties {
    highlightText=XX to claim
    highlightType=Utility
    renderTransform {angle=0, scale=<0.8 0.8>, shear=<0 0>, translation=<0 0>}
    renderTransformPivot=<0 0.5>
   }
   slot {class=VerticalBoxSlot, properties {horizontalAlignment=HAlign_Left, verticalAlignment=VAlign_Top, padding=<ltrb 0 8 0 0>}}
  }
  Overlay_0 {
   widgetClassPath=&Overlay
   parent=VerticalBox_MainContent
   widget=Overlay_0
   slot {class=VerticalBoxSlot, properties {verticalAlignment=VAlign_Center}}
  }
  Image_RewardPreview {
   widgetClassPath=&FortLazyImage
   bIsVariable=true
   parent=Overlay_0
   widget=Image_RewardPreview
   properties {brush {
     bIsDynamicallyLoaded=false
     drawAs=Image
     imageType=FullColor
     mirroring=NoMirror
     resourceName=None
     resourceObject=/SeasonPass/Widgets/Materials/M_UI_Upsell_BuyPassPreview_Reward
     tiling=NoTile
     imageSize=<128 240>
     margin=<ltrb 0 0 0 0>
     outlineSettings {
      bUseBrushTransparency=false
      roundingType=HalfHeightRadius
      width=0
      color {colorUseRule=UseColor_Specified, specifiedColor=<rgba 0 0 0 0>}
      cornerRadii=<0 0 0 0>
     }
     tintColor {colorUseRule=UseColor_Specified, specifiedColor=<rgba 1 1 1 1>}
     uVRegion {bIsValid=false, max=<0 0>, min=<0 0>}
    }}
   slot {class=OverlaySlot, properties {horizontalAlignment=HAlign_Fill, verticalAlignment=VAlign_Center, padding=<ltrb 50 -55 0 0>}}
  }
  VerticalBox_PreviewText {
   widgetClassPath=&VerticalBox
   parent=Overlay_0
   widget=VerticalBox_PreviewText
   slot {class=OverlaySlot, properties {verticalAlignment=VAlign_Center, padding=<ltrb 0 -30 0 0>}}
  }
  Text_Title {
   widgetClassPath=&CommonTextBlock
   bIsVariable=true
   parent=VerticalBox_PreviewText
   widget=Text_Title
   properties {
    lineHeightPercentage=0.85
    text="[PH] Preview Title"
    textTransformPolicy=ToUpper
    wrapTextAt=270
    font {
     bForceMonospaced=false
     bMaterialIsStencil=false
     fontMaterial=null
     fontObject=&HeadingNow
     letterSpacing=0
     monospacedWidth=1.1091
     size=22.5
     skewAmount=0
     typefaceFontName=76 Bold Italic
     outlineSettings {
      bApplyOutlineToDropShadows=false
      bMiteredCorners=false
      bSeparateFillAlpha=false
      outlineMaterial=null
      outlineSize=0
      outlineColor=<rgba 0 0 0 1>
     }
    }
    shadowOffset=<0 0>
   }
   slot {class=VerticalBoxSlot}
  }
  LoudButton_BuySeasonPass {
   widgetClassPath=/FortUILibrary/UIKit/Buttons/WBP_UIKit_Button_Loud
   bIsVariable=true
   parent=VerticalBox_MainContent
   widget=LoudButton_BuySeasonPass
   properties {
    bShouldUseFallbackDefaultInputAction=false
    buttonSize=Small
    hoveredSound=/Game/Sounds/Fort_UI_Sounds/FrontEnd/BattlePass/BattlePass_Button_Hover_Tile_Default_Cue
    minWidthOverride=430
    mobileFontSizeMultiplierOverride=1.33
    mobileMinHeight=60
    navigation=&self:WidgetTree.LoudButton_BuySeasonPass.WidgetNavigation_0
    overrideDefaultHeight=true
    overrideDefaultMobileFontSizeMultiplier=true
    pressedSound=/Game/Sounds/Fort_UI_Sounds/FrontEnd/BattlePass/BattlePass_Button_Confirm_Buy_BattlePass_Cue
    text="[PH] GET THE SEASON PASS"
    visibility=Visible
    triggeringInputAction {dataTable=&InputActionDataTable, rowName=GenericLeftStickClick}
   }
   slot {class=VerticalBoxSlot, properties {horizontalAlignment=HAlign_Center, padding=<ltrb 0 -15 0 0>}}
  }
 ]
}
