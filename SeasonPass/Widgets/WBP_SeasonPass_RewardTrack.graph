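# Season Pass reward track widget, apparently for the post-match screen, that switches between match recap and an expandable battle pass tray.
# SeasonPassSelectionVM supplies pages, tracks and rewards, with previous/next navigation sliding two named track slots.
# Claim-all runs asynchronously with a delayed progress modal, and a claim-pages subview replaces the track.
# UMG animations drive transitions, including background mask and vault-world material parameters.
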WidgetBlueprint /SeasonPass/Widgets/WBP_SeasonPass_RewardTrack
animations [
 Intro {
  binding_count=5
  length_seconds=0.6
  object=&self:Intro
  sequence {
   bindings [
    SizeBox_LeftContent {
     id=#1
     possessed_class=SizeBox
     tracks [
      {
       type=MovieScene2DTransformTrack
       sections [{
         type=MovieScene2DTransformSection
         end_frame=30
         start_frame=0
         channels [
          Translation.X {type=MovieSceneScriptingFloatChannel, keys [[0,-100],[27,0]]}
          Translation.Y {type=MovieSceneScriptingFloatChannel, const [0,2]}
         ]
        }]
      }
      {
       type=MovieSceneFloatTrack
       display_name=Render Opacity
       sections [{type=MovieSceneFloatSection, end_frame=30, start_frame=0, channels [None {type=MovieSceneScriptingFloatChannel, keys [[0,0],[27,1]]}]}]
      }
     ]
    }
    LImg_Background {
     id=#2
     possessed_class=FortLazyImage
     tracks [{
       type=MovieSceneWidgetMaterialTrack
       sections [{
         type=MovieSceneParameterSection
         end_frame=36
         start_frame=-1
         channels [
          PM_Transition_MaskFix {type=MovieSceneScriptingFloatChannel, const [1,2]}
          PM_Transition_MaskHeight {type=MovieSceneScriptingFloatChannel, keys [[0,-0.5],[9.73,-0.5],[30,-0.1]]}
          PM_Transition_RevealOverride {type=MovieSceneScriptingFloatChannel, const [1,2]}
          PM_Transition_ShowMask {type=MovieSceneScriptingFloatChannel, const [1,2]}
         ]
        }]
      }]
    }
    Ovr_BattlePassScroll {id=#3, possessed_class=Overlay, tracks []}
    VBox_TrackArea {
     id=#4
     possessed_class=VerticalBox
     tracks [
      {
       type=MovieSceneFloatTrack
       display_name=Render Opacity
       sections [{type=MovieSceneFloatSection, end_frame=33, start_frame=0, channels [None {type=MovieSceneScriptingFloatChannel, keys [[6,0],[33,1]]}]}]
      }
      {
       type=MovieScene2DTransformTrack
       sections [{
         type=MovieScene2DTransformSection
         end_frame=33
         start_frame=0
         channels [
          Translation.X {type=MovieSceneScriptingFloatChannel, const [0,2]}
          Translation.Y {type=MovieSceneScriptingFloatChannel, keys [[6,60],[33,0]]}
          Angle {type=MovieSceneScriptingFloatChannel}
          Scale.X {type=MovieSceneScriptingFloatChannel}
          Scale.Y {type=MovieSceneScriptingFloatChannel}
          Shear.X {type=MovieSceneScriptingFloatChannel}
          Shear.Y {type=MovieSceneScriptingFloatChannel}
         ]
        }]
      }
     ]
    }
    SeasonPass_Level {
     id=#5
     possessed_class=SeasonPass_Level_C
     tracks [
      {
       type=MovieScene2DTransformTrack
       sections [{
         type=MovieScene2DTransformSection
         end_frame=30
         start_frame=-1
         channels [
          Translation.X {type=MovieSceneScriptingFloatChannel, keys [[-0.001,-100],[0,100],[20.999,0]]}
          Translation.Y {type=MovieSceneScriptingFloatChannel, const [0,2]}
         ]
        }]
      }
      {
       type=MovieSceneFloatTrack
       display_name=Render Opacity
       sections [{type=MovieSceneFloatSection, end_frame=30, start_frame=0, channels [None {type=MovieSceneScriptingFloatChannel, keys [[0,0],[21,1]]}]}]
      }
     ]
    }
   ]
   display_rate {numerator=60}
   tracks [{
     type=MovieSceneMaterialParameterCollectionTrack
     display_name=MPC_UI_Master_VaultWorld_Settings
     sections [{
       type=MovieSceneParameterSection
       channels [
        FocalPointPositionR {type=MovieSceneScriptingFloatChannel, keys [[9,0.4],[27,0]]}
        FocalPointPositionG {type=MovieSceneScriptingFloatChannel, keys [[9,0],[27,0.1]]}
        FocalPointPositionB {type=MovieSceneScriptingFloatChannel}
        FocalPointPositionA {type=MovieSceneScriptingFloatChannel, keys [[0,0],[6,0],[30,1],[36,1]]}
       ]
      }]
    }]
  }
 }
 ToViewReward {
  binding_count=4
  length_seconds=0.25
  object=&self:ToViewReward
  sequence {bindings [
    Button_ClaimPages {
     id=#6
     possessed_class=WBP_UIKit_Button_Quiet_C
     tracks [
      {
       type=MovieSceneFloatTrack
       display_name=Render Opacity
       sections [{type=MovieSceneFloatSection, end_frame=5, start_frame=0, channels [None {type=MovieSceneScriptingFloatChannel, keys [[0,1],[2,0],[5,0]]}]}]
      }
      {
       type=MovieSceneByteTrack
       display_name=Visibility
       sections [{
         type=MovieSceneByteSection
         end_frame=5
         start_frame=0
         channels [None {type=MovieSceneScriptingByteChannel, keys [[0,4],[2,4],[3,1],[5,1]]}]
        }]
      }
     ]
    }
    RewardCards_A {
     id=#7
     possessed_class=DynamicEntryBox
     tracks [{
       type=MovieSceneBoolTrack
       display_name=Is Enabled
       sections [{
         type=MovieSceneBoolSection
         end_frame=5
         start_frame=0
         channels [None {type=MovieSceneScriptingBoolChannel, keys [[0,true],[1,false],[5,false]]}]
        }]
      }]
    }
    VBox_TrackArea {
     id=#8
     possessed_class=VerticalBox
     tracks [
      {
       type=MovieSceneFloatTrack
       display_name=Render Opacity
       sections [{type=MovieSceneFloatSection, end_frame=5, start_frame=0, channels [None {type=MovieSceneScriptingFloatChannel, keys [[0,1],[5,0]]}]}]
      }
      {
       type=MovieScene2DTransformTrack
       sections [{
         type=MovieScene2DTransformSection
         end_frame=5
         start_frame=0
         channels [
          Translation.X {type=MovieSceneScriptingFloatChannel, const [0,2]}
          Translation.Y {type=MovieSceneScriptingFloatChannel, keys [[0,0],[5,200]]}
          Angle {type=MovieSceneScriptingFloatChannel}
          Scale.X {type=MovieSceneScriptingFloatChannel}
          Scale.Y {type=MovieSceneScriptingFloatChannel}
          Shear.X {type=MovieSceneScriptingFloatChannel}
          Shear.Y {type=MovieSceneScriptingFloatChannel}
         ]
        }]
      }
     ]
    }
    SeasonPass_RewardTrack_StatusBar {
     id=#9
     possessed_class=SeasonPass_RewardTrack_StatusBar_C
     tracks [{
       type=MovieSceneFloatTrack
       display_name=Render Opacity
       sections [{type=MovieSceneFloatSection, end_frame=5, start_frame=0, channels [None {type=MovieSceneScriptingFloatChannel, keys [[0,1],[5,0]]}]}]
      }]
    }
   ]}
 }
 Close {
  binding_count=4
  length_seconds=0.6
  object=&self:Close
  sequence {
   bindings [
    SeasonPass_Level {
     id=#10
     possessed_class=SeasonPass_Level_C
     tracks [
      {
       type=MovieScene2DTransformTrack
       sections [{
         type=MovieScene2DTransformSection
         end_frame=12
         start_frame=0
         channels [
          Translation.X {type=MovieSceneScriptingFloatChannel, keys [[0,0],[6,200],[12,200]]}
          Translation.Y {type=MovieSceneScriptingFloatChannel, const [0,2]}
          Angle {type=MovieSceneScriptingFloatChannel}
          Scale.X {type=MovieSceneScriptingFloatChannel}
          Scale.Y {type=MovieSceneScriptingFloatChannel}
          Shear.X {type=MovieSceneScriptingFloatChannel}
          Shear.Y {type=MovieSceneScriptingFloatChannel}
         ]
        }]
      }
      {
       type=MovieSceneFloatTrack
       display_name=Render Opacity
       sections [{type=MovieSceneFloatSection, end_frame=12, start_frame=0, channels [None {type=MovieSceneScriptingFloatChannel, keys [[0,1],[6,0],[12,0]]}]}]
      }
     ]
    }
    SizeBox_LeftContent {
     id=#11
     possessed_class=SizeBox
     tracks [
      {
       type=MovieScene2DTransformTrack
       sections [{
         type=MovieScene2DTransformSection
         end_frame=12
         start_frame=0
         channels [
          Translation.X {type=MovieSceneScriptingFloatChannel, keys [[0,0],[12,-200]]}
          Translation.Y {type=MovieSceneScriptingFloatChannel, const [0,1]}
          Angle {type=MovieSceneScriptingFloatChannel}
          Scale.X {type=MovieSceneScriptingFloatChannel}
          Scale.Y {type=MovieSceneScriptingFloatChannel}
          Shear.X {type=MovieSceneScriptingFloatChannel}
          Shear.Y {type=MovieSceneScriptingFloatChannel}
         ]
        }]
      }
      {
       type=MovieSceneFloatTrack
       display_name=Render Opacity
       sections [{type=MovieSceneFloatSection, end_frame=12, start_frame=0, channels [None {type=MovieSceneScriptingFloatChannel, keys [[0,1],[12,0]]}]}]
      }
     ]
    }
    Ovr_BattlePassScroll {
     id=#12
     possessed_class=Overlay
     tracks [{
       type=MovieSceneFloatTrack
       display_name=Render Opacity
       sections [{type=MovieSceneFloatSection, end_frame=12, start_frame=0, channels [None {type=MovieSceneScriptingFloatChannel, keys [[0,1],[6,0],[12,0]]}]}]
      }]
    }
    VBox_TrackArea {
     id=#13
     possessed_class=VerticalBox
     tracks [
      {
       type=MovieSceneFloatTrack
       display_name=Render Opacity
       sections [{type=MovieSceneFloatSection, end_frame=12, start_frame=0, channels [None {type=MovieSceneScriptingFloatChannel, keys [[0,1],[12,0]]}]}]
      }
      {
       type=MovieScene2DTransformTrack
       sections [{
         type=MovieScene2DTransformSection
         end_frame=12
         start_frame=0
         channels [
          Translation.X {type=MovieSceneScriptingFloatChannel, const [0,1]}
          Translation.Y {type=MovieSceneScriptingFloatChannel, keys [[0,0],[12,100]]}
          Angle {type=MovieSceneScriptingFloatChannel}
          Scale.X {type=MovieSceneScriptingFloatChannel}
          Scale.Y {type=MovieSceneScriptingFloatChannel}
          Shear.X {type=MovieSceneScriptingFloatChannel}
          Shear.Y {type=MovieSceneScriptingFloatChannel}
         ]
        }]
      }
     ]
    }
   ]
   tracks [
    {
     type=MovieSceneMaterialParameterCollectionTrack
     display_name=MPC_UI_Master_VaultWorld_Settings
     sections [{
       type=MovieSceneParameterSection
       channels [
        FocalPointPositionR {type=MovieSceneScriptingFloatChannel, keys [[3,0],[9,0.4]]}
        FocalPointPositionG {type=MovieSceneScriptingFloatChannel, keys [[3,0.1],[9,0]]}
        FocalPointPositionB {type=MovieSceneScriptingFloatChannel}
        FocalPointPositionA {type=MovieSceneScriptingFloatChannel, keys [[0,1],[2,1],[10,0],[12,0]]}
       ]
      }]
    }
    {
     type=MovieSceneEventTrack
     sections [{
       type=MovieSceneEventTriggerSection
       channels [None {type=MovieSceneScriptingEventChannel, const ["<Struct 'MovieSceneEvent' (0x469195c2d00) {}>", 1]}]
      }]
    }
   ]
  }
 }
 ItemPreviewFraming {
  binding_count=1
  length_seconds=0.25
  object=&self:ItemPreviewFraming
  sequence {bindings [SeasonPassCategory_C_29 {
     id=#14
     possessed_class=SeasonPassCategory_C
     tracks [{
       type=MovieSceneDoubleTrack
       display_name=View Reward Anim Value
       sections [{type=MovieSceneDoubleSection, end_frame=5, start_frame=0, channels [None {type=MovieSceneScriptingDoubleChannel, keys [[0,0],[5,1]]}]}]
      }]
    }]}
 }
 NextPage {
  binding_count=3
  length_seconds=0.4
  object=&self:NextPage
  sequence {
   bindings [
    NamedSlot_Track_Next {
     id=#15
     possessed_class=NamedSlot
     tracks [
      {
       type=MovieSceneFloatTrack
       display_name=Render Opacity
       sections [{
         type=MovieSceneFloatSection
         end_frame=12
         start_frame=0
         channels [None {type=MovieSceneScriptingFloatChannel, keys [[0,0],[3,0.0026],[12,1]]}]
        }]
      }
      {
       type=MovieScene2DTransformTrack
       sections [{
         type=MovieScene2DTransformSection
         end_frame=12
         start_frame=0
         channels [Translation.X {type=MovieSceneScriptingFloatChannel, keys [[0,300],[12,0]]}]
        }]
      }
     ]
    }
    SeasonPass_RewardTrack_StatusBar {
     id=#16
     possessed_class=SeasonPass_RewardTrack_StatusBar_C
     tracks [
      {
       type=MovieSceneFloatTrack
       display_name=Render Opacity
       sections [{type=MovieSceneFloatSection, end_frame=12, start_frame=0, channels [None {type=MovieSceneScriptingFloatChannel, keys [[0,0],[8,0],[12,1]]}]}]
      }
      {
       type=MovieScene2DTransformTrack
       sections [{
         type=MovieScene2DTransformSection
         end_frame=12
         start_frame=0
         channels [
          Translation.X {type=MovieSceneScriptingFloatChannel, keys [[0,0],[8,15],[12,0]]}
          Translation.Y {type=MovieSceneScriptingFloatChannel, const [0,1]}
          Angle {type=MovieSceneScriptingFloatChannel, const [0,1]}
          Scale.X {type=MovieSceneScriptingFloatChannel, const [1,1]}
          Scale.Y {type=MovieSceneScriptingFloatChannel, const [1,1]}
          Shear.X {type=MovieSceneScriptingFloatChannel, const [0,1]}
          Shear.Y {type=MovieSceneScriptingFloatChannel, const [0,1]}
         ]
        }]
      }
     ]
    }
    NamedSlot_Track_Current {
     id=#17
     possessed_class=NamedSlot
     tracks [
      {
       type=MovieSceneFloatTrack
       display_name=Render Opacity
       sections [{
         type=MovieSceneFloatSection
         end_frame=12
         start_frame=0
         channels [None {type=MovieSceneScriptingFloatChannel, keys [[0,1],[7.5,0],[12,0]]}]
        }]
      }
      {
       type=MovieScene2DTransformTrack
       sections [{
         type=MovieScene2DTransformSection
         end_frame=12
         start_frame=0
         channels [Translation.X {type=MovieSceneScriptingFloatChannel, keys [[0,0],[12,-300]]}]
        }]
      }
     ]
    }
   ]
   display_rate {numerator=30}
  }
 }
 PreviousPage {
  binding_count=3
  length_seconds=0.4
  object=&self:PreviousPage
  sequence {
   bindings [
    NamedSlot_Track_Next {
     id=#18
     possessed_class=NamedSlot
     tracks [
      {
       type=MovieSceneFloatTrack
       display_name=Render Opacity
       sections [{
         type=MovieSceneFloatSection
         end_frame=12
         start_frame=0
         channels [None {type=MovieSceneScriptingFloatChannel, keys [[0,0],[3,0.0026],[12,1]]}]
        }]
      }
      {
       type=MovieScene2DTransformTrack
       sections [{
         type=MovieScene2DTransformSection
         end_frame=12
         start_frame=0
         channels [Translation.X {type=MovieSceneScriptingFloatChannel, keys [[0,-300],[12,0]]}]
        }]
      }
     ]
    }
    SeasonPass_RewardTrack_StatusBar {
     id=#19
     possessed_class=SeasonPass_RewardTrack_StatusBar_C
     tracks [
      {
       type=MovieSceneFloatTrack
       display_name=Render Opacity
       sections [{type=MovieSceneFloatSection, end_frame=12, start_frame=0, channels [None {type=MovieSceneScriptingFloatChannel, keys [[0,0],[8,0],[12,1]]}]}]
      }
      {
       type=MovieScene2DTransformTrack
       sections [{
         type=MovieScene2DTransformSection
         end_frame=12
         start_frame=0
         channels [
          Translation.X {type=MovieSceneScriptingFloatChannel, keys [[0,0],[8,-15],[12,0]]}
          Translation.Y {type=MovieSceneScriptingFloatChannel, const [0,1]}
         ]
        }]
      }
     ]
    }
    NamedSlot_Track_Current {
     id=#20
     possessed_class=NamedSlot
     tracks [
      {
       type=MovieSceneFloatTrack
       display_name=Render Opacity
       sections [{type=MovieSceneFloatSection, end_frame=12, start_frame=0, channels [None {type=MovieSceneScriptingFloatChannel, keys [[0,1],[9,0],[12,0]]}]}]
      }
      {
       type=MovieScene2DTransformTrack
       sections [{
         type=MovieScene2DTransformSection
         end_frame=12
         start_frame=0
         channels [Translation.X {type=MovieSceneScriptingFloatChannel, keys [[0,0],[12,300]]}]
        }]
      }
     ]
    }
   ]
   display_rate {numerator=30}
  }
 }
 ToClaimPages {
  binding_count=3
  length_seconds=1
  object=&self:ToClaimPages
  sequence {
   bindings [
    SeasonPass_ClaimPages {
     id=#21
     possessed_class=SeasonPass_ClaimPages_C
     tracks [
      {
       type=MovieSceneFloatTrack
       display_name=Render Opacity
       sections [{type=MovieSceneFloatSection, end_frame=20, start_frame=0, channels [None {type=MovieSceneScriptingFloatChannel, keys [[0,0],[10,0],[20,1]]}]}]
      }
      {
       type=MovieSceneByteTrack
       display_name=Visibility
       sections [{
         type=MovieSceneByteSection
         end_frame=20
         start_frame=0
         channels [None {type=MovieSceneScriptingByteChannel, keys [[0,1],[9,1],[10,0],[20,0]]}]
        }]
      }
     ]
    }
    SeasonPass_Level {
     id=#22
     possessed_class=SeasonPass_Level_C
     tracks [
      {
       type=MovieSceneFloatTrack
       display_name=Render Opacity
       sections [{type=MovieSceneFloatSection, end_frame=20, start_frame=0, channels [None {type=MovieSceneScriptingFloatChannel, keys [[0,1],[10,0],[20,0]]}]}]
      }
      {
       type=MovieSceneByteTrack
       display_name=Visibility
       sections [{
         type=MovieSceneByteSection
         end_frame=20
         start_frame=0
         channels [None {type=MovieSceneScriptingByteChannel, keys [[0,4],[10,4],[11,1],[20,1]]}]
        }]
      }
     ]
    }
    VBox_TrackArea {
     id=#23
     possessed_class=VerticalBox
     tracks [
      {
       type=MovieSceneFloatTrack
       display_name=Render Opacity
       sections [{type=MovieSceneFloatSection, end_frame=20, start_frame=0, channels [None {type=MovieSceneScriptingFloatChannel, keys [[0,1],[10,0],[20,0]]}]}]
      }
      {
       type=MovieSceneByteTrack
       display_name=Visibility
       sections [{
         type=MovieSceneByteSection
         end_frame=20
         start_frame=0
         channels [None {type=MovieSceneScriptingByteChannel, keys [[0,4],[10,4],[11,1],[20,1]]}]
        }]
      }
     ]
    }
   ]
   tracks [
    {
     type=MovieSceneMaterialParameterCollectionTrack
     display_name=MPC_UI_Master_VaultWorld_Settings
     sections [{
       type=MovieSceneParameterSection
       channels [
        FocalPointPositionR {type=MovieSceneScriptingFloatChannel, keys [[0,0],[20,0.25]]}
        FocalPointPositionG {type=MovieSceneScriptingFloatChannel, const [0.1,2]}
        FocalPointPositionB {type=MovieSceneScriptingFloatChannel, const [0,2]}
        FocalPointPositionA {type=MovieSceneScriptingFloatChannel, const [1,2]}
       ]
      }]
    }
    {
     type=MovieSceneEventTrack
     sections [{
       type=MovieSceneEventTriggerSection
       channels [None {type=MovieSceneScriptingEventChannel, const ["<Struct 'MovieSceneEvent' (0x469195c4c40) {}>", 1]}]
      }]
    }
   ]
  }
 }
 Anim_TrayExpand {
  binding_count=8
  length_seconds=0.85
  object=&self:Anim_TrayExpand
  sequence {bindings [
    SeasonPass_RewardTrack_StatusBar {
     id=#24
     possessed_class=SeasonPass_RewardTrack_StatusBar_C
     tracks [
      {
       type=MovieSceneFloatTrack
       display_name=Render Opacity
       sections [{type=MovieSceneFloatSection, end_frame=10, start_frame=0, channels [None {type=MovieSceneScriptingFloatChannel, keys [[0,0],[10,1]]}]}]
      }
      {
       type=MovieScene2DTransformTrack
       sections [{
         type=MovieScene2DTransformSection
         end_frame=10
         start_frame=0
         channels [Translation.Y {type=MovieSceneScriptingFloatChannel, keys [[0,50],[10,0]]}]
        }]
      }
     ]
    }
    Ovr_MatchRecap {
     id=#25
     possessed_class=Overlay
     tracks [
      {
       type=MovieSceneFloatTrack
       display_name=Render Opacity
       sections [{type=MovieSceneFloatSection, end_frame=12, start_frame=0, channels [None {type=MovieSceneScriptingFloatChannel, keys [[0,0],[5,0],[10,1]]}]}]
      }
      {
       type=MovieScene2DTransformTrack
       sections [{
         type=MovieScene2DTransformSection
         end_frame=12
         start_frame=0
         channels [
          Translation.X {type=MovieSceneScriptingFloatChannel, default=0}
          Translation.Y {type=MovieSceneScriptingFloatChannel, keys [[0,-25],[5,25],[10,0]]}
         ]
        }]
      }
     ]
    }
    LImg_Background {
     id=#26
     possessed_class=FortLazyImage
     tracks [{
       type=MovieSceneWidgetMaterialTrack
       sections [{
         type=MovieSceneParameterSection
         end_frame=17
         start_frame=0
         channels [
          PM_Transition_MaskHeight {type=MovieSceneScriptingFloatChannel, keys [[0,-0.1],[17,0.45]]}
          PM_Transition_ShowMask {type=MovieSceneScriptingFloatChannel, const [1,2]}
          PM_Transition_RevealOverride {type=MovieSceneScriptingFloatChannel, const [1,2]}
          PM_Transition_MaskFix {type=MovieSceneScriptingFloatChannel, const [1,2]}
         ]
        }]
      }]
    }
    TrayContent {
     id=#27
     possessed_class=Overlay
     tracks [
      {
       type=MovieSceneFloatTrack
       display_name=Render Opacity
       sections [{type=MovieSceneFloatSection, end_frame=10, start_frame=0, channels [None {type=MovieSceneScriptingFloatChannel, keys [[0,0],[10,1]]}]}]
      }
      {
       type=MovieScene2DTransformTrack
       sections [{
         type=MovieScene2DTransformSection
         end_frame=10
         start_frame=0
         channels [Translation.Y {type=MovieSceneScriptingFloatChannel, keys [[0,50],[10,0]]}]
        }]
      }
     ]
    }
    Ovr_BattlePassScroll {
     id=#28
     possessed_class=Overlay
     tracks [
      {
       type=MovieScene2DTransformTrack
       sections [{
         type=MovieScene2DTransformSection
         end_frame=10
         start_frame=0
         channels [
          Translation.X {type=MovieSceneScriptingFloatChannel, const [0,1]}
          Translation.Y {type=MovieSceneScriptingFloatChannel, keys [[0,0],[4,-25],[10,-25]]}
         ]
        }]
      }
      {
       type=MovieSceneFloatTrack
       display_name=Render Opacity
       sections [{type=MovieSceneFloatSection, end_frame=10, start_frame=0, channels [None {type=MovieSceneScriptingFloatChannel, keys [[0,1],[4,0],[10,0]]}]}]
      }
     ]
    }
    WBP_RewardInfoWidget {
     id=#29
     possessed_class=WBP_RewardInfoWidget_C
     tracks [
      {
       type=MovieSceneFloatTrack
       display_name=Render Opacity
       sections [{type=MovieSceneFloatSection, end_frame=10, start_frame=0, channels [None {type=MovieSceneScriptingFloatChannel, keys [[0,1],[10,0]]}]}]
      }
      {
       type=MovieScene2DTransformTrack
       sections [{
         type=MovieScene2DTransformSection
         end_frame=10
         start_frame=0
         channels [Translation.Y {type=MovieSceneScriptingFloatChannel, keys [[0,0],[10,-50]]}]
        }]
      }
     ]
    }
    TabNavigation {
     id=#30
     possessed_class=Overlay
     tracks [
      {
       type=MovieSceneFloatTrack
       display_name=Render Opacity
       sections [{type=MovieSceneFloatSection, end_frame=10, start_frame=0, channels [None {type=MovieSceneScriptingFloatChannel, keys [[0,0],[10,1]]}]}]
      }
      {
       type=MovieScene2DTransformTrack
       sections [{
         type=MovieScene2DTransformSection
         end_frame=10
         start_frame=0
         channels [Translation.Y {type=MovieSceneScriptingFloatChannel, keys [[0,50],[10,0]]}]
        }]
      }
      {
       type=MovieSceneByteTrack
       display_name=Visibility
       sections [{type=MovieSceneByteSection, end_frame=10, start_frame=0, channels [None {type=MovieSceneScriptingByteChannel, keys [[0,2],[1,4],[10,4]]}]}]
      }
     ]
    }
    Button_Battlepass {
     id=#31
     possessed_class=Button
     tracks [{
       type=MovieSceneByteTrack
       display_name=Visibility
       sections [{
         type=MovieSceneByteSection
         end_frame=10
         start_frame=0
         channels [None {type=MovieSceneScriptingByteChannel, thinned_from=4, keys [[0,0],[1,2],[10,2]]}]
        }]
      }]
    }
   ]}
 }
 Anim_HoverMatchRecap {
  binding_count=1
  length_seconds=0.3
  object=&self:Anim_HoverMatchRecap
  sequence {bindings [CTxt_MatchRecapName {
     id=#32
     possessed_class=CommonTextBlock
     tracks [{
       type=MovieScene2DTransformTrack
       sections [{
         type=MovieScene2DTransformSection
         end_frame=6
         start_frame=0
         channels [
          Translation.X {type=MovieSceneScriptingFloatChannel, const [0,1]}
          Translation.Y {type=MovieSceneScriptingFloatChannel, const [0,1]}
          Angle {type=MovieSceneScriptingFloatChannel, const [0,1]}
          Scale.X {type=MovieSceneScriptingFloatChannel, keys [[0,1],[6,1.1]]}
          Scale.Y {type=MovieSceneScriptingFloatChannel, keys [[0,1],[6,1.1]]}
          Shear.X {type=MovieSceneScriptingFloatChannel, const [0,1]}
          Shear.Y {type=MovieSceneScriptingFloatChannel, const [0,1]}
         ]
        }]
      }]
    }]}
 }
 Anim_HoverBattlepass {
  binding_count=1
  length_seconds=0.3
  object=&self:Anim_HoverBattlepass
  sequence {bindings [CTxt_SeasonPassName {
     id=#33
     possessed_class=CommonTextBlock
     tracks [{
       type=MovieScene2DTransformTrack
       sections [{
         type=MovieScene2DTransformSection
         end_frame=6
         start_frame=0
         channels [
          Translation.X {type=MovieSceneScriptingFloatChannel, const [0,1]}
          Translation.Y {type=MovieSceneScriptingFloatChannel, const [0,1]}
          Angle {type=MovieSceneScriptingFloatChannel, const [0,1]}
          Scale.X {type=MovieSceneScriptingFloatChannel, keys [[0,1],[6,1.1]]}
          Scale.Y {type=MovieSceneScriptingFloatChannel, keys [[0,1],[6,1.1]]}
          Shear.X {type=MovieSceneScriptingFloatChannel, const [0,1]}
          Shear.Y {type=MovieSceneScriptingFloatChannel, const [0,1]}
         ]
        }]
      }]
    }]}
 }
]
graphs [
 AreAllSelectedPageItemVMsReady {nodes [
   AreAllSelectedPageItemVMsReady() {is=title K2Node_FunctionEntry}
   n1 = Get SeasonPassSelectionVM() -> (SeasonPassSelectionVM:Season Pass Selection VM Object Reference)
   > n2 = For Each Loop(Array:Array of Season Pass Track VM Object References=Get Tracks!(self:Season Pass Page VM Object Reference=Get SelectedPage!(self:Season Pass Selection VM Object Reference=@n1) -> (SelectedPage:Season Pass Page VM Object Reference)) -> (Tracks:Array of Season Pass Track VM Object References)) -> (Array Element:Season Pass Track VM Object Reference)
   Return Node!(execute=@n2.Completed, AllItemVMsReady:bool=true)
   > n3 = For Each Loop(Array:Array of Season Pass Reward VM Object References=Get Rewards!(self:Season Pass Track VM Object Reference=@n2) -> (Rewards:Array of Season Pass Reward VM Object References)) -> (Array Element:Season Pass Reward VM Object Reference)
   n4 = Get ItemVMs!(self:Season Pass Reward VM Object Reference=@n3) -> (ItemVMs:Array of Fort Item VM Object References)
   > Branch(Condition="Equal (Integer)"(A=Length(TargetArray:Array of Fort Item VM Object References=@n4)))
   > Return Node(AllItemVMsReady=false)
  ]}
 BP_GetDesiredFocusTarget {nodes [
   Get Desired Focus Target()
   > n1 = Branch(Condition=Get InitialPurchaseState() -> (InitialPurchaseState:bool))
   n2 = GetAllEntries!(self:Dynamic Entry Box Base Object Reference=Get ActiveRewardList() -> (ActiveRewardList:Dynamic Entry Box Object Reference)) -> (ReturnValue:Array of User Widget Object References)
   n3 = Get SelectedRewardIndex() -> (SelectedRewardIndex:int=0)
   n4 = Get SelectedRewardIndex() -> (SelectedRewardIndex=0)
   > Return Node!(
    execute
    ReturnValue:Widget Object Reference=Select(
     Option 0:User Widget Object Reference="Get (a copy)"(Array:Array of User Widget Object References=@n2, Dimension 1=@n3) -> (Output:User Widget Object Reference)
     Option 1:User Widget Object Reference=Get ModeSpecificReward() -> (ModeSpecificReward:WBP Reward Track Item Button Object Reference)
     Index="Equal (Integer)"(A=@n4, B=-1)
    ) -> (ReturnValue:User Widget Object Reference)
   )
   Return Node(
    execute=@n1.else
    ReturnValue=Get LoudButton_BuySeasonPass!(self:WBP Season Pass Buy Pass Preview Object Reference=Get WBP_BuyPassPreview() -> (WBP_BuyPassPreview:WBP Season Pass Buy Pass Preview Object Reference)) -> (LoudButton_BuySeasonPass:WBP UIKit Button Loud Object Reference)
   )
  ]}
 BP_GetDesiredInputConfig {nodes [
   Get Desired Input Config() {is=title K2Node_FunctionEntry}
   > Return Node(ReturnValue:UIInput Config Structure=Make UIInput Config!(
     bIgnoreMoveInput:bool=False
     bIgnoreLookInput:bool=False
     InputMode:ECommonInputMode Enum=All
     MouseCaptureMode:EMouseCaptureMode Enum=CaptureDuringMouseDown
     MouseLockMode:EMouseLockMode Enum=DoNotLock
     bHideCursorDuringViewportCapture:bool=True
    ) -> (UIInputConfig:UIInput Config Structure) {is=title K2Node_MakeStruct})
  ]}
 BranchOnInputType {nodes [
   Inputs!() -> (In:Exec)
   > n1 = Switch on ECommonInputType!(execute, Selection:ECommonInputType Enum=GetCurrentInputType(self=Get CommonInputSubsystem())) -> (MouseAndKeyboard:Exec, Gamepad:Exec, Touch:Exec, Count:Exec) {is=title K2Node_SwitchEnum}
   Outputs!(MnKOrTouch:Exec=[@n1.MouseAndKeyboard, @n1.Touch], Gamepad:Exec=@n1.Gamepad)
  ]}
 CF_FindBattlepass {nodes [
   CF_FindBattlepass!() -> (then, SeasonPasses:Array of Season Pass VM Object References) {is=title K2Node_FunctionEntry}
   n1 = Get SeasonPassSystemVM() -> (SeasonPassSystemVM:Season Pass System VM Object Reference)
   > Is Valid(InputObject=@n1)
   > Return Node(ReturnValue:Season Pass VM Object Reference=FindSeasonPass!(self:Season Pass System VM Object Reference=@n1, TemplateId:String=battlepassdata) -> (ReturnValue:Season Pass VM Object Reference))
  ]}
 ClearRewardFocus {nodes [
   ClearRewardFocus() {is=title K2Node_FunctionEntry}
   > SetVisibility(self=Get CPanel_ContextMenuRoot() -> (CPanel_ContextMenuRoot:Canvas Panel Object Reference), InVisibility=Hidden)
  ]}
 CreateSeasonPassVariant {nodes [
   CreateSeasonPassVariant!() -> (then, ItemVM:Fort Item VM Object Reference, ItemDef:Fort Item Definition Object Reference) {is=title K2Node_FunctionEntry}
   > n1 = Create Entry!(execute, self:Dynamic Entry Box Object Reference=Get VariantCards() -> (VariantCards:Fort Dynamic Entry Box Object Reference)) -> (then, ReturnValue:User Widget Object Reference)
   > n2 = Cast To WBP_UIKit_ItemCard_BattlePass!(execute, Object=@n1) -> (then, CastFailed, AsWBP UIKit Item Card Battle Pass:WBP UIKit Item Card Battle Pass Object Reference) {is=title K2Node_DynamicCast}
   > InitializeItemCard(self:WBP UIKit Item Card Base Object Reference=@n2)
   > SetFortItemVM!(
    execute
    self:WBP UIKit Item Card Battle Pass Object Reference=@n2
    Viewmodel:Fort Item VM Object Reference=Get ItemVM() -> (ItemVM:Fort Item VM Object Reference)
   ) -> (then)
   > Bind Event to On Card Focused(self:WBP UIKit Item Card Battle Pass Object Reference=@n2, Delegate=Create Event()) {is=title K2Node_AddDelegate}
   > Bind Event to Viewing Reward State Changed(self:WBP Season Pass Reward Track Object Reference, Delegate=Create Event(self=@n2)) {is=title K2Node_AddDelegate}
   > Bind Event to New Variant Card Selected(self:WBP Season Pass Reward Track Object Reference, Delegate=Create Event(self=@n2)) {is=title K2Node_AddDelegate}
  ]}
 DecreasePage {nodes [
   DecreasePage() {is=title K2Node_FunctionEntry}
   n1 = Get IsTransitionActive() -> (IsTransitionActive:bool)
   > Branch(Condition=AND Boolean(A=Get IsBattlepassOpen() -> (IsBattlepassOpen:bool), B=NOT Boolean(A=@n1)))
   > Set IsTransitionActive(IsTransitionActive:bool=true) -> (Output_Get:bool)
   > ClearRewardFocus() {is=title K2Node_CallFunction}
   > ResetActiveTrackSlot!(execute, self, ActiveTrack:Named Slot Object Reference=Get NamedSlot_Track_Next() -> (NamedSlot_Track_Next:Named Slot Object Reference)) -> (then)
   > n2 = GetLinearSeasonPassPageNavTarget!(
    execute
    SeasonPassSelectionVM:Season Pass Selection VM Object Reference=Get SeasonPassSelectionVM()
    IsPrevious:bool=true
    ShowLockedCategories:bool=true
    UseMainSetOnly:bool=Get bUseMainSetOnly() -> (bUseMainSetOnly:bool)
   ) -> (
    then
    NavigateToRewardTarget:Season Pass Reward VM Object Reference
    SeasonPassCategoryTarget:Season Pass Category VM Object Reference
    SelectedPage:int=0
   )
   > Set NewPageRewardTarget(NewPageRewardTarget:Season Pass Reward VM Object Reference=@n2) -> (Output_Get:Season Pass Reward VM Object Reference)
   > Set SelectedCategory!(
    execute
    SelectedCategory:Season Pass Category VM Object Reference=@n2.SeasonPassCategoryTarget
    self:Season Pass Selection VM Object Reference=Get SeasonPassSelectionVM()
   ) -> (then, Output_Get:Season Pass Category VM Object Reference)
   > Set Page Index!(execute, self, Page Index:int=@n2.SelectedPage) -> (then) {is=title K2Node_CallFunction}
   > BindToAnimationFinished(Animation=Get PreviousPage() -> (PreviousPage:Widget Animation Object Reference), Delegate=Create Event())
   > n3 = Branch(Condition=IsAnimationPlaying(InAnimation=Get NextPage() -> (NextPage:Widget Animation Object Reference)))
   > n4 = StopAnimation(InAnimation=Get NextPage())
   PlayAnimationForward(execute=[@n4.then, @n3.else], InAnimation=Get PreviousPage())
   > n5 = Cast To FortPlayerController(Object=GetOwningPlayer())
   > "Fire Event Element Interacted (Non-Localized)"!(
    execute
    FortPC:Fort Player Controller Object Reference=@n5
    ElementName:String=PostMatchScreen.SeasonPass_RewardTrack.NavigatedToPreviousPage
    AdditionalInfo:Flow Analytics Additional Info Structure
   ) -> (then)
  ]}
 DisplayContextMenuPosition {nodes [
   n1 = DisplayContextMenuPosition!() -> (then, Position:Vector 2D Structure) {is=title K2Node_FunctionEntry}
   > Set IsMenuPositionSet(IsMenuPositionSet:bool="Not Equal (Vector2D)"!(A:Vector 2D Structure=@n1, B:Vector 2D Structure, ErrorTolerance:float) -> (ReturnValue:bool=~) {is=title K2Node_PromotableOperator}) -> (Output_Get:bool)
   > SetRenderOpacity(
    self=Get CPanel_ContextMenuRoot()
    InOpacity=Select(Option 0:float=0, Option 1:float=1, Index=Get IsMenuPositionSet() -> (IsMenuPositionSet:bool)) -> (ReturnValue:float)
   )
   > Branch(Condition=Get IsMenuPositionSet())
   > SetPosition!(
    execute
    self:Canvas Panel Slot Object Reference=SlotAsCanvasSlot!(Widget:Widget Object Reference=Get SeasonPassCategory_Buttons_ContextMenu() -> (SeasonPassCategory_Buttons_ContextMenu:Season Pass Category Buttons Context Menu Object Reference)) -> (ReturnValue:Canvas Panel Slot Object Reference)
    InPosition:Vector 2D Structure=@n1
   ) -> (then)
   > PlayIntro_VIZ(self:Season Pass Category Buttons Context Menu Object Reference=Get SeasonPassCategory_Buttons_ContextMenu())
  ]}
 EventGraph {nodes [
   "On Unhovered (Button_MatchRecap)"() {is=title K2Node_ComponentBoundEvent}
   n1 = Get Anim_HoverMatchRecap() -> (Anim_HoverMatchRecap:Widget Animation Object Reference)
   > PlayAnimation(InAnimation=@n1, PlayMode=Reverse)
   "On Hovered (Button_Battlepass)"() {is=title K2Node_ComponentBoundEvent}
   n2 = Get Anim_HoverBattlepass() -> (Anim_HoverBattlepass:Widget Animation Object Reference)
   > PlayAnimation(InAnimation=@n2)
   "On Hovered (Button_MatchRecap)"() {is=title K2Node_ComponentBoundEvent}
   > PlayAnimation(InAnimation=@n1)
   "On Clicked (Button_MatchRecap)"() {is=title K2Node_ComponentBoundEvent}
   > CloseBattlepass_VIZ!(execute, self, bSkipAnimation:bool=false) -> (then)
   n3 = "On Clicked (Button_Back)"!() -> (OutputDelegate, then, Button:Common Button Base Object Reference) {is=title K2Node_ComponentBoundEvent}
   n4 = Branch(execute=[@n3.then, @n13.then], Condition=Get IsClaimPagesVisible() -> (IsClaimPagesVisible:bool))
   > DismissClaimPagesEvent()
   PlayAnimation(execute=@n4.else, InAnimation=Get Close() -> (Close:Widget Animation Object Reference))
   > SetVisibility(self=Get Button_Back() -> (Button_Back:WBP UIKit Button Quiet Object Reference), InVisibility=HitTestInvisible)
   > SetVisibility(self=Get Button_ClaimPages() -> (Button_ClaimPages:WBP UIKit Button Quiet Object Reference), InVisibility=Collapsed)
   "On Clicked (Button_ClaimAllRewards)"!() -> (OutputDelegate, then, Button:Common Button Base Object Reference) {is=title K2Node_ComponentBoundEvent}
   > Set bIsEnabled!(
    execute
    bIsEnabled:bool=false
    self:Widget Object Reference=Get Button_ClaimAllRewards() -> (Button_ClaimAllRewards:WBP UIKit Button Quiet Object Reference)
   ) -> (then, Output_Get:bool)
   > n5 = ClaimAllClaimableRewards!(
    execute:Exec
    SeasonPass:Season Pass Element Base VM Object Reference=Get SelectedPass!(self:Season Pass Selection VM Object Reference=Get SeasonPassSelectionVM()) -> (SelectedPass:Season Pass VM Object Reference)
    bTryPurchasePrerequisites:bool=false
   ) -> (then:Exec, OnOperationCompleted:Exec, Success:bool=~) {is=title K2Node_AsyncAction}
   > Set IsClaimAllInProgress(IsClaimAllInProgress:bool=true) -> (Output_Get:bool)
   > Delay(Duration=0.4)
   > n6 = Branch(Condition=NOT Boolean(A=Get IsClaimAllInProgress() -> (IsClaimAllInProgress:bool)))
   ShowProgressModal!(
    execute=@n6.else
    Title:Text="LOCTEXT(\"Loading rewards...\")"
    Description:Text="LOCTEXT(\"Hang tight, we\\'re loading all your unclaimed rewards.\")"
   ) -> (then)
   > Set IsModalVisible(IsModalVisible:bool=true) -> (Output_Get:bool)
   Set IsClaimAllInProgress(execute=@n5.OnOperationCompleted, IsClaimAllInProgress=false)
   > Branch()
   > Delay(Duration=1)
   > HideProgressModal!(execute) -> (then)
   > Set IsModalVisible(IsModalVisible=false)
   "On Clicked (SecondaryLeft)"!() -> (OutputDelegate, then, Button:Common Button Base Object Reference) {is=title K2Node_ComponentBoundEvent}
   > DecreasePage() {is=title K2Node_CallFunction}
   "On Clicked (SecondaryRight)"!() -> (OutputDelegate, then, Button:Common Button Base Object Reference) {is=title K2Node_ComponentBoundEvent}
   > IncreasePage()
   "On Clicked (Button_ClaimPages)"!() -> (OutputDelegate, then, Button:Common Button Base Object Reference) {is=title K2Node_ComponentBoundEvent}
   > n7 = Branch(Condition=Get IsClaimPagesVisible())
   Set ClaimPagesBackLastSelectedReward(
    execute=@n7.else
    ClaimPagesBackLastSelectedReward:Season Pass Reward VM Object Reference=Get SelectedReward!(self:Season Pass Selection VM Object Reference=Get SeasonPassSelectionVM()) -> (SelectedReward:Season Pass Reward VM Object Reference)
   ) -> (Output_Get:Season Pass Reward VM Object Reference)
   > n8 = Play Animation with Finished event!(
    execute:Exec
    Widget:User Widget Object Reference=Self-Reference()
    InAnimation:Widget Animation Object Reference=Get ToClaimPages() -> (ToClaimPages:Widget Animation Object Reference)
    StartAtTime:float=0
    NumLoopsToPlay:int=1
    PlayMode:EUMGSequencePlayMode Enum
    PlaybackSpeed:float=1
   ) -> (then:Exec, Result:Widget Animation Handle Structure, Finished:Exec) {is=title K2Node_PlayAnimation2}
   n9 = Get Button_Back()
   > SetVisibility(
    self=[
     @n9
     Get Button_ClaimPages()
     Get WBP_UIKit_TabGroup() -> (WBP_UIKit_TabGroup:WBP UIKit Tab Group Object Reference)
     Get ClaimAllButtonWrapper() -> (ClaimAllButtonWrapper:Overlay Object Reference)
    ]
    InVisibility=Collapsed
   )
   n10 = Get SeasonPass_ClaimPages() -> (SeasonPass_ClaimPages:Season Pass Claim Pages Object Reference)
   > Initialize(self:Season Pass Claim Pages Object Reference=@n10)
   > SetPageIndexAndUpdate!(execute, self:Season Pass Claim Pages Object Reference=@n10, NewIndex:int=Get PageIndex() -> (PageIndex:int=0)) -> (then)
   Set IsClaimPagesVisible(execute=@n8.Finished, IsClaimPagesVisible:bool=true) -> (Output_Get:bool)
   > SetVisibility(self=@n9, InVisibility=SelfHitTestInvisible)
   > ActivateWidget(self=Get SeasonPass_ClaimPages())
   "On Clicked (Button_Battlepass)"() {is=title K2Node_ComponentBoundEvent}
   > OpenBattlepass_VIZ()
   "On Unhovered (Button_Battlepass)"() {is=title K2Node_ComponentBoundEvent}
   > PlayAnimation(InAnimation=@n2, PlayMode=Reverse)
   n11 = "On UIKit Tab Selected (WBP_UIKit_TabGroup)"!() -> (OutputDelegate, then, TabID:Name=None, TabVM:MVVM Base Viewmodel Object Reference) {is=title K2Node_ComponentBoundEvent}
   > n12 = Cast To SeasonPassCategoryVM!(execute, Object=@n11.TabVM) -> (then, CastFailed, AsSeason Pass Category VM:Season Pass Category VM Object Reference) {is=title K2Node_DynamicCast}
   > NavigateToReward!(
    execute
    self:Season Pass Selection VM Object Reference=Get SeasonPassSelectionVM()
    Reward:Season Pass Reward VM Object Reference=GetFirstUnlockedReward!(self:Season Pass Category VM Object Reference=@n12) -> (ReturnValue:Season Pass Reward VM Object Reference)
   ) -> (then)
   > SetRenderOpacity(self=Get CPanel_ContextMenuRoot(), InOpacity=0)
   > Delay(Duration=0.2)
   > RefreshContextMenuPosition()
   n13 = Back() {is=title K2Node_CustomEvent}
   ShowBattlepassPurchase() {is=title K2Node_CustomEvent}
   > n14 = PushPrimaryContentSoftClassAsync!(
    execute:Exec
    SoftWidgetClass:Common Activatable Widget Soft Class Reference=/SeasonPass/Giftbox/BP_BattlePassUnlockAnim.BP_BattlePassUnlockAnim_C
   ) -> (then:Exec, OnPrimaryContentPushed:Exec, ActivatableWidget:Common Activatable Widget Object Reference) {is=title K2Node_AsyncAction}
   n15 = Cast To BP_BattlePassUnlockAnim!(execute=@n14.OnPrimaryContentPushed, Object=@n14) -> (then, CastFailed, AsBP Battle Pass Unlock Anim:BP Battle Pass Unlock Anim Object Reference) {is=title K2Node_DynamicCast}
   > n16 = SetPass!(
    execute
    self:BP Battle Pass Unlock Anim Object Reference=@n15
    Pass:Season Pass VM Object Reference=Get SeasonPassVM() -> (SeasonPassVM:Season Pass VM Object Reference)
   ) -> (then)
   n17 = HandleBattlepassPurchaseAnimFinished() {is=title K2Node_CustomEvent}
   Bind Event to On Unlock Animation Finished(execute=@n16.then, self:BP Battle Pass Unlock Anim Object Reference=@n15, Delegate=@n17) {is=title K2Node_AddDelegate}
   > PlayIntro(self:BP Battle Pass Unlock Anim Object Reference=@n15)
   Disable RewardsAvailable Audio() {is=title K2Node_CustomEvent}
   > Set IsRewardsAvailableAudioAllowed(IsRewardsAvailableAudioAllowed:bool=false) -> (Output_Get:bool)
   > StopAudio(self:WBP Reward Info Widget Object Reference=Get WBP_RewardInfoWidget() -> (WBP_RewardInfoWidget:WBP Reward Info Widget Object Reference))
   OnInitialPageRewardItemVMsChanged!() -> (OutputDelegate, then, Object:Object Reference, Field:Field Notification Id Structure) {is=title K2Node_CustomEvent}
   > Branch(Condition=AreAllSelectedPageItemVMsReady!() -> (AllItemVMsReady:bool) {is=title K2Node_CallFunction})
   n18 = Get SeasonPassSelectionVM()
   > Initialize Selected Page And Reward!(
    execute
    self
    PassPageVM:Season Pass Page VM Object Reference=Get SelectedPage(self=@n18)
    RewardToSelect:Season Pass Reward VM Object Reference=Get SelectedReward(self=@n18)
   ) -> (then)
   n19 = RequestLoadBackgroundMaterial!() -> (OutputDelegate, then, BackgroundMaterialAsset:Soft Object Reference) {is=title K2Node_CustomEvent}
   > n20 = Async Load Asset(Asset=@n19.BackgroundMaterialAsset)
   n21 = Cast To MaterialInterface!(execute=@n20.Completed, Object=@n20) -> (then, CastFailed, AsMaterial Interface:Material Interface Object Reference) {is=title K2Node_DynamicCast}
   > n22 = CreateDynamicMaterialInstance!(execute, Parent=@n21, OptionalName=None, CreationFlags=None) -> (then, ReturnValue)
   > SetScalarParameterValue(self=@n22, ParameterName=OverrideTheme, Value=0)
   > SetBrushFromLazyMaterial!(
    execute
    self:Common Lazy Image Object Reference=Get LImg_Background() -> (LImg_Background:Fort Lazy Image Object Reference)
    LazyMaterial:"Material Interface Soft Object Reference (by ref)"=Make Soft Reference!(Input:Material Instance Dynamic Object Reference=@n22) -> (Output:Material Instance Dynamic Soft Object Reference)
   ) -> (then)
   OnXpRundownComplete() {is=title K2Node_CustomEvent}
   > BattlepassIntro_VIZ()
   > Delay(Duration=2)
   > ShowHideRewardsAvailableWidget!(execute, self, ShouldShow:bool=true, AllowAudio:bool=Get IsRewardsAvailableAudioAllowed() -> (IsRewardsAvailableAudioAllowed:bool)) -> (then)
   n23 = Update Reward List!() -> (OutputDelegate, then, ItemList:Array of Season Pass Reward VM Object References, RewardToSelect:Season Pass Reward VM Object Reference) {is=title K2Node_CustomEvent}
   > Reset!(execute, self:Dynamic Entry Box Object Reference=Get ActiveRewardList(), bDeleteWidgets:bool=false) -> (then)
   > UpdateClaimPagesButtonVisibility()
   > n24 = For Each Loop(Array=@n23.ItemList)
   > n25 = Create Entry(self=Get ActiveRewardList())
   > n26 = Cast To WBP_RewardTrack_ItemButton!(execute, Object=@n25) -> (then, CastFailed, AsWBP Reward Track Item Button:WBP Reward Track Item Button Object Reference) {is=title K2Node_DynamicCast}
   > Set SeasonPassRewardVM!(execute, SeasonPassRewardVM:Season Pass Reward VM Object Reference, self:WBP Reward Track Item Button Object Reference=@n26) -> (then, Output_Get:Season Pass Reward VM Object Reference)
   > SetItemCardSeasonPassReward!(execute, self, RewardVM:Season Pass Reward VM Object Reference=@n24, ItemCard:WBP Reward Track Item Button Object Reference=@n26) -> (then)
   n27 = QuickCloseClaimPages() {is=title K2Node_CustomEvent}
   > QuickDismissClaimPages()
   n28 = OnVariantFocused!() -> (OutputDelegate, then, Card:WBP UIKit Item Card Battle Pass Object Reference) {is=title K2Node_CustomEvent}
   > Call New Variant Card Selected!(execute:Exec, self:WBP Season Pass Reward Track Object Reference, NewCard:WBP UIKit Item Card Battle Pass Object Reference=@n28.Card) -> (then:Exec) {is=title K2Node_CallDelegate}
   > Set FortItemDetailsVM!(
    execute
    FortItemDetailsVM:Fort Item Details VM Object Reference=Get ItemDetails!(self:Fort Item VM Object Reference=Get FortItemVM!(self:WBP UIKit Item Card Battle Pass Object Reference=@n28.Card) -> (FortItemVM:Fort Item VM Object Reference)) -> (ItemDetails:Fort Item Details VM Object Reference)
    self:WBP UIKit Description With Subtitle Object Reference=Get WBP_UIKit_DescriptionWithSubtitle() -> (WBP_UIKit_DescriptionWithSubtitle:WBP UIKit Description With Subtitle Object Reference)
   ) -> (then, Output_Get:Fort Item Details VM Object Reference)
   DismissClaimPagesEvent() {is=title K2Node_CustomEvent}
   > n29 = Play Animation with Finished event(Widget=Self-Reference(), InAnimation=Get ToClaimPages(), StartAtTime=0, NumLoopsToPlay=1, PlayMode=Reverse, PlaybackSpeed=2)
   > SetVisibility(self=Get WBP_UIKit_TabGroup(), InVisibility=SelfHitTestInvisible)
   n30 = Get Button_Back()
   > SetVisibility(self=@n30, InVisibility=Collapsed)
   DeactivateWidget(execute=@n29.Finished, self=Get SeasonPass_ClaimPages())
   > SetVisibility(self=@n30, InVisibility=SelfHitTestInvisible)
   n31 = Get SeasonPassCategoryVM() -> (SeasonPassCategoryVM:Season Pass Category VM Object Reference)
   > SetVisibility(
    self=Get ClaimAllButtonWrapper()
    InVisibility=GetClaimAllRewardsButtonVisibility!(self, ClaimableRewardsAmount:int=GetNumPurchaseableRewards!(self:Season Pass Element Base VM Object Reference=@n31) -> (ReturnValue:int=0)) -> (Result:ESlateVisibility Enum=Visible)
   )
   > Set SelectedReward!(
    execute
    SelectedReward:Season Pass Reward VM Object Reference=Get ClaimPagesBackLastSelectedReward() -> (ClaimPagesBackLastSelectedReward:Season Pass Reward VM Object Reference)
    self:Season Pass Selection VM Object Reference=Get SeasonPassSelectionVM()
   ) -> (then, Output_Get:Season Pass Reward VM Object Reference)
   > Set IsClaimPagesVisible(IsClaimPagesVisible=false)
   > UpdateClaimPagesButtonVisibility()
   InitializeScreen() {is=title K2Node_CustomEvent}
   > Branch(Condition=NOT Boolean(A=Get IsScreenInitialized() -> (IsScreenInitialized:bool)))
   > Set IsTransitionActive(IsTransitionActive=false)
   n32 = Get SeasonPassSelectionVM()
   > n33 = Is Valid(InputObject=Get SelectedPage(self=@n32))
   n34 = Get SeasonPassSelectionVM()
   n35 = Get SelectedPage(self=@n34)
   n36 = Is Valid(exec=[@n42.then, @n33.Is Valid], InputObject=@n35)
   > Initialize Selected Page And Reward(PassPageVM=@n35, RewardToSelect=Get SelectedReward(self=@n34))
   > n37 = Set IsScreenInitialized(IsScreenInitialized:bool=true) -> (Output_Get:bool)
   n38 = Get SeasonPassVM()
   Fire Event Sub View Opened!(
    execute=[@n37.then, @n40.then]
    View:Widget Object Reference=Self-Reference()
    AdditionalInfo:Flow Analytics Additional Info Structure="Add Text (Localized)"!(
     Info:"Flow Analytics Additional Info Structure (by ref)"=Make Flow Analytics Additional Info!() -> (FlowAnalyticsAdditionalInfo:Flow Analytics Additional Info Structure) {is=title K2Node_MakeStruct}
     Key:String=PassName
     Value:"Text (by ref)"=GetDisplayName!(self:Season Pass VM Object Reference=@n38) -> (ReturnValue:Text)
    ) -> (ReturnValue:Flow Analytics Additional Info Structure)
    bIsOverlay:bool=false
   ) -> (then)
   > SetVisibility(self=Get Button_Back(), InVisibility=SelfHitTestInvisible)
   > n39 = Branch(Condition=Get bIsEnabled!(self:Widget Object Reference=Get Button_ClaimAllRewards()) -> (bIsEnabled:bool))
   Set bIsEnabled(execute=@n39.else, bIsEnabled=true, self=Get Button_ClaimAllRewards())
   n40 = Set Page Index(execute=@n36.Is Not Valid, Page Index=0)
   n41 = Get SelectedCategory!(self:Season Pass Selection VM Object Reference=@n32) -> (SelectedCategory:Season Pass Category VM Object Reference)
   Is Valid(exec=@n33.Is Not Valid, InputObject=@n41)
   > n42 = NavigateToReward(self=@n32, Reward=GetFirstUnlockedReward(self=@n41))
   n43 = OnRewardPressed!() -> (OutputDelegate, then, ItemCard:WBP UIKit Item Card Base Object Reference) {is=title K2Node_CustomEvent}
   > n44 = Cast To WBP_RewardTrack_ItemButton(Object=@n43.ItemCard)
   > Branch on Input Type!(In:Exec) -> (MnKOrTouch:Exec, Gamepad:Exec) {is=title K2Node_MacroInstance}
   > Branch(Condition=NOT Boolean(A=Get IsBattlepassOpen()))
   > OpenBattlepass_VIZ()
   n45 = Get SeasonPassRewardVM!(self:WBP Reward Track Item Button Object Reference=@n44) -> (SeasonPassRewardVM:Season Pass Reward VM Object Reference)
   n46 = Get SeasonPassSelectionVM()
   > Branch(Condition="Not Equal (Object)"(A:Object Reference=@n45, B:Object Reference=Get SelectedReward(self=@n46)) -> (ReturnValue:bool=~) {is=title K2Node_PromotableOperator})
   > Set SelectedReward(SelectedReward=@n45, self=Get SeasonPassSelectionVM())
   Event On Deactivated()
   > CloseBattlepass_VIZ(bSkipAnimation=true)
   > Hide_VIZ!(execute, self:WBP Season Pass Buy Pass Preview Object Reference=Get WBP_BuyPassPreview(), NeedsToAnimate:bool=false) -> (then)
   Event Construct()
   > SetMobileStyle() {is=title K2Node_CallFunction}
   n47 = Get SeasonPass_ClaimPages()
   > Bind Event to Finish Claim Event Dispatcher(self:Season Pass Claim Pages Object Reference=@n47, Delegate=@n13) {is=title K2Node_AddDelegate}
   > Bind Event to Quick Close Claim Pages Dispatcher(self:Season Pass Claim Pages Object Reference=@n47, Delegate=@n27) {is=title K2Node_AddDelegate}
   > n48 = StartListeningToStatefulEvent!(
    execute:Exec
    Target:Gameplay Event Router Component Object Reference=GetGlobalEventRouter()
    EventType:Script Struct Object Reference=/Script/FortniteUI.MatchXPScreenEvent_XPRundownSequenceComplete
    Context:Object Reference
    EventBubblingRule:EEventBubblingRule Enum
    None:%=Self-Reference()
   ) -> (
    then:Exec
    AsyncTaskProxy:Async Action Start Listening To Stateful Event Object Reference
    OnEventReceived:Exec
    OnSavedEventStateExists:Exec
    OnEventStateCleared:Exec
    Payload:Match XPScreen Event XPRundown Sequence Complete Structure
   ) {is=title K2Node_StartListeningToStatefulEventAsyncAction}
   > n49 = StartListeningToEvent!(
    execute:Exec
    Target:Gameplay Event Router Component Object Reference=GetGlobalEventRouter()
    EventType:Script Struct Object Reference=/Script/FortniteUI.MatchXPScreenEvent_ReadyUpClicked
    Context:Object Reference
    EventBubblingRule:EEventBubblingRule Enum
    None:%=Self-Reference()
   ) -> (
    then:Exec
    AsyncTaskProxy:Async Action Start Listening To Event Object Reference
    OnEventReceived:Exec
    Payload:Match XPScreen Event Ready Up Clicked Structure
   )
   > SetVisibility(self=Get VisSwitcher_PassPurchase() -> (VisSwitcher_PassPurchase:Common Visibility Switcher Object Reference), InVisibility=Collapsed)
   Disable RewardsAvailable Audio(execute=@n49.OnEventReceived) {is=title K2Node_CallFunction}
   OnXpRundownComplete(execute=[@n48.OnEventReceived, @n48.OnSavedEventStateExists]) {is=title K2Node_CallFunction}
   Event PreConstruct()
   > UpdateActiveTrack!(
    execute
    self
    ActiveTrack:Named Slot Object Reference=Get NamedSlot_Track_Current() -> (NamedSlot_Track_Current:Named Slot Object Reference)
   ) -> (then)
   Event OnInitialized()
   > n50 = GetConsoleVariableBoolValue!(execute, VariableName:String=Athena.MatchXPScreenWidget.S41SuppressPostGameBattlePassUpsellPreview) -> (then, ReturnValue:bool)
   > Set S41SuppressPostGameBattlePassUpsellPreview(S41SuppressPostGameBattlePassUpsellPreview:bool=@n50) -> (Output_Get:bool)
   Event OnAddedToFocusPath!() -> (OutputDelegate, then, InFocusEvent:Focus Event Structure) {is=title K2Node_Event}
   > n51 = Branch on Input Type()
   n52 = IsBattlePassTrackActive!() -> (ReturnValue:bool)
   Branch(execute=@n51.Gamepad, Condition=@n52)
   > OpenBattlepass_VIZ()
   > n53 = Cast To FortPlayerController(Object=GetOwningPlayer())
   > "Fire Event Element Interacted (Non-Localized)"(FortPC=@n53, ElementName=PostMatchScreen.SeasonPass_RewardTrack.FocusedGamepad)
   Event OnRemovedFromFocusPath!() -> (OutputDelegate, then, InFocusEvent:Focus Event Structure) {is=title K2Node_Event}
   > n54 = Branch on Input Type()
   Branch(execute=@n54.Gamepad, Condition=@n52)
   > CloseBattlepass_VIZ(bSkipAnimation=false)
   > n55 = Cast To FortPlayerController(Object=GetOwningPlayer())
   > "Fire Event Element Interacted (Non-Localized)"(FortPC=@n55, ElementName=PostMatchScreen.SeasonPass_RewardTrack.UnfocusedGamepad)
   n56 = "Animation Finished (NextPage)"!() -> (OutputDelegate:Delegate, then:Exec) {is=title K2Node_WidgetAnimationEvent}
   RefreshContextMenuPosition(execute=[@n56.then, @n57.then])
   n57 = "Animation Finished (PreviousPage)"!() -> (OutputDelegate:Delegate, then:Exec) {is=title K2Node_WidgetAnimationEvent}
   RequestRefreshFocus(execute=@n58.then, self:Common Activatable Widget Object Reference)
   n58 = RewardTrackIntro_Viz(execute=@n59.then)
   n59 = DeactivateWidget(execute=@n17.then, self=@n15)
  ]}
 Get Page Text {nodes [
   n1 = Get Page Text!() -> (then, Page Index:int) {is=title K2Node_FunctionEntry}
   n2 = Get SeasonPassCategoryVM()
   > n3 = Is Valid(InputObject=@n2)
   Return Node!(execute=@n3.Is Not Valid, Result:Text)
   n4 = Get Pages!(self:Season Pass Category VM Object Reference=@n2) -> (Pages:Array of Season Pass Page VM Object References)
   > Return Node(Result=Format Text!(
     Format="LOCTEXT(\"{PageIndex}/{PageCount}\")"
     PageIndex:int=int + int(A=@n1)
     PageCount:int=Length(TargetArray:Array of Season Pass Page VM Object References=@n4)
    ) -> (Result))
  ]}
 Get Page Text Visibility {nodes [
   n1 = Get Page Text Visibility!() -> (then, Category:Season Pass Category VM Object Reference) {is=title K2Node_FunctionEntry}
   > n2 = Is Valid(InputObject=@n1)
   n3 = Length(TargetArray=Get Pages(self=@n1))
   > Return Node(Result:ESlateVisibility Enum=Select(Option 0:ESlateVisibility Enum=Hidden, Option 1:ESlateVisibility Enum=SelfHitTestInvisible, Index="integer > integer"(A=@n3, B=1)) -> (ReturnValue:ESlateVisibility Enum))
   Return Node(execute=@n2.Is Not Valid, Result=Hidden)
  ]}
 GetClaimAllRewardsButtonText {nodes [
   n1 = GetClaimAllRewardsButtonText!() -> (then, ClaimableRewardsAmount:int) {is=title K2Node_FunctionEntry}
   > Return Node!(execute, ButtonText:Text=Format Text!(Format="LOCTEXT(\"Claim all rewards ({RewardsAmount})\")", RewardsAmount:int=@n1) -> (Result))
  ]}
 GetClaimAllRewardsButtonVisibility {nodes [
   n1 = GetClaimAllRewardsButtonVisibility!() -> (then, ClaimableRewardsAmount:int) {is=title K2Node_FunctionEntry}
   > Return Node!(execute, Result:ESlateVisibility Enum=Select(Option 0=Collapsed, Option 1=SelfHitTestInvisible, Index="integer > integer"(A=@n1)))
  ]}
 GetContainsBonusInfo {nodes [
   n1 = Get ItemDefinition!(self:Fort Item VM Object Reference=Get CurrentItemVM!(self:Fort Item Preview Widget Object Reference) -> (CurrentItemVM:Fort Item VM Object Reference)) -> (ItemDefinition:Fort Item Definition Object Reference)
   n2 = Get ItemDefinition(self=GetContainsBonusInfo!() -> (then, FortItemVM:Fort Item VM Object Reference) {is=title K2Node_FunctionEntry})
   Return Node!(
    execute
    ContainsBonusInfo:bool="Not Equal Exactly (String)"(A=GetTemplateId!(self:Mcp Item Definition Base Object Reference=@n2) -> (ReturnValue:String), B=GetTemplateId(self=@n1))
   )
  ]}
 GetFramingDataFromItem {nodes [
   n1 = GetFramingDataFromItem!() -> (then, ItemVM:Fort Item VM Object Reference) {is=title K2Node_FunctionEntry}
   n2 = GetItemType!(ItemDefinition:Item Definition Base Object Reference=Get ItemDefinition(self=@n1)) -> (ReturnValue:EFortItemType Enum=WorldItem)
   n3 = Find!(
    TargetMap:Map of S Item Preview Framing Data Key Structures to S Item Preview Framing Data Values Structures=Get ItemPreviewFramingOverrides() -> (ItemPreviewFramingOverrides:Map of S Item Preview Framing Data Key Structures to S Item Preview Framing Data Values Structures)
    Key:"S Item Preview Framing Data Key Structure (by ref)"=Make S Item Preview Framing Data Key!(
     ItemType_2_#34:EFortItemType Enum=@n2
     ProductTag_5_#35:Gameplay Tag Structure=GetCurrentProduct!(self:Fort Item VM Object Reference=@n1) -> (ReturnValue:Gameplay Tag Structure)
    ) -> (S_ItemPreviewFramingData_Key:S Item Preview Framing Data Key Structure) {is=title K2Node_MakeStruct}
   ) -> (
    Value_LandingData_16_#36:S Item Preview Framing Data Structure
    Value_ViewRewardData_18_#37:S Item Preview Framing Data Structure
    ReturnValue
   )
   > n4 = Branch(Condition=@n3.ReturnValue)
   > Return Node!(
    execute
    FramingData:S Item Preview Framing Data Structure=Select(
     Option 0:S Item Preview Framing Data Structure=@n3
     Option 1:S Item Preview Framing Data Structure=@n3.Value_ViewRewardData_18_#37
     Index=Get IsViewingReward() -> (IsViewingReward:bool)
    ) -> (ReturnValue:S Item Preview Framing Data Structure)
   )
   n5 = Find(
    TargetMap=Get ItemPreviewFramingOverrides()
    Key=Make S Item Preview Framing Data Key(ItemType_2_#34=@n2, ProductTag_5_#35=(TagName=""))
   )
   n6 = Branch(execute=@n4.else, Condition=@n5.ReturnValue)
   > Return Node(FramingData=Select(Option 0=@n5, Option 1=@n5.Value_ViewRewardData_18_#37, Index=Get IsViewingReward()))
   Return Node(
    execute=@n6.else
    FramingData=Select(
     Option 0=Get FallbackFramingData_Landing() -> (FallbackFramingData_Landing:S Item Preview Framing Data Structure)
     Option 1=Get FallbackFramingData_Landing_ViewReward() -> (FallbackFramingData_Landing_ViewReward:S Item Preview Framing Data Structure)
     Index=Get IsViewingReward()
    )
   )
  ]}
 GetLastClaimedReward {nodes [
   n1 = GetLastClaimedReward!() -> (then, Categories:Array of Season Pass Category VM Object References) {is=title K2Node_FunctionEntry}
   > n2 = Branch(Condition="integer > integer"(A=Length(TargetArray:Array of Season Pass Category VM Object References=@n1)))
   > n3 = For Each Loop(Array:Array of Season Pass Category VM Object References=@n1) -> (Array Element:Season Pass Category VM Object Reference)
   Return Node!(
    execute=@n3.Completed
    Reward:Season Pass Reward VM Object Reference=Get LocalLastClaimedReward() -> (LocalLastClaimedReward:Season Pass Reward VM Object Reference)
   )
   > n4 = For Each Loop(Array:Array of Season Pass Page VM Object References=Get Pages(self=@n3)) -> (Array Element:Season Pass Page VM Object Reference)
   > n5 = For Each Loop(Array:Array of Season Pass Track VM Object References=Get Tracks(self=@n4)) -> (Array Element:Season Pass Track VM Object Reference)
   > n6 = For Each Loop(Array:Array of Season Pass Reward VM Object References=Get Rewards(self=@n5)) -> (Array Element:Season Pass Reward VM Object Reference)
   > n7 = Branch(Condition=Get bIsRewardOwned!(self:Season Pass Reward VM Object Reference=@n6) -> (bIsRewardOwned:bool))
   > Set LocalLastClaimedReward(LocalLastClaimedReward:Season Pass Reward VM Object Reference=@n6) -> (Output_Get:Season Pass Reward VM Object Reference)
   Return Node(execute=@n7.else, Reward=Get LocalLastClaimedReward())
   Return Node(execute=@n2.else)
  ]}
 GetPage {nodes [
   n1 = GetPage!() -> (then, PageIndex:int, Category:Season Pass Category VM Object Reference) {is=title K2Node_FunctionEntry}
   > n2 = Is Valid(InputObject=@n1.Category)
   Return Node!(execute=@n2.Is Not Valid, PassPageVM:Season Pass Page VM Object Reference)
   > Return Node(PassPageVM="Get (a copy)"(Array:Array of Season Pass Page VM Object References=Get Pages(self=@n1.Category), Dimension 1=@n1) -> (Output:Season Pass Page VM Object Reference))
  ]}
 GetPrimaryContentSetup {nodes [
   GetPrimaryContentSetup() {is=title K2Node_FunctionEntry}
   > Return Node!(
    execute
    ReturnValue:Primary Content Setup Structure=Make Primary Content Setup!(
     bShowBottomBar:bool=false
     bShowChatWidget:bool=false
     bEnableNewTextChat:bool=False
     bShowTopBar:bool=false
     bTabSupportsInterface:bool=False
    ) -> (PrimaryContentSetup:Primary Content Setup Structure) {is=title K2Node_MakeStruct}
   )
  ]}
 GetRewardsInPage {nodes [
   n1 = GetRewardsInPage!() -> (then, Page:Season Pass Page VM Object Reference) {is=title K2Node_FunctionEntry}
   > n2 = Is Valid(InputObject=@n1)
   Return Node!(
    execute=@n2.Is Not Valid
    Rewards:Array of Season Pass Reward VM Object References=Make Array!() -> (Array:Array of Season Pass Reward VM Object References)
   )
   n3 = Get Tracks(self=@n1)
   > Return Node(Rewards=Get Rewards(self="Get (a copy)"(Array:Array of Season Pass Track VM Object References=@n3) -> (Output:Season Pass Track VM Object Reference)))
  ]}
 HasValidSeasonPassData {nodes [
   HasValidSeasonPassData() {is=title K2Node_FunctionEntry}
   > n1 = Is Valid(InputObject=Get SeasonPassVM())
   n2 = Get Categories!(self:Season Pass VM Object Reference=Get SeasonPassVM()) -> (Categories:Array of Season Pass Category VM Object References)
   > n3 = Branch(Condition="integer > integer"(A=Length(TargetArray=@n2)))
   > Return Node!(execute, bHasValidData:bool=true)
   Return Node(execute=[@n3.else, @n1.Is Not Valid], bHasValidData=false)
  ]}
 IncreasePage {nodes [
   IncreasePage() {is=title K2Node_FunctionEntry}
   n1 = Get IsTransitionActive()
   > Branch(Condition=AND Boolean(A=Get IsBattlepassOpen(), B=NOT Boolean(A=@n1)))
   > Set IsTransitionActive(IsTransitionActive=true)
   > ClearRewardFocus()
   > ResetActiveTrackSlot(ActiveTrack=Get NamedSlot_Track_Next())
   > n2 = GetLinearSeasonPassPageNavTarget(SeasonPassSelectionVM=Get SeasonPassSelectionVM(), IsPrevious=false, ShowLockedCategories=true, UseMainSetOnly=Get bUseMainSetOnly()) -> (SelectedPage=0)
   > Set NewPageRewardTarget(NewPageRewardTarget=@n2)
   > Set SelectedCategory(SelectedCategory=@n2.SeasonPassCategoryTarget, self=Get SeasonPassSelectionVM())
   > Set Page Index(Page Index=@n2.SelectedPage)
   > BindToAnimationFinished(Animation=Get NextPage(), Delegate=Create Event())
   > n3 = Branch(Condition=IsAnimationPlaying(InAnimation=Get PreviousPage()))
   > n4 = StopAnimation(InAnimation=Get PreviousPage())
   PlayAnimationForward(execute=[@n4.then, @n3.else], InAnimation=Get NextPage())
   > n5 = Cast To FortPlayerController(Object=GetOwningPlayer())
   > "Fire Event Element Interacted (Non-Localized)"(FortPC=@n5, ElementName=PostMatchScreen.SeasonPass_RewardTrack.NavigatedToNextPage)
  ]}
 Initialize Selected Page And Reward {nodes [
   n1 = Initialize Selected Page And Reward!() -> (then, PassPageVM:Season Pass Page VM Object Reference, RewardToSelect:Season Pass Reward VM Object Reference) {is=title K2Node_FunctionEntry}
   > Is Valid(InputObject=@n1)
   n2 = Get SeasonPassCategoryVM()
   > Is Valid(InputObject=@n2)
   n3 = Get Pages(self=@n2)
   n4 = Get PassPageVM() -> (PassPageVM:Season Pass Page VM Object Reference)
   > Set PageIndex(PageIndex:int="Max (Integer)"(
     A=Find Item!(TargetArray:Array of Season Pass Page VM Object References=@n3, ItemToFind:"Season Pass Page VM Object Reference (by ref)"=@n4) -> (ReturnValue)
     B=0
    )) -> (Output_Get:int=0)
   > UpdateNavigationArrows()
   n5 = GetRewardsAndQuest!(SeasonPassPage:Season Pass Page VM Object Reference=Get PassPageVM()) -> (Rewards:Array of Season Pass Reward VM Object References, Quest:Season Pass Reward VM Object Reference)
   > Set Local_RewardsInPage(Local_RewardsInPage:Array of Season Pass Reward VM Object References=@n5) -> (Output_Get:Array of Season Pass Reward VM Object References)
   > Set QuestReward_VM(QuestReward_VM:Season Pass Reward VM Object Reference=@n5.Quest) -> (Output_Get:Season Pass Reward VM Object Reference)
   > SetVisibility(
    self=Get Ovr_ModeSpecificQuestArea() -> (Ovr_ModeSpecificQuestArea:Overlay Object Reference)
    InVisibility=Select(Option 0:ESlateVisibility Enum=Collapsed, Option 1:ESlateVisibility Enum=SelfHitTestInvisible, Index=IsValid(Object=@n5.Quest)) -> (ReturnValue:ESlateVisibility Enum)
   )
   > SetSeasonPassRewardVM!(
    execute
    self:WBP Reward Track Item Button Object Reference=Get ModeSpecificReward()
    Viewmodel:Season Pass Reward VM Object Reference=@n5.Quest
   ) -> (then)
   n6 = Find Item(
    TargetArray:Array of Season Pass Reward VM Object References=Get Local_RewardsInPage() -> (Local_RewardsInPage:Array of Season Pass Reward VM Object References)
    ItemToFind:"Season Pass Reward VM Object Reference (by ref)"=Get RewardToSelect() -> (RewardToSelect:Season Pass Reward VM Object Reference)
   )
   > n7 = Branch(Condition="Equal (Integer)"(A=@n6, B=-1))
   > Set SelectedReward(SelectedReward=Get QuestReward_VM() -> (QuestReward_VM:Season Pass Reward VM Object Reference), self=Get SeasonPassSelectionVM())
   > n8 = Set SelectedRewardIndex(SelectedRewardIndex:int=-1) -> (Output_Get:int=0)
   Update Reward List!(
    execute=[@n9.then, @n8.then]
    self
    ItemList:Array of Season Pass Reward VM Object References=Get Local_RewardsInPage()
    RewardToSelect:Season Pass Reward VM Object Reference=Get SelectedReward(self=Get SeasonPassSelectionVM())
   ) -> (then) {is=title K2Node_CallFunction}
   Set SelectedRewardIndex(execute=@n7.else, SelectedRewardIndex=@n6) -> (Output_Get=0)
   > n9 = Set SelectedReward(
    SelectedReward="Get (a copy)"(
     Array:Array of Season Pass Reward VM Object References=Get Local_RewardsInPage()
     Dimension 1=Get SelectedRewardIndex() -> (SelectedRewardIndex=0)
    ) -> (Output:Season Pass Reward VM Object Reference)
    self=Get SeasonPassSelectionVM()
   )
  ]}
 InitializeCustomItemDetails {nodes [
   n1 = InitializeCustomItemDetails!() -> (then, ItemVM:Fort Item VM Object Reference) {is=title K2Node_FunctionEntry}
   > ClearChildren(self=Get NamedSlot_CustomBanner() -> (NamedSlot_CustomBanner:Named Slot Object Reference))
   n2 = Get ItemDetails(self=@n1)
   > CreateCustomItemDetails!(
    execute
    self:WBP Custom Item Details Container Object Reference=Get WBP_CustomItemDetailsContainer() -> (WBP_CustomItemDetailsContainer:WBP Custom Item Details Container Object Reference)
    Item Details VM:Fort Item Details VM Object Reference=@n2
   ) -> (then)
   n3 = Get CustomInspectScreenBanner!(self:Fort Locker Category Config Object Reference=GetItemCategory!(self:Fort Item VM Object Reference=Get Item!(self:Fort Item Details VM Object Reference=@n2) -> (Item:Fort Item VM Object Reference)) -> (ReturnValue:Fort Locker Category Config Object Reference)) -> (CustomInspectScreenBanner:User Widget Soft Class Reference)
   > Branch(Condition=IsValidClass!(Class:Object Class Reference=Resolve Soft Reference!(Input:User Widget Soft Class Reference=@n3) -> (Output:User Widget Class Reference)) -> (ReturnValue:bool))
   > n4 = Create Widget!(
    execute:Exec
    Class:User Widget Class Reference=Resolve Soft Reference(Input=@n3)
    OwningPlayer:Player Controller Object Reference=GetOwningPlayer()
   ) -> (then:Exec, ReturnValue:User Widget Object Reference)
   > AddChild(self=Get NamedSlot_CustomBanner(), Content=@n4)
   > n5 = Cast To WBP_CustomBanner_Base!(execute, Object=@n4) -> (then, CastFailed, AsWBP Custom Banner Base:WBP Custom Banner Base Object Reference) {is=title K2Node_DynamicCast}
   > SetSingleItem!(execute, self:WBP Custom Banner Base Object Reference=@n5, ItemVM:Fort Item VM Object Reference=Get ItemVM()) -> (then)
   > SetViewType!(execute, self:WBP Custom Banner Base Object Reference=@n5, IsFullDetailsView:bool=true) -> (then)
  ]}
 IsBattlePassTrackActive {nodes [
   IsBattlePassTrackActive() {is=title K2Node_FunctionEntry}
   n1 = Get VisSwitcher_PassPurchase()
   > Return Node!(
    execute
    ReturnValue:bool="Equal (Object)"(
     A=GetActiveWidget!(self:Common Visibility Switcher Object Reference=@n1) -> (ReturnValue:Widget Object Reference)
     B=Get Overlay_RewardTrack_Root() -> (Overlay_RewardTrack_Root:Overlay Object Reference)
    )
   )
  ]}
 NewVariantCardSelected {nodes [NewVariantCardSelected!() -> (then, NewCard:WBP UIKit Item Card Battle Pass Object Reference) {is=title K2Node_FunctionEntry}]}
 OnMouseWheel {nodes [
   n1 = OnMouseWheel!() -> (then, MyGeometry:Geometry Structure, MouseEvent:"Pointer Event Structure (by ref)") {is=title K2Node_FunctionEntry}
   > n2 = Branch(Condition=OR Boolean(A=Get IsViewingReward(), B=Get IsClaimPagesVisible()))
   > Return Node(ReturnValue:Event Reply Structure=Unhandled())
   n3 = Branch(
    execute=@n2.else
    Condition="float > float"(A=Get Wheel Delta!(Input:"Pointer Event Structure (by ref)"=@n1.MouseEvent) -> (ReturnValue:float) {is=title K2Node_CallFunction})
   )
   > n4 = DecreasePage()
   Return Node(execute=[@n4.then, @n5.then], ReturnValue=Handled())
   n5 = IncreasePage(execute=@n3.else) {is=title K2Node_CallFunction}
  ]}
 OnPageTransitionAnimationFinished {nodes [
   OnPageTransitionAnimationFinished() {is=title K2Node_FunctionEntry}
   > Set IsTransitionActive(IsTransitionActive=false)
   > SwapTracks()
   n1 = Get NamedSlot_Track_Current()
   > SetRenderOpacity(self=@n1, InOpacity=1)
   > SetRenderTranslation(self=@n1)
   > SetRenderOpacity(self=Get NamedSlot_Track_Next(), InOpacity=0)
   > SetVisibility(self=Get NamedSlot_Track_Next(), InVisibility=Hidden)
   > Set LastScrolledDirection(LastScrolledDirection:EVaultWorldTransitionDirection Enum=None) -> (Output_Get:EVaultWorldTransitionDirection Enum=Forward)
   > RestoreRewardFocus()
  ]}
 OnRewardsLeftRightNavFailed {nodes [
   n1 = OnRewardsLeftRightNavFailed!() -> (then, Navigation:EUINavigation Enum) {is=title K2Node_FunctionEntry}
   > n2 = Switch on EUINavigation!(execute, Selection:EUINavigation Enum=@n1) -> (Left:Exec, Right:Exec, Up:Exec, Down:Exec, Next:Exec, Previous:Exec, Invalid:Exec) {is=title K2Node_SwitchEnum}
   > n3 = IsBoundaryOnPageSelected!(execute, SeasonPassSelectionVM:Season Pass Selection VM Object Reference=Get SeasonPassSelectionVM(), IsFirst:bool=true) -> (then, IsSelected:bool)
   > n4 = Branch(Condition=@n3)
   > DecreasePage()
   Return Node(
    execute=@n4.else
    ReturnValue:Widget Object Reference=GetDesiredFocusTarget!(self:Common Activatable Widget Object Reference) -> (ReturnValue:Widget Object Reference)
   )
   n5 = IsBoundaryOnPageSelected(execute=@n2.Right, SeasonPassSelectionVM=Get SeasonPassSelectionVM(), IsFirst=false)
   n6 = Get SelectedReward(self=Get SeasonPassSelectionVM())
   n7 = Get ModeSpecificReward()
   > n8 = Branch(Condition=OR Boolean(A=@n5, B="Equal (Object)"(A=@n6, B=@n7)))
   > IncreasePage()
   Return Node(execute=@n8.else, ReturnValue=GetDesiredFocusTarget())
  ]}
 QuickDismissClaimPages {nodes [
   QuickDismissClaimPages() {is=title K2Node_FunctionEntry}
   > n1 = Sequence()
   > SetVectorParameterValue!(
    execute
    Collection:Material Parameter Collection Object Reference=/Game/Environments/FrontEnd/Blueprints/ItemPreview/VaultWorldTransitions/MaterialCollectionParameters/MPC_UI_Master_VaultWorld_Settings
    ParameterName:Name=FocalPointPosition
    ParameterValue_R:float=0
    ParameterValue_G:float=0.1
    ParameterValue_B:float=0
    ParameterValue_A:float=1
   ) -> (then) {is=title K2Node_CallMaterialParameterCollectionFunction}
   n2 = Get Button_ClaimPages()
   n3 = Get SeasonPass_Level() -> (SeasonPass_Level:Season Pass Level Object Reference)
   > SetVisibility(self=[Get WBP_UIKit_TabGroup(), @n2, @n3], InVisibility=SelfHitTestInvisible)
   > SetRenderOpacity(self=[@n2, @n3], InOpacity=1)
   > Set!(
    execute
    RenderTransform_Translation:Vector 2D Structure=(X=0,Y=0)
    RenderTransform_Scale:Vector 2D Structure=(X=1,Y=1)
    RenderTransform_Shear:Vector 2D Structure=(X=0,Y=0)
    RenderTransform_Angle:float=0
    self:Widget Object Reference
   ) -> (then, Output_Get:Widget Transform Structure)
   n4 = Get SeasonPass_ClaimPages()
   > SetVisibility(self=@n4, InVisibility=Collapsed)
   > SetRenderOpacity(self=@n4, InOpacity=0)
   Set IsClaimPagesVisible(execute=@n1.then_1, IsClaimPagesVisible=false)
   > Set SelectedReward(SelectedReward=Get ClaimPagesBackLastSelectedReward(), self=Get SeasonPassSelectionVM())
   n5 = Get SeasonPassCategoryVM()
   > SetVisibility(
    self=Get ClaimAllButtonWrapper()
    InVisibility=GetClaimAllRewardsButtonVisibility(ClaimableRewardsAmount=GetNumPurchaseableRewards(self=@n5) -> (ReturnValue=0)) -> (Result=Visible) {is=title K2Node_CallFunction}
   )
   > DeactivateWidget(self=Get SeasonPass_ClaimPages())
  ]}
 RefreshContextMenuPosition {nodes [
   RefreshContextMenuPosition() {is=title K2Node_FunctionEntry}
   n1 = GetAllEntries(self=Get ActiveRewardList())
   n2 = Get SelectedRewardIndex() -> (SelectedRewardIndex=0)
   > n3 = Branch(Condition=Is Valid Index!(TargetArray:Array of User Widget Object References=@n1, IndexToTest:int=@n2) -> (ReturnValue:bool))
   > UpdateContextMenuPosition!(
    execute
    self
    AnchorWidget:Widget Object Reference="Get (a copy)"(Array:Array of User Widget Object References=@n1, Dimension 1=@n2) -> (Output:User Widget Object Reference)
   ) -> (then)
   UpdateContextMenuPosition(execute=@n3.else, AnchorWidget=Get ModeSpecificReward())
  ]}
 ResetActiveTrackSlot {nodes [
   n1 = ResetActiveTrackSlot!() -> (then, ActiveTrack:Named Slot Object Reference) {is=title K2Node_FunctionEntry}
   n2 = Create Event()
   > UnbindFromAnimationFinished!(execute, self:User Widget Object Reference, Animation:Widget Animation Object Reference=Get NextPage(), Delegate:Delegate=@n2) -> (then)
   > UnbindFromAnimationFinished(Animation=Get PreviousPage(), Delegate=@n2)
   > UpdateActiveTrack(ActiveTrack=@n1)
   > Set Visibility(Visibility=SelfHitTestInvisible, self=@n1)
  ]}
 RestoreRewardFocus {nodes [
   RestoreRewardFocus() {is=title K2Node_FunctionEntry}
   > Branch(Condition=Get IsBattlepassOpen())
   > SetFocus(self=GetDesiredFocusTarget())
   > SetVisibility(self=Get CPanel_ContextMenuRoot(), InVisibility=SelfHitTestInvisible)
  ]}
 Sequencer Events {nodes [
   OnClosePass() {is=title K2Node_CustomEvent}
   > PopContentWidget!(execute) -> (then)
   ReinitializeSelectedPage() {is=title K2Node_CustomEvent}
   > n1 = Branch(Condition=Get IsClaimPagesVisible())
   n2 = Get SeasonPassSelectionVM()
   Initialize Selected Page And Reward(execute=@n1.else, PassPageVM=Get SelectedPage(self=@n2), RewardToSelect=Get SelectedReward(self=@n2)) {is=title K2Node_CallFunction}
  ]}
 Set Page Index {nodes [
   n1 = Set Page Index!() -> (then, Page Index:int) {is=title K2Node_FunctionEntry}
   > Set PageIndex(PageIndex=@n1) -> (Output_Get=0)
   n2 = Get SeasonPassCategoryVM()
   > Set SelectedPage!(
    execute
    SelectedPage:Season Pass Page VM Object Reference="Get (a copy)"(Array:Array of Season Pass Page VM Object References=Get Pages(self=@n2), Dimension 1=Get PageIndex() -> (PageIndex=0)) -> (Output:Season Pass Page VM Object Reference)
    self:Season Pass Selection VM Object Reference=Get SeasonPassSelectionVM()
   ) -> (then, Output_Get:Season Pass Page VM Object Reference)
  ]}
 SetItemCardSeasonPassReward {nodes [
   n1 = SetItemCardSeasonPassReward!() -> (then, RewardVM:Season Pass Reward VM Object Reference, ItemCard:WBP Reward Track Item Button Object Reference) {is=title K2Node_FunctionEntry}
   > Is Valid(InputObject=@n1.ItemCard)
   > SetSeasonPassRewardVM(self=@n1.ItemCard, Viewmodel=Get RewardVM() -> (RewardVM:Season Pass Reward VM Object Reference))
   n2 = Get TileSize!(self:Athena Season Item Entry Base Object Reference=GetReward!(self:Season Pass Reward VM Object Reference=Get RewardVM()) -> (ReturnValue:Athena Season Item Entry Base Object Reference)) -> (TileSize:EPageItemTileSize Enum=Size_1_x_1)
   > SetItemCardSize!(
    execute
    self:WBP UIKit Item Card Base Object Reference=Get ItemCard() -> (ItemCard:WBP Reward Track Item Button Object Reference)
    ItemSize_X:double=float * float(
     A=Get BaseTileSizeX() -> (BaseTileSizeX:double)
     B=Select!(Size_1_x_1:double=1, Size_2_x_1:double=2, Size_2_x_2:double=2, Size_3_x_2:double=3, Size_3_x_3:double=3, Index:EPageItemTileSize Enum=@n2) -> (ReturnValue:double)
    ) {is=title K2Node_PromotableOperator}
    ItemSize_Y:double=160
   ) -> (then)
   > SetSeasonPassVM!(execute, self:WBP Reward Track Item Button Object Reference=Get ItemCard(), Viewmodel:Season Pass VM Object Reference=Get SeasonPassVM()) -> (then)
   > Bind Event to On Card Pressed(self:WBP UIKit Item Card Base Object Reference=Get ItemCard(), Delegate=Create Event()) {is=title K2Node_AddDelegate}
   > Return Node!(execute)
  ]}
 SetMobileStyle {nodes [
   SetMobileStyle() {is=title K2Node_FunctionEntry}
   > Branch(Condition=IsUsingCompactUI())
   > SetMaxDesiredWidth!(
    execute
    self:Size Box Object Reference=Get SizeBox_LeftContent() -> (SizeBox_LeftContent:Size Box Object Reference)
    InMaxDesiredWidth:float=640
   ) -> (then)
   > SetPadding!(
    execute
    self:Size Box Slot Object Reference=SlotAsSizeBoxSlot!(Widget:Widget Object Reference=Get TrayContent() -> (TrayContent:Overlay Object Reference)) -> (ReturnValue:Size Box Slot Object Reference)
    InPadding_Left:float=48
    InPadding_Top:float=24
    InPadding_Right:float=48
    InPadding_Bottom:float=24
   ) -> (then)
   > SetSize!(
    execute
    self:Vertical Box Slot Object Reference=SlotAsVerticalBoxSlot(Widget=Get LegalandActionBar() -> (LegalandActionBar:Overlay Object Reference))
    InSize_Value:float=0.16
    InSize_SizeRule:ESlateSizeRule Enum=Fill
   ) -> (then)
   > SetPadding(
    self:Overlay Slot Object Reference=SlotAsOverlaySlot(Widget=Get SeasonPass_Level())
    InPadding_Left=0
    InPadding_Top=-4
    InPadding_Right=0
    InPadding_Bottom=0
   )
   > SetPadding(
    self:Vertical Box Slot Object Reference=SlotAsVerticalBoxSlot(Widget=Get WBP_UIKit_TabGroup())
    InPadding_Left=0
    InPadding_Top=0
    InPadding_Right=0
    InPadding_Bottom=-96
   )
   > SetPadding(self:Size Box Slot Object Reference=SlotAsSizeBoxSlot(), InPadding_Left=0, InPadding_Top=0, InPadding_Right=0, InPadding_Bottom=0)
   > SetPadding(
    self:Grid Slot Object Reference=SlotAsGridSlot(Widget=Get HBox_Header() -> (HBox_Header:Horizontal Box Object Reference))
    InPadding_Left=0
    InPadding_Top=110
    InPadding_Right=0
    InPadding_Bottom=0
   )
  ]}
 SetViewRewardAnimValue {nodes [
   Get!(execute:Exec, self:Fort Item Preview Widget Object Reference) -> (then:Exec, else:Exec, CurrentItemVM:Fort Item VM Object Reference)
   n1 = Get!() -> (
    TargetFramingData_Offset_16_#36_X:double
    TargetFramingData_Offset_16_#36_Y:double
    TargetFramingData_ZoomLevel_15_#38:double=~
    TargetFramingData_CameraMode_19_#39:EPreviewCameraPositionMode Enum
   )
   n2 = Get!() -> (
    CurrentItemFramingData_Offset_16_#36_X:double
    CurrentItemFramingData_Offset_16_#36_Y:double
    CurrentItemFramingData_ZoomLevel_15_#38:double=~
    CurrentItemFramingData_CameraMode_19_#39:EPreviewCameraPositionMode Enum
   )
   n3 = "Lerp (Vector)"!(
    A_X:double=@n2
    A_Y:double=@n2.CurrentItemFramingData_Offset_16_#36_Y
    A_Z:double=@n2.CurrentItemFramingData_ZoomLevel_15_#38
    B_X:double=@n1
    B_Y:double=@n1.TargetFramingData_Offset_16_#36_Y
    B_Z:double=@n1.TargetFramingData_ZoomLevel_15_#38
    Alpha:float=Get AnimationValue() -> (AnimationValue:double)
   ) -> (ReturnValue_X:double, ReturnValue_Y:double, ReturnValue_Z:double)
   > Set!(
    execute
    CurrentItemFramingData_Offset_16_#36_X:double=@n3
    CurrentItemFramingData_Offset_16_#36_Y:double=@n3.ReturnValue_Y
    CurrentItemFramingData_ZoomLevel_15_#38:double=@n3.ReturnValue_Z
    CurrentItemFramingData_CameraMode_19_#39:EPreviewCameraPositionMode Enum=Configured
   ) -> (
    then
    Output_Get_Offset_16_#36:Vector 2D Structure
    Output_Get_ZoomLevel_15_#38:double=~
    Output_Get_CameraMode_19_#39:EPreviewCameraPositionMode Enum
   )
   n4 = Get() -> (CurrentItemFramingData_ZoomLevel_15_#38=~)
   > SetZoomLevel!(
    execute
    self:Fort Item Preview Screen Object Reference
    NewZoomLevel:float=@n4.CurrentItemFramingData_ZoomLevel_15_#38
    bSnap:bool=true
    bZoomEvenIfNotAllowed:bool=true
   ) -> (then)
   > n5 = Branch(Condition=Get IsClaimPagesVisible())
   SetRenderTranslation!(
    execute=@n5.else
    self
    Translation_X:double=@n4
    Translation_Y:double=@n4.CurrentItemFramingData_Offset_16_#36_Y
   ) -> (then)
   SetViewRewardAnimValue!() -> (then, AnimationValue:double=~) {is=title K2Node_FunctionEntry}
  ]}
 SetupCategories {nodes [
   SetupCategories!() -> (then, Categories:Array of Season Pass Category VM Object References) {is=title K2Node_FunctionEntry}
   n1 = FilterSeasonPassCategories!(ShowLocked:bool=true, UseMainSetOnly:bool=false, SeasonPass:Season Pass VM Object Reference=Get SeasonPassVM()) -> (FilteredCategories:Array of Season Pass Category VM Object References, MainSetComplete:bool)
   > Set bUseMainSetOnly(bUseMainSetOnly:bool=NOT Boolean(A=@n1.MainSetComplete)) -> (Output_Get:bool)
   > n2 = For Each Loop(Array:Array of Season Pass Category VM Object References=@n1) -> (Array Element:Season Pass Category VM Object Reference)
   > AddTab!(
    execute
    self:WBP UIKit Tab Group Base Object Reference=Get WBP_UIKit_TabGroup()
    Tab VM:MVVM Base Viewmodel Object Reference=@n2
    Tab ID:Name=String To Name(InString=GetDisplayName!(Object:Object Reference=@n2) -> (ReturnValue:String))
    Index:int=-1
   ) -> (then)
  ]}
 ShowHideRewardsAvailableWidget {nodes [
   n1 = ShowHideRewardsAvailableWidget!() -> (then, ShouldShow:bool=~, AllowAudio:bool=true) {is=title K2Node_FunctionEntry}
   > n2 = Branch(Condition=IsVisible(self=Get WBP_RewardInfoWidget()))
   > n3 = Branch(Condition=@n1)
   > Show_VIZ!(
    execute
    self:WBP Reward Info Widget Object Reference=Get WBP_RewardInfoWidget()
    NeedsToAnimate:bool=NOT Boolean(A=Get HasRewardsStingerInitiallyShown() -> (HasRewardsStingerInitiallyShown:bool))
    PlayAudio:bool=Get AllowAudio() -> (AllowAudio:bool)
   ) -> (then)
   > Set HasRewardsStingerInitiallyShown(HasRewardsStingerInitiallyShown:bool=true) -> (Output_Get:bool)
   Hide_VIZ(execute=@n3.else, self:WBP Reward Info Widget Object Reference=Get WBP_RewardInfoWidget(), NeedsToAnimate=true)
   Branch(execute=@n2.else, Condition=Get ShouldShow() -> (ShouldShow:bool))
   > Set QueueIntroAnimation!(execute, QueueIntroAnimation:bool=true, self:WBP Reward Info Widget Object Reference=Get WBP_RewardInfoWidget()) -> (then, Output_Get:bool)
  ]}
 SwapTracks {nodes [
   n1 = Create Event()
   n2 = Get ActiveRewardList()
   SetNavigationRuleCustomBoundary!(execute, self:Widget Object Reference=@n2, Direction:EUINavigation Enum=Left, InCustomDelegate:Delegate=@n1) -> (then)
   > SetNavigationRuleCustomBoundary(self=@n2, Direction=Right, InCustomDelegate=@n1)
   SwapTracks() {is=title K2Node_FunctionEntry}
   > Set Content_Current(Content_Current:Widget Object Reference=GetChildAt!(self:Panel Widget Object Reference=Get NamedSlot_Track_Current(), Index:int=0) -> (ReturnValue:Widget Object Reference)) -> (Output_Get:Widget Object Reference)
   > Set Content_Next(Content_Next:Widget Object Reference=GetChildAt(self=Get NamedSlot_Track_Next(), Index=0)) -> (Output_Get:Widget Object Reference)
   > ClearChildren(self=[Get NamedSlot_Track_Next(), Get NamedSlot_Track_Current()])
   > AddChild(self=Get NamedSlot_Track_Current(), Content=Get Content_Next() -> (Content_Next:Widget Object Reference))
   > AddChild(self=Get NamedSlot_Track_Next(), Content=Get Content_Current() -> (Content_Current:Widget Object Reference))
   > UpdateActiveTrack(ActiveTrack=Get NamedSlot_Track_Current())
  ]}
 UpdateActiveTrack {nodes [
   n1 = UpdateActiveTrack!() -> (then, ActiveTrack:Named Slot Object Reference) {is=title K2Node_FunctionEntry}
   > n2 = Set ActiveTrack(ActiveTrack:Named Slot Object Reference=@n1) -> (Output_Get:Named Slot Object Reference)
   > n3 = Cast To DynamicEntryBox!(execute, Object=GetChildAt(self=@n2, Index=0)) -> (then, CastFailed, AsDynamic Entry Box:Dynamic Entry Box Object Reference) {is=title K2Node_DynamicCast}
   > Set ActiveRewardList(ActiveRewardList:Dynamic Entry Box Object Reference=@n3) -> (Output_Get:Dynamic Entry Box Object Reference)
   Set ActiveRewardList(execute=@n3.CastFailed)
  ]}
 UpdateClaimPagesButtonVisibility {nodes [
   UpdateClaimPagesButtonVisibility() {is=title K2Node_FunctionEntry}
   n1 = Get SeasonPassSelectionVM()
   n2 = NOT Boolean(A=Get bIsDependentCategoryLocked!(self:Season Pass Page VM Object Reference=Get SelectedPage(self=@n1)) -> (bIsDependentCategoryLocked:bool))
   n3 = NeedsButtonOnPassCategory!(self:Season Pass Claim Pages Object Reference=Get SeasonPass_ClaimPages()) -> (Value:bool)
   n4 = Get bDoesOwnsPass!(self:Season Pass VM Object Reference=Get SeasonPassVM()) -> (bDoesOwnsPass:bool)
   > SetVisibility(
    self=Get Button_ClaimPages()
    InVisibility=Select!(
     Option 0:ESlateVisibility Enum=Collapsed
     Option 1:ESlateVisibility Enum=Visible
     Index:bool=AND Boolean!(A=@n3, B=@n4, C:bool=@n2) -> (ReturnValue)
    ) -> (ReturnValue:ESlateVisibility Enum)
   )
  ]}
 UpdateContextMenuPosition {nodes [
   n1 = UpdateContextMenuPosition!() -> (then, AnchorWidget:Widget Object Reference) {is=title K2Node_FunctionEntry}
   > Set AnchorCachedGeometry(AnchorCachedGeometry:Geometry Structure=GetCachedGeometry!(self:Widget Object Reference=@n1) -> (ReturnValue:"Geometry Structure (by ref)")) -> (Output_Get:Geometry Structure)
   n2 = GetLocalSize!(Geometry:"Geometry Structure (by ref)"=Get AnchorCachedGeometry() -> (AnchorCachedGeometry:Geometry Structure)) -> (ReturnValue_X:double, ReturnValue_Y:double)
   n3 = vector2d + vector2d(
    A:Vector 2D Structure=MakeVector2D(X=float / float(A=@n2, B=2), Y=@n2.ReturnValue_Y)
    B:Vector 2D Structure=Get ContextMenuOffset() -> (ContextMenuOffset:Vector 2D Structure)
   ) -> (ReturnValue:Vector 2D Structure) {is=title K2Node_PromotableOperator}
   n4 = Get AnchorCachedGeometry()
   n5 = Get CPanel_ContextMenuRoot()
   > Set with Broadcast ContextMenuPosition!(
    execute
    ContextMenuPosition:Vector 2D Structure="Absolute to Local (Coordinates)"!(
     Geometry:"Geometry Structure (by ref)"=GetCachedGeometry(self=@n5)
     AbsoluteCoordinate:Vector 2D Structure="Local to Absolute (Coordinates)"!(Geometry:"Geometry Structure (by ref)"=@n4, LocalCoordinate:Vector 2D Structure=@n3) -> (ReturnValue:Vector 2D Structure)
    ) -> (ReturnValue:Vector 2D Structure)
   ) -> (then, Output_Get:Vector 2D Structure)
  ]}
 UpdateNavigationArrows {nodes [
   UpdateNavigationArrows() {is=title K2Node_FunctionEntry}
   > n1 = GetLinearSeasonPassPageNavTarget(SeasonPassSelectionVM=Get SeasonPassSelectionVM(), IsPrevious=false, ShowLockedCategories=false, UseMainSetOnly=Get bUseMainSetOnly()) -> (SelectedPage=0)
   > SetIsEnabled(
    self=Get SecondaryRight() -> (SecondaryRight:WBP UIKit Button Generic Object Reference)
    bInIsEnabled=IsValid(Object=@n1.SeasonPassCategoryTarget)
   )
   > n2 = GetLinearSeasonPassPageNavTarget(SeasonPassSelectionVM=Get SeasonPassSelectionVM(), IsPrevious=true, ShowLockedCategories=false, UseMainSetOnly=Get bUseMainSetOnly()) -> (SelectedPage=0)
   > SetIsEnabled(
    self=Get SecondaryLeft() -> (SecondaryLeft:WBP UIKit Button Generic Object Reference)
    bInIsEnabled=IsValid(Object=@n2.SeasonPassCategoryTarget)
   )
  ]}
 UpdateRewardsListWidth {nodes [
   UpdateRewardsListWidth() {is=title K2Node_FunctionEntry}
   n1 = Get ActiveRewardList()
   > ForceLayoutPrepass(self:Widget Object Reference=@n1)
   > SetWidthOverride(
    self=Get SBox_Tracks() -> (SBox_Tracks:Size Box Object Reference)
    InWidthOverride=float + float(A=GetDesiredSize!(self:Widget Object Reference=@n1) -> (ReturnValue_X:double, ReturnValue_Y:double), B=32)
   )
   > Return Node()
  ]}
 VM_BattlepassPurchased {nodes [
   n1 = VM_BattlepassPurchased!() -> (then, IsPurchased:bool=~) {is=title K2Node_FunctionEntry}
   > n2 = Branch(Condition=NOT Boolean(A=Get IsWidgetInitialized() -> (IsWidgetInitialized:bool)))
   > Set InitialPurchaseState(InitialPurchaseState:bool=@n1) -> (Output_Get:bool)
   > Set IsWidgetInitialized(IsWidgetInitialized:bool=true) -> (Output_Get:bool)
   Branch(execute=@n2.else, Condition="Not Equal (Boolean)"(A=Get InitialPurchaseState(), B=@n1) -> (ReturnValue=~))
   > ShowBattlepassPurchase() {is=title K2Node_CallFunction}
   > Set InitialPurchaseState(InitialPurchaseState=true)
  ]}
 VM_CategoriesUpdated {nodes [
   n1 = VM_CategoriesUpdated!() -> (then, Categories:Array of Season Pass Category VM Object References) {is=title K2Node_FunctionEntry}
   > Branch(Condition="integer > integer"(A=Length(TargetArray=@n1)))
   > SetupCategories!(execute, self, Categories:Array of Season Pass Category VM Object References=@n1) -> (then) {is=title K2Node_CallFunction}
   > Branch(Condition=Get bIsWaitingSeasonPassDataToPlayIntro() -> (bIsWaitingSeasonPassDataToPlayIntro:bool))
   > BattlepassIntro_VIZ()
  ]}
 VM_CategoryChanged {nodes [
   n1 = VM_CategoryChanged!() -> (then, Category:Season Pass Category VM Object Reference) {is=title K2Node_FunctionEntry}
   > Is Valid(InputObject=@n1)
   > Set SeasonPassCategoryVM(SeasonPassCategoryVM:Season Pass Category VM Object Reference=@n1) -> (Output_Get:Season Pass Category VM Object Reference)
   > SelectTabByID!(
    execute
    self:Common Tab List Widget Base Object Reference=Get WBP_UIKit_TabGroup()
    TabNameID:Name=String To Name(InString=GetDisplayName(Object=@n1))
    bSuppressClickFeedback:bool=true
   ) -> (then, ReturnValue:bool)
  ]}
 VM_CategoryColorChanged {nodes [
   n1 = VM_CategoryColorChanged!() -> (then, ThemeData:Season Pass Category Theme Data Structure) {is=title K2Node_FunctionEntry}
   > SetVectorParameterValue!(
    execute
    Collection=/SeasonPass/MaterialFunctions/MPC_UI_Pass_Properties
    ParameterName=ThemeColor
    ParameterValue:"Linear Color Structure (by ref)"=CF_GetThemeCharacterColor!(ThemeData:Season Pass Category Theme Data Structure=@n1) -> (CharacterColor:Linear Color Structure)
   ) -> (then)
  ]}
 VM_ClaimableRewardsCountChanged {nodes [
   n1 = VM_ClaimableRewardsCountChanged!() -> (then, ClaimableRewards:int) {is=title K2Node_FunctionEntry}
   n2 = Get Button_ClaimAllRewards()
   > SetText!(
    execute
    self:Fort CTAButton Object Reference=@n2
    ButtonText:"Text (by ref)"=GetClaimAllRewardsButtonText!(self, ClaimableRewardsAmount:int=@n1) -> (ButtonText:Text) {is=title K2Node_CallFunction}
   ) -> (then)
   n3 = GetClaimAllRewardsButtonVisibility(ClaimableRewardsAmount=@n1) -> (Result=Visible)
   > SetVisibility(self=@n2, InVisibility=@n3)
   > SetVisibility(self=Get ClaimAllButtonWrapper(), InVisibility=Select(Option 0=@n3, Option 1=Collapsed, Index=Get IsClaimPagesVisible()))
  ]}
 VM_HandleThemeData {nodes [
   n1 = VM_HandleThemeData!() -> (then, ThemeData:"Season Pass Theme Data Structure (by ref)") {is=title K2Node_FunctionEntry}
   > RequestLoadBackgroundMaterial!(
    execute
    self
    BackgroundMaterialAsset:Soft Object Reference=CF_GetPassBackgroundMaterial!(ThemeData:Season Pass Theme Data Structure=@n1) -> (UIBackground Material:Soft Object Reference)
   ) -> (then) {is=title K2Node_CallFunction}
  ]}
 VM_InitializeSelection {nodes [
   n1 = VM_InitializeSelection!() -> (then, SeasonPassVM:Season Pass VM Object Reference) {is=title K2Node_FunctionEntry}
   n2 = FilterSeasonPassCategories(ShowLocked=true, UseMainSetOnly=false, SeasonPass=Get SeasonPassVM())
   n3 = GetFirstUnlockedReward(self:Season Pass VM Object Reference=@n1)
   n4 = "Equal (Integer)"(
    A=Find Item(
     TargetArray:Array of Season Pass Category VM Object References=@n2
     ItemToFind:"Season Pass Category VM Object Reference (by ref)"=GetOwnerCategory!(self:Season Pass Reward VM Object Reference=@n3) -> (ReturnValue:Season Pass Category VM Object Reference)
    )
    B=-1
   )
   > n5 = Branch(Condition=NOT Boolean(A=AND Boolean!(A=@n2.MainSetComplete, B=@n4) -> (ReturnValue)))
   > n6 = NavigateToReward(self=Get SeasonPassSelectionVM(), Reward=@n3)
   InitializeScreen(execute=[@n6.then, @n14.then, @n12.then]) {is=title K2Node_CallFunction}
   n7 = Get SeasonPassSelectionVM()
   > n8 = For Each Loop(Array:Array of Season Pass Track VM Object References=Get Tracks(self=Get SelectedPage(self=@n7))) -> (Array Element:Season Pass Track VM Object Reference)
   > n9 = For Each Loop(Array:Array of Season Pass Reward VM Object References=Get Rewards(self=@n8)) -> (Array Element:Season Pass Reward VM Object Reference)
   > Add Field Value Changed Delegate!(
    execute
    self:MVVM Base Viewmodel Object Reference=@n9
    FieldId:Field Notification Id Structure=(FieldName="ItemVMs")
    Delegate:Delegate=Create Event()
   ) -> (then)
   n10 = GetLastClaimedReward!(Categories:Array of Season Pass Category VM Object References=@n2) -> (Reward:Season Pass Reward VM Object Reference) {is=title K2Node_CallFunction}
   n11 = Is Valid(exec=@n5.else, InputObject=@n10)
   > n12 = NavigateToReward(self=Get SeasonPassSelectionVM(), Reward=@n10)
   n13 = Get Pages(self="Get (a copy)"(Array:Array of Season Pass Category VM Object References=@n2) -> (Output:Season Pass Category VM Object Reference))
   n14 = NavigateToReward(
    execute=@n11.Is Not Valid
    self=Get SeasonPassSelectionVM()
    Reward=GetFirstOrLastRewardOnSeasonPassPage!(
     SeasonPassPageVM:"Season Pass Page VM Object Reference (by ref)"="Get (a copy)"(Array:Array of Season Pass Page VM Object References=@n13) -> (Output:Season Pass Page VM Object Reference)
     IsFirst:bool=true
    ) -> (SeasonPassRewardVM:Season Pass Reward VM Object Reference)
   )
  ]}
 VM_ItemVMsChanged {nodes [
   VM_ItemVMsChanged!() -> (then, ItemVMs:Array of Fort Item VM Object References) {is=title K2Node_FunctionEntry}
   > Reset(self=Get VariantCards(), bDeleteWidgets=false)
   n1 = Get ItemVMs!() -> (ItemVMs)
   > Branch(Condition="integer > integer"(A=Length(TargetArray:Array of Fort Item VM Object References=@n1), B=1))
   > n2 = For Each Loop(Array:Array of Fort Item VM Object References=Get ItemVMs()) -> (Array Element:Fort Item VM Object Reference)
   > CreateSeasonPassVariant!(execute, self, ItemVM:Fort Item VM Object Reference=@n2, ItemDef:Fort Item Definition Object Reference=Get ItemDefinition(self=@n2)) -> (then) {is=title K2Node_CallFunction}
  ]}
 VM_PageOwnedRewardsChanged {nodes [
   n1 = VM_PageOwnedRewardsChanged!() -> (then, OwnedFromPage:int) {is=title K2Node_FunctionEntry}
   n2 = GetRewardCount!(self:Season Pass Page VM Object Reference=Get SelectedPage(self=Get SeasonPassSelectionVM())) -> (ReturnValue:int=0)
   > Branch(Condition="Equal (Integer)"(A=@n1, B=int - int(A=@n2)))
   n3 = Get Pages(self=Get SeasonPassCategoryVM())
   n4 = Get SeasonPassSelectionVM()
   > Branch(Condition="Not Equal (Object)"(A="Get (a copy)"(Array=@n3), B=Get SelectedPage(self=@n4)) -> (ReturnValue=~))
  ]}
 VM_PreviewDataChanged {nodes [
   n1 = VM_PreviewDataChanged!() -> (then, PreviewItemVM:Fort Cosmetic Preview Data Structure) {is=title K2Node_FunctionEntry}
   > Branch(Condition=Get IsViewingReward())
   SetPreviewData!(execute, self:Fort Item Preview Widget Object Reference, Data:"Fort Cosmetic Preview Data Structure (by ref)"=@n1) -> (then)
  ]}
 VM_PreviewItemChanged {nodes [
   n1 = VM_PreviewItemChanged!() -> (then, PreviewItemVM:Fort Item VM Object Reference) {is=title K2Node_FunctionEntry}
   > Is Valid(InputObject=@n1)
   n2 = Get WBP_UIKit_DescriptionWithSubtitle()
   n3 = Get PreviewItemVM() -> (PreviewItemVM:Fort Item VM Object Reference)
   > Set FortItemDetailsVM(FortItemDetailsVM=Get ItemDetails(self=@n3), self=@n2)
   n4 = Get ItemDefinition(self=@n3)
   > ShowDynamicSubtitle!(
    execute
    self:WBP UIKit Description With Subtitle Object Reference=@n2
    Show:bool="Equal (Enum)"(A:EFortItemType Enum=GetItemType(ItemDefinition=@n4) -> (ReturnValue=WorldItem), B:EFortItemType Enum=SparksSong)
   ) -> (then)
   > SetFortItemVM(self:Reward Counter Object Reference=Get RewardCounter() -> (RewardCounter:Reward Counter Object Reference), Viewmodel=Get PreviewItemVM())
  ]}
 VM_SeasonPassesUpdated {nodes [
   n1 = VM_SeasonPassesUpdated!() -> (then, SeasonPasses:Array of Season Pass VM Object References) {is=title K2Node_FunctionEntry}
   > For Each Loop(Array:Array of Season Pass VM Object References=@n1) -> (Array Element:Season Pass VM Object Reference)
   SetSeasonPassVM(
    self:Self Object Reference
    Viewmodel=CF_FindBattlepass!(self, SeasonPasses:Array of Season Pass VM Object References=Get SeasonPassSystemVM()) -> (ReturnValue:Season Pass VM Object Reference) {is=title K2Node_CallFunction}
   )
  ]}
 VM_SelectedPageChanged {nodes [
   n1 = VM_SelectedPageChanged!() -> (then, PassPageVM:Season Pass Page VM Object Reference) {is=title K2Node_FunctionEntry}
   n2 = Get NewPageRewardTarget() -> (NewPageRewardTarget:Season Pass Reward VM Object Reference)
   > n3 = Is Valid(InputObject=@n2)
   > Initialize Selected Page And Reward(PassPageVM=@n1, RewardToSelect=@n2)
   Initialize Selected Page And Reward(execute=@n3.Is Not Valid, PassPageVM=@n1, RewardToSelect=Get SelectedReward(self=Get SeasonPassSelectionVM()))
  ]}
 VM_SelectedPassChanged {nodes [
   n1 = VM_SelectedPassChanged!() -> (then, SelectedPass:Season Pass VM Object Reference) {is=title K2Node_FunctionEntry}
   > Is Valid(InputObject=@n1)
   n2 = GetDisplayName!(self:Season Pass VM Object Reference=@n1) -> (ReturnValue:Text)
   > Set NewTabId(NewTabId:Name=String To Name(InString="To String (Text)"(InText=@n2))) -> (Output_Get:Name=None)
   > Set LastSelectedTabID(LastSelectedTabID:int=GetSelectedTabIndex!(self:UIKit Tab Group Object Reference=Get WBP_UIKit_TabGroup()) -> (ReturnValue:int=0)) -> (Output_Get:int=0)
   > SelectTabByID(self=Get WBP_UIKit_TabGroup(), TabNameID=Get NewTabId() -> (NewTabId:Name=None), bSuppressClickFeedback=true)
  ]}
 VM_SelectedRewardUpdated {nodes [
   n1 = VM_SelectedRewardUpdated!() -> (then, SelectedReward:Season Pass Reward VM Object Reference) {is=title K2Node_FunctionEntry}
   > Is Valid(InputObject=@n1)
   n2 = Get SeasonPassSelectionVM()
   n3 = Find Item(
    TargetArray:Array of Season Pass Reward VM Object References=GetRewardsInPage!(self, Page:Season Pass Page VM Object Reference=Get SelectedPage(self=@n2)) -> (Rewards:Array of Season Pass Reward VM Object References) {is=title K2Node_CallFunction}
    ItemToFind:"Season Pass Reward VM Object Reference (by ref)"=@n1
   )
   n4 = Get SeasonPassRewardVM(self=Get ModeSpecificReward())
   > Branch(Condition=OR Boolean(A="Not Equal (Integer)"(A=@n3, B=-1), B="Equal (Object)"(A=@n1, B=@n4)))
   > Set SelectedRewardIndex(SelectedRewardIndex=@n3) -> (Output_Get=0)
   > RefreshContextMenuPosition() {is=title K2Node_CallFunction}
  ]}
 ViewingRewardStateChanged {nodes [ViewingRewardStateChanged!() -> (then, IsViewingReward:bool=~) {is=title K2Node_FunctionEntry}]}
 Visualizations {nodes [
   OpenBattlepass_VIZ() {is=title K2Node_CustomEvent}
   > Set IsBattlepassOpen(IsBattlepassOpen:bool=true) -> (Output_Get:bool)
   > SetIsEnabled(self=Get Button_MatchRecap() -> (Button_MatchRecap:Button Object Reference), bInIsEnabled=true)
   > SetIsEnabled(self=Get Button_Battlepass() -> (Button_Battlepass:Button Object Reference), bInIsEnabled=false)
   > SetVisibility(self=Get CPanel_ContextMenuRoot(), InVisibility=SelfHitTestInvisible)
   > RefreshContextMenuPosition()
   > PlayAnimation(InAnimation=Get Anim_TrayExpand() -> (Anim_TrayExpand:Widget Animation Object Reference))
   > Broadcast Event!(
    execute
    self:Gameplay Event Router Component Object Reference=GetGlobalEventRouter()
    EventData:Match XPScreen Event Focus Battlepass Structure=Make Match XPScreen Event Focus Battlepass!(IsBattlepassFocused:bool=true) -> (MatchXPScreenEvent_FocusBattlepass:Match XPScreen Event Focus Battlepass Structure) {is=title K2Node_MakeStruct}
    EventContext:Object Reference
   ) -> (then, ReturnValue:bool)
   > PlaySound2D(Sound=Get BattlePassTray_Open() -> (BattlePassTray_Open:Sound Base Object Reference))
   n1 = CloseBattlepass_VIZ!() -> (OutputDelegate, then, bSkipAnimation:bool=~) {is=title K2Node_CustomEvent}
   > Set IsBattlepassOpen(IsBattlepassOpen=false)
   > SetIsEnabled(self=Get Button_MatchRecap(), bInIsEnabled=false)
   > SetIsEnabled(self=Get Button_Battlepass(), bInIsEnabled=true)
   > ClearRewardFocus()
   n2 = Get Anim_TrayExpand()
   > PlayAnimation(
    InAnimation=Get Anim_TrayExpand()
    StartAtTime=SelectFloat(A=GetEndTime(self=@n2), B=0, bPickA=@n1.bSkipAnimation)
    PlayMode=Reverse
    PlaybackSpeed=2
   )
   > Broadcast Event(self=GetGlobalEventRouter(), EventData=Make Match XPScreen Event Focus Battlepass(IsBattlepassFocused=False))
   > PlaySound2D(Sound=Get BattlePassTray_Close() -> (BattlePassTray_Close:Sound Base Object Reference))
   n3 = Get WBP_RewardInfoWidget()
   n4 = Get IsCompact!(self:WBP Reward Info Widget Object Reference=@n3) -> (IsCompact:bool)
   > Branch(Condition=AND Boolean(A=IsVisible(self=@n3), B=NOT Boolean(A=@n4)))
   > Delay(Duration=GetEndTime(self=Get Anim_TrayExpand()))
   > ShowHideRewardsAvailableWidget(ShouldShow=false, AllowAudio=false) {is=title K2Node_CallFunction}
   BattlepassIntro_VIZ() {is=title K2Node_CustomEvent}
   > Set bIsWaitingSeasonPassDataToPlayIntro(bIsWaitingSeasonPassDataToPlayIntro:bool=true) -> (Output_Get:bool)
   > Branch(Condition=HasValidSeasonPassData!() -> (bHasValidData:bool) {is=title K2Node_CallFunction})
   > Set bIsWaitingSeasonPassDataToPlayIntro(bIsWaitingSeasonPassDataToPlayIntro=false)
   > SetVisibility(self=Get VisSwitcher_PassPurchase(), InVisibility=SelfHitTestInvisible)
   > n5 = Branch(Condition=Get InitialPurchaseState())
   n6 = Get VisSwitcher_PassPurchase()
   SetActiveWidget(execute=[@n10.then, @n5.then, @n8.then], self=@n6, Widget=Get Overlay_RewardTrack_Root())
   > PlayAnimation(InAnimation=Get Intro())
   n7 = Get CPanel_ContextMenuRoot()
   > SetVisibility(self=@n7, InVisibility=SelfHitTestInvisible)
   > Set RenderOpacity!(execute, RenderOpacity:float=0, self:Widget Object Reference=@n7) -> (then, Output_Get:float)
   > DelayUntilNextFrame!(execute) -> (then)
   > SetVisibility(self=@n7, InVisibility=Collapsed)
   n8 = Branch(execute=@n5.else, Condition=Get S41SuppressPostGameBattlePassUpsellPreview() -> (S41SuppressPostGameBattlePassUpsellPreview:bool))
   n9 = Get WBP_BuyPassPreview()
   SetActiveWidget(execute=@n8.else, self=@n6, Widget=@n9)
   > Show_VIZ(self:WBP Season Pass Buy Pass Preview Object Reference=@n9, NeedsToAnimate=true, PlayAudio=true)
   > RequestRefreshFocus()
   n10 = RewardTrackIntro_Viz() {is=title K2Node_CustomEvent}
  ]}
]
variables [
 ActiveRewardList {type=DynamicEntryBox, default=null}
 ActiveTrack {type=NamedSlot, default=null}
 BaseTileSizeX {type=number, default=128}
 BattlePassTray_Close {type=SoundBase, default=/Game/Sounds/Fort_UI_Sounds/FrontEnd/DefaultSounds/PressedButtons/MSS_UI_Button_Pressed_Default}
 BattlePassTray_Open {type=SoundBase, default=&MSS_UI_Button_Pressed_Default}
 bIsWaitingSeasonPassDataToPlayIntro {type=boolean}
 bUseMainSetOnly {type=boolean}
 ClaimPagesBackLastSelectedReward {type=SeasonPassRewardVM, default=null}
 ContextMenuOffset {type=Vector2D, default=<0 15>}
 ContextMenuPosition {type=Vector2D, default=<0 0>}
 CurrentItemFramingData {type=S_ItemPreviewFramingData, default {cameraMode=Configured, zoomLevel=0, offset=<0 0>}}
 EnableNavigationPageWrapping {type=boolean}
 FallbackFramingData_Landing {type=S_ItemPreviewFramingData, default {cameraMode=Configured, zoomLevel=0, offset=<0 0>}}
 FallbackFramingData_Landing_ViewReward {type=S_ItemPreviewFramingData, default {cameraMode=Configured, zoomLevel=1, offset=<0 0>}}
 HasRewardsStingerInitiallyShown {type=boolean}
 InitialPurchaseState {type=boolean}
 IsBattlepassOpen {type=boolean}
 IsClaimAllInProgress {type=boolean}
 IsClaimPagesVisible {type=boolean}
 IsModalVisible {type=boolean}
 IsRewardsAvailableAudioAllowed {type=boolean, default=true}
 IsScreenInitialized {type=boolean}
 IsTransitionActive {type=boolean}
 IsViewingReward {type=boolean}
 IsWidgetInitialized {type=boolean}
 ItemPreviewFramingOverrides {default {
   "(ItemType_2_#34=AthenaBackpack,ProductTag_5_#35=())" {landing Data {cameraMode=FrameCapsule, zoomLevel=0.75, offset=<0 25>}, viewReward Data {cameraMode=Configured, zoomLevel=1, offset=<0 0>}}
   "(ItemType_2_#34=AthenaCharacter,ProductTag_5_#35=(TagName=\"Product.BR\"))" {landing Data {cameraMode=FrameCapsule, zoomLevel=0.5, offset=<0 0>}, viewReward Data {cameraMode=Configured, zoomLevel=0, offset=<0 0>}}
   "(ItemType_2_#34=AthenaCharacter,ProductTag_5_#35=(TagName=\"Product.Juno\"))" {landing Data {cameraMode=FrameCapsule, zoomLevel=0.5, offset=<0 -80>}, viewReward Data {cameraMode=FrameCapsule, zoomLevel=0, offset=<0 0>}}
   "(ItemType_2_#34=AthenaDance,ProductTag_5_#35=())" {landing Data {cameraMode=FrameCapsule, zoomLevel=0, offset=<0 0>}, viewReward Data {cameraMode=FrameCapsule, zoomLevel=0.2, offset=<0 0>}}
   "(ItemType_2_#34=AthenaGlider,ProductTag_5_#35=())" {
    landing Data {cameraMode=Configured, zoomLevel=-0.5, offset=<0 -120>}
    viewReward Data {cameraMode=Configured, zoomLevel=0.9, offset=<0 -50>}
   }
   "(ItemType_2_#34=AthenaItemWrap,ProductTag_5_#35=())" {
    landing Data {cameraMode=FrameCapsule, zoomLevel=0.25, offset=<-50 40>}
    viewReward Data {cameraMode=Configured, zoomLevel=0.4, offset=<0 0>}
   }
   "(ItemType_2_#34=AthenaLoadingScreen,ProductTag_5_#35=())" {landing Data {cameraMode=Configured, zoomLevel=0, offset=<40 0>}, viewReward Data {cameraMode=Configured, zoomLevel=1, offset=<40 0>}}
   "(ItemType_2_#34=JunoBuildingProp,ProductTag_5_#35=())" {
    landing Data {cameraMode=FrameCapsule, zoomLevel=0, offset=<0 -50>}
    viewReward Data {cameraMode=FrameCapsule, zoomLevel=0.2, offset=<160 0>}
   }
   "(ItemType_2_#34=JunoBuildingSet,ProductTag_5_#35=())" {
    landing Data {cameraMode=FrameCapsule, zoomLevel=0, offset=<0 -50>}
    viewReward Data {cameraMode=FrameCapsule, zoomLevel=0.2, offset=<160 0>}
   }
  }}
 LastScrolledDirection {type=EVaultWorldTransitionDirection, default=None}
 LastSelectedTabID {type=integer}
 NewPageRewardTarget {type=SeasonPassRewardVM, default=null}
 NewVariantCardSelected {}
 PageIndex {type=integer}
 RequestedBackgroundSoftPath {type=SoftObjectPath, default=""}
 S41SuppressPostGameBattlePassUpsellPreview {type=boolean}
 SelectedRewardIndex {type=integer}
 TargetFramingData {type=S_ItemPreviewFramingData, default {cameraMode=Configured, zoomLevel=0, offset=<0 0>}}
 ViewingRewardStateChanged {}
 ViewRewardAnimValue {type=number}
]
widgets {
 info {
  inheritedWidgetCount=0
  namedSlotCount=1
  parentClass=&CommonActivatableWidget
  rootWidgetClass=&CommonVisibilitySwitcher
  widgetCount=76
 }
 widgets [
  VisSwitcher_PassPurchase {widgetClassPath=&CommonVisibilitySwitcher, bIsVariable=true, widget=VisSwitcher_PassPurchase}
  WBP_BuyPassPreview {
   widgetClassPath=&/WBP_SeasonPass_BuyPassPreview
   bIsVariable=true
   parent=VisSwitcher_PassPurchase
   widget=WBP_BuyPassPreview
   slot {class=CommonVisibilitySwitcherSlot, properties {horizontalAlignment=HAlign_Left, verticalAlignment=VAlign_Bottom, padding=<ltrb 50 0 0 0>}}
  }
  Overlay_RewardTrack_Root {
   widgetClassPath=&Overlay
   bIsVariable=true
   parent=VisSwitcher_PassPurchase
   widget=Overlay_RewardTrack_Root
   slot {class=CommonVisibilitySwitcherSlot}
  }
  LImg_Background {
   widgetClassPath=&FortLazyImage
   bIsVariable=true
   parent=Overlay_RewardTrack_Root
   widget=LImg_Background
   properties {brush {
     bIsDynamicallyLoaded=false
     drawAs=Image
     imageType=FullColor
     mirroring=NoMirror
     resourceName=None
     resourceObject=/SeasonPass/Materials/RewardTrack/M_RewardTrack_TrayPlaceholder
     tiling=NoTile
     imageSize=<32 32>
     margin=<ltrb 0 0 0 0>
     outlineSettings {
      bUseBrushTransparency=false
      roundingType=HalfHeightRadius
      width=0
      color {colorUseRule=UseColor_Specified, specifiedColor=<rgba 0 0 0 0>}
      cornerRadii=<0 0 0 0>
     }
     tintColor {colorUseRule=UseColor_Specified, specifiedColor=<rgba 1 1 1 1>}
     uVRegion {bIsValid=false, max=<0 0>, min=<0 0>}
    }}
   slot {class=OverlaySlot, properties {horizontalAlignment=HAlign_Fill, verticalAlignment=VAlign_Fill, padding=<ltrb -130 -30 -130 -130>}}
  }
  TrayContent {
   widgetClassPath=&Overlay
   bIsVariable=true
   parent=Overlay_RewardTrack_Root
   widget=TrayContent
   properties {renderOpacity=0}
   slot {class=OverlaySlot, properties {horizontalAlignment=HAlign_Fill, verticalAlignment=VAlign_Fill}}
  }
  GridPanel_TrayContent {
   widgetClassPath=&GridPanel
   parent=TrayContent
   widget=GridPanel_TrayContent
   properties {bOverride_Cursor=true, columnFill [1], rowFill [0,1,0]}
   slot {class=OverlaySlot, properties {horizontalAlignment=HAlign_Fill, verticalAlignment=VAlign_Fill}}
  }
  HBox_Header {
   widgetClassPath=&HorizontalBox
   bIsVariable=true
   parent=GridPanel_TrayContent
   widget=HBox_Header
   slot {class=GridSlot, properties {horizontalAlignment=HAlign_Right, verticalAlignment=VAlign_Top}}
  }
  WBP_CreatorBadge {
   widgetClassPath=/FortniteCoreUI/Items/WBP_CreatorBadge
   bIsVariable=true
   parent=HBox_Header
   widget=WBP_CreatorBadge
   slot {class=HorizontalBoxSlot, properties {horizontalAlignment=HAlign_Left, verticalAlignment=VAlign_Top, padding=<ltrb 0 0 20 0>}}
  }
  Overlay_2 {
   widgetClassPath=&Overlay
   parent=HBox_Header
   widget=Overlay_2
   slot {class=HorizontalBoxSlot, properties {horizontalAlignment=HAlign_Left, verticalAlignment=VAlign_Top}}
  }
  PostBufferUpdate_PassLevel {
   widgetClassPath=/Script/UMG.PostBufferUpdate
   bIsVariable=true
   parent=Overlay_2
   widget=PostBufferUpdate_PassLevel
   properties {
    visibility=HitTestInvisible
    updateBufferInfos [{
      bufferToUpdate=ESlatePostRT_0
      postParamUpdater=&self:WidgetTree.PostBufferUpdate_PassLevel.PostBufferBlurUpdater_0
     }]
   }
   slot {class=OverlaySlot, properties {horizontalAlignment=HAlign_Fill, verticalAlignment=VAlign_Fill, padding=<ltrb -10 -10 -10 -10>}}
  }
  SeasonPass_Level {
   widgetClassPath=/FortniteCoreUI/MultiProduct/SeasonPass_Level
   bIsVariable=true
   parent=Overlay_2
   widget=SeasonPass_Level
   slot {class=OverlaySlot}
  }
  FortVisualAttachment_62 {
   widgetClassPath=&FortVisualAttachment
   parent=HBox_Header
   widget=FortVisualAttachment_62
   properties {contentAnchor=<1 0>}
   slot {class=HorizontalBoxSlot, properties {horizontalAlignment=HAlign_Right, verticalAlignment=VAlign_Bottom}}
  }
  WBP_SeasonPass_RemainingSetsToUnlock {
   widgetClassPath=/SeasonPass/Widgets/Reboot/WBP_SeasonPass_RemainingSetsToUnlock
   bIsVariable=true
   parent=FortVisualAttachment_62
   widget=WBP_SeasonPass_RemainingSetsToUnlock
   slot {class=SizeBoxSlot, properties {horizontalAlignment=HAlign_Right, verticalAlignment=VAlign_Center, padding=<ltrb 0 30 0 5>}}
  }
  SeasonPass_ClaimPages {
   widgetClassPath=&/SeasonPass_ClaimPages
   bIsVariable=true
   parent=GridPanel_TrayContent
   widget=SeasonPass_ClaimPages
   properties {bIsFocusable=false, renderOpacity=0, visibility=Collapsed}
   slot {class=GridSlot, properties {horizontalAlignment=HAlign_Left, row=1}}
  }
  SizeBox_LeftContent {
   widgetClassPath=&SizeBox
   bIsVariable=true
   parent=GridPanel_TrayContent
   widget=SizeBox_LeftContent
   properties {bOverride_MaxDesiredWidth=true, maxDesiredWidth=600}
   slot {class=GridSlot, properties {horizontalAlignment=HAlign_Left, row=1, verticalAlignment=VAlign_Center}}
  }
  VerticalBox_1 {widgetClassPath=&VerticalBox, parent=SizeBox_LeftContent, widget=VerticalBox_1, slot {class=SizeBoxSlot}}
  RewardCounter {
   widgetClassPath=&/RewardCounter
   bIsVariable=true
   parent=VerticalBox_1
   widget=RewardCounter
   slot {class=VerticalBoxSlot, properties {horizontalAlignment=HAlign_Left, verticalAlignment=VAlign_Top, padding=<ltrb 0 0 0 14>}}
  }
  WBP_UIKit_DescriptionWithSubtitle {
   widgetClassPath=/FortUILibrary/UIKit/Items/WBP_UIKit_DescriptionWithSubtitle
   bIsVariable=true
   parent=VerticalBox_1
   widget=WBP_UIKit_DescriptionWithSubtitle
   properties {
    bodyTextRow=3
    bottomSlotRow=5
    compatibilityRow=4
    hideBodyText=false
    hideGameplayTagsText=false
    hideSetText=false
    subtitleRow=3
    tagsOrientation=Orient_Vertical
    titleMaxHeight=155
    use Rich Description=false
   }
   slot {class=VerticalBoxSlot}
  }
  BonusInfo_VariantCard_Switcher {
   widgetClassPath=&CommonVisibilitySwitcher
   bIsVariable=true
   parent=VerticalBox_1
   widget=BonusInfo_VariantCard_Switcher
   slot {class=VerticalBoxSlot}
  }
  BonusInfoMiniTagNexus {
   widgetClassPath=&/BonusInfoMiniTagNexus
   bIsVariable=true
   parent=BonusInfo_VariantCard_Switcher
   widget=BonusInfoMiniTagNexus
   properties {bSupportsActivationFocus=false}
   slot {class=CommonVisibilitySwitcherSlot}
  }
  VariantCards {
   widgetClassPath=/Script/FortniteUI.FortDynamicEntryBox
   bIsVariable=true
   parent=BonusInfo_VariantCard_Switcher
   widget=VariantCards
   properties {
    entryWidgetClass=/FortUILibrary/UIKit/Items/WBP_UIKit_ItemCard_BattlePass.WBP_UIKit_ItemCard_BattlePass_C
    numDesignerPreviewEntries=2
    visibility=Collapsed
    entrySpacing=<14 0>
   }
   slot {class=CommonVisibilitySwitcherSlot, properties {padding=<ltrb 0 20 0 0>}}
  }
  WBP_CustomItemDetailsContainer {
   widgetClassPath=/FortniteCoreUI/Customization/MPL/WBP_CustomItemDetailsContainer
   bIsVariable=true
   parent=BonusInfo_VariantCard_Switcher
   widget=WBP_CustomItemDetailsContainer
   properties {visibility=Collapsed}
   slot {class=CommonVisibilitySwitcherSlot, properties {padding=<ltrb 0 12 12 12>}}
  }
  NamedSlot_CustomBanner {
   widgetClassPath=&NamedSlot
   bIsVariable=true
   parent=VerticalBox_1
   widget=NamedSlot_CustomBanner
   properties {visibility=Collapsed}
   slot {class=VerticalBoxSlot, properties {verticalAlignment=VAlign_Bottom, padding=<ltrb 0 0 0 12>}}
  }
  TrackContainerSpacer {
   widgetClassPath=&Spacer
   parent=GridPanel_TrayContent
   widget=TrackContainerSpacer
   properties {size=<1 250>}
   slot {class=GridSlot, properties {horizontalAlignment=HAlign_Left, row=2, verticalAlignment=VAlign_Top}}
  }
  LegalandActionBar {
   widgetClassPath=&Overlay
   bIsVariable=true
   parent=GridPanel_TrayContent
   widget=LegalandActionBar
   slot {class=GridSlot, properties {row=3, verticalAlignment=VAlign_Bottom}}
  }
  LegalText {
   widgetClassPath=&CommonTextBlock
   bIsVariable=true
   parent=LegalandActionBar
   widget=LegalText
   properties {
    autoWrapText=true
    bIsScrollingEnabled=false
    mobileFontSizeMultiplier=1.25
    text=Legal text goes here
    visibility=Collapsed
    wrapTextAt=720
    colorAndOpacity {colorUseRule=UseColor_Specified, specifiedColor=<rgba 1 1 1 0.6>}
    font {
     bForceMonospaced=false
     bMaterialIsStencil=false
     fontMaterial=null
     fontObject=&HeadingNow
     letterSpacing=0
     monospacedWidth=1
     size=24
     skewAmount=0
     typefaceFontName=86 Bold
     outlineSettings {
      bApplyOutlineToDropShadows=false
      bMiteredCorners=false
      bSeparateFillAlpha=false
      outlineMaterial=null
      outlineSize=0
      outlineColor=<rgba 0 0 0 1>
     }
    }
    shadowOffset=<0 0>
   }
   slot {class=OverlaySlot, properties {horizontalAlignment=HAlign_Fill, verticalAlignment=VAlign_Fill}}
  }
  HorizontalBox_1 {
   widgetClassPath=&HorizontalBox
   parent=LegalandActionBar
   widget=HorizontalBox_1
   properties {visibility=Collapsed}
   slot {class=OverlaySlot, properties {horizontalAlignment=HAlign_Fill}}
  }
  Button_ClaimPages {
   widgetClassPath=/FortUILibrary/UIKit/Buttons/WBP_UIKit_Button_Quiet
   bIsVariable=true
   parent=HorizontalBox_1
   widget=Button_ClaimPages
   properties {
    bIsFocusable=false
    text=Claim Pages
    visibility=Collapsed
    triggeringInputAction {dataTable=&InputActionDataTable, rowName=GenericLeftStickClick}
   }
   slot {class=HorizontalBoxSlot, properties {horizontalAlignment=HAlign_Left, verticalAlignment=VAlign_Top}}
  }
  Spacer_149 {
   widgetClassPath=&Spacer
   parent=HorizontalBox_1
   widget=Spacer_149
   slot {class=HorizontalBoxSlot, properties {size {sizeRule=Fill, value=1}}}
  }
  WBP_ActionBarContainer {
   widgetClassPath=/Game/UI/Frontend/BottomBar/WBP_ActionBarContainer
   bIsVariable=true
   parent=HorizontalBox_1
   widget=WBP_ActionBarContainer
   slot {class=HorizontalBoxSlot}
  }
  Button_Bar {widgetClassPath=&HorizontalBox, namedSlotHost=WBP_ActionBarContainer, widget=Button_Bar}
  ActionButtonsWrapper {
   widgetClassPath=&HorizontalBox
   bIsVariable=true
   parent=Button_Bar
   widget=ActionButtonsWrapper
   properties {visibility=Collapsed}
   slot {class=HorizontalBoxSlot, properties {verticalAlignment=VAlign_Center}}
  }
  Button_PreviewAction {
   widgetClassPath=&WBP_UIKit_Button_Quiet
   parent=ActionButtonsWrapper
   widget=Button_PreviewAction
   properties {
    bAutoUpdateFromInputAction=true
    bDisplayInputActionWhenNotInteractable=false
    bIsFocusable=false
    bShouldUseFallbackDefaultInputAction=false
    minWidth=48
    text=""
   }
   slot {class=HorizontalBoxSlot, properties {padding=<ltrb 0 0 24 0>}}
  }
  Button_PreviewAction1 {
   widgetClassPath=&WBP_UIKit_Button_Quiet
   parent=ActionButtonsWrapper
   widget=Button_PreviewAction1
   properties {
    bAutoUpdateFromInputAction=true
    bDisplayInputActionWhenNotInteractable=false
    bIsFocusable=false
    bShouldUseFallbackDefaultInputAction=false
    minWidth=48
    text=""
   }
   slot {class=HorizontalBoxSlot, properties {padding=<ltrb 0 0 24 0>}}
  }
  Button_PreviewAction2 {
   widgetClassPath=&WBP_UIKit_Button_Quiet
   parent=ActionButtonsWrapper
   widget=Button_PreviewAction2
   properties {
    bAutoUpdateFromInputAction=true
    bDisplayInputActionWhenNotInteractable=false
    bIsFocusable=false
    bShouldUseFallbackDefaultInputAction=false
    minWidth=48
    text=""
   }
   slot {class=HorizontalBoxSlot, properties {padding=<ltrb 0 0 24 0>}}
  }
  ClaimAllButtonWrapper {
   widgetClassPath=&Overlay
   bIsVariable=true
   parent=Button_Bar
   widget=ClaimAllButtonWrapper
   slot {class=HorizontalBoxSlot, properties {verticalAlignment=VAlign_Center}}
  }
  Button_ClaimAllRewards {
   widgetClassPath=&WBP_UIKit_Button_Quiet
   bIsVariable=true
   parent=ClaimAllButtonWrapper
   widget=Button_ClaimAllRewards
   properties {
    bIsFocusable=false
    text="Claim all Rewards (0)"
    visibility=Collapsed
    triggeringInputAction {dataTable=&InputActionDataTable, rowName=GenericRightStickClick}
   }
   slot {class=OverlaySlot, properties {padding=<ltrb 0 0 24 0>}}
  }
  Button_Back {
   widgetClassPath=&WBP_UIKit_Button_Quiet
   bIsVariable=true
   parent=Button_Bar
   widget=Button_Back
   properties {
    bIsFocusable=false
    text=Back
    visibility=Visible
    triggeringInputAction {dataTable=&InputActionDataTable, rowName=GenericBack}
   }
   slot {class=HorizontalBoxSlot, properties {verticalAlignment=VAlign_Center}}
  }
  SizeBox_RewardTrack {
   widgetClassPath=&SizeBox
   parent=Overlay_RewardTrack_Root
   widget=SizeBox_RewardTrack
   properties {bOverride_WidthOverride=true, heightOverride=180, widthOverride=2000}
   slot {class=OverlaySlot, properties {horizontalAlignment=HAlign_Fill, verticalAlignment=VAlign_Fill}}
  }
  Overlay_51 {widgetClassPath=&Overlay, parent=SizeBox_RewardTrack, widget=Overlay_51, slot {class=SizeBoxSlot}}
  VBox_TrackArea {
   widgetClassPath=&VerticalBox
   parent=Overlay_51
   widget=VBox_TrackArea
   slot {class=OverlaySlot, properties {horizontalAlignment=HAlign_Center, verticalAlignment=VAlign_Bottom, padding=<ltrb 0 0 0 5>}}
  }
  TabNavigation {
   widgetClassPath=&Overlay
   bIsVariable=true
   parent=VBox_TrackArea
   widget=TabNavigation
   properties {renderOpacity=0, visibility=Hidden}
   slot {class=VerticalBoxSlot, properties {horizontalAlignment=HAlign_Center, verticalAlignment=VAlign_Bottom}}
  }
  PostBufferUpdate_TabNavigation {
   widgetClassPath=&PostBufferUpdate
   bIsVariable=true
   parent=TabNavigation
   widget=PostBufferUpdate_TabNavigation
   properties {
    visibility=HitTestInvisible
    updateBufferInfos [{
      bufferToUpdate=ESlatePostRT_0
      postParamUpdater=&self:WidgetTree.PostBufferUpdate_TabNavigation.PostBufferBlurUpdater_0
     }]
   }
   slot {class=OverlaySlot, properties {horizontalAlignment=HAlign_Fill, verticalAlignment=VAlign_Fill, padding=<ltrb -100 -10 -100 -10>}}
  }
  WBP_UIKit_TabGroup {
   widgetClassPath=/FortUILibrary/UIKit/Tabs/WBP_UIKit_TabGroup
   bIsVariable=true
   parent=TabNavigation
   widget=WBP_UIKit_TabGroup
   properties {
    backplateBrightness=Obscure
    backplateCornerRadius=Circle
    buttonClass=/SeasonPass/Widgets/Reboot/WBP_RewardTrack_TopNav_Button.WBP_RewardTrack_TopNav_Button_C
    inputActions=None
    promptBackplateBrightness=Dark
    scrollClipping=OnDemand
    tabSpacing=8
    usePromptBackplate=false
    backplatePadding=<ltrb 5 5 5 5>
    initialTabsData [
     {
      bEnableTabId=false
      tabId=None
      tabViewModel=&self:WidgetTree.WBP_UIKit_TabGroup.UIKitSelectableButtonViewModel_0
     }
     {
      bEnableTabId=false
      tabId=None
      tabViewModel=&self:WidgetTree.WBP_UIKit_TabGroup.UIKitSelectableButtonViewModel_2
     }
     {
      bEnableTabId=false
      tabId=None
      tabViewModel=&self:WidgetTree.WBP_UIKit_TabGroup.UIKitSelectableButtonViewModel_1
     }
    ]
   }
   slot {class=OverlaySlot, properties {horizontalAlignment=HAlign_Center}}
  }
  PreviousButton {
   widgetClassPath=&FortVisualAttachment
   parent=TabNavigation
   widget=PreviousButton
   properties {contentAnchor=<1 0.5>}
   slot {class=OverlaySlot, properties {verticalAlignment=VAlign_Center}}
  }
  SecondaryLeft {
   widgetClassPath=&WBP_UIKit_Button_Generic
   bIsVariable=true
   parent=PreviousButton
   widget=SecondaryLeft
   properties {
    bIsFocusable=false
    iconMaterial=/FortUILibrary/Foundation/Buttons/Materials/Arrows/MI_UIKit_Icon_Button_Previous
    minHeight=48
    mobileMinHeight=80
    mobileMinWidth=80
    shouldPromptDirectionFollowCulture=false
    text=""
    useTextBlock=false
    buttonPadding=<ltrb 8 0 8 0>
    iconSize=<32 32>
    promptNudge=<-48 0>
    triggeringInputAction {dataTable=&InputActionDataTable, rowName=GenericLeftTrigger}
   }
   slot {class=SizeBoxSlot}
  }
  NextButton {
   widgetClassPath=&FortVisualAttachment
   parent=TabNavigation
   widget=NextButton
   properties {contentAnchor=<0 0.5>}
   slot {class=OverlaySlot, properties {horizontalAlignment=HAlign_Right, verticalAlignment=VAlign_Center}}
  }
  SecondaryRight {
   widgetClassPath=&WBP_UIKit_Button_Generic
   bIsVariable=true
   parent=NextButton
   widget=SecondaryRight
   properties {
    bIsFocusable=false
    iconMaterial=/FortUILibrary/Foundation/Buttons/Materials/Arrows/MI_UIKit_Icon_Button_Next
    minHeight=48
    mobileMinHeight=80
    mobileMinWidth=80
    promptHorizontalAlignment=HAlign_Right
    shouldPromptDirectionFollowCulture=false
    text=""
    useTextBlock=false
    buttonPadding=<ltrb 8 0 8 0>
    iconSize=<32 32>
    promptNudge=<48 0>
    triggeringInputAction {dataTable=&InputActionDataTable, rowName=GenericRightTrigger}
   }
   slot {class=SizeBoxSlot}
  }
  Spacer_92 {
   widgetClassPath=&Spacer
   parent=VBox_TrackArea
   widget=Spacer_92
   properties {size=<1 50>}
   slot {class=VerticalBoxSlot, properties {horizontalAlignment=HAlign_Left, verticalAlignment=VAlign_Top}}
  }
  Status {
   widgetClassPath=&Overlay
   parent=VBox_TrackArea
   widget=Status
   slot {class=VerticalBoxSlot, properties {verticalAlignment=VAlign_Top, padding=<ltrb -15 10 0 0>}}
  }
  Overlay_1 {widgetClassPath=&Overlay, parent=VBox_TrackArea, widget=Overlay_1, slot {class=VerticalBoxSlot}}
  SBox_TracksAndModeWidget {
   widgetClassPath=&SizeBox
   parent=Overlay_1
   widget=SBox_TracksAndModeWidget
   properties {
    bOverride_HeightOverride=true
    heightOverride=180
    navigation=&self:WidgetTree.SBox_TracksAndModeWidget.WidgetNavigation_0
   }
   slot {class=OverlaySlot}
  }
  HorizontalBox_0 {
   widgetClassPath=&HorizontalBox
   parent=SBox_TracksAndModeWidget
   widget=HorizontalBox_0
   properties {navigation=&self:WidgetTree.HorizontalBox_0.WidgetNavigation_0}
   slot {class=SizeBoxSlot, properties {horizontalAlignment=HAlign_Center}}
  }
  SBox_Tracks {widgetClassPath=&SizeBox, bIsVariable=true, parent=HorizontalBox_0, widget=SBox_Tracks, slot {class=HorizontalBoxSlot}}
  Overlay_138 {widgetClassPath=&Overlay, parent=SBox_Tracks, widget=Overlay_138, slot {class=SizeBoxSlot}}
  NamedSlot_Track_Current {
   widgetClassPath=&NamedSlot
   bIsVariable=true
   parent=Overlay_138
   widget=NamedSlot_Track_Current
   slot {class=OverlaySlot, properties {horizontalAlignment=HAlign_Center, verticalAlignment=VAlign_Fill}}
  }
  RewardCards_A {
   widgetClassPath=&DynamicEntryBox
   bIsVariable=true
   parent=NamedSlot_Track_Current
   widget=RewardCards_A
   properties {
    entryHorizontalAlignment=HAlign_Left
    entryVerticalAlignment=VAlign_Center
    entryWidgetClass=/SeasonPass/Widgets/Reboot/WBP_RewardTrack_ItemButton.WBP_RewardTrack_ItemButton_C
    numDesignerPreviewEntries=8
    entrySpacing=<16 0>
   }
   slot {class=PanelSlot}
  }
  NamedSlot_Track_Next {
   widgetClassPath=&NamedSlot
   bIsVariable=true
   parent=Overlay_138
   widget=NamedSlot_Track_Next
   slot {class=OverlaySlot, properties {horizontalAlignment=HAlign_Center, verticalAlignment=VAlign_Fill}}
  }
  RewardCards_B {
   widgetClassPath=&DynamicEntryBox
   bIsVariable=true
   parent=NamedSlot_Track_Next
   widget=RewardCards_B
   properties {
    entryHorizontalAlignment=HAlign_Left
    entryVerticalAlignment=VAlign_Center
    entryWidgetClass=&WBP_RewardTrack_ItemButton_C
    numDesignerPreviewEntries=8
    entrySpacing=<16 0>
   }
   slot {class=PanelSlot}
  }
  FortVisualAttachment_56 {
   widgetClassPath=&FortVisualAttachment
   parent=Overlay_138
   widget=FortVisualAttachment_56
   properties {contentAnchor=<0.1 1.05>}
   slot {class=OverlaySlot}
  }
  WBP_RewardInfoWidget {
   widgetClassPath=&/WBP_RewardInfoWidget
   bIsVariable=true
   parent=FortVisualAttachment_56
   widget=WBP_RewardInfoWidget
   properties {padding=<ltrb 0 0 0 10>}
   slot {class=SizeBoxSlot, properties {horizontalAlignment=HAlign_Left, verticalAlignment=VAlign_Top}}
  }
  FortVisualAttachment_74 {
   widgetClassPath=&FortVisualAttachment
   parent=Overlay_138
   widget=FortVisualAttachment_74
   properties {contentAnchor=<0 0.9>}
   slot {class=OverlaySlot, properties {padding=<ltrb 20 0 0 0>}}
  }
  SeasonPass_RewardTrack_StatusBar {
   widgetClassPath=&/SeasonPass_RewardTrack_StatusBar
   bIsVariable=true
   parent=FortVisualAttachment_74
   widget=SeasonPass_RewardTrack_StatusBar
   properties {isAbridged=true, renderOpacity=0, padding=<ltrb 0 0 0 10>}
   slot {class=SizeBoxSlot, properties {horizontalAlignment=HAlign_Left, verticalAlignment=VAlign_Top, padding=<ltrb -10 0 0 0>}}
  }
  Ovr_ModeSpecificQuestArea {
   widgetClassPath=&Overlay
   bIsVariable=true
   parent=HorizontalBox_0
   widget=Ovr_ModeSpecificQuestArea
   slot {class=HorizontalBoxSlot, properties {horizontalAlignment=HAlign_Left, verticalAlignment=VAlign_Center}}
  }
  Img_Ellipses {
   widgetClassPath=&Image
   parent=Ovr_ModeSpecificQuestArea
   widget=Img_Ellipses
   properties {brush {
     bIsDynamicallyLoaded=false
     drawAs=Image
     imageType=FullColor
     mirroring=NoMirror
     resourceName=None
     resourceObject=/FortUILibrary/Textures/Icons/T_UI_Icon_UIKit_ContextMenu
     tiling=NoTile
     imageSize=<32 32>
     margin=<ltrb 0 0 0 0>
     outlineSettings {
      bUseBrushTransparency=false
      roundingType=HalfHeightRadius
      width=0
      color {colorUseRule=UseColor_Specified, specifiedColor=<rgba 0 0 0 0>}
      cornerRadii=<0 0 0 0>
     }
     tintColor {colorUseRule=UseColor_Specified, specifiedColor=<rgba 1 1 1 1>}
     uVRegion {bIsValid=false, max=<0 0>, min=<0 0>}
    }}
   slot {class=OverlaySlot, properties {verticalAlignment=VAlign_Center, padding=<ltrb 12 0 0 0>}}
  }
  ModeSpecificReward {
   widgetClassPath=/SeasonPass/Widgets/Reboot/WBP_RewardTrack_ItemButton
   bIsVariable=true
   parent=Ovr_ModeSpecificQuestArea
   widget=ModeSpecificReward
   properties {isModeSpecificReward=true, itemCardSize=<256 160>}
   slot {class=OverlaySlot, properties {verticalAlignment=VAlign_Bottom, padding=<ltrb 56 0 0 0>}}
  }
  Ovr_BattlePassScroll {
   widgetClassPath=&Overlay
   parent=Overlay_RewardTrack_Root
   widget=Ovr_BattlePassScroll
   slot {class=OverlaySlot, properties {horizontalAlignment=HAlign_Center, verticalAlignment=VAlign_Bottom, padding=<ltrb 0 240 0 200>}}
  }
  CTxt_SeasonPassName {
   widgetClassPath=&CommonTextBlock
   parent=Ovr_BattlePassScroll
   widget=CTxt_SeasonPassName
   properties {
    text=Battle Pass
    textTransformPolicy=ToUpper
    font {
     bForceMonospaced=false
     bMaterialIsStencil=false
     fontMaterial=null
     fontObject=&HeadingNow
     letterSpacing=0
     monospacedWidth=1
     size=24
     skewAmount=0
     typefaceFontName=76 Bold
     outlineSettings {
      bApplyOutlineToDropShadows=false
      bMiteredCorners=false
      bSeparateFillAlpha=false
      outlineMaterial=null
      outlineSize=0
      outlineColor=<rgba 0 0 0 1>
     }
    }
    shadowOffset=<0 0>
   }
   slot {class=OverlaySlot, properties {horizontalAlignment=HAlign_Center}}
  }
  Icon_Chevron {
   widgetClassPath=&Image
   bIsVariable=true
   parent=Ovr_BattlePassScroll
   widget=Icon_Chevron
   properties {
    visibility=SelfHitTestInvisible
    brush {
     bIsDynamicallyLoaded=false
     drawAs=Image
     imageType=FullColor
     mirroring=NoMirror
     resourceName=None
     resourceObject=/FortUILibrary/Textures/MaterialInstances/MI_UIKit_Icon_Chevron_Up
     tiling=NoTile
     imageSize=<32 32>
     margin=<ltrb 0 0 0 0>
     outlineSettings {
      bUseBrushTransparency=false
      roundingType=HalfHeightRadius
      width=0
      color {colorUseRule=UseColor_Specified, specifiedColor=<rgba 0 0 0 0>}
      cornerRadii=<0 0 0 0>
     }
     tintColor {colorUseRule=UseColor_Specified, specifiedColor=<rgba 1 1 1 1>}
     uVRegion {bIsValid=false, max=<0 0>, min=<0 0>}
    }
    renderTransform {angle=0, scale=<1 -1>, shear=<0 0>, translation=<0 0>}
   }
   slot {class=OverlaySlot, properties {horizontalAlignment=HAlign_Center, padding=<ltrb 0 50 0 0>}}
  }
  Button_Battlepass {
   widgetClassPath=&Button
   parent=Ovr_BattlePassScroll
   widget=Button_Battlepass
   properties {
    isFocusable=false
    backgroundColor=<rgba 1 1 1 0>
    colorAndOpacity=<rgba 1 1 1 0>
    widgetStyle {
     clickedSlateSound {resourceObject=null}
     disabled {
      bIsDynamicallyLoaded=false
      drawAs=NoDrawType
      imageType=NoImage
      mirroring=NoMirror
      resourceName=None
      resourceObject=null
      tiling=NoTile
      imageSize=<0 0>
      margin=<ltrb 0 0 0 0>
      outlineSettings {
       bUseBrushTransparency=false
       roundingType=HalfHeightRadius
       width=0
       color {colorUseRule=UseColor_Specified, specifiedColor=<rgba 0 0 0 0>}
       cornerRadii=<0 0 0 0>
      }
      tintColor {colorUseRule=UseColor_Specified, specifiedColor=<rgba 1 1 1 1>}
      uVRegion {bIsValid=false, max=<0 0>, min=<0 0>}
     }
     disabledForeground {colorUseRule=UseColor_Specified, specifiedColor=<rgba 0.5271 0.5271 0.5271 1>}
     hovered {
      bIsDynamicallyLoaded=false
      drawAs=NoDrawType
      imageType=NoImage
      mirroring=NoMirror
      resourceName=None
      resourceObject=null
      tiling=NoTile
      imageSize=<0 0>
      margin=<ltrb 0 0 0 0>
      outlineSettings {
       bUseBrushTransparency=true
       roundingType=FixedRadius
       width=1
       color {colorUseRule=UseColor_Specified, specifiedColor=<rgba 0.7243 0.7243 0.7243 1>}
       cornerRadii=<4 4 4 4>
      }
      tintColor {colorUseRule=UseColor_Specified, specifiedColor=<rgba 1 0 0.7689 0.5>}
      uVRegion {bIsValid=false, max=<0 0>, min=<0 0>}
     }
     hoveredForeground {colorUseRule=UseColor_Specified, specifiedColor=<rgba 0 1 1 0>}
     hoveredSlateSound {resourceObject=null}
     normal {
      bIsDynamicallyLoaded=false
      drawAs=NoDrawType
      imageType=NoImage
      mirroring=NoMirror
      resourceName=None
      resourceObject=null
      tiling=NoTile
      imageSize=<0 0>
      margin=<ltrb 0 0 0 0>
      outlineSettings {
       bUseBrushTransparency=true
       roundingType=FixedRadius
       width=1
       color {colorUseRule=UseColor_Specified, specifiedColor=<rgba 0.6951 0.6951 0.6951 1>}
       cornerRadii=<4 4 4 4>
      }
      tintColor {colorUseRule=UseColor_Specified, specifiedColor=<rgba 1 0 0.7689 0.5>}
      uVRegion {bIsValid=false, max=<0 0>, min=<0 0>}
     }
     normalForeground {colorUseRule=UseColor_Specified, specifiedColor=<rgba 0 1 1 0>}
     normalPadding=<ltrb 0 0 0 0>
     pressed {
      bIsDynamicallyLoaded=false
      drawAs=NoDrawType
      imageType=NoImage
      mirroring=NoMirror
      resourceName=None
      resourceObject=null
      tiling=NoTile
      imageSize=<0 0>
      margin=<ltrb 0 0 0 0>
      outlineSettings {
       bUseBrushTransparency=true
       roundingType=FixedRadius
       width=1
       color {colorUseRule=UseColor_Specified, specifiedColor=<rgba 0.7243 0.7243 0.7243 1>}
       cornerRadii=<4 4 4 4>
      }
      tintColor {colorUseRule=UseColor_Specified, specifiedColor=<rgba 1 0.4688 0.8772 0.5>}
      uVRegion {bIsValid=false, max=<0 0>, min=<0 0>}
     }
     pressedForeground {colorUseRule=UseColor_Specified, specifiedColor=<rgba 0 1 1 0>}
     pressedPadding=<ltrb 0 0 0 0>
     pressedSlateSound {resourceObject=null}
    }
   }
   slot {class=OverlaySlot, properties {horizontalAlignment=HAlign_Fill, verticalAlignment=VAlign_Fill}}
  }
  Ovr_MatchRecap {
   widgetClassPath=&Overlay
   parent=Overlay_RewardTrack_Root
   widget=Ovr_MatchRecap
   properties {renderOpacity=0}
   slot {class=OverlaySlot, properties {horizontalAlignment=HAlign_Center, padding=<ltrb 0 240 0 200>}}
  }
  CTxt_MatchRecapName {
   widgetClassPath=&CommonTextBlock
   parent=Ovr_MatchRecap
   widget=CTxt_MatchRecapName
   properties {
    text=Match Recap
    textTransformPolicy=ToUpper
    font {
     bForceMonospaced=false
     bMaterialIsStencil=false
     fontMaterial=null
     fontObject=&HeadingNow
     letterSpacing=0
     monospacedWidth=1
     size=24
     skewAmount=0
     typefaceFontName=76 Bold
     outlineSettings {
      bApplyOutlineToDropShadows=false
      bMiteredCorners=false
      bSeparateFillAlpha=false
      outlineMaterial=null
      outlineSize=0
      outlineColor=<rgba 0 0 0 1>
     }
    }
    shadowOffset=<0 0>
   }
   slot {class=OverlaySlot, properties {horizontalAlignment=HAlign_Center}}
  }
  Icon_Chevron_1 {
   widgetClassPath=&Image
   bIsVariable=true
   parent=Ovr_MatchRecap
   widget=Icon_Chevron_1
   properties {
    visibility=SelfHitTestInvisible
    brush {
     bIsDynamicallyLoaded=false
     drawAs=Image
     imageType=FullColor
     mirroring=NoMirror
     resourceName=None
     resourceObject=&MI_UIKit_Icon_Chevron_Up
     tiling=NoTile
     imageSize=<32 32>
     margin=<ltrb 0 0 0 0>
     outlineSettings {
      bUseBrushTransparency=false
      roundingType=HalfHeightRadius
      width=0
      color {colorUseRule=UseColor_Specified, specifiedColor=<rgba 0 0 0 0>}
      cornerRadii=<0 0 0 0>
     }
     tintColor {colorUseRule=UseColor_Specified, specifiedColor=<rgba 1 1 1 1>}
     uVRegion {bIsValid=false, max=<0 0>, min=<0 0>}
    }
    renderTransform {angle=180, scale=<1 -1>, shear=<0 0>, translation=<0 0>}
   }
   slot {class=OverlaySlot, properties {horizontalAlignment=HAlign_Center, padding=<ltrb 0 50 0 0>}}
  }
  Button_MatchRecap {
   widgetClassPath=&Button
   parent=Ovr_MatchRecap
   widget=Button_MatchRecap
   properties {
    isFocusable=false
    widgetStyle {
     clickedSlateSound {resourceObject=null}
     disabled {
      bIsDynamicallyLoaded=false
      drawAs=NoDrawType
      imageType=NoImage
      mirroring=NoMirror
      resourceName=None
      resourceObject=null
      tiling=NoTile
      imageSize=<0 0>
      margin=<ltrb 0 0 0 0>
      outlineSettings {
       bUseBrushTransparency=false
       roundingType=HalfHeightRadius
       width=0
       color {colorUseRule=UseColor_Specified, specifiedColor=<rgba 0 0 0 0>}
       cornerRadii=<0 0 0 0>
      }
      tintColor {colorUseRule=UseColor_Specified, specifiedColor=<rgba 1 1 1 1>}
      uVRegion {bIsValid=false, max=<0 0>, min=<0 0>}
     }
     disabledForeground {colorUseRule=UseColor_Specified, specifiedColor=<rgba 0.5271 0.5271 0.5271 1>}
     hovered {
      bIsDynamicallyLoaded=false
      drawAs=NoDrawType
      imageType=NoImage
      mirroring=NoMirror
      resourceName=None
      resourceObject=null
      tiling=NoTile
      imageSize=<0 0>
      margin=<ltrb 0 0 0 0>
      outlineSettings {
       bUseBrushTransparency=true
       roundingType=FixedRadius
       width=1
       color {colorUseRule=UseColor_Specified, specifiedColor=<rgba 0.7243 0.7243 0.7243 1>}
       cornerRadii=<4 4 4 4>
      }
      tintColor {colorUseRule=UseColor_Specified, specifiedColor=<rgba 0.7243 0.7243 0.7243 1>}
      uVRegion {bIsValid=false, max=<0 0>, min=<0 0>}
     }
     hoveredForeground {colorUseRule=UseColor_Specified, specifiedColor=<rgba 0 1 1 1>}
     hoveredSlateSound {resourceObject=null}
     normal {
      bIsDynamicallyLoaded=false
      drawAs=NoDrawType
      imageType=NoImage
      mirroring=NoMirror
      resourceName=None
      resourceObject=null
      tiling=NoTile
      imageSize=<0 0>
      margin=<ltrb 0 0 0 0>
      outlineSettings {
       bUseBrushTransparency=true
       roundingType=FixedRadius
       width=1
       color {colorUseRule=UseColor_Specified, specifiedColor=<rgba 0.6951 0.6951 0.6951 1>}
       cornerRadii=<4 4 4 4>
      }
      tintColor {colorUseRule=UseColor_Specified, specifiedColor=<rgba 0.4955 0.4955 0.4955 1>}
      uVRegion {bIsValid=false, max=<0 0>, min=<0 0>}
     }
     normalForeground {colorUseRule=UseColor_Specified, specifiedColor=<rgba 0 1 1 1>}
     normalPadding=<ltrb 0 0 0 0>
     pressed {
      bIsDynamicallyLoaded=false
      drawAs=NoDrawType
      imageType=NoImage
      mirroring=NoMirror
      resourceName=None
      resourceObject=null
      tiling=NoTile
      imageSize=<0 0>
      margin=<ltrb 0 0 0 0>
      outlineSettings {
       bUseBrushTransparency=true
       roundingType=FixedRadius
       width=1
       color {colorUseRule=UseColor_Specified, specifiedColor=<rgba 0.7243 0.7243 0.7243 1>}
       cornerRadii=<4 4 4 4>
      }
      tintColor {colorUseRule=UseColor_Specified, specifiedColor=<rgba 0.3843 0.3843 0.3843 1>}
      uVRegion {bIsValid=false, max=<0 0>, min=<0 0>}
     }
     pressedForeground {colorUseRule=UseColor_Specified, specifiedColor=<rgba 0 1 1 1>}
     pressedPadding=<ltrb 0 0 0 0>
     pressedSlateSound {resourceObject=null}
    }
   }
   slot {class=OverlaySlot, properties {horizontalAlignment=HAlign_Fill, verticalAlignment=VAlign_Fill}}
  }
  CPanel_ContextMenuRoot {
   widgetClassPath=&CanvasPanel
   bIsVariable=true
   parent=Overlay_RewardTrack_Root
   widget=CPanel_ContextMenuRoot
   properties {visibility=Collapsed}
   slot {class=OverlaySlot, properties {horizontalAlignment=HAlign_Fill, verticalAlignment=VAlign_Fill}}
  }
  SeasonPassCategory_Buttons_ContextMenu {
   widgetClassPath=&/SeasonPassCategory_Buttons_ContextMenu
   bIsVariable=true
   parent=CPanel_ContextMenuRoot
   widget=SeasonPassCategory_Buttons_ContextMenu
   slot {class=CanvasPanelSlot, properties {layoutData {alignment=<0 0>, anchors {maximum=<0 0>, minimum=<0 0>}, offsets=<ltrb 0 0 0 0>}}}
  }
 ]
}
