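# Widget Blueprint price switcher for simple offers, derived from WBP_PriceSwitcher_Base in the item shop tile system.
# On Construct, if HasRealPrice is true, it tells the FortMtxStoreVM catalog view model to start monitoring for updates.
# On Destruct, under the same check, it stops that monitoring.
# It declares no widgets or variables of its own.
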WidgetBlueprint /FortUILibrary/UIKit/Price/WBP_PriceSwitcher_SimpleOffer
graphs [EventGraph {nodes [
   Event Construct()
   > Branch(Condition=Get HasRealPrice() -> (HasRealPrice:bool))
   n1 = Get FortMtxStoreVM() -> (FortMtxStoreVM:Fort Mtx Store VM Object Reference)
   > StartMonitoringForUpdates!(execute, self:Fort Mcp Catalog VM Object Reference=@n1, RequestingObject:Object Reference) -> (then)
   Event Destruct()
   > Branch(Condition=Get HasRealPrice())
   > StopMonitoringForUpdates!(execute, self:Fort Mcp Catalog VM Object Reference=@n1, RequestingObject:Object Reference) -> (then)
  ]}]
variables []
widgets {
 info {
  inheritedWidgetCount=0
  parentClass=/FortUILibrary/Systems/ItemShopTile/WBP_PriceSwitcher_Base.WBP_PriceSwitcher_Base_C
  rootWidgetClass=null
  widgetCount=0
 }
 widgets []
}
