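# M_UI_Icon_Locked is a translucent UI material that draws a locked icon for the SeasonPass interface.
# The red channel of a signed-distance-field lock texture feeds MF_UI_AntiAliasing, with parameters for threshold and smoothing.
# Its Outer result blends from a near-black constant color to the default lerp value.
# A separate UV distance-from-center branch uses an inverted smoothstep to build a circular mask with a soft edge.
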Material /SeasonPass/Materials/M_UI_Icon_Locked
graph {
 properties {
  bAutomaticallySetUsageInEditor=false
  bUsedWithStaticMesh=false
  blendMode=BLEND_Translucent
  materialDomain=MD_UI
  usedShadingModels=Default Lit
 }
 nodes [
  n1 = TextureSample() {samplerType=SAMPLERTYPE_LinearColor, texture=/FortUILibrary/Textures/Icons/T_UI_Icon_UIKit_Locked_SDF}
  LinearInterpolate(
   A=Constant3Vector() {constant=<rgba 0.0048 0.004 0.0116 1>}
   Alpha=MaterialFunctionCall(
    DistanceField=ComponentMask(None=@n1.RGB) {r=true}
    Threshold=ScalarParameter() {defaultValue=0.68, parameterName=Fill Aliasing}
    Smoothing=ScalarParameter() {parameterName=Smoothing, sliderMax=1}
   ) {materialFunction=&MF_UI_AntiAliasing}.Outer
  )
  n2 = ComponentMask(None=TextureCoordinate()) {g=true, r=true}
  n3 = Constant() {r=0.5}
  OneMinus(None=SmoothStep(Value=Distance(A=@n2, B=@n3)) {constMax=0.5, constMin=0.495, constValue=0.49})
 ]
}
