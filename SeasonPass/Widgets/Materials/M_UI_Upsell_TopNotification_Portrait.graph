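# UI material for the portrait upsell top notification: a circular badge with launch-week character art and an icon.
# An offer color and its hue-shifted, darkened variant form the base, split by a procedural rotated bolt shape.
# Character art is positioned, scaled and masked by circle and step masks with a content offset.
# An MSDF icon, converted to SDF and anti-aliased, is composited in white on top.
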Material /SeasonPass/Widgets/Materials/M_UI_Upsell_TopNotification_Portrait
graph {
 properties {
  bAutomaticallySetUsageInEditor=false
  bUsedWithStaticMesh=false
  blendMode=BLEND_Translucent
  materialDomain=MD_UI
  usedShadingModels=Default Lit
 }
 nodes [
  n1 = NamedRerouteUsage()
  n2 = AppendVector(A=ScalarParameter() {defaultValue=0.1, parameterName=CharacterPos_X}, B=ScalarParameter() {defaultValue=0.1, parameterName=CharacterPos_Y})
  n3 = TextureSample(UVs=MaterialFunctionCall(UVs=Add(A=@n1, B=@n2), Texture Scale=ScalarParameter() {defaultValue=0.825, parameterName=CharacterScale}) {materialFunction=&ScaleUVsByCenter}.UVs) {texture=/SeasonPass/Widgets/Textures/T_UI_UpsellScreen_LaunchWeekBonus}
  n4 = Multiply(A=@n3.A, B=NamedRerouteUsage())
  n5 = NamedRerouteUsage()
  n6 = AppendVector(A=ScalarParameter() {defaultValue=-0.145, parameterName=IconPos_X}, B=ScalarParameter() {defaultValue=-0.15, parameterName=IconPos_Y})
  n7 = TextureSampleParameter2D(UVs=MaterialFunctionCall(UVs=Add(A=@n5, B=@n6), Texture Scale=ScalarParameter() {defaultValue=0.3, parameterName=IconScale}) {materialFunction=&ScaleUVsByCenter}.UVs) {parameterName=Icon, samplerType=SAMPLERTYPE_LinearColor, texture=/SeasonPass/Widgets/Textures/T_UI_Icon_LaunchWeek_MSDF}
  LinearInterpolate(
   A=LinearInterpolate(A=NamedRerouteUsage(), B=@n3.RGB, Alpha=@n4)
   Alpha=MaterialFunctionCall(
    DistanceField=MaterialFunctionCall!(MSDF RGB=@n7.RGB) {materialFunction=/FortUILibrary/MaterialFunctions/Shapes/Modifiers/MF_UI_MSDFtoSDF}.SDF
    Threshold=Constant() {r=0.5}
   ) {materialFunction=&MF_UI_AntiAliasing}.Outer
  )
  Max(A=@n4, B=NamedRerouteUsage())
  NamedRerouteDeclaration(None=MaterialFunctionCall() {materialFunction=&MF_UI_UVScaling}.Scaled to Fit) {name=UV, nodeColor=<rgba 0 0 0 1>}
  n8 = ScalarParameter() {defaultValue=0.8, parameterName=Circle Mask Size}
  n9 = Multiply(A=Distance(A=NamedRerouteUsage(), B=Constant() {r=0.5})) {constB=2}
  NamedRerouteDeclaration(None=OneMinus(None=MaterialFunctionCall!(Threshold=@n8, Gradient=@n9, Filter Width Override, Filter Width Fudge Factor) {materialFunction=/Game/UI/Materials/AAThreshold}.Result)) {name=Circle, nodeColor=<rgba 0 0 0 1>}
  n10 = NamedRerouteUsage()
  n11 = Multiply(A=Distance(A=ComponentMask(None=@n10) {r=true}, B=Constant() {r=0.5})) {constB=2}
  n12 = ScalarParameter() {defaultValue=0.8, parameterName=Circle Mask Size}
  n13 = ComponentMask(None=@n10) {g=true}
  n14 = Max(A=Min(A=Step(Y=@n11, X=@n12), B=Step(Y=@n13) {constX=0.5}), B=NamedRerouteUsage())
  n15 = Step(Y=OneMinus(None=@n13), X=Add(A=Constant() {r=1}, B=ScalarParameter() {parameterName=Content Offset})) {constX=1.075}
  n16 = NamedRerouteUsage()
  NamedRerouteDeclaration(None=Min(A=Multiply(A=@n14, B=@n15), B=OneMinus(None=@n16))) {name=Content Mask, nodeColor=<rgba 0 0 0 1>}
  n17 = Add(A=NamedRerouteUsage(), B=Constant2Vector() {r=-0.05})
  n18 = ComponentMask(None=@n17) {r=true}
  n19 = ComponentMask(None=@n17) {g=true}
  n20 = ScalarParameter() {defaultValue=0.06, parameterName=Bolt_ Size}
  n21 = LinearInterpolate(A=Multiply(A=@n20) {constB=-1}, B=@n20, Alpha=Step(Y=@n18) {constX=0.5})
  n22 = MaterialFunctionCall(UVs=AppendVector(A=@n18, B=Add(A=@n19, B=@n21)), "Rotation Angle (0-1)"=ScalarParameter() {defaultValue=0.04, parameterName=Bolt_Rotation}) {materialFunction=&CustomRotator}
  n23 = Saturate(None=OneMinus(None=ComponentMask(None=@n22.Rotated Values) {r=true}))
  n24 = Saturate(None=ComponentMask(None=@n22.Rotated Values) {g=true})
  NamedRerouteDeclaration(None=SmoothStep(Value=Divide(A=@n23, B=@n24)) {constMax=0.95, constMin=1}) {name=Bolt, nodeColor=<rgba 0 0 0 1>}
  n25 = VectorParameter() {group=Theming, parameterName=Offer Color, defaultValue=<rgba 0.8714 0.3467 0.0561 1>}
  n26 = Multiply(A=MaterialFunctionCall(Hue Shift Percentage=Constant() {r=-0.002}, Texture=@n25.RGB) {materialFunction=&HueShift}.Result) {constB=0.8}
  NamedRerouteDeclaration(None=LinearInterpolate(A=@n25.RGB, B=Multiply(A=@n26, B=@n26), Alpha=NamedRerouteUsage())) {name=Base Colors, nodeColor=<rgba 0 0 0 1>}
 ]
}
