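# Season Pass widget wrapping WBP_UIKit_CompatibilityTags for cross-product rewards, configured to skip the Product.Mars tag.
# CF_GetOtherItemDetailsVM returns details of the first reward item whose product differs from the previewed item's and is not an AthenaDance, otherwise none.
# CF_GetVisibility collapses unless the reward holds multiple items.
# The event graph is empty; logic lives in these two conversion-style functions.
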WidgetBlueprint /SeasonPass/Widgets/Reboot/WBP_CrossProductCompatibilityTags
graphs [
 CF_GetOtherItemDetailsVM {nodes [
   n1 = CF_GetOtherItemDetailsVM!() -> (then, PreviewedItemVM:Fort Item Details VM Object Reference, SelectedReward:Season Pass Reward VM Object Reference) {is=title K2Node_FunctionEntry}
   > n2 = For Each Loop(Array:Array of Fort Item VM Object References=Get ItemVMs!(self:Season Pass Reward VM Object Reference=@n1.SelectedReward) -> (ItemVMs:Array of Fort Item VM Object References)) -> (Array Element:Fort Item VM Object Reference)
   Return Node!(execute=@n2.Completed, ItemDetailsVM:Fort Item Details VM Object Reference)
   > n3 = Set LoopItemVM(LoopItemVM:Fort Item VM Object Reference=@n2) -> (Output_Get:Fort Item VM Object Reference)
   n4 = Get Item!(self:Fort Item Details VM Object Reference=Get PreviewedItemVM() -> (PreviewedItemVM:Fort Item Details VM Object Reference)) -> (Item:Fort Item VM Object Reference)
   > n5 = Branch(Condition=MatchesTag!(
     TagOne:Gameplay Tag Structure=GetCurrentProduct!(self:Fort Item VM Object Reference=@n3) -> (ReturnValue:Gameplay Tag Structure)
     TagTwo:Gameplay Tag Structure=GetCurrentProduct(self=@n4)
     bExactMatch:bool=true
    ) -> (ReturnValue:bool))
   n6 = Get LoopItemVM() -> (LoopItemVM:Fort Item VM Object Reference)
   n7 = Get ItemDefinition!(self:Fort Item VM Object Reference=@n6) -> (ItemDefinition:Fort Item Definition Object Reference)
   Branch(
    execute=@n5.else
    Condition="Not Equal (Enum)"!(
     A:EFortItemType Enum=GetItemType!(ItemDefinition:Item Definition Base Object Reference=@n7) -> (ReturnValue:EFortItemType Enum=WorldItem)
     B:EFortItemType Enum=AthenaDance
    ) -> (ReturnValue:bool)
   )
   > Return Node(ItemDetailsVM=Get ItemDetails!(self:Fort Item VM Object Reference=@n6) -> (ItemDetails:Fort Item Details VM Object Reference))
  ]}
 CF_GetVisibility {nodes [
   n1 = CF_GetVisibility!() -> (then, SeasonPassRewardVM:Season Pass Reward VM Object Reference) {is=title K2Node_FunctionEntry}
   n2 = Length(TargetArray:Array of Fort Item VM Object References=Get ItemVMs(self=@n1))
   > Return Node!(execute, ReturnValue:ESlateVisibility Enum=Select(Option 0=Collapsed, Option 1=SelfHitTestInvisible, Index="integer > integer"(A=@n2, B=1)))
  ]}
 EventGraph {links [], nodes []}
]
variables []
widgets {
 info {
  inheritedWidgetCount=0
  parentClass=&UserWidget
  rootWidgetClass=/FortUILibrary/UIKit/Items/WBP_UIKit_CompatibilityTags.WBP_UIKit_CompatibilityTags_C
  widgetCount=1
 }
 widgets [WBP_UIKit_CompatibilityTags {
   widgetClassPath=/FortUILibrary/UIKit/Items/WBP_UIKit_CompatibilityTags
   bIsVariable=true
   widget=WBP_UIKit_CompatibilityTags
   properties {skipTagList [{tagName=Product.Mars}]}
  }]
}
