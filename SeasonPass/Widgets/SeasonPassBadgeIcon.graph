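# SeasonPassBadgeIcon is a widget showing a season pass badge whose material reflects pass ownership.
# A SizeBox, sized from Size in PreConstruct, wraps a lazy image that SetLargeIcon can swap.
# Brush field-change notifications and icon swaps refresh the dynamic material's IsPaid scalar from SeasonPassVM, defaulting to unowned without one.
# SetDoesOwnPass sets an override OR'd with ownership, but updates immediately with ownership false, ignoring the view model until refresh.
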WidgetBlueprint /SeasonPass/Widgets/SeasonPassBadgeIcon
graphs [
 EventGraph {nodes [
   n1 = PassBadgeBrushChanged!() -> (OutputDelegate, then, Object:Object Reference, Field:Field Notification Id Structure) {is=title K2Node_CustomEvent}
   > RefreshPassOwnership()
   Event OnInitialized()
   > Add Field Value Changed Delegate!(
    execute
    self:Widget Object Reference=Get Pass_Badge() -> (Pass_Badge:Fort Lazy Image Object Reference)
    FieldId:Field Notification Id Structure=(FieldName="Brush")
    Delegate:Delegate=@n1
   ) -> (then)
   Event PreConstruct()
   > SetBadgeSize!(execute, self, Size:Vector 2D Structure=Get Size() -> (Size:Vector 2D Structure)) -> (then)
  ]}
 OnBadgeBrushUpdated {nodes [
   OnBadgeBrushUpdated!() -> (then, Brush:Slate Brush Structure) {is=title K2Node_FunctionEntry}
   > RefreshPassOwnership()
  ]}
 RefreshPassOwnership {nodes [
   RefreshPassOwnership() {is=title K2Node_FunctionEntry}
   n1 = Get SeasonPassVM() -> (SeasonPassVM:Season Pass VM Object Reference)
   > n2 = Is Valid(InputObject=@n1)
   UpdatePassOwnershipVisual!(execute=@n2.Is Not Valid, self, OwnsPass:bool=false) -> (then)
   UpdatePassOwnershipVisual(execute=@n2.Is Valid, OwnsPass=Get bDoesOwnsPass!(self:Season Pass VM Object Reference=@n1) -> (bDoesOwnsPass:bool))
  ]}
 SetBadgeSize {nodes [
   n1 = SetBadgeSize!() -> (then, Size:Vector 2D Structure) {is=title K2Node_FunctionEntry}
   n2 = BreakVector2D(InVec=@n1)
   n3 = Get SizeBox() -> (SizeBox:Size Box Object Reference)
   > SetWidthOverride(self=@n3, InWidthOverride=@n2)
   > SetHeightOverride(self=@n3, InHeightOverride=@n2.Y)
  ]}
 SetDoesOwnPass {nodes [
   n1 = SetDoesOwnPass!() -> (then, DoesOwnPass:bool=~) {is=title K2Node_FunctionEntry}
   > Set OverrideToOwnPass(OverrideToOwnPass:bool=@n1) -> (Output_Get:bool)
   > UpdatePassOwnershipVisual(OwnsPass=false)
  ]}
 SetLargeIcon {nodes [
   n1 = SetLargeIcon!() -> (then, Icon:Soft Object Reference) {is=title K2Node_FunctionEntry}
   > SetBrushFromLazyDisplayAsset!(
    execute
    self:Common Lazy Image Object Reference=Get Pass_Badge()
    LazyObject:"Soft Object Reference (by ref)"=@n1
    bMatchTextureSize:bool=false
   ) -> (then)
   > RefreshPassOwnership() {is=title K2Node_CallFunction}
  ]}
 UpdatePassOwnershipVisual {nodes [
   n1 = UpdatePassOwnershipVisual!() -> (then, OwnsPass:bool=~) {is=title K2Node_FunctionEntry}
   > n2 = GetDynamicMaterial(self=Get Pass_Badge())
   n3 = Get OverrideToOwnPass() -> (OverrideToOwnPass:bool)
   > SetScalarParameterValue(self=@n2, ParameterName=IsPaid, Value=SelectFloat(A=1, B=0, bPickA=OR Boolean(A=@n1, B=@n3)))
  ]}
]
variables [
 Icon {type=Object, default=null}
 OverrideToOwnPass {type=boolean}
 Size {type=Vector2D, default=<64 64>}
]
widgets {
 info {inheritedWidgetCount=0, parentClass=&UserWidget, rootWidgetClass=&SizeBox, widgetCount=2}
 widgets [
  SizeBox {
   widgetClassPath=&SizeBox
   bIsVariable=true
   widget=SizeBox
   properties {bOverride_HeightOverride=true, bOverride_WidthOverride=true, heightOverride=64, widthOverride=64}
  }
  Pass_Badge {
   widgetClassPath=&FortLazyImage
   bIsVariable=true
   parent=SizeBox
   widget=Pass_Badge
   properties {brush {
     bIsDynamicallyLoaded=false
     drawAs=Image
     imageType=FullColor
     mirroring=NoMirror
     resourceName=None
     resourceObject=/SeasonPass/Materials/HeaderIcon/MI_UI_NexusHeaderIcon_BattlePass
     tiling=NoTile
     imageSize=<140 140>
     margin=<ltrb 0 0 0 0>
     outlineSettings {
      bUseBrushTransparency=false
      roundingType=HalfHeightRadius
      width=0
      color {colorUseRule=UseColor_Specified, specifiedColor=<rgba 0 0 0 0>}
      cornerRadii=<0 0 0 0>
     }
     tintColor {colorUseRule=UseColor_Specified, specifiedColor=<rgba 1 1 1 1>}
     uVRegion {bIsValid=false, max=<0 0>, min=<0 0>}
    }}
   slot {class=SizeBoxSlot}
  }
 ]
}
