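# Season pass widget showing the selected reward's status tags, unlock progress and action buttons.
# View-model-driven update functions toggle tags from a header bitmask and switch between buy, claim and claim-all buttons.
# Clicks fire analytics, then open purchase or customization screens, equip, or claim asynchronously, disabling claim buttons meanwhile.
# Config selects hold or tap claiming; hold progress feeds the selection view model, with a gated retriggerable delay clearing IsClaimingAll after release.
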WidgetBlueprint /SeasonPass/Widgets/SeasonPassCategory_ButtonsSwitcher
graphs [
 CF_GetClaimAllButtonsVisibility {nodes [
   n1 = CF_GetClaimAllButtonsVisibility!() -> (then, ClaimableRewardsInPage:int, IsPageLocked:bool=~) {is=title K2Node_FunctionEntry}
   n2 = NOT Boolean(A=@n1.IsPageLocked)
   n3 = "integer > integer"(A=@n1)
   > Return Node!(execute, Visibility:ESlateVisibility Enum=Select(Option 0=Collapsed, Option 1=SelfHitTestInvisible, Index=AND Boolean(A=@n3, B=@n2)))
  ]}
 EventGraph {nodes [
   n1 = "On Clicked (Button_ViewReward)"!() -> (OutputDelegate, then, Button:Common Button Base Object Reference) {is=title K2Node_ComponentBoundEvent}
   n2 = Make Flow Analytics Additional Info!() -> (FlowAnalyticsAdditionalInfo:Flow Analytics Additional Info Structure) {is=title K2Node_MakeStruct}
   n3 = Get SeasonPassSelectionVM() -> (SeasonPassSelectionVM:Season Pass Selection VM Object Reference)
   > Fire Event Element Interacted(
    Element=@n1.Button
    AdditionalInfo=Get Analytics Info Item Name!(
     MakeFlowAnalyticsAdditionalInfo:"Flow Analytics Additional Info Structure (by ref)"=Get Analytics Info Pass Category Name!(
      MakeFlowAnalyticsAdditionalInfo:"Flow Analytics Additional Info Structure (by ref)"=@n2
      Key:String
      SeasonPassSelectedVM:Season Pass Selection VM Object Reference=@n3
     ) -> (PassName:Flow Analytics Additional Info Structure) {is=title K2Node_MacroInstance}
     Key:String=SelectReward
     FortItemVM:Fort Item VM Object Reference=Get FocusedFortItemVM() -> (FocusedFortItemVM:Fort Item VM Object Reference)
    ) -> (ItemName:Flow Analytics Additional Info Structure) {is=title K2Node_MacroInstance}
   )
   > Call View Reward Button Clicked!(execute:Exec, self:Season Pass Category Buttons Switcher Object Reference) -> (then:Exec) {is=title K2Node_CallDelegate}
   n4 = "On Clicked (LoudButton_BuySeasonPass)"!() -> (OutputDelegate, then, Button:Common Button Base Object Reference) {is=title K2Node_ComponentBoundEvent}
   n5 = Get Analytics Info Pass Category Name(MakeFlowAnalyticsAdditionalInfo=Make Flow Analytics Additional Info(), SeasonPassSelectedVM=Get SeasonPassSelectionVM())
   > Fire Event Element Interacted(Element=@n4.Button, AdditionalInfo=@n5)
   > n6 = PushPrimaryContentSoftClassAsync!(
    execute:Exec
    SoftWidgetClass:Common Activatable Widget Soft Class Reference=GetActivatableWidgetForAlias!(WidgetAlias:Gameplay Tag Structure=(TagName="UI.Alias.PassPurchase")) -> (ReturnValue:Common Activatable Widget Soft Class Reference)
   ) -> (then:Exec, OnPrimaryContentPushed:Exec, ActivatableWidget:Common Activatable Widget Object Reference) {is=title K2Node_AsyncAction}
   SetViewModelByClass!(
    execute=@n6.OnPrimaryContentPushed
    self:MVVM View Object Reference=GetExtension!(self:User Widget Object Reference=@n6, ExtensionType:User Widget Extension Class Reference=/Script/ModelViewViewModel.MVVMView) -> (ReturnValue:MVVM View Object Reference)
    NewValue:Notify Field Value Changed Interface=Get SeasonPassVM() -> (SeasonPassVM:Season Pass VM Object Reference)
   ) -> (then, ReturnValue:bool)
   n7 = "On Pressed (LoudButton_ClaimAll)"!() -> (OutputDelegate, then, Button:Common Button Base Object Reference) {is=title K2Node_ComponentBoundEvent}
   n8 = Get Analytics Info Pass Category Name(MakeFlowAnalyticsAdditionalInfo=Make Flow Analytics Additional Info(), SeasonPassSelectedVM=Get SeasonPassSelectionVM())
   > n9 = Fire Event Element Interacted(Element=@n7.Button, AdditionalInfo=@n8)
   n10 = Set IsClaimingAll!(execute=[@n9.then, @n18.then], IsClaimingAll:bool=true, self:Season Pass Selection VM Object Reference=Get SeasonPassSelectionVM()) -> (then, Output_Get:bool)
   Gate!(Enter:Exec=@n13.then, Open:Exec=@n12.then_1, Close:Exec=@n10.then, Toggle:Exec, bStartClosed:bool=true) -> (Exit:Exec) {is=title K2Node_MacroInstance}
   > Set IsClaimingAll(IsClaimingAll=false, self=Get SeasonPassSelectionVM())
   n11 = "On Released (LoudButton_ClaimAll)"() {is=title K2Node_ComponentBoundEvent}
   n12 = Sequence(execute=[@n11.then, @n19.then])
   > n13 = RetriggerableDelay(Duration=0.8)
   n14 = "On Pressed (RegularButton_ClaimReward)"!() -> (OutputDelegate, then, Button:Common Button Base Object Reference) {is=title K2Node_ComponentBoundEvent}
   n15 = Make Flow Analytics Additional Info()
   n16 = Get SeasonPassSelectionVM()
   > Fire Event Element Interacted(
    Element=@n14.Button
    AdditionalInfo=Get Analytics Info Item Name(
     MakeFlowAnalyticsAdditionalInfo=Get Analytics Info Pass Category Name(MakeFlowAnalyticsAdditionalInfo=@n15, SeasonPassSelectedVM=@n16)
     Key=SelectReward
     FortItemVM=Get FocusedFortItemVM()
    )
   )
   > Set IsClaimingAll(IsClaimingAll=false, self=Get SeasonPassSelectionVM())
   n17 = "On Pressed (RegularButton_ClaimAll)"!() -> (OutputDelegate, then, Button:Common Button Base Object Reference) {is=title K2Node_ComponentBoundEvent}
   > n18 = Fire Event Element Interacted(Element=@n17.Button, AdditionalInfo=@n8)
   n19 = "On Released (RegularButton_ClaimAll)"() {is=title K2Node_ComponentBoundEvent}
   n20 = "On Clicked (RegularButton_ClaimAll)"!() -> (OutputDelegate, then, Button:Common Button Base Object Reference) {is=title K2Node_ComponentBoundEvent}
   n21 = Get Analytics Info Pass Category Name(MakeFlowAnalyticsAdditionalInfo=Make Flow Analytics Additional Info(), SeasonPassSelectedVM=Get SeasonPassSelectionVM())
   > n22 = Fire Event Element Interacted(Element=@n20.Button, AdditionalInfo=@n21)
   n23 = ClaimAllClaimableRewards!(
    execute:Exec=[@n37.then, @n22.then]
    SeasonPass:Season Pass Element Base VM Object Reference=Get SelectedPage!(self:Season Pass Selection VM Object Reference=Get SeasonPassSelectionVM()) -> (SelectedPage:Season Pass Page VM Object Reference)
    bTryPurchasePrerequisites:bool=false
   ) -> (then:Exec, OnOperationCompleted:Exec, Success:bool=~) {is=title K2Node_AsyncAction}
   > SetClaimButtonsEnabledState!(execute, self, bInIsEnabled:bool=false) -> (then)
   SetClaimButtonsEnabledState(execute=@n23.OnOperationCompleted, bInIsEnabled=true)
   n24 = "On Clicked (LoudButton_BuyRewards)"!() -> (OutputDelegate, then, Button:Common Button Base Object Reference) {is=title K2Node_ComponentBoundEvent}
   > Fire Event Element Interacted(Element=@n24.Button, AdditionalInfo=@n5)
   > n25 = Create Widget!(
    execute:Exec
    Class:User Widget Class Reference=&/SeasonPass_BuyRewards.SeasonPass_BuyRewards_C
    OwningPlayer:Player Controller Object Reference
    SeasonPassVM:Season Pass VM Object Reference=Get SeasonPassVM()
    SeasonPassCategoryVM:Season Pass Category VM Object Reference
   ) -> (then:Exec, ReturnValue:Season Pass Buy Rewards Object Reference)
   > Set ItemToPurchase!(
    execute
    ItemToPurchase:Season Pass Reward VM Object Reference=Get SelectedReward!(self:Season Pass Selection VM Object Reference=Get SeasonPassSelectionVM()) -> (SelectedReward:Season Pass Reward VM Object Reference)
    self:Season Pass Buy Rewards Object Reference=@n25
   ) -> (then, Output_Get:Season Pass Reward VM Object Reference)
   > TryPushWidgetAsModalPanel!(execute, PushWidget:User Widget Object Reference=@n25) -> (then)
   n26 = "On Clicked (Button_Equip)"!() -> (OutputDelegate, then, Button:Common Button Base Object Reference) {is=title K2Node_ComponentBoundEvent}
   n27 = Make Flow Analytics Additional Info()
   n28 = Get SeasonPassSelectionVM()
   > Fire Event Element Interacted(
    Element=@n26.Button
    AdditionalInfo=Get Analytics Info Item Name(
     MakeFlowAnalyticsAdditionalInfo=Get Analytics Info Pass Category Name(MakeFlowAnalyticsAdditionalInfo=@n27, SeasonPassSelectedVM=@n28)
     Key=SelectReward
     FortItemVM=Get FocusedFortItemVM()
    )
   )
   > n29 = Branch(Condition=Get bRequiresCustomization!(self:Fort Item VM Object Reference=Get FocusedFortItemVM()) -> (bRequiresCustomization:bool))
   n30 = TryGetSoftCustomActivatableWidgetClass!(
    self:Fort Locker Category Config Object Reference=GetItemCategory!(self:Fort Item VM Object Reference=Get FocusedFortItemVM()) -> (ReturnValue:Fort Locker Category Config Object Reference)
    ViewRequest:Gameplay Tag Structure=(TagName="UI.ItemPreview.PluginBridge.Mimosa.CustomizeMenu")
   ) -> (WidgetClass:Common Activatable Widget Soft Class Reference, ReturnValue:bool)
   > Branch(Condition=@n30.ReturnValue)
   > n31 = PushPrimaryContentSoftClassAsync(SoftWidgetClass=@n30)
   SetViewModelByClass!(
    execute=@n31.OnPrimaryContentPushed
    Widget:User Widget Object Reference=@n31
    ViewModel:Notify Field Value Changed Interface=Get FocusedFortItemVM()
   ) -> (then)
   EquipSeasonPassItem!(
    execute:Exec=@n29.else
    SeasonPass:Season Pass VM Object Reference=Get SeasonPassVM()
    Item:Season Pass Reward VM Object Reference=Get SelectedReward(self=Get SeasonPassSelectionVM())
    SlotIndex:int=0
    Variants:Array of Mcp Variant Channel Info Structures
    bApplyToAllSlots:bool=false
    bAllowMocking:bool=true
   ) -> (then:Exec, OnOperationCompleted:Exec, Success:bool=~) {is=title K2Node_AsyncAction}
   n32 = "On Clicked (RegularButton_ClaimReward)"!() -> (OutputDelegate, then, Button:Common Button Base Object Reference) {is=title K2Node_ComponentBoundEvent}
   n33 = Make Flow Analytics Additional Info()
   n34 = Get SeasonPassSelectionVM()
   > Fire Event Element Interacted(
    Element=@n32.Button
    AdditionalInfo=Get Analytics Info Item Name(
     MakeFlowAnalyticsAdditionalInfo=Get Analytics Info Pass Category Name(MakeFlowAnalyticsAdditionalInfo=@n33, SeasonPassSelectedVM=@n34)
     Key=SelectReward
     FortItemVM=Get FocusedFortItemVM()
    )
   )
   > n35 = PurchaseItem!(
    execute:Exec
    Item:Season Pass Reward VM Object Reference=Get SelectedReward(self=Get SeasonPassSelectionVM())
    SeasonPass:Season Pass VM Object Reference=Get SeasonPassVM()
    bTryPurchasePrerequisites:bool=false
   ) -> (then:Exec, OnOperationCompleted:Exec, Success:bool=~) {is=title K2Node_AsyncAction}
   > SetClaimButtonsEnabledState(bInIsEnabled=false)
   SetClaimButtonsEnabledState(execute=@n35.OnOperationCompleted, bInIsEnabled=true)
   n36 = "On Clicked (LoudButton_ClaimAll)"!() -> (OutputDelegate, then, Button:Common Button Base Object Reference) {is=title K2Node_ComponentBoundEvent}
   > n37 = Fire Event Element Interacted(Element=@n36.Button, AdditionalInfo=@n21)
   n38 = OnHoldActionProgress!() -> (OutputDelegate, then, HeldPercent:double=~) {is=title K2Node_CustomEvent}
   > Set OfferPurchaseHoldProgress!(execute, OfferPurchaseHoldProgress:float=@n38.HeldPercent, self:Season Pass Selection VM Object Reference=Get SeasonPassSelectionVM()) -> (then, Output_Get:float)
   n39 = OnVisibilityChanged_Event!() -> (OutputDelegate, then, InVisibility:ESlateVisibility Enum=Visible) {is=title K2Node_CustomEvent}
   > ResetHoldableButtons()
   Event Construct()
   > Bind Event to On Hold Action Progress(
    self:WBP UIKit Button CTA Base Object Reference=Get LoudButton_ClaimAll() -> (LoudButton_ClaimAll:WBP UIKit Button Loud Object Reference)
    Delegate=@n38
   ) {is=title K2Node_AddDelegate}
   > Bind Event to On Hold Action Progress(self=Get RegularButton_ClaimReward() -> (RegularButton_ClaimReward:WBP UIKit Button Regular Object Reference), Delegate=@n38)
   > Bind Event to On Hold Action Progress(self=Get RegularButton_ClaimAll() -> (RegularButton_ClaimAll:WBP UIKit Button Regular Object Reference), Delegate=@n38)
   > SetHoldOrTap()
   > Assign On Visibility Changed!(execute:Exec, self:User Widget Object Reference, Delegate:"Delegate (by ref)"=@n39) -> (then:Exec) {is=title K2Node_AssignDelegate}
   > Set UseResetHoldableButtonsInPassesRewardTrack(UseResetHoldableButtonsInPassesRewardTrack:bool=GetConfigDataBoolWithDefault!(EntryName:String=bUseResetHoldableButtonsInPassesRewardTrack, bDefaultValue:bool=true) -> (ReturnValue:bool)) -> (Output_Get:bool)
   n40 = Event PreConstruct()
   > SetCompactUIStyle() {is=title K2Node_CallFunction}
   > Branch(Condition=@n40.IsDesignTime)
   > UpdateProgressBarColors!(execute, self, ShouldUseAlternativeColors:bool=Get Preview_UseAlterProgressBar() -> (Preview_UseAlterProgressBar:bool)) -> (then)
  ]}
 GetFocusedItemVM {nodes [
   n1 = GetFocusedItemVM!() -> (then, Reward:Season Pass Reward VM Object Reference) {is=title K2Node_FunctionEntry}
   > n2 = Is Valid(InputObject=@n1)
   Return Node!(execute=@n2.Is Not Valid, Result:Fort Item VM Object Reference)
   > Return Node(Result="Get (a copy)"(Array:Array of Fort Item VM Object References=Get ItemVMs!(self:Season Pass Reward VM Object Reference=@n1) -> (ItemVMs:Array of Fort Item VM Object References)) -> (Output:Fort Item VM Object Reference))
  ]}
 GetRequiredPassText {nodes [
   n1 = GetRequiredPassText!() -> (then, PassName:Text) {is=title K2Node_FunctionEntry}
   > Return Node(Result:Text=Format Text!(Format="LOCTEXT(\"{PassName} required\")", PassName:Text=@n1) -> (Result))
  ]}
 GetWidgetFromEnum {nodes [
   n1 = GetWidgetFromEnum!() -> (then, SeasonPassRewardHeader:ESeasonPassRewardHeader Enum=None) {is=title K2Node_FunctionEntry}
   > Return Node!(
    execute
    NewParam:Widget Object Reference=Select!(
     Prerequisite:Widget Object Reference=Get Tag_Prerequisite() -> (Tag_Prerequisite:Border Object Reference)
     SeasonPassRequired:Widget Object Reference
     ComingSoon:Widget Object Reference=Get Tag_ComingSoon() -> (Tag_ComingSoon:Border Object Reference)
     ReachLevel:Widget Object Reference=Get Tag_GrantByLevel() -> (Tag_GrantByLevel:Border Object Reference)
     Index:ESeasonPassRewardHeader Enum=@n1
    ) -> (ReturnValue:Widget Object Reference)
   )
  ]}
 HeaderChanged {nodes [
   HeaderChanged!() -> (then, SeasonPassRewardHeaders:int) {is=title K2Node_FunctionEntry}
   > n1 = ForEach ESeasonPassRewardHeader!(execute:Exec, SkipHidden:bool) -> (LoopBody:Exec, EnumValue:ESeasonPassRewardHeader Enum, then:Exec) {is=title K2Node_ForEachElementInEnum}
   > n2 = GetWidgetFromEnum!(execute, self, SeasonPassRewardHeader:ESeasonPassRewardHeader Enum=@n1) -> (then, NewParam:Widget Object Reference) {is=title K2Node_CallFunction}
   > Is Valid(InputObject=@n2)
   n3 = Bitwise AND(A:int=Get SeasonPassRewardHeaders() -> (SeasonPassRewardHeaders:int=0), B:int="To Integer (Byte)"!(InByte:Byte=@n1) -> (ReturnValue:int=0)) -> (ReturnValue:int=0) {is=title K2Node_CommutativeAssociativeBinaryOperator}
   > SetVisibility(
    self=@n2
    InVisibility=Select!(Option 0:ESlateVisibility Enum=Collapsed, Option 1:ESlateVisibility Enum=Visible, Index:bool="Not Equal (Integer)"(A=@n3)) -> (ReturnValue:ESlateVisibility Enum)
   )
  ]}
 IsEquippedChanged {nodes [
   n1 = IsEquippedChanged!() -> (then, bIsEquipped:bool=~) {is=title K2Node_FunctionEntry}
   > SetVisibility(
    self=Get Tag_Equipped() -> (Tag_Equipped:Border Object Reference)
    InVisibility=Select(Option 0=Collapsed, Option 1=SelfHitTestInvisible, Index=@n1)
   )
   > SetIsEnabled(self=Get Button_Equip() -> (Button_Equip:WBP UIKit Button Regular Object Reference), bInIsEnabled=NOT Boolean(A=@n1))
   n2 = Get bIsRewardOwned!(self:Season Pass Reward VM Object Reference=Get SelectedReward(self=Get SeasonPassSelectionVM())) -> (bIsRewardOwned:bool)
   > SetVisibility(
    self=Get SizeBox_Claimed() -> (SizeBox_Claimed:Size Box Object Reference)
    InVisibility=Select(
     Option 0=Select(Option 0=Collapsed, Option 1=SelfHitTestInvisible, Index=@n2)
     Option 1=Collapsed
     Index=Get bIsEquipped() -> (bIsEquipped:bool)
    )
   )
  ]}
 LockedTextChanged {nodes [
   n1 = LockedTextChanged!() -> (then, LockedText:Text) {is=title K2Node_FunctionEntry}
   n2 = Get Text_LockCondition() -> (Text_LockCondition:Common Text Object Reference)
   > SetVisibility(self=@n2, InVisibility=Select(Option 0=Visible, Option 1=Collapsed, Index=TextIsEmpty(InText=@n1)))
   > Set Text(Text=Get LockedText() -> (LockedText:Text), self=@n2)
  ]}
 ResetHoldableButtons {nodes [
   ResetHoldableButtons() {is=title K2Node_FunctionEntry}
   > Branch(Condition=Get UseResetHoldableButtonsInPassesRewardTrack() -> (UseResetHoldableButtonsInPassesRewardTrack:bool))
   n1 = Get RegularButton_ClaimReward()
   n2 = Get RegularButton_ClaimAll()
   n3 = Get LoudButton_ClaimAll()
   > FadeOutSound(self:WBP UIKit Button CTA Base Object Reference=[@n1, @n2, @n3])
   > SetUseNaturalNavigation!(execute, self:Fort CTAButton Object Reference=[@n1, @n2, @n3], bInUseNaturalNavigation:bool=true) -> (then)
   > SetUseNaturalNavigation(self=[@n1, @n2, @n3], bInUseNaturalNavigation=false)
   > SetActionProgress!(execute, self:Fort CTAButton Object Reference=[@n1, @n2, @n3], HeldPercent:float=0) -> (then)
  ]}
 SetClaimButtonsEnabledState {nodes [
   n1 = SetClaimButtonsEnabledState!() -> (then, bInIsEnabled:bool=~) {is=title K2Node_FunctionEntry}
   > SetIsEnabled(self=[Get RegularButton_ClaimAll(), Get RegularButton_ClaimReward(), Get LoudButton_ClaimAll()], bInIsEnabled=@n1)
  ]}
 SetCompactUIStyle {nodes [
   SetCompactUIStyle() {is=title K2Node_FunctionEntry}
   > Branch(Condition=IsUsingCompactUI())
   > Branch(Condition=CompareToScreenAspectRatio!(Comparison:EAspectRatioComparison Enum=GreaterThan, AspectRatio:EAspectRatio Enum=Ratio_4_3) -> (ReturnValue:bool))
   n1 = Get CompactUIPadding() -> (CompactUIPadding:double)
   > SetPadding!(
    execute
    self:Vertical Box Slot Object Reference=[
     SlotAsVerticalBoxSlot(Widget=Get Text_LockCondition())
     SlotAsVerticalBoxSlot(Widget=Get ProgressHorizontalBox() -> (ProgressHorizontalBox:Horizontal Box Object Reference))
     SlotAsVerticalBoxSlot(Widget=Get SizeBox_Claimed())
     SlotAsVerticalBoxSlot(Widget=Get TopButtonsSwitcher() -> (TopButtonsSwitcher:Common Visibility Switcher Object Reference))
     SlotAsVerticalBoxSlot(Widget=Get Button_ViewReward() -> (Button_ViewReward:WBP UIKit Button Regular Object Reference))
     SlotAsVerticalBoxSlot(Widget=Get Button_Equip())
     SlotAsVerticalBoxSlot(Widget=Get ClaimAllLoudSwitcher() -> (ClaimAllLoudSwitcher:Common Visibility Switcher Object Reference))
     SlotAsVerticalBoxSlot(Widget=Get Tag_Prerequisite())
     SlotAsVerticalBoxSlot(Widget=Get Tag_GrantByLevel())
     SlotAsVerticalBoxSlot(Widget=Get Tag_Equipped())
     SlotAsVerticalBoxSlot(Widget=Get Tag_ComingSoon())
    ]
    InPadding_Left:float=0
    InPadding_Top:float=0
    InPadding_Right:float=0
    InPadding_Bottom:float=@n1
   ) -> (then)
   > SetPadding(
    self=SlotAsVerticalBoxSlot(Widget=Get Divider() -> (Divider:Image Object Reference))
    InPadding_Left=0
    InPadding_Top=@n1
    InPadding_Right=0
    InPadding_Bottom=@n1
   )
  ]}
 SetCustomInfoButtonLazyContent {nodes [
   n1 = SetCustomInfoButtonLazyContent!() -> (then, Widget:User Widget Object Reference) {is=title K2Node_FunctionEntry}
   > SetViewModelByClass(Widget=@n1, ViewModel=Get FocusedFortItemVM())
  ]}
 SetHoldOrTap {nodes [
   SetHoldOrTap() {is=title K2Node_FunctionEntry}
   n1 = GetConfigDataBoolWithDefault(EntryName=bPassesTapNoHold, bDefaultValue=false)
   n2 = Get LoudButton_ClaimAll()
   n3 = Get RegularButton_ClaimReward()
   n4 = Get RegularButton_ClaimAll()
   > SetRequiresHold!(execute, self:Common Button Base Object Reference=[@n2, @n3, @n4], bInRequiresHold:bool=NOT Boolean(A=@n1)) -> (then)
   > SetTriggeringInputAction!(
    execute
    self:Common Button Base Object Reference=@n3
    InputActionRow_DataTable:Data Table Object Reference=/SeasonPass/Data/SeasonPassInputActionDataTable
    InputActionRow_RowName:Name=Select(Option 0:Name=ClaimReward_Hold, Option 1:Name=ClaimReward, Index=@n1) -> (ReturnValue:Name)
   ) -> (then)
   > SetTriggeringInputAction(
    self=[@n2, @n4]
    InputActionRow_DataTable=&SeasonPassInputActionDataTable
    InputActionRow_RowName=Select(Option 0=ClaimAll_Hold, Option 1=ClaimAll, Index=@n1)
   )
  ]}
 SetProgressInfo {nodes [
   n1 = SetProgressInfo!() -> (then, ProgressCount:int, ProgressTotal:int) {is=title K2Node_FunctionEntry}
   > Set Text(Text="To Text (Integer)"(Value=@n1), self=Get ProgressTextX() -> (ProgressTextX:Common Text Object Reference))
   > Set Text(
    Text=Format Text!(Format="LOCTEXT(\"/{CountForUnlock}\")", CountForUnlock:int=Get ProgressTotal() -> (ProgressTotal:int=0)) -> (Result)
    self=Get ProgressTextOf() -> (ProgressTextOf:Common Text Object Reference)
   )
   > SetProgress!(
    execute
    self:WBP UIKit Progress Bar Base Object Reference=[
     Get ProgressBar() -> (ProgressBar:WBP UIKit Progress Bar Object Reference)
     Get ProgressBar_Alter() -> (ProgressBar_Alter:WBP UIKit Progress Bar Object Reference)
    ]
    Progress:double=float / float(A:int=Get ProgressCount() -> (ProgressCount:int=0), B:int=Get ProgressTotal() -> (ProgressTotal=0))
   ) -> (then)
   > Set Visibility(Visibility=Visible, self=Get ProgressBarSwitcher() -> (ProgressBarSwitcher:Common Visibility Switcher Object Reference))
   > Set Text(
    Text=Format Text!(Format="LOCTEXT(\"Level {0}\")", 0:int=Get ProgressTotal() -> (ProgressTotal=0)) -> (Result)
    self=Get Text_GrantByLevel() -> (Text_GrantByLevel:Common Text Object Reference)
   )
  ]}
 UpdateCustomInfoButton {nodes [
   n1 = UpdateCustomInfoButton!() -> (then, Target:Fort Locker Category Config Object Reference) {is=title K2Node_FunctionEntry}
   > Is Valid(InputObject=@n1)
   > SetLazyContentWithCallback!(
    execute
    self:Common Lazy Widget Object Reference=Get Button_CustomInfo() -> (Button_CustomInfo:Common Lazy Widget Object Reference)
    SoftWidget:User Widget Soft Class Reference=TryGetSoftCustomActivatableWidgetClass(self=@n1, ViewRequest=(TagName="UI.SeasonPass.CustomInfoButton"))
    OnCreatedCallback:"Delegate (by ref)"=Create Event()
   ) -> (then)
  ]}
 UpdateIsOwned {nodes [
   n1 = UpdateIsOwned!() -> (then, bIsRewardOwned:bool=~) {is=title K2Node_FunctionEntry}
   > SetVisibility(
    self=Get SizeBox_Claimed()
    InVisibility=Select(Option 0:ESlateVisibility Enum=Collapsed, Option 1:ESlateVisibility Enum=Visible, Index=@n1) -> (ReturnValue:ESlateVisibility Enum)
   )
  ]}
 UpdateProgressBarColors {nodes [
   UpdateProgressBarColors!() -> (then, ShouldUseAlternativeColors:bool=~) {is=title K2Node_FunctionEntry}
   > SetActiveWidget(
    self=Get ProgressBarSwitcher()
    Widget=Select(
     Option 0:WBP UIKit Progress Bar Object Reference=Get ProgressBar()
     Option 1:WBP UIKit Progress Bar Object Reference=Get ProgressBar_Alter()
     Index=Get ShouldUseAlternativeColors() -> (ShouldUseAlternativeColors:bool)
    ) -> (ReturnValue:WBP UIKit Progress Bar Object Reference)
   )
   > SetColorAndOpacity!(
    execute
    self=Get ProgressTextX()
    InColorAndOpacity_SpecifiedColor:Linear Color Structure=SelectColor(A=(R=1,G=1,B=1,A=1), B=(R=0.9301,G=1,B=0.0103,A=1), bPickA=Get ShouldUseAlternativeColors())
    InColorAndOpacity_ColorUseRule:ESlateColorStylingMode Enum
   ) -> (then)
  ]}
 UpdateRequiresProgressBar {nodes [
   n1 = UpdateRequiresProgressBar!() -> (then, bDoesRequireProgressBar:bool=~) {is=title K2Node_FunctionEntry}
   > n2 = Branch(Condition=@n1)
   > SetVisibility(self=Get ProgressHorizontalBox(), InVisibility=Visible)
   n3 = Get SelectedReward(self=Get SeasonPassSelectionVM())
   > SetProgressInfo!(
    execute
    self
    ProgressCount:int=Get ProgressForUnlock!(self:Season Pass Reward VM Object Reference=@n3) -> (ProgressForUnlock:int=0)
    ProgressTotal:int=Get ProgressCountForUnlock!(self:Season Pass Reward VM Object Reference=@n3) -> (ProgressCountForUnlock:int=0)
   ) -> (then) {is=title K2Node_CallFunction}
   SetVisibility(execute=@n2.else, self=Get ProgressHorizontalBox(), InVisibility=Collapsed)
  ]}
 UpdateShouldHideLoudButton {nodes [
   n1 = UpdateShouldHideLoudButton!() -> (then, bShouldHideLoudButton:bool=~) {is=title K2Node_FunctionEntry}
   > SetVisibility(
    self=Get TopButtonsSwitcher()
    InVisibility=Select(Option 0:ESlateVisibility Enum=SelfHitTestInvisible, Option 1:ESlateVisibility Enum=Collapsed, Index=@n1) -> (ReturnValue:ESlateVisibility Enum)
   )
   > Branch(Condition=@n1)
   > SetActiveWidget(self=Get ClaimAllLoudSwitcher(), Widget=Get LoudButton_ClaimAll())
  ]}
 UpdateShouldShowBuyRewards {nodes [
   n1 = UpdateShouldShowBuyRewards!() -> (then, bShouldShowBuyRewards:bool=~) {is=title K2Node_FunctionEntry}
   > Branch(Condition=@n1)
   > SetActiveWidget(self=Get TopButtonsSwitcher(), Widget=Get LoudButton_BuyRewards() -> (LoudButton_BuyRewards:WBP UIKit Button Loud Object Reference))
   > SetActiveWidget(self=Get ClaimAllLoudSwitcher(), Widget=Get RegularButton_ClaimAll())
  ]}
 UpdateShouldShowBuySeasonPass {nodes [
   n1 = UpdateShouldShowBuySeasonPass!() -> (then, bshouldShowBuySeasonPass:bool=~) {is=title K2Node_FunctionEntry}
   > Branch(Condition=@n1)
   > SetActiveWidget(self=Get TopButtonsSwitcher(), Widget=Get LoudButton_BuySeasonPass() -> (LoudButton_BuySeasonPass:WBP UIKit Button Loud Object Reference))
   > SetActiveWidget(self=Get ClaimAllLoudSwitcher(), Widget=Get RegularButton_ClaimAll())
  ]}
 UpdateShouldShowClaimReward {nodes [
   n1 = UpdateShouldShowClaimReward!() -> (then, bShouldShowClaimReward:bool=~) {is=title K2Node_FunctionEntry}
   > Branch(Condition=@n1)
   > SetActiveWidget(self=Get TopButtonsSwitcher(), Widget=Get RegularButton_ClaimReward())
   > SetActiveWidget(self=Get ClaimAllLoudSwitcher(), Widget=Get LoudButton_ClaimAll())
  ]}
 UpdateShouldShowEquipButton {nodes [
   n1 = UpdateShouldShowEquipButton!() -> (then, bShouldShowEquipButton:bool=~) {is=title K2Node_FunctionEntry}
   > SetVisibility(self=Get Button_Equip(), InVisibility=Select(Option 0=Collapsed, Option 1=Visible, Index=@n1))
  ]}
 VM_SelectedRewardChanged {nodes [
   VM_SelectedRewardChanged!() -> (then, SelectedReward:Season Pass Reward VM Object Reference) {is=title K2Node_FunctionEntry}
   > ResetHoldableButtons() {is=title K2Node_CallFunction}
  ]}
 VM_SetClaimButtonProgress {nodes [
   n1 = VM_SetClaimButtonProgress!() -> (then, Progress:double=~) {is=title K2Node_FunctionEntry}
   > n2 = Branch(Condition=Get IsClaimingAll!(self:Season Pass Selection VM Object Reference=Get SeasonPassSelectionVM()) -> (IsClaimingAll:bool))
   SetActionProgress(execute=@n2.else, self=Get RegularButton_ClaimReward(), HeldPercent=@n1)
  ]}
 ViewRewardButtonClicked {nodes [ViewRewardButtonClicked() {is=title K2Node_FunctionEntry}]}
]
variables [
 As WBP About Item Dialog {type=WBP_AboutItem_Dialog_C, default=null}
 CustomizeMenu {type=CommonActivatableWidget, default=null}
 IsClaimAllInProgress {type=boolean}
 IsModalVisible {type=boolean}
 PageState {type=ESeasonPassCategory_ButtonSwitcher, default=Claimed}
 Preview_UseAlterProgressBar {type=boolean}
 UseResetHoldableButtonsInPassesRewardTrack {type=boolean, default=true}
 ViewRewardButtonClicked {}
]
widgets {
 info {inheritedWidgetCount=0, parentClass=&UserWidget, rootWidgetClass=&VerticalBox, widgetCount=36}
 widgets [
  VerticalBox_26 {widgetClassPath=&VerticalBox, widget=VerticalBox_26}
  Tag_Equipped {
   widgetClassPath=&Border
   bIsVariable=true
   parent=VerticalBox_26
   widget=Tag_Equipped
   properties {
    visibility=Collapsed
    background {
     bIsDynamicallyLoaded=false
     drawAs=Image
     imageType=FullColor
     mirroring=NoMirror
     resourceName=None
     resourceObject=/FortUILibrary/Systems/Badge/Materials/M_UIKit_Badge_Background
     tiling=NoTile
     imageSize=<32 32>
     margin=<ltrb 0 0 0 0>
     outlineSettings {
      bUseBrushTransparency=false
      roundingType=HalfHeightRadius
      width=0
      color {colorUseRule=UseColor_Specified, specifiedColor=<rgba 0 0 0 0>}
      cornerRadii=<0 0 0 0>
     }
     tintColor {colorUseRule=UseColor_Specified, specifiedColor=<rgba 0.0048 0.004 0.0116 1>}
     uVRegion {bIsValid=false, max=<0 0>, min=<0 0>}
    }
    padding=<ltrb 8 2 8 2>
   }
   slot {class=VerticalBoxSlot, properties {padding=<ltrb 0 0 0 12>}}
  }
  Text_Equipped {
   widgetClassPath=&CommonTextBlock
   parent=Tag_Equipped
   widget=Text_Equipped
   properties {
    justification=Center
    mobileFontSizeMultiplier=1.14
    scrollStyle=&TextScrollStyle-Base_C
    text=Equipped
    textTransformPolicy=ToUpper
    font {
     bForceMonospaced=false
     bMaterialIsStencil=false
     fontMaterial=null
     fontObject=&HeadingNow
     letterSpacing=0
     monospacedWidth=1
     size=15.75
     skewAmount=0
     typefaceFontName=86 Bold Italic
     outlineSettings {
      bApplyOutlineToDropShadows=false
      bMiteredCorners=false
      bSeparateFillAlpha=false
      outlineMaterial=null
      outlineSize=0
      outlineColor=<rgba 0 0 0 1>
     }
    }
    shadowOffset=<0 0>
   }
   slot {class=BorderSlot, properties {padding=<ltrb 8 2 8 2>}}
  }
  Tag_Prerequisite {
   widgetClassPath=&Border
   bIsVariable=true
   parent=VerticalBox_26
   widget=Tag_Prerequisite
   properties {
    horizontalAlignment=HAlign_Center
    background {
     bIsDynamicallyLoaded=false
     drawAs=Image
     imageType=FullColor
     mirroring=NoMirror
     resourceName=None
     resourceObject=&M_UIKit_Badge_Background
     tiling=NoTile
     imageSize=<32 32>
     margin=<ltrb 0 0 0 0>
     outlineSettings {
      bUseBrushTransparency=false
      roundingType=HalfHeightRadius
      width=0
      color {colorUseRule=UseColor_Specified, specifiedColor=<rgba 0 0 0 0>}
      cornerRadii=<0 0 0 0>
     }
     tintColor {colorUseRule=UseColor_Specified, specifiedColor=<rgba 1 0.063 0.3372 1>}
     uVRegion {bIsValid=false, max=<0 0>, min=<0 0>}
    }
    padding=<ltrb 8 2 8 2>
   }
   slot {class=VerticalBoxSlot, properties {padding=<ltrb 0 0 0 12>}}
  }
  HorizontalBox {
   widgetClassPath=&HorizontalBox
   parent=Tag_Prerequisite
   widget=HorizontalBox
   slot {class=BorderSlot, properties {horizontalAlignment=HAlign_Center, padding=<ltrb 8 2 8 2>}}
  }
  Icon_Lock {
   widgetClassPath=&Image
   parent=HorizontalBox
   widget=Icon_Lock
   properties {brush {
     bIsDynamicallyLoaded=false
     drawAs=Image
     imageType=FullColor
     mirroring=NoMirror
     resourceName=None
     resourceObject=/FortUILibrary/Textures/MaterialInstances/MI_UIKit_Icon_Locked
     tiling=NoTile
     imageSize=<21 21>
     margin=<ltrb 0 0 0 0>
     outlineSettings {
      bUseBrushTransparency=false
      roundingType=HalfHeightRadius
      width=0
      color {colorUseRule=UseColor_Specified, specifiedColor=<rgba 0 0 0 0>}
      cornerRadii=<0 0 0 0>
     }
     tintColor {colorUseRule=UseColor_Specified, specifiedColor=<rgba 1 1 1 1>}
     uVRegion {bIsValid=false, max=<0 0>, min=<0 0>}
    }}
   slot {class=HorizontalBoxSlot, properties {horizontalAlignment=HAlign_Center, verticalAlignment=VAlign_Center, padding=<ltrb 0 0 6 0>}}
  }
  Text_Prequisite {
   widgetClassPath=&CommonTextBlock
   parent=HorizontalBox
   widget=Text_Prequisite
   properties {
    justification=Center
    mobileFontSizeMultiplier=1.14
    scrollStyle=&TextScrollStyle-Base_C
    text=Prerequisite
    textTransformPolicy=ToUpper
    font {
     bForceMonospaced=false
     bMaterialIsStencil=false
     fontMaterial=null
     fontObject=&HeadingNow
     letterSpacing=0
     monospacedWidth=1
     size=15.75
     skewAmount=0
     typefaceFontName=86 Bold Italic
     outlineSettings {
      bApplyOutlineToDropShadows=false
      bMiteredCorners=false
      bSeparateFillAlpha=false
      outlineMaterial=null
      outlineSize=0
      outlineColor=<rgba 0 0 0 1>
     }
    }
    shadowOffset=<0 0>
   }
   slot {class=HorizontalBoxSlot, properties {horizontalAlignment=HAlign_Center, verticalAlignment=VAlign_Center}}
  }
  Tag_ComingSoon {
   widgetClassPath=&Border
   bIsVariable=true
   parent=VerticalBox_26
   widget=Tag_ComingSoon
   properties {
    horizontalAlignment=HAlign_Center
    background {
     bIsDynamicallyLoaded=false
     drawAs=Image
     imageType=FullColor
     mirroring=NoMirror
     resourceName=None
     resourceObject=&M_UIKit_Badge_Background
     tiling=NoTile
     imageSize=<32 32>
     margin=<ltrb 0 0 0 0>
     outlineSettings {
      bUseBrushTransparency=false
      roundingType=HalfHeightRadius
      width=0
      color {colorUseRule=UseColor_Specified, specifiedColor=<rgba 0 0 0 0>}
      cornerRadii=<0 0 0 0>
     }
     tintColor {colorUseRule=UseColor_Specified, specifiedColor=<rgba 1 0.063 0.3372 1>}
     uVRegion {bIsValid=false, max=<0 0>, min=<0 0>}
    }
    padding=<ltrb 8 2 8 2>
   }
   slot {class=VerticalBoxSlot, properties {padding=<ltrb 0 0 0 12>}}
  }
  HorizontalBox_0 {
   widgetClassPath=&HorizontalBox
   parent=Tag_ComingSoon
   widget=HorizontalBox_0
   slot {class=BorderSlot, properties {horizontalAlignment=HAlign_Center, padding=<ltrb 8 2 8 2>}}
  }
  Icon_Clock {
   widgetClassPath=&Image
   parent=HorizontalBox_0
   widget=Icon_Clock
   properties {brush {
     bIsDynamicallyLoaded=false
     drawAs=Image
     imageType=FullColor
     mirroring=NoMirror
     resourceName=None
     resourceObject=/FortUILibrary/Textures/MaterialInstances/MI_UIKit_Icon_UpdatePending
     tiling=NoTile
     imageSize=<21 21>
     margin=<ltrb 0 0 0 0>
     outlineSettings {
      bUseBrushTransparency=false
      roundingType=HalfHeightRadius
      width=0
      color {colorUseRule=UseColor_Specified, specifiedColor=<rgba 0 0 0 0>}
      cornerRadii=<0 0 0 0>
     }
     tintColor {colorUseRule=UseColor_Specified, specifiedColor=<rgba 1 1 1 1>}
     uVRegion {bIsValid=false, max=<0 0>, min=<0 0>}
    }}
   slot {class=HorizontalBoxSlot, properties {horizontalAlignment=HAlign_Center, verticalAlignment=VAlign_Center, padding=<ltrb 0 0 6 0>}}
  }
  Text_ComingSoon {
   widgetClassPath=&CommonTextBlock
   parent=HorizontalBox_0
   widget=Text_ComingSoon
   properties {
    justification=Center
    mobileFontSizeMultiplier=1.14
    scrollStyle=&TextScrollStyle-Base_C
    text=Coming Soon
    textTransformPolicy=ToUpper
    font {
     bForceMonospaced=false
     bMaterialIsStencil=false
     fontMaterial=null
     fontObject=&HeadingNow
     letterSpacing=0
     monospacedWidth=1
     size=15.75
     skewAmount=0
     typefaceFontName=86 Bold Italic
     outlineSettings {
      bApplyOutlineToDropShadows=false
      bMiteredCorners=false
      bSeparateFillAlpha=false
      outlineMaterial=null
      outlineSize=0
      outlineColor=<rgba 0 0 0 1>
     }
    }
    shadowOffset=<0 0>
   }
   slot {class=HorizontalBoxSlot, properties {horizontalAlignment=HAlign_Center, verticalAlignment=VAlign_Center}}
  }
  Tag_GrantByLevel {
   widgetClassPath=&Border
   bIsVariable=true
   parent=VerticalBox_26
   widget=Tag_GrantByLevel
   properties {
    horizontalAlignment=HAlign_Center
    background {
     bIsDynamicallyLoaded=false
     drawAs=Image
     imageType=FullColor
     mirroring=NoMirror
     resourceName=None
     resourceObject=&M_UIKit_Badge_Background
     tiling=NoTile
     imageSize=<32 32>
     margin=<ltrb 0 0 0 0>
     outlineSettings {
      bUseBrushTransparency=false
      roundingType=HalfHeightRadius
      width=0
      color {colorUseRule=UseColor_Specified, specifiedColor=<rgba 0 0 0 0>}
      cornerRadii=<0 0 0 0>
     }
     tintColor {colorUseRule=UseColor_Specified, specifiedColor=<rgba 1 0.063 0.3372 1>}
     uVRegion {bIsValid=false, max=<0 0>, min=<0 0>}
    }
    padding=<ltrb 8 2 8 2>
   }
   slot {class=VerticalBoxSlot, properties {padding=<ltrb 0 0 0 12>}}
  }
  HorizontalBox_1 {
   widgetClassPath=&HorizontalBox
   parent=Tag_GrantByLevel
   widget=HorizontalBox_1
   slot {class=BorderSlot, properties {horizontalAlignment=HAlign_Center, padding=<ltrb 8 2 8 2>}}
  }
  Icon_Lock_1 {
   widgetClassPath=&Image
   parent=HorizontalBox_1
   widget=Icon_Lock_1
   properties {brush {
     bIsDynamicallyLoaded=false
     drawAs=Image
     imageType=FullColor
     mirroring=NoMirror
     resourceName=None
     resourceObject=&MI_UIKit_Icon_Locked
     tiling=NoTile
     imageSize=<21 21>
     margin=<ltrb 0 0 0 0>
     outlineSettings {
      bUseBrushTransparency=false
      roundingType=HalfHeightRadius
      width=0
      color {colorUseRule=UseColor_Specified, specifiedColor=<rgba 0 0 0 0>}
      cornerRadii=<0 0 0 0>
     }
     tintColor {colorUseRule=UseColor_Specified, specifiedColor=<rgba 1 1 1 1>}
     uVRegion {bIsValid=false, max=<0 0>, min=<0 0>}
    }}
   slot {class=HorizontalBoxSlot, properties {horizontalAlignment=HAlign_Center, verticalAlignment=VAlign_Center, padding=<ltrb 0 0 6 0>}}
  }
  Text_GrantByLevel {
   widgetClassPath=&CommonTextBlock
   bIsVariable=true
   parent=HorizontalBox_1
   widget=Text_GrantByLevel
   properties {
    justification=Center
    mobileFontSizeMultiplier=1.14
    scrollStyle=&TextScrollStyle-Base_C
    text="Level {0}"
    textTransformPolicy=ToUpper
    font {
     bForceMonospaced=false
     bMaterialIsStencil=false
     fontMaterial=null
     fontObject=&HeadingNow
     letterSpacing=0
     monospacedWidth=1
     size=15.75
     skewAmount=0
     typefaceFontName=86 Bold Italic
     outlineSettings {
      bApplyOutlineToDropShadows=false
      bMiteredCorners=false
      bSeparateFillAlpha=false
      outlineMaterial=null
      outlineSize=0
      outlineColor=<rgba 0 0 0 1>
     }
    }
    shadowOffset=<0 0>
   }
   slot {class=HorizontalBoxSlot, properties {horizontalAlignment=HAlign_Center, verticalAlignment=VAlign_Center}}
  }
  Text_LockCondition {
   widgetClassPath=&CommonTextBlock
   bIsVariable=true
   parent=VerticalBox_26
   widget=Text_LockCondition
   properties {
    autoWrapText=true
    bAutoCollapseWithEmptyText=true
    text="[PH] Lock condition..."
    wrapTextAt=392
    colorAndOpacity {colorUseRule=UseColor_Specified, specifiedColor=<rgba 1 1 1 0.65>}
    font {
     bForceMonospaced=false
     bMaterialIsStencil=false
     fontMaterial=null
     fontObject=&HeadingNow
     letterSpacing=0
     monospacedWidth=1
     size=24
     skewAmount=0
     typefaceFontName=86 Bold
     outlineSettings {
      bApplyOutlineToDropShadows=false
      bMiteredCorners=false
      bSeparateFillAlpha=false
      outlineMaterial=null
      outlineSize=0
      outlineColor=<rgba 0 0 0 1>
     }
    }
    shadowOffset=<0 0>
   }
   slot {class=VerticalBoxSlot, properties {verticalAlignment=VAlign_Bottom, padding=<ltrb 0 0 0 12>}}
  }
  ProgressHorizontalBox {
   widgetClassPath=&HorizontalBox
   bIsVariable=true
   parent=VerticalBox_26
   widget=ProgressHorizontalBox
   slot {class=VerticalBoxSlot, properties {verticalAlignment=VAlign_Bottom, padding=<ltrb 0 0 0 12>}}
  }
  ProgressBarSwitcher {
   widgetClassPath=&CommonVisibilitySwitcher
   bIsVariable=true
   parent=ProgressHorizontalBox
   widget=ProgressBarSwitcher
   slot {class=HorizontalBoxSlot, properties {verticalAlignment=VAlign_Center, padding=<ltrb 0 0 16 0>, size {sizeRule=Fill, value=1}}}
  }
  ProgressBar {
   widgetClassPath=/FortUILibrary/UIKit/ProgressBar/WBP_UIKit_ProgressBar
   bIsVariable=true
   parent=ProgressBarSwitcher
   widget=ProgressBar
   properties {
    height=16
    linearMaterial=/SeasonPass/Materials/ProgressBar/MI_UI_NexusProgerssBar
    outline_PX=6
    progressValue=0.5
    useCircular=false
    useOutline=true
    emptyColor=<rgba 0.3573 0.3573 0.3573 0.305>
   }
   slot {class=CommonVisibilitySwitcherSlot}
  }
  ProgressBar_Alter {
   widgetClassPath=&WBP_UIKit_ProgressBar
   bIsVariable=true
   parent=ProgressBarSwitcher
   widget=ProgressBar_Alter
   properties {
    height=16
    linearMaterial=/SeasonPass/Materials/ProgressBar/MI_UI_NexusProgerssBar_Exp
    outline_PX=6
    progressValue=0.5
    useCircular=false
    useOutline=true
    emptyColor=<rgba 0.3573 0.3573 0.3573 0.305>
   }
   slot {class=CommonVisibilitySwitcherSlot}
  }
  ProgressTextX {
   widgetClassPath=&CommonTextBlock
   bIsVariable=true
   parent=ProgressHorizontalBox
   widget=ProgressTextX
   properties {
    text=0
    colorAndOpacity {colorUseRule=UseColor_Specified, specifiedColor=<rgba 0.9301 1 0.0103 1>}
    font {
     bForceMonospaced=false
     bMaterialIsStencil=false
     fontMaterial=null
     fontObject=&NotoSans
     letterSpacing=0
     monospacedWidth=1
     size=24
     skewAmount=0
     typefaceFontName=NotoSans-Regular
     outlineSettings {
      bApplyOutlineToDropShadows=false
      bMiteredCorners=false
      bSeparateFillAlpha=false
      outlineMaterial=null
      outlineSize=0
      outlineColor=<rgba 0 0 0 1>
     }
    }
    shadowOffset=<0 0>
   }
   slot {class=HorizontalBoxSlot}
  }
  ProgressTextOf {
   widgetClassPath=&CommonTextBlock
   bIsVariable=true
   parent=ProgressHorizontalBox
   widget=ProgressTextOf
   properties {
    text=/100
    font {
     bForceMonospaced=false
     bMaterialIsStencil=false
     fontMaterial=null
     fontObject=&NotoSans
     letterSpacing=0
     monospacedWidth=1
     size=24
     skewAmount=0
     typefaceFontName=NotoSans-Regular
     outlineSettings {
      bApplyOutlineToDropShadows=false
      bMiteredCorners=false
      bSeparateFillAlpha=false
      outlineMaterial=null
      outlineSize=0
      outlineColor=<rgba 0 0 0 1>
     }
    }
    shadowOffset=<0 0>
   }
   slot {class=HorizontalBoxSlot}
  }
  SizeBox_Claimed {
   widgetClassPath=&SizeBox
   bIsVariable=true
   parent=VerticalBox_26
   widget=SizeBox_Claimed
   properties {bOverride_MinDesiredHeight=true, minDesiredHeight=56}
   slot {class=VerticalBoxSlot, properties {padding=<ltrb 0 0 0 12>}}
  }
  Claimed_Text {
   widgetClassPath=&CommonTextBlock
   parent=SizeBox_Claimed
   widget=Claimed_Text
   properties {
    justification=Center
    text=CLAIMED
    colorAndOpacity {colorUseRule=UseColor_Specified, specifiedColor=<rgba 0.0395 1 0.2307 1>}
    font {
     bForceMonospaced=false
     bMaterialIsStencil=false
     fontMaterial=null
     fontObject=&HeadingNow
     letterSpacing=0
     monospacedWidth=1
     size=22.5
     skewAmount=0
     typefaceFontName=86 Bold Italic
     outlineSettings {
      bApplyOutlineToDropShadows=false
      bMiteredCorners=false
      bSeparateFillAlpha=false
      outlineMaterial=null
      outlineSize=0
      outlineColor=<rgba 0 0 0 1>
     }
    }
    shadowOffset=<0 0>
   }
   slot {class=SizeBoxSlot, properties {verticalAlignment=VAlign_Center}}
  }
  TopButtonsSwitcher {
   widgetClassPath=&CommonVisibilitySwitcher
   bIsVariable=true
   parent=VerticalBox_26
   widget=TopButtonsSwitcher
   slot {class=VerticalBoxSlot, properties {padding=<ltrb 0 0 0 12>, size {sizeRule=Fill, value=1}}}
  }
  LoudButton_BuySeasonPass {
   widgetClassPath=/FortUILibrary/UIKit/Buttons/WBP_UIKit_Button_Loud
   bIsVariable=true
   parent=TopButtonsSwitcher
   widget=LoudButton_BuySeasonPass
   properties {
    bHideInputActionWithKeyboard=true
    bIsFocusable=false
    hoveredSound=/Game/Sounds/Fort_UI_Sounds/FrontEnd/BattlePass/BattlePass_Button_Hover_Tile_Default_Cue
    mobileFontSizeMultiplierOverride=1.33
    mobileMinHeight=80
    overrideDefaultHeight=true
    overrideDefaultMobileFontSizeMultiplier=true
    pressedSound=/Game/Sounds/Fort_UI_Sounds/FrontEnd/BattlePass/BattlePass_Button_Confirm_Buy_BattlePass_Cue
    text="[PH] GET THE SEASON PASS"
    triggeringInputAction {dataTable=&SeasonPassInputActionDataTable, rowName=PurchaseBattlePass_S23}
   }
   slot {class=CommonVisibilitySwitcherSlot}
  }
  LoudButton_BuyRewards {
   widgetClassPath=&WBP_UIKit_Button_Loud
   bIsVariable=true
   parent=TopButtonsSwitcher
   widget=LoudButton_BuyRewards
   properties {
    bHideInputActionWithKeyboard=true
    bIsFocusable=false
    hoveredSound=&BattlePass_Button_Hover_Tile_Default_Cue
    mobileFontSizeMultiplierOverride=1.33
    mobileMinHeight=80
    overrideDefaultHeight=true
    overrideDefaultMobileFontSizeMultiplier=true
    pressedSound=&BattlePass_Button_Confirm_Buy_BattlePass_Cue
    text=BUY REWARDS
    triggeringInputAction {dataTable=&SeasonPassInputActionDataTable, rowName=PurchaseBattlePass_S23}
   }
   slot {class=CommonVisibilitySwitcherSlot}
  }
  RegularButton_ClaimReward {
   widgetClassPath=&WBP_UIKit_Button_Regular
   bIsVariable=true
   parent=TopButtonsSwitcher
   widget=RegularButton_ClaimReward
   properties {
    bIsFocusable=false
    holdData=/FortUILibrary/Input/PurchaseInteraction_HoldData.PurchaseInteraction_HoldData_C
    hoveredSound=&BattlePass_Button_Hover_Tile_Default_Cue
    mobileFontSizeMultiplierOverride=1.33
    overrideDefaultHeight=true
    overrideDefaultMobileFontSizeMultiplier=true
    pressedSound=/Game/Sounds/Fort_UI_Sounds/FrontEnd/BattlePass/BattlePass_Button_Click_Whoosh_Cue
    text=CLAIM REWARD
    triggeringInputAction {dataTable=&SeasonPassInputActionDataTable, rowName=ClaimReward_Hold}
   }
   slot {class=CommonVisibilitySwitcherSlot, properties {verticalAlignment=VAlign_Top}}
  }
  Button_ViewReward {
   widgetClassPath=&WBP_UIKit_Button_Regular
   bIsVariable=true
   parent=VerticalBox_26
   widget=Button_ViewReward
   properties {
    bHideInputActionWithKeyboard=true
    bIsFocusable=false
    hoveredSound=&BattlePass_Button_Hover_Tile_Default_Cue
    mobileFontSizeMultiplierOverride=1.33
    overrideDefaultHeight=true
    overrideDefaultMobileFontSizeMultiplier=true
    pressedSound=&BattlePass_Button_Click_Whoosh_Cue
    text=VIEW REWARD
    triggeringInputAction {dataTable=&SeasonPassInputActionDataTable, rowName=GiftBattlePass}
   }
   slot {class=VerticalBoxSlot, properties {verticalAlignment=VAlign_Top, padding=<ltrb 0 0 0 12>}}
  }
  Button_Equip {
   widgetClassPath=&WBP_UIKit_Button_Regular
   bIsVariable=true
   parent=VerticalBox_26
   widget=Button_Equip
   properties {
    bHideInputActionWithKeyboard=true
    bIsFocusable=false
    hoveredSound=&BattlePass_Button_Hover_Tile_Default_Cue
    mobileFontSizeMultiplierOverride=1.33
    overrideDefaultHeight=true
    overrideDefaultMobileFontSizeMultiplier=true
    pressedSound=&BattlePass_Button_Click_Whoosh_Cue
    text=Equip
    triggeringInputAction {dataTable=&SeasonPassInputActionDataTable, rowName=EquipItemPreview}
   }
   slot {class=VerticalBoxSlot, properties {verticalAlignment=VAlign_Top, padding=<ltrb 0 0 0 12>}}
  }
  Button_CustomInfo {
   widgetClassPath=/Script/CommonUI.CommonLazyWidget
   bIsVariable=true
   parent=VerticalBox_26
   widget=Button_CustomInfo
   slot {class=VerticalBoxSlot}
  }
  ClaimAllSection {widgetClassPath=&VerticalBox, parent=VerticalBox_26, widget=ClaimAllSection, slot {class=VerticalBoxSlot}}
  Divider {
   widgetClassPath=&Image
   bIsVariable=true
   parent=ClaimAllSection
   widget=Divider
   properties {
    brush {
     bIsDynamicallyLoaded=false
     drawAs=Image
     imageType=NoImage
     mirroring=NoMirror
     resourceName=None
     resourceObject=null
     tiling=NoTile
     imageSize=<32 2>
     margin=<ltrb 0 0 0 0>
     outlineSettings {
      bUseBrushTransparency=false
      roundingType=HalfHeightRadius
      width=0
      color {colorUseRule=UseColor_Specified, specifiedColor=<rgba 0 0 0 0>}
      cornerRadii=<0 0 0 0>
     }
     tintColor {colorUseRule=UseColor_Specified, specifiedColor=<rgba 1 1 1 1>}
     uVRegion {bIsValid=false, max=<0 0>, min=<0 0>}
    }
    colorAndOpacity=<rgba 0.5776 0.5776 0.5776 0.2>
   }
   slot {class=VerticalBoxSlot, properties {padding=<ltrb 0 12 0 12>}}
  }
  ClaimAllLoudSwitcher {
   widgetClassPath=&CommonVisibilitySwitcher
   bIsVariable=true
   parent=ClaimAllSection
   widget=ClaimAllLoudSwitcher
   slot {class=VerticalBoxSlot}
  }
  LoudButton_ClaimAll {
   widgetClassPath=&WBP_UIKit_Button_Loud
   bIsVariable=true
   parent=ClaimAllLoudSwitcher
   widget=LoudButton_ClaimAll
   properties {
    bIsFocusable=false
    bRequiresHold=true
    holdData=&PurchaseInteraction_HoldData_C
    hoveredSound=&BattlePass_Button_Hover_Tile_Default_Cue
    mobileFontSizeMultiplierOverride=1.33
    mobileMinHeight=80
    overrideDefaultHeight=true
    overrideDefaultMobileFontSizeMultiplier=true
    pressedSound=&BattlePass_Button_Confirm_Buy_BattlePass_Cue
    text=CLAIM ALL
    triggeringInputAction {dataTable=&SeasonPassInputActionDataTable, rowName=ClaimAll_Hold}
   }
   slot {class=CommonVisibilitySwitcherSlot}
  }
  RegularButton_ClaimAll {
   widgetClassPath=&WBP_UIKit_Button_Regular
   bIsVariable=true
   parent=ClaimAllLoudSwitcher
   widget=RegularButton_ClaimAll
   properties {
    bIsFocusable=false
    holdData=&PurchaseInteraction_HoldData_C
    text=CLAIM ALL
    triggeringInputAction {dataTable=&SeasonPassInputActionDataTable, rowName=ClaimAll_Hold}
   }
   slot {class=CommonVisibilitySwitcherSlot}
  }
 ]
}
