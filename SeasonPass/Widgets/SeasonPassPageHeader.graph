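# SeasonPassPageHeader is a UserWidget that lays out a header for Season Pass pages.
# A horizontal box places a SeasonPassBadgeIcon widget beside a left-aligned, vertically centered CommonTextBlock title in the HeadingNow font.
# The title holds placeholder text, and both children are exposed as variables.
# It defines no variables of its own, and its event graph holds only unconnected PreConstruct, Construct and Tick events, so it has no logic.
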WidgetBlueprint /SeasonPass/Widgets/SeasonPassPageHeader
graphs [EventGraph {nodes [
   Event PreConstruct()
   Event Construct()
   Event Tick()
  ]}]
variables []
widgets {
 info {inheritedWidgetCount=0, parentClass=&UserWidget, rootWidgetClass=&HorizontalBox, widgetCount=3}
 widgets [
  HorizontalBox_0 {widgetClassPath=&HorizontalBox, widget=HorizontalBox_0}
  SeasonPassBadgeIcon {
   widgetClassPath=&/SeasonPassBadgeIcon
   bIsVariable=true
   parent=HorizontalBox_0
   widget=SeasonPassBadgeIcon
   slot {class=HorizontalBoxSlot}
  }
  Title {
   widgetClassPath=&CommonTextBlock
   bIsVariable=true
   parent=HorizontalBox_0
   widget=Title
   properties {
    text=Title
    font {
     bForceMonospaced=false
     bMaterialIsStencil=false
     fontMaterial=null
     fontObject=&HeadingNow
     letterSpacing=0
     monospacedWidth=1
     size=24
     skewAmount=0
     typefaceFontName=75 Medium
     outlineSettings {
      bApplyOutlineToDropShadows=false
      bMiteredCorners=false
      bSeparateFillAlpha=false
      outlineMaterial=null
      outlineSize=0
      outlineColor=<rgba 0 0 0 1>
     }
    }
    shadowOffset=<0 0>
   }
   slot {class=HorizontalBoxSlot, properties {horizontalAlignment=HAlign_Left, verticalAlignment=VAlign_Center, padding=<ltrb 16 0 0 0>}}
  }
 ]
}
