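# Top navigation button for the Season Pass UI, built on the generic UIKit modular button, with a claimable-rewards indicator.
# PreConstruct sets selection timings and, in compact UI, adjusts padding, collapses the main stack and reconfigures the indicator.
# Compact UI also drives the icon material's Selected parameter and async-loads pass icons, comparing against the last requested path to discard stale loads.
# Outside compact UI, setting a pass icon collapses the icon block.
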WidgetBlueprint /SeasonPass/Widgets/SeasonPass_Button_TopNav
graphs [
 EventGraph {nodes [
   n1 = SetPassIconEvent!() -> (OutputDelegate, then, Asset:Soft Object Reference) {is=title K2Node_CustomEvent}
   > n2 = Async Load Asset(Asset=@n1.Asset)
   Branch(
    execute=@n2.Completed
    Condition="Equal Exactly (String)"(
     A=To Soft Object Path!(SoftObjectReference:Soft Object Reference=@n1.Asset) -> (ReturnValue_PathString:String)
     B=Get LastRequestedSoftObjectPath() -> (LastRequestedSoftObjectPath:String)
    )
   )
   > SetIcon!(execute, self:WBP UIKit Button Generic Base Object Reference, Icon:Object Reference=@n2) -> (then)
   > SetIconSize!(execute, self:WBP UIKit Button Generic Base Object Reference, Icon:Vector 2D Structure=(X=48,Y=48)) -> (then)
   n3 = Event PreConstruct()
   > Parent: PreConstruct(IsDesignTime=@n3.IsDesignTime)
   > SetBlocksTiming!(
    execute
    self:UIKit Modular Button Object Reference
    Timing:"UIKit Block Timing Structure (by ref)"=Make UIKit Block Timing!(
     Hovering:float=0
     Unhovering:float=0
     Focusing:float=0
     Unfocusing:float=0
     Pressing:float=0
     Releasing:float=0
     Disabling:float=0
     Enabling:float=0
     Selecting:float=0.15
     DeselectingIdle:float=0.1
     DeselectingFocused:float=0.1
    ) -> (UIKitBlockTiming:UIKit Block Timing Structure) {is=title K2Node_MakeStruct}
   ) -> (then)
   > SetCompactUIStyle() {is=title K2Node_CallFunction}
   Event On Selected()
   > Parent: On Selected() {is=title K2Node_CallParentFunction}
   n4 = IsUsingCompactUI()
   > Branch(Condition=@n4)
   > SetScalarParameterValue(
    self=Get ImageMID!(self:WBP UIKit Block Image Object Reference=Get Block_Icon() -> (Block_Icon:WBP UIKit Block Image Object Reference)) -> (ImageMID:Material Instance Dynamic Object Reference)
    ParameterName=Selected
    Value=1
   )
   Event On Deselected()
   > Parent: On Deselected() {is=title K2Node_CallParentFunction}
   > Branch(Condition=@n4)
   > SetScalarParameterValue(self=Get ImageMID(self=Get Block_Icon()), ParameterName=Selected, Value=0)
  ]}
 SetClaimableRewards {nodes [
   n1 = SetClaimableRewards!() -> (then, RewardsAmount:int) {is=title K2Node_FunctionEntry}
   n2 = Get StatusIndicator() -> (StatusIndicator:WBP UIKit Status Indicator Object Reference)
   > Set Bang State!(execute, self:Fort Bang Wrapper NUI Object Reference=@n2, bEnabled:bool="integer > integer"(A=@n1), Count:int=0) -> (then) {is=title K2Node_CallFunction}
   > OverrideCount!(execute, self:WBP UIKit Status Indicator Base Object Reference=@n2, Count:int=Get RewardsAmount() -> (RewardsAmount:int=0)) -> (then)
  ]}
 SetCompactUIStyle {nodes [
   SetCompactUIStyle() {is=title K2Node_FunctionEntry}
   > Branch(Condition=IsUsingCompactUI())
   > SetPadding!(
    execute
    self:Grid Slot Object Reference=SlotAsGridSlot(Widget=Get Main() -> (Main:Grid Panel Object Reference))
    InPadding_Left:float=12
    InPadding_Top:float=12
    InPadding_Right:float=8
    InPadding_Bottom:float=12
   ) -> (then)
   > SetVisibility(self=Get Main_Stack() -> (Main_Stack:Stack Box Object Reference), InVisibility=Collapsed)
   > SetStatusIndicatorConfiguration!(
    execute
    self:WBP UIKit Button Generic Base Object Reference
    Configuration:E_UI_StatusIndicator_Configuration Enum=NewEnumerator2
    Count:int=0
   ) -> (then)
   > SetPadding(self=SlotAsGridSlot(Widget=Get StatusIndicator()), InPadding_Left=-26, InPadding_Top=0, InPadding_Right=6, InPadding_Bottom=0)
  ]}
 SetPassIcon {nodes [
   SetPassIcon!() -> (then, IconImage:Soft Object Reference) {is=title K2Node_FunctionEntry}
   > n1 = Branch(Condition=IsUsingCompactUI())
   > Set LastRequestedSoftObjectPath(LastRequestedSoftObjectPath:String=To Soft Object Path(SoftObjectReference=Get IconImage() -> (IconImage:Soft Object Reference))) -> (Output_Get:String)
   > SetPassIconEvent!(execute, self, Asset:Soft Object Reference=Get IconImage()) -> (then) {is=title K2Node_CallFunction}
   SetVisibility(execute=@n1.else, self=Get Block_Icon(), InVisibility=Collapsed)
  ]}
]
variables [LastRequestedSoftObjectPath {type=string, default=""}]
widgets {
 info {
  inheritedWidgetCount=0
  parentClass=/FortUILibrary/Systems/ModularButton/WBP_UIKit_ButtonGeneric_Base.WBP_UIKit_ButtonGeneric_Base_C
  rootWidgetClass=null
  widgetCount=0
 }
 widgets []
}
