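# Sidebar onboarding tooltip widget for the Season Pass: a coach mark announcing that a reward unlocks at Level 100 without requiring the Battle Pass.
# It wraps a UIKit coach mark, activating it on show and deactivating the tooltip on tooltip end or after a "Got it" click is handled.
# Deactivation notifies close completion and deactivates the coach mark.
# GetCoachMarkText inserts the reward name from an item details view model, with generic fallback text; SetCoachMarkText applies given text.
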WidgetBlueprint /SeasonPass/Widgets/SeasonPass_GrantRewardByLevel_Coachmark
graphs [
 EventGraph {nodes [
   "Button Other Action Clicked (WBP_UIKit_CoachMark)"() {is=title K2Node_ComponentBoundEvent}
   > n1 = HandleOtherActionClicked(self:Fort Sidebar Onboard Tooltip Widget Object Reference)
   DeactivateWidget(execute=[@n1.then, @n2.then])
   Event On Deactivated()
   > NotifyCloseAnimationFinished(self:Fort Sidebar Onboard Tooltip Widget Object Reference)
   > DeactivateWidget(self=Get WBP_UIKit_CoachMark() -> (WBP_UIKit_CoachMark:WBP UIKit Coach Mark Object Reference))
   n2 = Event OnEndTooltip() {is=title K2Node_Event}
   Event OnShowTooltip() {is=title K2Node_Event}
   > ActivateWidget(self=Get WBP_UIKit_CoachMark())
  ]}
 GetCoachMarkText {nodes [
   n1 = GetCoachMarkText!() -> (then, ItemDetailsVM:Fort Item Details VM Object Reference) {is=title K2Node_FunctionEntry}
   > n2 = Is Valid(InputObject=@n1)
   Return Node!(execute=@n2.Is Not Valid, Text:Text="LOCTEXT(\"Reach Level 100 to unlock the reward! Battle Pass not required.\")")
   > Return Node(Text=Format Text!(
     Format="LOCTEXT(\"Reach Level 100 to unlock the {Reward}! Battle Pass not required.\")"
     Reward:Text=GetDisplayName!(self:Fort Item Details VM Object Reference=@n1) -> (ReturnValue:Text)
    ) -> (Result))
  ]}
 SetCoachMarkText {nodes [
   n1 = SetCoachMarkText!() -> (then, Text:Text) {is=title K2Node_FunctionEntry}
   > SetTextDescription!(execute, self:WBP UIKit Coach Mark Base Object Reference=Get WBP_UIKit_CoachMark(), Text:Text=@n1) -> (then, Empty:bool)
  ]}
]
variables []
widgets {
 info {
  inheritedWidgetCount=2
  parentClass=/Script/FortniteUI.FortSidebarOnboardTooltipWidget
  rootWidgetClass=/FortUILibrary/UIKit/CoachMark/WBP_UIKit_CoachMark.WBP_UIKit_CoachMark_C
  widgetCount=3
 }
 widgets [
  WBP_UIKit_CoachMark {
   widgetClassPath=/FortUILibrary/UIKit/CoachMark/WBP_UIKit_CoachMark
   bIsVariable=true
   widget=WBP_UIKit_CoachMark
   properties {
    arrowPosition=Bottom_Left
    shouldShowDismissButton=false
    textDescription=""
    textHeader=Reward Available on Level 100
    textOtherActionButton=Got it
    input Action Other {dataTable=&InputActionDataTable, rowName=GenericFaceButtonBottom}
   }
  }
  Dismiss_Button {widgetClassPath=&CommonButtonLegacy, bInherited=true, widget=null}
  OtherAction_Button {widgetClassPath=&CommonButtonLegacy, bInherited=true, widget=null}
 ]
}
