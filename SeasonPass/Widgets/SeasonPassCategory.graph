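# SeasonPassCategory is the activatable Season Pass screen for browsing a category's paged reward track and previewing rewards and their variants.
# View models fill entry boxes with reward buttons sized by tile, plus variant cards that update the item description.
# Page changes animate between current and next track slots, and reward viewing is a separate state.
# Backgrounds load async with a stale-request check, while Intro and Close animate the vault-world material parameter collection.
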WidgetBlueprint /SeasonPass/Widgets/SeasonPassCategory
animations [
 Intro {
  binding_count=7
  length_seconds=0.6
  object=&self:Intro
  sequence {
   bindings [
    SizeBox_LeftContent {
     id=#1
     possessed_class=SizeBox
     tracks [
      {
       type=MovieScene2DTransformTrack
       sections [{
         type=MovieScene2DTransformSection
         end_frame=10
         start_frame=0
         channels [
          Translation.X {type=MovieSceneScriptingFloatChannel, keys [[0,-100],[9,0]]}
          Translation.Y {type=MovieSceneScriptingFloatChannel, const [0,2]}
         ]
        }]
      }
      {
       type=MovieSceneFloatTrack
       display_name=Render Opacity
       sections [{type=MovieSceneFloatSection, end_frame=10, start_frame=0, channels [None {type=MovieSceneScriptingFloatChannel, keys [[0,0],[9,1]]}]}]
      }
     ]
    }
    SeasonPassPageHeader {
     id=#2
     possessed_class=SeasonPassPageHeader_C
     tracks [
      {
       type=MovieSceneFloatTrack
       display_name=Render Opacity
       sections [{type=MovieSceneFloatSection, end_frame=10, start_frame=0, channels [None {type=MovieSceneScriptingFloatChannel, keys [[0,0],[7,1]]}]}]
      }
      {
       type=MovieScene2DTransformTrack
       sections [{
         type=MovieScene2DTransformSection
         end_frame=10
         start_frame=-1
         channels [
          Translation.X {type=MovieSceneScriptingFloatChannel, keys [[-0.0003,-100],[6.9997,0]]}
          Translation.Y {type=MovieSceneScriptingFloatChannel, const [0,2]}
         ]
        }]
      }
     ]
    }
    SeasonPass_Level {
     id=#3
     possessed_class=SeasonPass_Level_C
     tracks [
      {
       type=MovieScene2DTransformTrack
       sections [{
         type=MovieScene2DTransformSection
         end_frame=10
         start_frame=-1
         channels [
          Translation.X {type=MovieSceneScriptingFloatChannel, keys [[-0.0003,-100],[0,100],[6.9997,0]]}
          Translation.Y {type=MovieSceneScriptingFloatChannel, const [0,2]}
         ]
        }]
      }
      {
       type=MovieSceneFloatTrack
       display_name=Render Opacity
       sections [{type=MovieSceneFloatSection, end_frame=10, start_frame=0, channels [None {type=MovieSceneScriptingFloatChannel, keys [[0,0],[7,1]]}]}]
      }
     ]
    }
    VerticalBox_3 {
     id=#4
     possessed_class=VerticalBox
     tracks [
      {
       type=MovieSceneFloatTrack
       display_name=Render Opacity
       sections [{type=MovieSceneFloatSection, end_frame=9, start_frame=0, channels [None {type=MovieSceneScriptingFloatChannel, keys [[0,0],[9,1]]}]}]
      }
      {
       type=MovieScene2DTransformTrack
       sections [{
         type=MovieScene2DTransformSection
         end_frame=10
         start_frame=0
         channels [
          Translation.X {type=MovieSceneScriptingFloatChannel, keys [[0,100],[9,0]]}
          Translation.Y {type=MovieSceneScriptingFloatChannel, const [0,2]}
         ]
        }]
      }
     ]
    }
    HorizontalBox_NavigationArrows {
     id=#5
     possessed_class=HorizontalBox
     tracks [{
       type=MovieSceneFloatTrack
       display_name=Render Opacity
       sections [{type=MovieSceneFloatSection, end_frame=12, start_frame=0, channels [None {type=MovieSceneScriptingFloatChannel, keys [[8,0],[12,1]]}]}]
      }]
    }
    TrackAndStatusBarContainer {
     id=#6
     possessed_class=VerticalBox
     tracks [
      {
       type=MovieSceneFloatTrack
       display_name=Render Opacity
       sections [{type=MovieSceneFloatSection, end_frame=11, start_frame=0, channels [None {type=MovieSceneScriptingFloatChannel, keys [[2,0],[11,1]]}]}]
      }
      {
       type=MovieScene2DTransformTrack
       sections [{
         type=MovieScene2DTransformSection
         end_frame=11
         start_frame=0
         channels [
          Translation.X {type=MovieSceneScriptingFloatChannel, const [0,2]}
          Translation.Y {type=MovieSceneScriptingFloatChannel, keys [[2,60],[11,0]]}
          Angle {type=MovieSceneScriptingFloatChannel}
          Scale.X {type=MovieSceneScriptingFloatChannel}
          Scale.Y {type=MovieSceneScriptingFloatChannel}
          Shear.X {type=MovieSceneScriptingFloatChannel}
          Shear.Y {type=MovieSceneScriptingFloatChannel}
         ]
        }]
      }
     ]
    }
    SeasonPass_RewardTrack_StatusBar {
     id=#7
     possessed_class=SeasonPass_RewardTrack_StatusBar_C
     tracks [{
       type=MovieSceneFloatTrack
       display_name=Render Opacity
       sections [{type=MovieSceneFloatSection, end_frame=12, start_frame=0, channels [None {type=MovieSceneScriptingFloatChannel, keys [[8,0],[12,1]]}]}]
      }]
    }
   ]
   tracks [{
     type=MovieSceneMaterialParameterCollectionTrack
     display_name=MPC_UI_Master_VaultWorld_Settings
     sections [{
       type=MovieSceneParameterSection
       channels [
        FocalPointPositionR {type=MovieSceneScriptingFloatChannel, keys [[3,0.4],[9,0]]}
        FocalPointPositionG {type=MovieSceneScriptingFloatChannel, keys [[3,0],[9,0.1]]}
        FocalPointPositionB {type=MovieSceneScriptingFloatChannel}
        FocalPointPositionA {type=MovieSceneScriptingFloatChannel, keys [[0,0],[2,0],[10,1],[12,1]]}
       ]
      }]
    }]
  }
 }
 ToViewReward {
  binding_count=6
  length_seconds=0.25
  object=&self:ToViewReward
  sequence {bindings [
    SecondaryCurrentPageText {
     id=#8
     possessed_class=CommonTextBlock
     tracks [{
       type=MovieSceneFloatTrack
       display_name=Render Opacity
       sections [{type=MovieSceneFloatSection, end_frame=5, start_frame=0, channels [None {type=MovieSceneScriptingFloatChannel, keys [[0,1],[5,0]]}]}]
      }]
    }
    HorizontalBox_NavigationArrows {
     id=#9
     possessed_class=HorizontalBox
     tracks [
      {
       type=MovieScene2DTransformTrack
       sections [{
         type=MovieScene2DTransformSection
         end_frame=5
         start_frame=0
         channels [
          Translation.X {type=MovieSceneScriptingFloatChannel, const [0,2]}
          Translation.Y {type=MovieSceneScriptingFloatChannel, keys [[0,0],[5,200]]}
          Angle {type=MovieSceneScriptingFloatChannel}
          Scale.X {type=MovieSceneScriptingFloatChannel}
          Scale.Y {type=MovieSceneScriptingFloatChannel}
          Shear.X {type=MovieSceneScriptingFloatChannel}
          Shear.Y {type=MovieSceneScriptingFloatChannel}
         ]
        }]
      }
      {
       type=MovieSceneFloatTrack
       display_name=Render Opacity
       sections [{type=MovieSceneFloatSection, end_frame=5, start_frame=0, channels [None {type=MovieSceneScriptingFloatChannel, keys [[0,1],[5,0]]}]}]
      }
     ]
    }
    TrackAndStatusBarContainer {
     id=#10
     possessed_class=VerticalBox
     tracks [
      {
       type=MovieScene2DTransformTrack
       sections [{
         type=MovieScene2DTransformSection
         end_frame=5
         start_frame=0
         channels [
          Translation.X {type=MovieSceneScriptingFloatChannel, const [0,2]}
          Translation.Y {type=MovieSceneScriptingFloatChannel, keys [[0,0],[5,200]]}
          Angle {type=MovieSceneScriptingFloatChannel}
          Scale.X {type=MovieSceneScriptingFloatChannel}
          Scale.Y {type=MovieSceneScriptingFloatChannel}
          Shear.X {type=MovieSceneScriptingFloatChannel}
          Shear.Y {type=MovieSceneScriptingFloatChannel}
         ]
        }]
      }
      {
       type=MovieSceneFloatTrack
       display_name=Render Opacity
       sections [{type=MovieSceneFloatSection, end_frame=5, start_frame=0, channels [None {type=MovieSceneScriptingFloatChannel, keys [[0,1],[5,0]]}]}]
      }
     ]
    }
    RewardCards_A {
     id=#11
     possessed_class=DynamicEntryBox
     tracks [{
       type=MovieSceneBoolTrack
       display_name=Is Enabled
       sections [{
         type=MovieSceneBoolSection
         end_frame=5
         start_frame=0
         channels [None {type=MovieSceneScriptingBoolChannel, keys [[0,true],[1,false],[5,false]]}]
        }]
      }]
    }
    SeasonPassCategoryItemPreviewWidget {
     id=#12
     possessed_class=WBP_ItemPreview_C
     tracks [{
       type=MovieSceneMarginTrack
       display_name=Padding
       sections [{
         type=MovieSceneMarginSection
         end_frame=5
         start_frame=0
         channels [
          Left {type=MovieSceneScriptingFloatChannel, const [0,1]}
          Top {type=MovieSceneScriptingFloatChannel, const [0,1]}
          Right {type=MovieSceneScriptingFloatChannel, const [0,1]}
          Bottom {type=MovieSceneScriptingFloatChannel, keys [[0,128],[5,0]]}
         ]
        }]
      }]
    }
    SeasonPass_RewardTrack_StatusBar {
     id=#13
     possessed_class=SeasonPass_RewardTrack_StatusBar_C
     tracks [{
       type=MovieSceneFloatTrack
       display_name=Render Opacity
       sections [{type=MovieSceneFloatSection, end_frame=5, start_frame=0, channels [None {type=MovieSceneScriptingFloatChannel, keys [[0,1],[5,0]]}]}]
      }]
    }
   ]}
 }
 Close {
  binding_count=5
  length_seconds=0.6
  object=&self:Close
  sequence {
   bindings [
    SeasonPassPageHeader {
     id=#14
     possessed_class=SeasonPassPageHeader_C
     tracks [
      {
       type=MovieScene2DTransformTrack
       sections [{
         type=MovieScene2DTransformSection
         end_frame=12
         start_frame=0
         channels [
          Translation.X {type=MovieSceneScriptingFloatChannel, keys [[0,0],[6,-200],[12,-200]]}
          Translation.Y {type=MovieSceneScriptingFloatChannel}
          Angle {type=MovieSceneScriptingFloatChannel}
          Scale.X {type=MovieSceneScriptingFloatChannel}
          Scale.Y {type=MovieSceneScriptingFloatChannel}
          Shear.X {type=MovieSceneScriptingFloatChannel}
          Shear.Y {type=MovieSceneScriptingFloatChannel}
         ]
        }]
      }
      {
       type=MovieSceneFloatTrack
       display_name=Render Opacity
       sections [{type=MovieSceneFloatSection, end_frame=12, start_frame=0, channels [None {type=MovieSceneScriptingFloatChannel, keys [[0,1],[6,0],[12,0]]}]}]
      }
     ]
    }
    VerticalBox_3 {
     id=#15
     possessed_class=VerticalBox
     tracks [
      {
       type=MovieSceneFloatTrack
       display_name=Render Opacity
       sections [{type=MovieSceneFloatSection, end_frame=12, start_frame=0, channels [None {type=MovieSceneScriptingFloatChannel, keys [[0,1],[12,0]]}]}]
      }
      {
       type=MovieScene2DTransformTrack
       sections [{
         type=MovieScene2DTransformSection
         end_frame=12
         start_frame=0
         channels [
          Translation.X {type=MovieSceneScriptingFloatChannel, keys [[0,0],[12,200]]}
          Translation.Y {type=MovieSceneScriptingFloatChannel, const [0,1]}
          Angle {type=MovieSceneScriptingFloatChannel}
          Scale.X {type=MovieSceneScriptingFloatChannel}
          Scale.Y {type=MovieSceneScriptingFloatChannel}
          Shear.X {type=MovieSceneScriptingFloatChannel}
          Shear.Y {type=MovieSceneScriptingFloatChannel}
         ]
        }]
      }
     ]
    }
    SeasonPass_Level {
     id=#16
     possessed_class=SeasonPass_Level_C
     tracks [
      {
       type=MovieScene2DTransformTrack
       sections [{
         type=MovieScene2DTransformSection
         end_frame=12
         start_frame=0
         channels [
          Translation.X {type=MovieSceneScriptingFloatChannel, keys [[0,0],[6,200],[12,200]]}
          Translation.Y {type=MovieSceneScriptingFloatChannel, const [0,2]}
          Angle {type=MovieSceneScriptingFloatChannel}
          Scale.X {type=MovieSceneScriptingFloatChannel}
          Scale.Y {type=MovieSceneScriptingFloatChannel}
          Shear.X {type=MovieSceneScriptingFloatChannel}
          Shear.Y {type=MovieSceneScriptingFloatChannel}
         ]
        }]
      }
      {
       type=MovieSceneFloatTrack
       display_name=Render Opacity
       sections [{type=MovieSceneFloatSection, end_frame=12, start_frame=0, channels [None {type=MovieSceneScriptingFloatChannel, keys [[0,1],[6,0],[12,0]]}]}]
      }
     ]
    }
    TrackAndStatusBarContainer {
     id=#17
     possessed_class=VerticalBox
     tracks [
      {
       type=MovieScene2DTransformTrack
       sections [{
         type=MovieScene2DTransformSection
         end_frame=12
         start_frame=0
         channels [
          Translation.X {type=MovieSceneScriptingFloatChannel, const [0,1]}
          Translation.Y {type=MovieSceneScriptingFloatChannel, keys [[0,0],[12,100]]}
          Angle {type=MovieSceneScriptingFloatChannel}
          Scale.X {type=MovieSceneScriptingFloatChannel}
          Scale.Y {type=MovieSceneScriptingFloatChannel}
          Shear.X {type=MovieSceneScriptingFloatChannel}
          Shear.Y {type=MovieSceneScriptingFloatChannel}
         ]
        }]
      }
      {
       type=MovieSceneFloatTrack
       display_name=Render Opacity
       sections [{type=MovieSceneFloatSection, end_frame=12, start_frame=0, channels [None {type=MovieSceneScriptingFloatChannel, keys [[0,1],[12,0]]}]}]
      }
     ]
    }
    SizeBox_LeftContent {
     id=#18
     possessed_class=SizeBox
     tracks [
      {
       type=MovieScene2DTransformTrack
       sections [{
         type=MovieScene2DTransformSection
         end_frame=12
         start_frame=0
         channels [
          Translation.X {type=MovieSceneScriptingFloatChannel, keys [[0,0],[12,-200]]}
          Translation.Y {type=MovieSceneScriptingFloatChannel, const [0,1]}
          Angle {type=MovieSceneScriptingFloatChannel}
          Scale.X {type=MovieSceneScriptingFloatChannel}
          Scale.Y {type=MovieSceneScriptingFloatChannel}
          Shear.X {type=MovieSceneScriptingFloatChannel}
          Shear.Y {type=MovieSceneScriptingFloatChannel}
         ]
        }]
      }
      {
       type=MovieSceneFloatTrack
       display_name=Render Opacity
       sections [{type=MovieSceneFloatSection, end_frame=12, start_frame=0, channels [None {type=MovieSceneScriptingFloatChannel, keys [[0,1],[12,0]]}]}]
      }
     ]
    }
   ]
   tracks [
    {
     type=MovieSceneMaterialParameterCollectionTrack
     display_name=MPC_UI_Master_VaultWorld_Settings
     sections [{
       type=MovieSceneParameterSection
       channels [
        FocalPointPositionR {type=MovieSceneScriptingFloatChannel, keys [[3,0],[9,0.4]]}
        FocalPointPositionG {type=MovieSceneScriptingFloatChannel, keys [[3,0.1],[9,0]]}
        FocalPointPositionB {type=MovieSceneScriptingFloatChannel}
        FocalPointPositionA {type=MovieSceneScriptingFloatChannel, keys [[0,1],[2,1],[10,0],[12,0]]}
       ]
      }]
    }
    {
     type=MovieSceneEventTrack
     sections [{
       type=MovieSceneEventTriggerSection
       channels [None {type=MovieSceneScriptingEventChannel, const ["<Struct 'MovieSceneEvent' (0x46919260460) {}>", 1]}]
      }]
    }
   ]
  }
 }
 ItemPreviewFraming {
  binding_count=1
  length_seconds=0.25
  object=&self:ItemPreviewFraming
  sequence {bindings [SeasonPassCategory_C_29 {
     id=#19
     possessed_class=SeasonPassCategory_C
     tracks [{
       type=MovieSceneDoubleTrack
       display_name=View Reward Anim Value
       sections [{type=MovieSceneDoubleSection, end_frame=5, start_frame=0, channels [None {type=MovieSceneScriptingDoubleChannel, keys [[0,0],[5,1]]}]}]
      }]
    }]}
 }
 NextPage {
  binding_count=4
  length_seconds=0.4
  object=&self:NextPage
  sequence {
   bindings [
    NamedSlot_Track_Next {
     id=#20
     possessed_class=NamedSlot
     tracks [
      {
       type=MovieSceneFloatTrack
       display_name=Render Opacity
       sections [{
         type=MovieSceneFloatSection
         end_frame=12
         start_frame=0
         channels [None {type=MovieSceneScriptingFloatChannel, keys [[0,0],[3,0.0026],[12,1]]}]
        }]
      }
      {
       type=MovieScene2DTransformTrack
       sections [{
         type=MovieScene2DTransformSection
         end_frame=12
         start_frame=0
         channels [Translation.X {type=MovieSceneScriptingFloatChannel, keys [[0,300],[12,0]]}]
        }]
      }
     ]
    }
    SecondaryNavigationArrows {
     id=#21
     possessed_class=Overlay
     tracks [{
       type=MovieSceneFloatTrack
       display_name=Render Opacity
       sections [{type=MovieSceneFloatSection, end_frame=12, start_frame=0, channels [None {type=MovieSceneScriptingFloatChannel, keys [[0,1],[4,0],[12,0]]}]}]
      }]
    }
    SeasonPass_RewardTrack_StatusBar {
     id=#22
     possessed_class=SeasonPass_RewardTrack_StatusBar_C
     tracks [
      {
       type=MovieSceneFloatTrack
       display_name=Render Opacity
       sections [{type=MovieSceneFloatSection, end_frame=12, start_frame=0, channels [None {type=MovieSceneScriptingFloatChannel, keys [[0,1],[9,0],[12,0]]}]}]
      }
      {
       type=MovieScene2DTransformTrack
       sections [{
         type=MovieScene2DTransformSection
         end_frame=12
         start_frame=0
         channels [
          Translation.X {type=MovieSceneScriptingFloatChannel, keys [[0,0],[8,-175],[12,0]]}
          Translation.Y {type=MovieSceneScriptingFloatChannel, const [0,1]}
          Angle {type=MovieSceneScriptingFloatChannel, const [0,1]}
          Scale.X {type=MovieSceneScriptingFloatChannel, const [1,1]}
          Scale.Y {type=MovieSceneScriptingFloatChannel, const [1,1]}
          Shear.X {type=MovieSceneScriptingFloatChannel, const [0,1]}
          Shear.Y {type=MovieSceneScriptingFloatChannel, const [0,1]}
         ]
        }]
      }
     ]
    }
    NamedSlot_Track_Current {
     id=#23
     possessed_class=NamedSlot
     tracks [
      {
       type=MovieSceneFloatTrack
       display_name=Render Opacity
       sections [{
         type=MovieSceneFloatSection
         end_frame=12
         start_frame=0
         channels [None {type=MovieSceneScriptingFloatChannel, keys [[0,1],[7.5,0],[12,0]]}]
        }]
      }
      {
       type=MovieScene2DTransformTrack
       sections [{
         type=MovieScene2DTransformSection
         end_frame=12
         start_frame=0
         channels [Translation.X {type=MovieSceneScriptingFloatChannel, keys [[0,0],[12,-300]]}]
        }]
      }
     ]
    }
   ]
   display_rate {numerator=30}
  }
 }
 PreviousPage {
  binding_count=4
  length_seconds=0.4
  object=&self:PreviousPage
  sequence {
   bindings [
    SeasonPass_RewardTrack_StatusBar {
     id=#24
     possessed_class=SeasonPass_RewardTrack_StatusBar_C
     tracks [
      {
       type=MovieSceneFloatTrack
       display_name=Render Opacity
       sections [{type=MovieSceneFloatSection, end_frame=12, start_frame=0, channels [None {type=MovieSceneScriptingFloatChannel, keys [[0,1],[9,0],[12,0]]}]}]
      }
      {
       type=MovieScene2DTransformTrack
       sections [{
         type=MovieScene2DTransformSection
         end_frame=12
         start_frame=0
         channels [
          Translation.X {type=MovieSceneScriptingFloatChannel, keys [[0,0],[8,175],[12,0]]}
          Translation.Y {type=MovieSceneScriptingFloatChannel, const [0,1]}
          Angle {type=MovieSceneScriptingFloatChannel, const [0,1]}
          Scale.X {type=MovieSceneScriptingFloatChannel, const [1,1]}
          Scale.Y {type=MovieSceneScriptingFloatChannel, const [1,1]}
          Shear.X {type=MovieSceneScriptingFloatChannel, const [0,1]}
          Shear.Y {type=MovieSceneScriptingFloatChannel, const [0,1]}
         ]
        }]
      }
     ]
    }
    NamedSlot_Track_Next {
     id=#25
     possessed_class=NamedSlot
     tracks [
      {
       type=MovieSceneFloatTrack
       display_name=Render Opacity
       sections [{
         type=MovieSceneFloatSection
         end_frame=12
         start_frame=0
         channels [None {type=MovieSceneScriptingFloatChannel, keys [[0,0],[3,0.0026],[12,1]]}]
        }]
      }
      {
       type=MovieScene2DTransformTrack
       sections [{
         type=MovieScene2DTransformSection
         end_frame=12
         start_frame=0
         channels [Translation.X {type=MovieSceneScriptingFloatChannel, keys [[0,-300],[12,0]]}]
        }]
      }
     ]
    }
    NamedSlot_Track_Current {
     id=#26
     possessed_class=NamedSlot
     tracks [
      {
       type=MovieSceneFloatTrack
       display_name=Render Opacity
       sections [{type=MovieSceneFloatSection, end_frame=12, start_frame=0, channels [None {type=MovieSceneScriptingFloatChannel, keys [[0,1],[9,0],[12,0]]}]}]
      }
      {
       type=MovieScene2DTransformTrack
       sections [{
         type=MovieScene2DTransformSection
         end_frame=12
         start_frame=0
         channels [Translation.X {type=MovieSceneScriptingFloatChannel, keys [[0,0],[12,300]]}]
        }]
      }
     ]
    }
    SecondaryNavigationArrows {
     id=#27
     possessed_class=Overlay
     tracks [{
       type=MovieSceneFloatTrack
       display_name=Render Opacity
       sections [{type=MovieSceneFloatSection, end_frame=12, start_frame=0, channels [None {type=MovieSceneScriptingFloatChannel, keys [[0,1],[4,0],[12,0]]}]}]
      }]
    }
   ]
   display_rate {numerator=30}
  }
 }
 ShowStatusBar {
  binding_count=2
  length_seconds=0.1489
  object=&self:ShowStatusBar
  sequence {
   bindings [
    SeasonPass_RewardTrack_StatusBar {
     id=#28
     possessed_class=SeasonPass_RewardTrack_StatusBar_C
     tracks [{
       type=MovieSceneFloatTrack
       display_name=Render Opacity
       sections [{
         type=MovieSceneFloatSection
         end_frame=4
         start_frame=0
         channels [None {type=MovieSceneScriptingFloatChannel, keys [[0,0],[1,0],[4.4665,1]]}]
        }]
      }]
    }
    SecondaryNavigationArrows {
     id=#29
     possessed_class=Overlay
     tracks [{
       type=MovieSceneFloatTrack
       display_name=Render Opacity
       sections [{
         type=MovieSceneFloatSection
         end_frame=4
         start_frame=0
         channels [None {type=MovieSceneScriptingFloatChannel, keys [[0,0],[1.117,0],[4.4665,1]]}]
        }]
      }]
    }
   ]
   display_rate {numerator=30}
  }
 }
]
graphs [
 BP_GetDesiredFocusTarget {nodes [
   Get Desired Focus Target()
   n1 = GetAllEntries!(self:Dynamic Entry Box Base Object Reference=Get VariantCards() -> (VariantCards:Fort Dynamic Entry Box Object Reference)) -> (ReturnValue:Array of User Widget Object References)
   n2 = GetAllEntries(self=Get ActiveRewardList() -> (ActiveRewardList:Dynamic Entry Box Object Reference))
   n3 = Get SelectedRewardIndex() -> (SelectedRewardIndex:int=0)
   > Return Node!(
    execute
    ReturnValue:Widget Object Reference=Select(
     Option 0:User Widget Object Reference="Get (a copy)"(Array:Array of User Widget Object References=@n2, Dimension 1=@n3) -> (Output:User Widget Object Reference)
     Option 1:User Widget Object Reference="Get (a copy)"(Array=@n1)
     Index=Get IsViewingReward() -> (IsViewingReward:bool)
    ) -> (ReturnValue:User Widget Object Reference)
   )
  ]}
 BP_GetDesiredInputConfig {nodes [
   Get Desired Input Config() {is=title K2Node_FunctionEntry}
   > Return Node(ReturnValue:UIInput Config Structure=Make UIInput Config!(
     bIgnoreMoveInput:bool=False
     bIgnoreLookInput:bool=False
     InputMode:ECommonInputMode Enum=All
     MouseCaptureMode:EMouseCaptureMode Enum=CaptureDuringMouseDown
     MouseLockMode:EMouseLockMode Enum=DoNotLock
     bHideCursorDuringViewportCapture:bool=True
    ) -> (UIInputConfig:UIInput Config Structure) {is=title K2Node_MakeStruct})
  ]}
 CreateSeasonPassReward {nodes [
   CreateSeasonPassReward!() -> (then, RewardVM:Season Pass Reward VM Object Reference) {is=title K2Node_FunctionEntry}
   > n1 = Create Entry!(execute, self:Dynamic Entry Box Object Reference=Get ActiveRewardList()) -> (then, ReturnValue:User Widget Object Reference)
   > n2 = Cast To SeasonPassRewardButton!(execute, Object=@n1) -> (then, CastFailed, AsSeason Pass Reward Button:Season Pass Reward Button Object Reference) {is=title K2Node_DynamicCast}
   > n3 = Set LoopRewardButton(LoopRewardButton:Season Pass Reward Button Object Reference=@n2) -> (Output_Get:Season Pass Reward Button Object Reference)
   > SetSeasonPassRewardVM!(
    execute
    self:Season Pass Reward Button Object Reference=@n3
    Viewmodel:Season Pass Reward VM Object Reference=Get RewardVM() -> (RewardVM:Season Pass Reward VM Object Reference)
   ) -> (then)
   n4 = Get TileSize!(self:Athena Season Item Entry Base Object Reference=GetReward!(self:Season Pass Reward VM Object Reference=Get RewardVM()) -> (ReturnValue:Athena Season Item Entry Base Object Reference)) -> (TileSize:EPageItemTileSize Enum=Size_1_x_1)
   > SetItemCardSize!(
    execute
    self:WBP UIKit Item Card Base Object Reference=Get LoopRewardButton() -> (LoopRewardButton:Season Pass Reward Button Object Reference)
    ItemSize_X:double=float * float(
     A=Get BaseTileSizeX() -> (BaseTileSizeX:double)
     B=Select!(Size_1_x_1:double=1, Size_2_x_1:double=2, Size_2_x_2:double=2, Size_3_x_2:double=3, Size_3_x_3:double=3, Index:EPageItemTileSize Enum=@n4) -> (ReturnValue:double)
    ) {is=title K2Node_PromotableOperator}
    ItemSize_Y:double=160
   ) -> (then)
   > Set SeasonPassVM!(
    execute
    SeasonPassVM:Season Pass VM Object Reference=Get SeasonPassVM() -> (SeasonPassVM:Season Pass VM Object Reference)
    self:Season Pass Reward Button Object Reference=Get LoopRewardButton()
   ) -> (then, Output_Get:Season Pass VM Object Reference)
   > Return Node!(execute, AsSeason Pass Reward Button:Season Pass Reward Button Object Reference=Get LoopRewardButton())
   Return Node(execute=@n2.CastFailed)
  ]}
 CreateSeasonPassVariant {nodes [
   CreateSeasonPassVariant!() -> (then, ItemVM:Fort Item VM Object Reference, ItemDef:Fort Item Definition Object Reference) {is=title K2Node_FunctionEntry}
   > n1 = Create Entry(self=Get VariantCards())
   > n2 = Cast To WBP_UIKit_ItemCard_BattlePass!(execute, Object=@n1) -> (then, CastFailed, AsWBP UIKit Item Card Battle Pass:WBP UIKit Item Card Battle Pass Object Reference) {is=title K2Node_DynamicCast}
   > InitializeItemCard(self:WBP UIKit Item Card Base Object Reference=@n2)
   > SetFortItemVM!(
    execute
    self:WBP UIKit Item Card Battle Pass Object Reference=@n2
    Viewmodel:Fort Item VM Object Reference=Get ItemVM() -> (ItemVM:Fort Item VM Object Reference)
   ) -> (then)
   > Bind Event to On Card Focused(self:WBP UIKit Item Card Battle Pass Object Reference=@n2, Delegate=Create Event()) {is=title K2Node_AddDelegate}
   > Bind Event to Viewing Reward State Changed(self:Season Pass Category Object Reference, Delegate=Create Event(self=@n2)) {is=title K2Node_AddDelegate}
   > Bind Event to New Variant Card Selected(self:Season Pass Category Object Reference, Delegate=Create Event(self=@n2)) {is=title K2Node_AddDelegate}
  ]}
 DecreasePage {nodes [
   DecreasePage() {is=title K2Node_FunctionEntry}
   n1 = Get IsViewingReward()
   n2 = Get PageIndex() -> (PageIndex:int=0)
   > Branch(Condition=AND Boolean(A="integer > integer"(A=@n2), B=NOT Boolean(A=@n1)))
   n3 = Create Event()
   > UnbindFromAnimationFinished!(
    execute
    self:User Widget Object Reference
    Animation:Widget Animation Object Reference=Get NextPage() -> (NextPage:Widget Animation Object Reference)
    Delegate:Delegate=@n3
   ) -> (then)
   > UnbindFromAnimationFinished(Animation=Get PreviousPage() -> (PreviousPage:Widget Animation Object Reference), Delegate=@n3)
   > UpdateActiveTrack!(execute, self, ActiveTrack:Named Slot Object Reference=Get NamedSlot_Track_Next() -> (NamedSlot_Track_Next:Named Slot Object Reference)) -> (then)
   > Set Visibility(Visibility=SelfHitTestInvisible, self=Get NamedSlot_Track_Next())
   n4 = Get PageIndex() -> (PageIndex=0)
   > Branch(Condition=OR Boolean(A="integer > integer"(A=@n4), B=Get EnableNavigationPageWrapping() -> (EnableNavigationPageWrapping:bool)))
   n5 = Length(TargetArray:Array of Season Pass Page VM Object References=Get Pages!(self:Season Pass Category VM Object Reference=Get SeasonPassCategoryVM() -> (SeasonPassCategoryVM:Season Pass Category VM Object Reference)) -> (Pages:Array of Season Pass Page VM Object References))
   > Set Page Index!(execute, self, Page Index:int="% (Integer)"(A=int + int!(A=@n4, B=@n5, C:int=-1) -> (ReturnValue), B=@n5) -> (ReturnValue=0)) -> (then) {is=title K2Node_CallFunction}
   > BindToAnimationFinished(Animation=Get PreviousPage(), Delegate=Create Event())
   > n6 = Branch(Condition=IsAnimationPlaying(InAnimation=Get NextPage()))
   > n7 = StopAnimation(InAnimation=Get NextPage())
   PlayAnimationForward(execute=[@n7.then, @n6.else], InAnimation=Get PreviousPage())
  ]}
 EventGraph {nodes [
   n1 = "On Clicked (Left)"!() -> (OutputDelegate, then, Button:Common Button Base Object Reference) {is=title K2Node_ComponentBoundEvent}
   > n2 = FireEventElementInteractedByName!(execute, Element Name:String=Button_Previous_Page, WidgetCallingFrom:Widget Object Reference=@n1.Button) -> (then, Success:bool)
   DecreasePage(execute=[@n8.then, @n2.then]) {is=title K2Node_CallFunction}
   n3 = "On Clicked (Right)"!() -> (OutputDelegate, then, Button:Common Button Base Object Reference) {is=title K2Node_ComponentBoundEvent}
   > n4 = FireEventElementInteractedByName(Element Name=Button_Next_Page, WidgetCallingFrom=@n3.Button)
   IncreasePage(execute=[@n4.then, @n10.then])
   n5 = "On Clicked (Button_Back)"!() -> (OutputDelegate, then, Button:Common Button Base Object Reference) {is=title K2Node_ComponentBoundEvent}
   > Fire Event Element Interacted(
    Element=@n5.Button
    AdditionalInfo=Get Analytics Info Pass Name!(
     MakeFlowAnalyticsAdditionalInfo:"Flow Analytics Additional Info Structure (by ref)"=Make Flow Analytics Additional Info!() -> (FlowAnalyticsAdditionalInfo:Flow Analytics Additional Info Structure) {is=title K2Node_MakeStruct}
     Key:String
     SeasonPassSelectedVM:Season Pass Selection VM Object Reference=Get SeasonPassSelectionVM() -> (SeasonPassSelectionVM:Season Pass Selection VM Object Reference)
    ) -> (PassName:Flow Analytics Additional Info Structure) {is=title K2Node_MacroInstance}
   )
   > n6 = Branch(Condition=Get IsViewingReward())
   > ClearCachedPreviewContextTags(self:Fort Item Preview Screen Object Reference=Get SeasonPassCategoryItemPreviewWidget() -> (SeasonPassCategoryItemPreviewWidget:WBP Item Preview Object Reference))
   > SetViewingRewardState!(execute, self, IsViewingReward:bool=false) -> (then)
   > SetRotationOffset!(
    execute
    self:Fort Item Preview Widget Object Reference=Get SeasonPassCategoryItemPreviewWidget()
    InRotationOffset_Roll:float=0
    InRotationOffset_Pitch:float=0
    InRotationOffset_Yaw:float=0
   ) -> (then)
   ClearPreviewItem(execute=@n6.else, self:Fort Item Preview Widget Object Reference=Get SeasonPassCategoryItemPreviewWidget())
   > PlayAnimation(InAnimation=Get Close() -> (Close:Widget Animation Object Reference))
   > SetVisibility(self=Get Button_Back() -> (Button_Back:WBP UIKit Button Quiet Object Reference), InVisibility=HitTestInvisible)
   n7 = "On Clicked (SecondaryLeft)"!() -> (OutputDelegate, then, Button:Common Button Base Object Reference) {is=title K2Node_ComponentBoundEvent}
   > n8 = FireEventElementInteractedByName(Element Name=Button_Previous_Page, WidgetCallingFrom=@n7.Button)
   n9 = "On Clicked (SecondaryRight)"!() -> (OutputDelegate, then, Button:Common Button Base Object Reference) {is=title K2Node_ComponentBoundEvent}
   > n10 = FireEventElementInteractedByName(Element Name=Button_Next_Page, WidgetCallingFrom=@n9.Button)
   n11 = "On UIKit Tab Selected (WBP_UIKit_TabGroup)"!() -> (OutputDelegate, then, TabID:Name=None, TabVM:MVVM Base Viewmodel Object Reference) {is=title K2Node_ComponentBoundEvent}
   > Branch(Condition=IsActivated!(self:Common Activatable Widget Object Reference) -> (ReturnValue:bool))
   > n12 = Cast To SeasonPassVM!(execute, Object=@n11.TabVM) -> (then, CastFailed, AsSeason Pass VM:Season Pass VM Object Reference) {is=title K2Node_DynamicCast}
   n13 = GetFirstUnlockedReward!(self:Season Pass VM Object Reference=@n12) -> (ReturnValue:Season Pass Reward VM Object Reference)
   > n14 = Branch(Condition=GetShouldShowNavigationTabs!(self:Season Pass Category VM Object Reference=GetOwnerCategory!(self:Season Pass Reward VM Object Reference=@n13) -> (ReturnValue:Season Pass Category VM Object Reference)) -> (ReturnValue:bool))
   > n15 = NavigateToReward!(execute, self:Season Pass Selection VM Object Reference=Get SeasonPassSelectionVM(), Reward:Season Pass Reward VM Object Reference=@n13) -> (then)
   RequestRefreshFocus(execute=[@n15.then, @n17.then], self:Common Activatable Widget Object Reference)
   n16 = Get Categories!(self:Season Pass VM Object Reference=@n12) -> (Categories:Array of Season Pass Category VM Object References)
   Branch(execute=@n14.else, Condition=Is Not Empty!(TargetArray:Array of Season Pass Category VM Object References=@n16) -> (ReturnValue:bool))
   > n17 = NavigateToReward(
    self=Get SeasonPassSelectionVM()
    Reward=GetFirstUnlockedReward(self:Season Pass Category VM Object Reference="Get (a copy)"(Array:Array of Season Pass Category VM Object References=@n16) -> (Output:Season Pass Category VM Object Reference))
   )
   n18 = RequestLoadBackground!() -> (OutputDelegate, then, BackgroundAsset:Soft Object Reference, Direction:EVaultWorldTransitionDirection Enum=Forward, FocalPointValue:double=~) {is=title K2Node_CustomEvent}
   > Set RequestedBackgroundSoftPath(RequestedBackgroundSoftPath:Soft Object Path Structure=To Soft Object Path!(SoftObjectReference:Soft Object Reference=@n18.BackgroundAsset) -> (ReturnValue:Soft Object Path Structure)) -> (Output_Get:Soft Object Path Structure)
   > n19 = Async Load Asset(Asset=@n18.BackgroundAsset)
   n20 = GetSoftObjectPath(Object=@n19)
   n21 = Get RequestedBackgroundSoftPath() -> (RequestedBackgroundSoftPath:Soft Object Path Structure)
   Branch(
    execute=@n19.Completed
    Condition="Equal Exactly (String)"(A=BreakSoftObjectPath(InSoftObjectPath=@n20), B=BreakSoftObjectPath(InSoftObjectPath=@n21))
   )
   > Branch(Condition=Get bIsActive() -> (bIsActive:bool))
   > UpdateBackground!(
    execute
    self:Fort Item Preview Background Object Reference=Get FortItemPreviewBackground() -> (FortItemPreviewBackground:Fort Item Preview Background Object Reference)
    LoadedBackground:Object Reference=@n19
    TransitionDirection:EVaultWorldTransitionDirection Enum=GetDesiredPendingDirection!(PendingDirection:EVaultWorldTransitionDirection Enum=@n18.Direction) -> (Result:EVaultWorldTransitionDirection Enum=Forward)
    PrimaryColor:Linear Color Structure=(R=1,G=1,B=1,A=0)
    SecondaryColor:Linear Color Structure=(R=0,G=0,B=0,A=0)
   ) -> (then)
   > SetScalarParameterValue!(
    execute
    Collection:Material Parameter Collection Object Reference=/Game/Environments/FrontEnd/Blueprints/ItemPreview/VaultWorldTransitions/MaterialCollectionParameters/MPC_UI_Master_VaultWorld_Settings
    ParameterName:Name=Is Horizontal Orientation
    ParameterValue:float=1
   ) -> (then) {is=title K2Node_CallMaterialParameterCollectionFunction}
   > n22 = Branch(Condition=IsAnimationPlaying(InAnimation=Get Intro()))
   SetVectorParameterValue!(
    execute=@n22.else
    Collection:Material Parameter Collection Object Reference=&MPC_UI_Master_VaultWorld_Settings
    ParameterName:Name=FocalPointPosition
    ParameterValue_R:float=@n18.FocalPointValue
    ParameterValue_G:float=0.1
    ParameterValue_B:float=0
    ParameterValue_A:float=1
   ) -> (then) {is=title K2Node_CallMaterialParameterCollectionFunction}
   n23 = CreateAndPushCustomBanner!() -> (OutputDelegate, then, SoftClass:User Widget Soft Class Reference, ItemVM:Fort Item VM Object Reference) {is=title K2Node_CustomEvent}
   > Branch(Condition=IsValidSoftClassReference!(SoftClassReference:"Object Soft Class Reference (by ref)"=@n23.SoftClass) -> (ReturnValue:bool))
   > n24 = CreateWidgetAsync!(
    execute:Exec
    UserWidgetSoftClass:User Widget Soft Class Reference=@n23.SoftClass
    OwningPlayer:Player Controller Object Reference=GetOwningPlayer()
    bSuspendInputUntilComplete:bool=true
   ) -> (then:Exec, OnComplete:Exec, OnFailure:Exec, UserWidget:User Widget Object Reference) {is=title K2Node_AsyncAction}
   AddChild(execute=@n24.OnComplete, self=Get NamedSlot_CustomBanner() -> (NamedSlot_CustomBanner:Named Slot Object Reference), Content=@n24)
   > n25 = Cast To WBP_CustomBanner_Base!(execute, Object=@n24) -> (then, CastFailed, AsWBP Custom Banner Base:WBP Custom Banner Base Object Reference) {is=title K2Node_DynamicCast}
   > SetViewType!(execute, self:WBP Custom Banner Base Object Reference=@n25, IsFullDetailsView:bool=true) -> (then)
   > SetContext!(execute, self:WBP Custom Banner Base Object Reference=@n25, Desired Area:E_MultiproductContext Enum=NewEnumerator3) -> (then)
   > SetSingleItem!(execute, self:WBP Custom Banner Base Object Reference=@n25, ItemVM:Fort Item VM Object Reference=@n23.ItemVM) -> (then)
   > SetActiveWidget(
    self=Get BonusInfo_VariantCard_Switcher() -> (BonusInfo_VariantCard_Switcher:Common Visibility Switcher Object Reference)
    Widget=Get WBP_CustomItemDetailsContainer() -> (WBP_CustomItemDetailsContainer:WBP Custom Item Details Container Object Reference)
   )
   n26 = Update Reward List!() -> (OutputDelegate, then, ItemList:Array of Season Pass Reward VM Object References, RewardToSelect:Season Pass Reward VM Object Reference) {is=title K2Node_CustomEvent}
   > Reset!(execute, self:Dynamic Entry Box Object Reference=Get ActiveRewardList(), bDeleteWidgets:bool=false) -> (then)
   > n27 = For Each Loop(Array:Array of Season Pass Reward VM Object References=@n26.ItemList) -> (Array Element:Season Pass Reward VM Object Reference)
   UpdateRewardsListWidth(execute=@n27.Completed)
   > RequestRefreshFocus()
   CreateSeasonPassReward!(execute=@n27.LoopBody, self, RewardVM:Season Pass Reward VM Object Reference=@n27) -> (then, AsSeason Pass Reward Button:Season Pass Reward Button Object Reference) {is=title K2Node_CallFunction}
   n28 = ViewRewardClicked() {is=title K2Node_CustomEvent}
   > n29 = Branch(Condition=IsAnimationPlaying(InAnimation=Get Intro()))
   SetViewingRewardState(execute=@n29.else, IsViewingReward=true)
   n30 = OnVariantFocused!() -> (OutputDelegate, then, Card:WBP UIKit Item Card Battle Pass Object Reference) {is=title K2Node_CustomEvent}
   > FireEventElementInteractedByName(Element Name=Button_Varient_Selected, WidgetCallingFrom=@n30.Card)
   > Call New Variant Card Selected!(execute:Exec, self:Season Pass Category Object Reference, NewCard:WBP UIKit Item Card Battle Pass Object Reference=@n30.Card) -> (then:Exec) {is=title K2Node_CallDelegate}
   n31 = Get FortItemVM!(self:WBP UIKit Item Card Battle Pass Object Reference=@n30.Card) -> (FortItemVM:Fort Item VM Object Reference)
   n32 = Get ItemDetails!(self:Fort Item VM Object Reference=@n31) -> (ItemDetails:Fort Item Details VM Object Reference)
   > Set FortItemDetailsVM!(
    execute
    FortItemDetailsVM:Fort Item Details VM Object Reference=@n32
    self:WBP UIKit Item Description Compact Object Reference=Get WBP_UIKit_ItemDescription_Compact() -> (WBP_UIKit_ItemDescription_Compact:WBP UIKit Item Description Compact Object Reference)
   ) -> (then, Output_Get:Fort Item Details VM Object Reference)
   n33 = Get ItemDefinition!(self:Fort Item VM Object Reference=@n31) -> (ItemDefinition:Fort Item Definition Object Reference)
   > Set ItemName!(
    execute
    ItemName:String="To String (Text)"(InText=Get ItemName!(self:Item Definition Base Object Reference=@n33) -> (ItemName:Text))
    self:WBP UIKit Item Description Compact Object Reference=Get WBP_UIKit_ItemDescription_Compact()
   ) -> (then, Output_Get:String)
   > SetFortItemDetailsVM!(
    execute
    self:WBP Creator Badge Object Reference=Get WBP_CreatorBadge() -> (WBP_CreatorBadge:WBP Creator Badge Object Reference)
    Viewmodel:Fort Item Details VM Object Reference=@n32
   ) -> (then)
   n34 = Get SeasonPassSelectionVM()
   > SetPreviewData!(
    execute
    self:Fort Item Preview Widget Object Reference=Get SeasonPassCategoryItemPreviewWidget()
    Data:"Fort Cosmetic Preview Data Structure (by ref)"=GetPreviewDataForItem!(
     self:Season Pass Reward VM Object Reference=Get SelectedReward!(self:Season Pass Selection VM Object Reference=@n34) -> (SelectedReward:Season Pass Reward VM Object Reference)
     ItemVM:Fort Item VM Object Reference=@n31
    ) -> (ReturnValue:Fort Cosmetic Preview Data Structure)
   ) -> (then)
   n35 = PushCustomItemDetails!() -> (OutputDelegate, then, WidgetClass:Common Activatable Widget Soft Class Reference, Item VM:Fort Item VM Object Reference) {is=title K2Node_CustomEvent}
   > n36 = PushPrimaryContentSoftClassAsync!(execute:Exec, SoftWidgetClass:Common Activatable Widget Soft Class Reference=@n35.WidgetClass) -> (then:Exec, OnPrimaryContentPushed:Exec, ActivatableWidget:Common Activatable Widget Object Reference) {is=title K2Node_AsyncAction}
   SetViewModelByClass!(
    execute=@n36.OnPrimaryContentPushed
    self:MVVM View Object Reference=GetExtension!(self:User Widget Object Reference=@n36, ExtensionType:User Widget Extension Class Reference=/Script/ModelViewViewModel.MVVMView) -> (ReturnValue:MVVM View Object Reference)
    NewValue:Notify Field Value Changed Interface=@n35.Item VM
   ) -> (then, ReturnValue:bool)
   Event On Deactivated()
   > SetZoomLevel!(
    execute
    self:Fort Item Preview Screen Object Reference=Get SeasonPassCategoryItemPreviewWidget()
    NewZoomLevel:float=0
    bSnap:bool=true
    bZoomEvenIfNotAllowed:bool=true
   ) -> (then)
   n37 = Get SeasonPassCategoryItemPreviewWidget()
   > DeactivateWidget(self=@n37)
   > FlushPendingScratchPadChanges!(execute) -> (then)
   Event Construct()
   > SetCompactUIStyle() {is=title K2Node_CallFunction}
   > Assign View Reward Button Clicked!(
    execute:Exec
    self:Season Pass Category Buttons Switcher Object Reference=Get SeasonPass_ButtonsSwitcher() -> (SeasonPass_ButtonsSwitcher:Season Pass Category Buttons Switcher Object Reference)
    Delegate:"Delegate (by ref)"=@n28
   ) -> (then:Exec) {is=title K2Node_AssignDelegate}
   > SetViewingRewardState(IsViewingReward=false)
   n38 = Get ToViewReward() -> (ToViewReward:Widget Animation Object Reference)
   n39 = Create Event()
   > UnbindFromAnimationFinished(Animation=@n38, Delegate=@n39)
   > BindToAnimationFinished(Animation=@n38, Delegate=@n39)
   n40 = Event PreConstruct()
   > n41 = Branch(Condition=@n40.IsDesignTime)
   UpdateActiveTrack(execute=[@n42.then, @n41.then], ActiveTrack=Get NamedSlot_Track_Current() -> (NamedSlot_Track_Current:Named Slot Object Reference))
   > SetMaxDesiredHeight!(
    execute
    self:Size Box Object Reference=Get SizeBoxTitle!(self:WBP UIKit Item Description Base Object Reference=Get WBP_UIKit_ItemDescription_Compact()) -> (SizeBoxTitle:Size Box Object Reference)
    InMaxDesiredHeight:float=250
   ) -> (then)
   n42 = PlayAnimation(execute=@n41.else, InAnimation=Get Intro())
   Event On Activated()
   > ActivateWidget(self=@n37)
   > SetPreviewData(
    self=@n37
    Data=GetPreviewData!(self:Season Pass Selection VM Object Reference=Get SeasonPassSelectionVM()) -> (ReturnValue:"Fort Cosmetic Preview Data Structure (by ref)")
   )
   n43 = Get SeasonPassSelectionVM()
   > n44 = Is Valid(InputObject=Get SelectedPage!(self:Season Pass Selection VM Object Reference=@n43) -> (SelectedPage:Season Pass Page VM Object Reference))
   n45 = NavigateToReward(
    execute=@n44.Is Not Valid
    self=@n43
    Reward=GetFirstUnlockedReward(self=Get SelectedCategory!(self:Season Pass Selection VM Object Reference=@n43) -> (SelectedCategory:Season Pass Category VM Object Reference))
   )
   n46 = Get SeasonPassSelectionVM()
   n47 = Get SelectedPage(self=@n46)
   n48 = Is Valid(exec=[@n45.then, @n44.Is Valid], InputObject=@n47)
   n49 = Set Page Index(execute=@n48.Is Not Valid, Page Index=0)
   UpdateItemPreviewFraming(execute=[@n50.then, @n49.then])
   > Fire Event Sub View Opened!(
    execute
    View:Widget Object Reference=Self-Reference()
    AdditionalInfo:Flow Analytics Additional Info Structure=Get Analytics Info Pass Name(MakeFlowAnalyticsAdditionalInfo=Make Flow Analytics Additional Info(), SeasonPassSelectedVM=Get SeasonPassSelectionVM())
    bIsOverlay:bool=false
   ) -> (then)
   > SetVisibility(self=Get Button_Back(), InVisibility=SelfHitTestInvisible)
   n50 = Initialize Selected Page And Reward!(
    execute=@n48.Is Valid
    self
    PassPageVM:Season Pass Page VM Object Reference=@n47
    RewardToSelect:Season Pass Reward VM Object Reference=Get SelectedReward(self=@n46)
   ) -> (then)
   VM_SelectedPassChanged!(
    execute
    self
    SelectedPass:Season Pass VM Object Reference=Get SelectedPass!(self:Season Pass Selection VM Object Reference=Get SeasonPassSelectionVM()) -> (SelectedPass:Season Pass VM Object Reference)
   ) -> (then)
  ]}
 Get Page Text {nodes [
   n1 = Get Page Text!() -> (then, Page Index:int) {is=title K2Node_FunctionEntry}
   n2 = Get SeasonPassCategoryVM()
   > n3 = Is Valid(InputObject=@n2)
   Return Node!(execute=@n3.Is Not Valid, Result:Text)
   n4 = Get Pages(self=@n2)
   > Return Node(Result=Format Text!(Format="LOCTEXT(\"{PageIndex}/{PageCount}\")", PageIndex:int=int + int!(A=@n1, B) -> (ReturnValue), PageCount:int=Length(TargetArray=@n4)) -> (Result))
  ]}
 Get Page Text Visibility {nodes [
   n1 = Get Page Text Visibility!() -> (then, Category:Season Pass Category VM Object Reference) {is=title K2Node_FunctionEntry}
   > n2 = Is Valid(InputObject=@n1)
   Return Node(execute=@n2.Is Not Valid, Result:ESlateVisibility Enum=Hidden)
   n3 = Length(TargetArray=Get Pages(self=@n1))
   > Return Node(Result=Select!(Option 0:ESlateVisibility Enum=Hidden, Option 1:ESlateVisibility Enum=SelfHitTestInvisible, Index:bool="integer > integer"(A=@n3, B=1)) -> (ReturnValue:ESlateVisibility Enum))
  ]}
 GetContainsBonusInfo {nodes [
   n1 = GetContainsBonusInfo!() -> (then, FortItemVM:Fort Item VM Object Reference) {is=title K2Node_FunctionEntry}
   n2 = Get ItemDefinition(self=Get CurrentItemVM!(self:Fort Item Preview Widget Object Reference=Get SeasonPassCategoryItemPreviewWidget()) -> (CurrentItemVM:Fort Item VM Object Reference))
   n3 = Get ItemDefinition(self=@n1)
   > Return Node!(
    execute
    ContainsBonusInfo:bool="Not Equal Exactly (String)"(A=GetTemplateId!(self:Mcp Item Definition Base Object Reference=@n3) -> (ReturnValue:String), B=GetTemplateId(self=@n2))
   )
  ]}
 GetDesiredPendingDirection {nodes [
   GetDesiredPendingDirection!() -> (then, PendingDirection:EVaultWorldTransitionDirection Enum=Forward) {is=title K2Node_FunctionEntry}
   > n1 = Branch(Condition=Get IsInitializingFromLandingPage() -> (IsInitializingFromLandingPage:bool))
   > Set IsInitializingFromLandingPage(IsInitializingFromLandingPage:bool=false) -> (Output_Get:bool)
   > Return Node!(execute, Result:EVaultWorldTransitionDirection Enum=None)
   Return Node(execute=@n1.else, Result=Get PendingDirection() -> (PendingDirection:EVaultWorldTransitionDirection Enum=Forward))
  ]}
 GetFramingDataFromItem {nodes [
   n1 = GetFramingDataFromItem!() -> (then, ItemVM:Fort Item VM Object Reference) {is=title K2Node_FunctionEntry}
   n2 = GetItemType!(ItemDefinition:Item Definition Base Object Reference=Get ItemDefinition(self=@n1)) -> (ReturnValue:EFortItemType Enum=WorldItem)
   n3 = Find!(
    TargetMap:Map of S Item Preview Framing Data Key Structures to S Item Preview Framing Data Values Structures=Get ItemPreviewFramingOverrides() -> (ItemPreviewFramingOverrides:Map of S Item Preview Framing Data Key Structures to S Item Preview Framing Data Values Structures)
    Key:"S Item Preview Framing Data Key Structure (by ref)"=Make S Item Preview Framing Data Key!(
     ItemType_2_#30:EFortItemType Enum=@n2
     ProductTag_5_#31:Gameplay Tag Structure=GetCurrentProduct!(self:Fort Item VM Object Reference=@n1) -> (ReturnValue:Gameplay Tag Structure)
    ) -> (S_ItemPreviewFramingData_Key:S Item Preview Framing Data Key Structure) {is=title K2Node_MakeStruct}
   ) -> (
    Value_LandingData_16_#32:S Item Preview Framing Data Structure
    Value_ViewRewardData_18_#33:S Item Preview Framing Data Structure
    ReturnValue
   )
   > n4 = Branch(Condition=@n3.ReturnValue)
   > Return Node!(
    execute
    FramingData:S Item Preview Framing Data Structure=Select(
     Option 0:S Item Preview Framing Data Structure=@n3
     Option 1:S Item Preview Framing Data Structure=@n3.Value_ViewRewardData_18_#33
     Index=Get IsViewingReward()
    ) -> (ReturnValue:S Item Preview Framing Data Structure)
   )
   n5 = Find(
    TargetMap=Get ItemPreviewFramingOverrides()
    Key=Make S Item Preview Framing Data Key(ItemType_2_#30=@n2, ProductTag_5_#31=(TagName=""))
   )
   n6 = Branch(execute=@n4.else, Condition=@n5.ReturnValue)
   > Return Node(FramingData=Select(Option 0=@n5, Option 1=@n5.Value_ViewRewardData_18_#33, Index=Get IsViewingReward()))
   Return Node(
    execute=@n6.else
    FramingData=Select(
     Option 0=Get FallbackFramingData_Landing() -> (FallbackFramingData_Landing:S Item Preview Framing Data Structure)
     Option 1=Get FallbackFramingData_Landing_ViewReward() -> (FallbackFramingData_Landing_ViewReward:S Item Preview Framing Data Structure)
     Index=Get IsViewingReward()
    )
   )
  ]}
 GetFrontendFlowInterruptConfig {nodes [
   GetFrontendFlowInterruptConfig() {is=title K2Node_FunctionEntry}
   > Return Node!(execute, ReturnValue:Frontend Flow Interrup Config Structure)
  ]}
 GetPage {nodes [
   n1 = GetPage!() -> (then, PageIndex:int, Category:Season Pass Category VM Object Reference) {is=title K2Node_FunctionEntry}
   > n2 = Is Valid(InputObject=@n1.Category)
   Return Node!(execute=@n2.Is Not Valid, PassPageVM:Season Pass Page VM Object Reference)
   > Return Node(PassPageVM="Get (a copy)"(Array:Array of Season Pass Page VM Object References=Get Pages(self=@n1.Category), Dimension 1=@n1) -> (Output:Season Pass Page VM Object Reference))
  ]}
 GetPrimaryContentSetup {nodes [
   GetPrimaryContentSetup() {is=title K2Node_FunctionEntry}
   > Return Node!(
    execute
    ReturnValue:Primary Content Setup Structure=Make Primary Content Setup!(
     bShowBottomBar:bool=false
     bShowChatWidget:bool=false
     bEnableNewTextChat:bool=False
     bShowTopBar:bool=false
     bTabSupportsInterface:bool=False
    ) -> (PrimaryContentSetup:Primary Content Setup Structure) {is=title K2Node_MakeStruct}
   )
  ]}
 GetRewardsInPage {nodes [
   n1 = GetRewardsInPage!() -> (then, Page:Season Pass Page VM Object Reference) {is=title K2Node_FunctionEntry}
   > n2 = Is Valid(InputObject=@n1)
   Return Node!(
    execute=@n2.Is Not Valid
    Rewards:Array of Season Pass Reward VM Object References=Make Array!() -> (Array:Array of Season Pass Reward VM Object References)
   )
   n3 = Get Tracks!(self:Season Pass Page VM Object Reference=@n1) -> (Tracks:Array of Season Pass Track VM Object References)
   > Return Node(Rewards=Get Rewards!(self:Season Pass Track VM Object Reference="Get (a copy)"(Array:Array of Season Pass Track VM Object References=@n3) -> (Output:Season Pass Track VM Object Reference)) -> (Rewards:Array of Season Pass Reward VM Object References))
  ]}
 HandleRewardByLevelCoachMark {nodes [
   HandleRewardByLevelCoachMark() {is=title K2Node_FunctionEntry}
   > n1 = Branch(Condition=ShouldShowOnboarding!(self:Fort Sidebar Onboard Tooltip Widget Object Reference=Get SeasonPass_GrantRewardByLevel_Coachmark() -> (SeasonPass_GrantRewardByLevel_Coachmark:Season Pass Grant Reward By Level Coachmark Object Reference)) -> (ReturnValue:bool))
   n2 = Get SeasonPassSelectionVM()
   > n3 = Branch(Condition=Get bDoesOwnsPass!(self:Season Pass VM Object Reference=Get SelectedPass(self=@n2)) -> (bDoesOwnsPass:bool))
   n4 = Get SeasonPass_GrantRewardByLevel_Coachmark()
   Branch(execute=[@n6.Is Not Valid, @n3.then, @n1.else], Condition=IsActivated(self=@n4))
   > SetVisibility(self=@n4, InVisibility=Collapsed)
   n5 = GetFirstFreeRewardClaimableByReachingLevel!(self:Season Pass Page VM Object Reference=Get SelectedPage(self=@n2)) -> (ReturnValue:Season Pass Reward VM Object Reference)
   n6 = Is Valid(exec=@n3.else, InputObject=@n5)
   > SetFortItemVM(
    self:Season Pass Grant Reward By Level Coachmark Object Reference=Get SeasonPass_GrantRewardByLevel_Coachmark()
    Viewmodel="Get (a copy)"(Array:Array of Fort Item VM Object References=Get ItemVMs!(self:Season Pass Reward VM Object Reference=@n5) -> (ItemVMs:Array of Fort Item VM Object References)) -> (Output:Fort Item VM Object Reference)
   )
   > ActivateWidget(self=Get SeasonPass_GrantRewardByLevel_Coachmark())
  ]}
 IncreasePage {nodes [
   IncreasePage() {is=title K2Node_FunctionEntry}
   n1 = Get IsViewingReward()
   n2 = Get SeasonPassCategoryVM()
   n3 = int - int(A=Length(TargetArray=Get Pages(self=@n2)))
   n4 = Get PageIndex() -> (PageIndex=0)
   > Branch(Condition=AND Boolean(A="integer < integer"(A=@n4, B=@n3), B=NOT Boolean(A=@n1)))
   n5 = Create Event()
   > UnbindFromAnimationFinished(Animation=Get NextPage(), Delegate=@n5)
   > UnbindFromAnimationFinished(Animation=Get PreviousPage(), Delegate=@n5)
   > UpdateActiveTrack(ActiveTrack=Get NamedSlot_Track_Next())
   > Set Visibility(Visibility=SelfHitTestInvisible, self=Get NamedSlot_Track_Next())
   n6 = Length(TargetArray=Get Pages(self=Get SeasonPassCategoryVM()))
   n7 = int + int(A=Get PageIndex() -> (PageIndex=0))
   > Branch(Condition=OR Boolean(A=Get EnableNavigationPageWrapping(), B="integer < integer"(A=@n7, B=@n6)))
   > Set Page Index(Page Index="% (Integer)"(A=@n7, B=@n6) -> (ReturnValue=0))
   > BindToAnimationFinished(Animation=Get NextPage(), Delegate=Create Event())
   > n8 = Branch(Condition=IsAnimationPlaying(InAnimation=Get PreviousPage()))
   > n9 = StopAnimation(InAnimation=Get PreviousPage())
   PlayAnimationForward(execute=[@n9.then, @n8.else], InAnimation=Get NextPage())
  ]}
 Initialize Selected Page And Reward {nodes [
   n1 = Initialize Selected Page And Reward!() -> (then, PassPageVM:Season Pass Page VM Object Reference, RewardToSelect:Season Pass Reward VM Object Reference) {is=title K2Node_FunctionEntry}
   > Is Valid(InputObject=@n1)
   n2 = Get Pages(self=Get SeasonPassCategoryVM())
   n3 = Get PassPageVM() -> (PassPageVM:Season Pass Page VM Object Reference)
   > Set PageIndex(PageIndex:int="Max (Integer)"(
     A=Find Item!(TargetArray:Array of Season Pass Page VM Object References=@n2, ItemToFind:"Season Pass Page VM Object Reference (by ref)"=@n3) -> (ReturnValue)
     B=0
    )) -> (Output_Get:int=0)
   > UpdateNavigationArrows()
   > UpdatePaginationText()
   > Set Local_RewardsInPage(Local_RewardsInPage:Array of Season Pass Reward VM Object References=GetRewardsInPage!(self, Page:Season Pass Page VM Object Reference=Get PassPageVM()) -> (Rewards:Array of Season Pass Reward VM Object References) {is=title K2Node_CallFunction}) -> (Output_Get:Array of Season Pass Reward VM Object References)
   n4 = Find Item(
    TargetArray:Array of Season Pass Reward VM Object References=Get Local_RewardsInPage() -> (Local_RewardsInPage:Array of Season Pass Reward VM Object References)
    ItemToFind:"Season Pass Reward VM Object Reference (by ref)"=Get RewardToSelect() -> (RewardToSelect:Season Pass Reward VM Object Reference)
   )
   > n5 = Branch(Condition="Equal (Integer)"(A=@n4, B=-1))
   n6 = Get Local_RewardsInPage()
   > n7 = Set SelectedRewardIndex(SelectedRewardIndex:int=Select!(
     Forward:int=0
     Backward:int=Last Index!(TargetArray:Array of Season Pass Reward VM Object References=@n6) -> (ReturnValue=0)
     None:int=0
     Index:EVaultWorldTransitionDirection Enum=Get LastScrolledDirection() -> (LastScrolledDirection:EVaultWorldTransitionDirection Enum=Forward)
    ) -> (ReturnValue:int)) -> (Output_Get:int=0)
   Set SelectedReward!(
    execute=[@n7.then, @n8.then]
    SelectedReward:Season Pass Reward VM Object Reference="Get (a copy)"(
     Array:Array of Season Pass Reward VM Object References=Get Local_RewardsInPage()
     Dimension 1=Get SelectedRewardIndex() -> (SelectedRewardIndex=0)
    ) -> (Output:Season Pass Reward VM Object Reference)
    self:Season Pass Selection VM Object Reference=Get SeasonPassSelectionVM()
   ) -> (then, Output_Get:Season Pass Reward VM Object Reference)
   > Update Reward List!(
    execute
    self
    ItemList:Array of Season Pass Reward VM Object References=Get Local_RewardsInPage()
    RewardToSelect:Season Pass Reward VM Object Reference=Get RewardToSelect()
   ) -> (then) {is=title K2Node_CallFunction}
   n8 = Set SelectedRewardIndex(execute=@n5.else, SelectedRewardIndex=@n4) -> (Output_Get=0)
  ]}
 InitializeCustomItemDetails {nodes [
   n1 = InitializeCustomItemDetails!() -> (then, ItemVM:Fort Item VM Object Reference) {is=title K2Node_FunctionEntry}
   > ClearChildren(self=Get NamedSlot_CustomBanner())
   n2 = Get ItemDetails(self=@n1)
   > CreateCustomItemDetails!(
    execute
    self:WBP Custom Item Details Container Object Reference=Get WBP_CustomItemDetailsContainer()
    Item Details VM:Fort Item Details VM Object Reference=@n2
   ) -> (then)
   n3 = Get Item!(self:Fort Item Details VM Object Reference=@n2) -> (Item:Fort Item VM Object Reference)
   > CreateAndPushCustomBanner!(
    execute
    self
    SoftClass:User Widget Soft Class Reference=Get CustomInspectScreenBanner!(self:Fort Locker Category Config Object Reference=GetItemCategory!(self:Fort Item VM Object Reference=@n3) -> (ReturnValue:Fort Locker Category Config Object Reference)) -> (CustomInspectScreenBanner:User Widget Soft Class Reference)
    ItemVM:Fort Item VM Object Reference=Get ItemVM()
   ) -> (then) {is=title K2Node_CallFunction}
  ]}
 NewVariantCardSelected {nodes [NewVariantCardSelected!() -> (then, NewCard:WBP UIKit Item Card Battle Pass Object Reference) {is=title K2Node_FunctionEntry}]}
 OnMouseWheel {nodes [
   n1 = OnMouseWheel!() -> (then, MyGeometry:Geometry Structure, MouseEvent:"Pointer Event Structure (by ref)") {is=title K2Node_FunctionEntry}
   > n2 = Branch(Condition=Get IsViewingReward())
   > Return Node!(execute, ReturnValue:Event Reply Structure=Unhandled())
   n3 = Branch(
    execute=@n2.else
    Condition="float > float"(A=Get Wheel Delta!(Input:"Pointer Event Structure (by ref)"=@n1.MouseEvent) -> (ReturnValue:float) {is=title K2Node_CallFunction})
   )
   > n4 = DecreasePage()
   Return Node(execute=[@n4.then, @n5.then], ReturnValue=Handled())
   n5 = IncreasePage(execute=@n3.else) {is=title K2Node_CallFunction}
  ]}
 OnPageTransitionAnimationFinished {nodes [
   OnPageTransitionAnimationFinished() {is=title K2Node_FunctionEntry}
   > SwapTracks()
   n1 = Get NamedSlot_Track_Current()
   > SetRenderOpacity(self=@n1, InOpacity=1)
   > SetRenderTranslation(self=@n1)
   > SetRenderOpacity(self=Get NamedSlot_Track_Next(), InOpacity=0)
   > SetVisibility(self=Get NamedSlot_Track_Next(), InVisibility=Hidden)
   > PlayAnimationForward(InAnimation=Get ShowStatusBar() -> (ShowStatusBar:Widget Animation Object Reference))
   > RequestRefreshFocus()
   > Set LastScrolledDirection(LastScrolledDirection:EVaultWorldTransitionDirection Enum=None) -> (Output_Get:EVaultWorldTransitionDirection Enum=Forward)
  ]}
 OnRewardsLeftRightNavFailed {nodes [
   n1 = OnRewardsLeftRightNavFailed!() -> (then, Navigation:EUINavigation Enum) {is=title K2Node_FunctionEntry}
   > n2 = Switch on EUINavigation!(execute, Selection:EUINavigation Enum=@n1) -> (Left:Exec, Right:Exec, Up:Exec, Down:Exec, Next:Exec, Previous:Exec, Invalid:Exec) {is=title K2Node_SwitchEnum}
   > n3 = Branch(Condition="integer > integer"(A=Get PageIndex() -> (PageIndex=0)))
   > Set LastScrolledDirection(LastScrolledDirection=Backward) -> (Output_Get=Forward)
   > n4 = DecreasePage()
   Return Node(execute=[@n4.then, @n7.then], ReturnValue:Widget Object Reference)
   Return Node(
    execute=@n3.else
    ReturnValue=GetDesiredFocusTarget!(self:Common Activatable Widget Object Reference) -> (ReturnValue:Widget Object Reference)
   )
   n5 = Length(TargetArray=Get Pages(self=Get SeasonPassCategoryVM()))
   n6 = Branch(execute=@n2.Right, Condition="integer < integer"(A=Get PageIndex() -> (PageIndex=0), B=int - int(A=@n5)))
   > Set LastScrolledDirection(LastScrolledDirection=Forward) -> (Output_Get=Forward)
   > n7 = IncreasePage()
   Return Node(execute=@n6.else, ReturnValue=GetDesiredFocusTarget())
  ]}
 "OnToViewRewardAnimationFinished " {nodes [
   "OnToViewRewardAnimationFinished "() {is=title K2Node_FunctionEntry}
   n1 = Get IsViewingReward()
   > SetVisibility(
    self=Get TrackAndStatusBarContainer() -> (TrackAndStatusBarContainer:Vertical Box Object Reference)
    InVisibility=Select!(Option 0:ESlateVisibility Enum=SelfHitTestInvisible, Option 1:ESlateVisibility Enum=Collapsed, Index:bool=@n1) -> (ReturnValue:ESlateVisibility Enum)
   )
   > SetVisibility(
    self=[Get NamedSlot_CustomBanner(), Get WBP_CustomItemDetailsContainer()]
    InVisibility=Select(Option 0=Collapsed, Option 1=SelfHitTestInvisible, Index=@n1)
   )
   > n2 = Branch(Condition=Get IsViewingReward())
   > n3 = Get!(execute:Exec, self:Fort Item Preview Widget Object Reference=Get SeasonPassCategoryItemPreviewWidget()) -> (then:Exec, else:Exec, CurrentItemVM:Fort Item VM Object Reference)
   > InitializeCustomItemDetails!(execute, self, ItemVM:Fort Item VM Object Reference=@n3) -> (then) {is=title K2Node_CallFunction}
   n4 = Get WBP_CustomItemDetailsContainer()
   n5 = Get NamedSlot_CustomDetails!(self:WBP Custom Item Details Container Object Reference=@n4) -> (NamedSlot_CustomDetails:Named Slot Object Reference)
   > n6 = SetActiveWidget(
    self=Get BonusInfo_VariantCard_Switcher()
    Widget=Select(
     Option 0:Widget Object Reference=Get VariantCards()
     Option 1:Widget Object Reference=Get WBP_CustomItemDetailsContainer()
     Index=HasAnyChildren!(self:Panel Widget Object Reference=@n5) -> (ReturnValue:bool)
    ) -> (ReturnValue:Widget Object Reference)
   )
   RequestRefreshFocus(execute=[@n6.then, @n7.then])
   n7 = RemoveCustomItemDetails(execute=@n2.else, self:WBP Custom Item Details Container Object Reference=@n4)
  ]}
 Sequencer Events {nodes [
   OnClosePass() {is=title K2Node_CustomEvent}
   > PopContentWidget!(execute) -> (then)
   ReinitializeSelectedPage() {is=title K2Node_CustomEvent}
   n1 = Get SeasonPassSelectionVM()
   > Initialize Selected Page And Reward(PassPageVM=Get SelectedPage(self=@n1), RewardToSelect=Get SelectedReward(self=@n1)) {is=title K2Node_CallFunction}
  ]}
 Set Page Index {nodes [
   n1 = Set Page Index!() -> (then, Page Index:int) {is=title K2Node_FunctionEntry}
   > Set PageIndex(PageIndex=@n1) -> (Output_Get=0)
   n2 = Get SeasonPassCategoryVM()
   > Set SelectedPage!(
    execute
    SelectedPage:Season Pass Page VM Object Reference="Get (a copy)"(Array:Array of Season Pass Page VM Object References=Get Pages(self=@n2), Dimension 1=Get PageIndex() -> (PageIndex=0)) -> (Output:Season Pass Page VM Object Reference)
    self:Season Pass Selection VM Object Reference=Get SeasonPassSelectionVM()
   ) -> (then, Output_Get:Season Pass Page VM Object Reference)
  ]}
 SetCompactUIStyle {nodes [
   SetCompactUIStyle() {is=title K2Node_FunctionEntry}
   > Branch(Condition=IsUsingCompactUI())
   > n1 = Sequence!(execute) -> (then_0, then_1, then_2:Exec, then_3:Exec)
   > SetMaxDesiredWidth!(
    execute
    self:Size Box Object Reference=Get SizeBox_LeftContent() -> (SizeBox_LeftContent:Size Box Object Reference)
    InMaxDesiredWidth:float=640
   ) -> (then)
   > SetWidthOverride(self=Get SizeBox_RightContent() -> (SizeBox_RightContent:Size Box Object Reference), InWidthOverride=456)
   > SetSize!(
    execute
    self:Vertical Box Slot Object Reference=SlotAsVerticalBoxSlot(Widget=Get LegalandActionBar() -> (LegalandActionBar:Overlay Object Reference))
    InSize_Value:float=0.16
    InSize_SizeRule:ESlateSizeRule Enum=Fill
   ) -> (then)
   SetPadding!(
    execute=@n1.then_1
    self:Vertical Box Slot Object Reference=SlotAsVerticalBoxSlot(Widget=Get Overlay_NavigationArrows() -> (Overlay_NavigationArrows:Overlay Object Reference))
    InPadding_Left:float=0
    InPadding_Top:float=24
    InPadding_Right:float=0
    InPadding_Bottom:float=12
   ) -> (then)
   > SetVisibility(
    self=[
     Get SecondaryCurrentPageText() -> (SecondaryCurrentPageText:Common Text Object Reference)
     Get SecondaryNavigationArrows() -> (SecondaryNavigationArrows:Overlay Object Reference)
    ]
    InVisibility=SelfHitTestInvisible
   )
   > SetVisibility(self=Get Overlay_NavigationArrows(), InVisibility=Collapsed)
   SetPadding(
    execute=@n1.then_2
    self:Size Box Slot Object Reference=SlotAsSizeBoxSlot!(Widget:Widget Object Reference=Get Overlay_Content() -> (Overlay_Content:Overlay Object Reference)) -> (ReturnValue:Size Box Slot Object Reference)
    InPadding_Left=48
    InPadding_Top=24
    InPadding_Right=48
    InPadding_Bottom=24
   )
   > SetPadding(
    self=SlotAsSizeBoxSlot(Widget=Get SeasonPassCategoryItemPreviewWidget())
    InPadding_Left=0
    InPadding_Top=0
    InPadding_Right=0
    InPadding_Bottom=0
   )
   Branch(
    execute=@n1.then_3
    Condition=CompareToScreenAspectRatio!(Comparison:EAspectRatioComparison Enum=GreaterThan, AspectRatio:EAspectRatio Enum=Ratio_4_3) -> (ReturnValue:bool)
   )
   > SetPadding(
    self:Grid Slot Object Reference=SlotAsGridSlot(Widget=Get WBP_UIKit_TabGroup() -> (WBP_UIKit_TabGroup:WBP UIKit Tab Group Object Reference))
    InPadding_Left=0
    InPadding_Top=0
    InPadding_Right=0
    InPadding_Bottom=-48
   )
   > SetPadding(
    self:Size Box Slot Object Reference=SlotAsSizeBoxSlot(Widget=Get SeasonPass_ButtonsSwitcher())
    InPadding_Left=0
    InPadding_Top=12
    InPadding_Right=0
    InPadding_Bottom=0
   )
   > SetPadding(
    self:Vertical Box Slot Object Reference=SlotAsVerticalBoxSlot(Widget=Get TrackContainer() -> (TrackContainer:Grid Panel Object Reference))
    InPadding_Left=0
    InPadding_Top=0
    InPadding_Right=0
    InPadding_Bottom=16
   )
  ]}
 SetIsInitializingFromLandingPage {nodes [
   n1 = SetIsInitializingFromLandingPage!() -> (then, IsInitializingFromLandingPage:bool=~) {is=title K2Node_FunctionEntry}
   > Set IsInitializingFromLandingPage(IsInitializingFromLandingPage=@n1)
  ]}
 SetViewRewardAnimValue {nodes [
   SetViewRewardAnimValue!() -> (then, AnimationValue:double=~) {is=title K2Node_FunctionEntry}
   > Get(self=Get SeasonPassCategoryItemPreviewWidget())
   n1 = Get!() -> (
    TargetFramingData_Offset_16_#32_X:double
    TargetFramingData_Offset_16_#32_Y:double
    TargetFramingData_ZoomLevel_15_#34:double=~
    TargetFramingData_CameraMode_19_#35:EPreviewCameraPositionMode Enum
   )
   n2 = Get!() -> (
    CurrentItemFramingData_Offset_16_#32_X:double
    CurrentItemFramingData_Offset_16_#32_Y:double
    CurrentItemFramingData_ZoomLevel_15_#34:double=~
    CurrentItemFramingData_CameraMode_19_#35:EPreviewCameraPositionMode Enum
   )
   n3 = "Lerp (Vector)"!(
    A_X:double=@n2
    A_Y:double=@n2.CurrentItemFramingData_Offset_16_#32_Y
    A_Z:double=@n2.CurrentItemFramingData_ZoomLevel_15_#34
    B_X:double=@n1
    B_Y:double=@n1.TargetFramingData_Offset_16_#32_Y
    B_Z:double=@n1.TargetFramingData_ZoomLevel_15_#34
    Alpha:float=Get AnimationValue() -> (AnimationValue:double)
   ) -> (ReturnValue_X:double, ReturnValue_Y:double, ReturnValue_Z:double)
   > Set!(
    execute
    CurrentItemFramingData_Offset_16_#32_X:double=@n3
    CurrentItemFramingData_Offset_16_#32_Y:double=@n3.ReturnValue_Y
    CurrentItemFramingData_ZoomLevel_15_#34:double=@n3.ReturnValue_Z
    CurrentItemFramingData_CameraMode_19_#35:EPreviewCameraPositionMode Enum=Configured
   ) -> (
    then
    Output_Get_Offset_16_#32:Vector 2D Structure
    Output_Get_ZoomLevel_15_#34:double=~
    Output_Get_CameraMode_19_#35:EPreviewCameraPositionMode Enum
   )
   n4 = Get() -> (CurrentItemFramingData_ZoomLevel_15_#34=~)
   > SetZoomLevel(
    self=Get SeasonPassCategoryItemPreviewWidget()
    NewZoomLevel=@n4.CurrentItemFramingData_ZoomLevel_15_#34
    bSnap=true
    bZoomEvenIfNotAllowed=true
   )
   > SetRenderTranslation!(
    execute
    self=Get SeasonPassCategoryItemPreviewWidget()
    Translation_X:double=@n4
    Translation_Y:double=@n4.CurrentItemFramingData_Offset_16_#32_Y
   ) -> (then)
  ]}
 SetViewingRewardState {nodes [
   SetViewingRewardState!() -> (then, IsViewingReward:bool=~) {is=title K2Node_FunctionEntry}
   > n1 = Branch(Condition=IsAnimationPlaying(InAnimation=Get Close()))
   n2 = Get CurrentItemVM(self=Get SeasonPassCategoryItemPreviewWidget())
   n3 = TryGetSoftCustomActivatableWidgetClass!(
    self:Fort Locker Category Config Object Reference=GetItemCategory(self=@n2)
    ViewRequest:Gameplay Tag Structure=(TagName="UI.ItemShop.ItemDetails")
   ) -> (WidgetClass:Common Activatable Widget Soft Class Reference, ReturnValue:bool)
   n4 = Branch(execute=@n1.else, Condition=@n3.ReturnValue)
   > Branch(Condition=Get IsViewingReward())
   > PushCustomItemDetails!(execute, self, WidgetClass:Common Activatable Widget Soft Class Reference=@n3, Item VM:Fort Item VM Object Reference=@n2) -> (then) {is=title K2Node_CallFunction}
   n5 = Get ItemDefinition(self=Get CurrentItemVM(self=Get SeasonPassCategoryItemPreviewWidget()))
   n6 = Branch(
    execute=@n4.else
    Condition="Equal (Enum)"(A:EFortItemType Enum=GetItemType(ItemDefinition=@n5) -> (ReturnValue=WorldItem), B:EFortItemType Enum=AthenaLoadingScreen)
   )
   > PreviewAction!(
    execute
    self:Fort Item Preview Screen Object Reference=Get SeasonPassCategoryItemPreviewWidget()
    ActionName:Name=SpecialPreviewAction_ViewLoadingScreen
   ) -> (then)
   Set IsViewingReward(execute=@n6.else, IsViewingReward:bool=Get IsViewingReward()) -> (Output_Get:bool)
   > n7 = Sequence!(execute) -> (then_0, then_1, then_2)
   n8 = Select(Option 0:ESlateVisibility Enum=SelfHitTestInvisible, Option 1:ESlateVisibility Enum=Collapsed, Index=Get IsViewingReward()) -> (ReturnValue:ESlateVisibility Enum)
   SetVisibility(
    execute=@n7.then_1
    self=[Get SeasonPass_Level() -> (SeasonPass_Level:Season Pass Level Object Reference), Get SeasonPass_ButtonsSwitcher()]
    InVisibility=@n8
   )
   n9 = Get SeasonPassCategoryVM()
   > SetVisibility(self=Get WBP_UIKit_TabGroup(), InVisibility=Select(Option 0=Collapsed, Option 1=@n8, Index=GetShouldShowNavigationTabs(self=@n9)))
   > SetIsEnabled(self=Get ActiveRewardList(), bInIsEnabled=NOT Boolean(A=Get IsViewingReward()))
   > SetVisibility(
    self=Get BonusInfo_VariantCard_Switcher()
    InVisibility=Select(Option 0=Collapsed, Option 1=SelfHitTestInvisible, Index=Get IsViewingReward())
   )
   SetVisibility(
    execute=@n7.then_0
    self=Get ActionButtonsWrapper() -> (ActionButtonsWrapper:Horizontal Box Object Reference)
    InVisibility=Select(Option 0=Collapsed, Option 1=SelfHitTestInvisible, Index=Get IsViewingReward())
   )
   > SetIsInteractionEnabled!(
    execute
    self:Common Button Base Object Reference=[Get Right() -> (Right:WBP UIKit Button Generic Object Reference), Get Left() -> (Left:WBP UIKit Button Generic Object Reference)]
    bInIsInteractionEnabled:bool=NOT Boolean(A=Get IsViewingReward())
   ) -> (then)
   PlayAnimation(
    execute=@n7.then_2
    InAnimation=Get ToViewReward()
    PlayMode=Select(Option 0:EUMGSequencePlayMode Enum=Reverse, Option 1:EUMGSequencePlayMode Enum, Index=Get IsViewingReward()) -> (ReturnValue:EUMGSequencePlayMode Enum)
    PlaybackSpeed=2
   )
   > SetAllowZooming!(
    execute
    self:Fort Item Preview Screen Object Reference=Get SeasonPassCategoryItemPreviewWidget()
    bInAllowZooming:bool=Get IsViewingReward()
   ) -> (then)
   > Set ScreenTag!(
    execute
    ScreenTag:Gameplay Tag Structure=Select(
     Option 0:Gameplay Tag Structure=(TagName="UI.ItemPreview.Screen.BattlePass.RewardTrack")
     Option 1:Gameplay Tag Structure=(TagName="UI.ItemPreview.Screen.BattlePass.ViewReward")
     Index=Get IsViewingReward()
    ) -> (ReturnValue:Gameplay Tag Structure)
    self:Fort Item Preview Screen Object Reference=Get SeasonPassCategoryItemPreviewWidget()
   ) -> (then, Output_Get:Gameplay Tag Structure)
   > SetPreviewData(self=Get SeasonPassCategoryItemPreviewWidget(), Data=GetPreviewData(self=Get SeasonPassSelectionVM()))
   > UpdateItemPreviewFraming()
   > Call Viewing Reward State Changed!(execute:Exec, self:Season Pass Category Object Reference, IsViewingReward:bool=Get IsViewingReward()) -> (then:Exec) {is=title K2Node_CallDelegate}
  ]}
 SetupPassesTabs {nodes [
   n1 = SetupPassesTabs!() -> (then, SeasonPasses:Array of Season Pass VM Object References) {is=title K2Node_FunctionEntry}
   > n2 = For Each Loop(Array:Array of Season Pass VM Object References=@n1) -> (Array Element:Season Pass VM Object Reference)
   n3 = String To Name(InString=Get Source String from Text!(Text:Text=GetDisplayName!(self:Season Pass VM Object Reference=@n2) -> (ReturnValue:Text)) -> (ReturnValue:String) {is=title K2Node_CallFunction})
   > Branch(Condition=OR Boolean!(A="Equal (Name)"(A=@n3, B="LEGO® Pass"), B="Equal (Name)"(A=@n3, B=MUSIC Pass), C:bool="Equal (Name)"(A=@n3, B=OG Pass)) -> (ReturnValue))
   > AddTab!(
    execute
    self:WBP UIKit Tab Group Base Object Reference=Get WBP_UIKit_TabGroup()
    Tab VM:MVVM Base Viewmodel Object Reference=@n2
    Tab ID:Name=@n3
    Index:int=-1
   ) -> (then)
  ]}
 SwapTracks {nodes [
   SwapTracks() {is=title K2Node_FunctionEntry}
   > Set Content_Current(Content_Current:Widget Object Reference=GetChildAt!(self:Panel Widget Object Reference=Get NamedSlot_Track_Current(), Index:int=0) -> (ReturnValue:Widget Object Reference)) -> (Output_Get:Widget Object Reference)
   > Set Content_Next(Content_Next:Widget Object Reference=GetChildAt(self=Get NamedSlot_Track_Next(), Index=0)) -> (Output_Get:Widget Object Reference)
   > ClearChildren(self=[Get NamedSlot_Track_Next(), Get NamedSlot_Track_Current()])
   > AddChild(self=Get NamedSlot_Track_Current(), Content=Get Content_Next() -> (Content_Next:Widget Object Reference))
   > AddChild(self=Get NamedSlot_Track_Next(), Content=Get Content_Current() -> (Content_Current:Widget Object Reference))
   > UpdateActiveTrack(ActiveTrack=Get NamedSlot_Track_Current())
   n1 = Create Event()
   n2 = Get ActiveRewardList()
   > SetNavigationRuleCustomBoundary!(execute, self:Widget Object Reference=@n2, Direction:EUINavigation Enum=Left, InCustomDelegate:Delegate=@n1) -> (then)
   > SetNavigationRuleCustomBoundary(self=@n2, Direction=Right, InCustomDelegate=@n1)
  ]}
 UpdateActiveTrack {nodes [
   n1 = UpdateActiveTrack!() -> (then, ActiveTrack:Named Slot Object Reference) {is=title K2Node_FunctionEntry}
   > n2 = Set ActiveTrack(ActiveTrack:Named Slot Object Reference=@n1) -> (Output_Get:Named Slot Object Reference)
   > n3 = Cast To DynamicEntryBox!(execute, Object=GetChildAt(self=@n2, Index=0)) -> (then, CastFailed, AsDynamic Entry Box:Dynamic Entry Box Object Reference) {is=title K2Node_DynamicCast}
   > Set ActiveRewardList(ActiveRewardList:Dynamic Entry Box Object Reference=@n3) -> (Output_Get:Dynamic Entry Box Object Reference)
   Set ActiveRewardList(execute=@n3.CastFailed)
  ]}
 UpdateCustomBannerLeft {nodes [
   n1 = UpdateCustomBannerLeft!() -> (then, ItemVM:Fort Item VM Object Reference) {is=title K2Node_FunctionEntry}
   > ClearChildren(self=Get NamedSlot_CustomBanner_Left() -> (NamedSlot_CustomBanner_Left:Named Slot Object Reference))
   n2 = TryGetSoftCustomWidgetClass!(
    self:Fort Locker Category Config Object Reference=GetItemCategory(self=@n1)
    ViewRequest:Gameplay Tag Structure=(TagName="UI.ItemPreview.PluginBridge.Mimosa.ShopISB")
   ) -> (WidgetClass:Common User Widget Soft Class Reference, ReturnValue:bool)
   > Branch(Condition=@n2.ReturnValue)
   > n3 = LoadClassAsset_Blocking(AssetClass=@n2)
   > n4 = Cast To CommonUserWidget Class!(execute:Exec, Class:Object Class Reference=@n3) -> (then:Exec, CastFailed:Exec, AsCommon User Widget:Common User Widget Class Reference) {is=title K2Node_ClassDynamicCast}
   > n5 = Create Widget!(execute:Exec, Class:User Widget Class Reference=@n4, OwningPlayer:Player Controller Object Reference=GetOwningPlayer()) -> (then:Exec, ReturnValue:Common User Widget Object Reference)
   > AddChild(self=Get NamedSlot_CustomBanner_Left(), Content=@n5)
   > n6 = Cast To WBP_CustomBanner_Base(Object=@n5)
   > SetSingleItem(self=@n6, ItemVM=Get ItemVM())
   > SetContext(self=@n6, Desired Area=NewEnumerator3)
  ]}
 UpdateItemPreviewFraming {nodes [
   UpdateItemPreviewFraming() {is=title K2Node_FunctionEntry}
   > n1 = Get!(execute:Exec, self:Fort Item Preview Widget Object Reference=Get SeasonPassCategoryItemPreviewWidget()) -> (then:Exec, else:Exec, CurrentItemVM:Fort Item VM Object Reference)
   > n2 = GetFramingDataFromItem!(execute, self, ItemVM:Fort Item VM Object Reference=@n1) -> (then, FramingData:S Item Preview Framing Data Structure) {is=title K2Node_CallFunction}
   > Set TargetFramingData!(execute, TargetFramingData:S Item Preview Framing Data Structure=@n2) -> (
    then
    Output_Get_Offset_16_#32:Vector 2D Structure
    Output_Get_ZoomLevel_15_#34:double=~
    Output_Get_CameraMode_19_#35:EPreviewCameraPositionMode Enum
   )
   > PlayAnimationForward(InAnimation=Get ItemPreviewFraming() -> (ItemPreviewFraming:Widget Animation Object Reference))
  ]}
 UpdateNavigationArrows {nodes [
   UpdateNavigationArrows() {is=title K2Node_FunctionEntry}
   n1 = Get PageIndex() -> (PageIndex=0)
   > SetIsEnabled(
    self=[Get Left(), Get SecondaryLeft() -> (SecondaryLeft:WBP UIKit Button Generic Object Reference)]
    bInIsEnabled="Not Equal (Integer)"(A=@n1)
   )
   n2 = Length(TargetArray=Get Pages(self=Get SeasonPassCategoryVM()))
   > SetIsEnabled(
    self=[Get SecondaryRight() -> (SecondaryRight:WBP UIKit Button Generic Object Reference), Get Right()]
    bInIsEnabled="integer < integer"(A=@n1, B=int - int(A=@n2))
   )
  ]}
 UpdatePaginationText {nodes [
   UpdatePaginationText() {is=title K2Node_FunctionEntry}
   > "SetText (Text)"(
    self=[Get SecondaryCurrentPageText(), Get CurrentPageText() -> (CurrentPageText:Common Text Object Reference)]
    InText=Get Page Text!(self, Page Index:int=Get PageIndex() -> (PageIndex=0)) -> (Result:Text) {is=title K2Node_CallFunction}
   )
  ]}
 UpdateRewardsListWidth {nodes [
   UpdateRewardsListWidth() {is=title K2Node_FunctionEntry}
   > ForceLayoutPrepass(self:Widget Object Reference)
   > n1 = Set WidthOverride!(
    execute
    WidthOverride:float=GetDesiredSize!(self:Widget Object Reference=Get ActiveRewardList()) -> (ReturnValue_X:double, ReturnValue_Y:double)
    self:Size Box Object Reference=Get StatusBarSizeBox() -> (StatusBarSizeBox:Size Box Object Reference)
   ) -> (then, Output_Get:float)
   > Set!(
    execute
    Size_X:double=float + float(A=@n1, B=32)
    Size_Y:double=1
    self:Spacer Object Reference=Get Spacer_RewardsListWidth() -> (Spacer_RewardsListWidth:Spacer Object Reference)
   ) -> (then, Output_Get:Vector 2D Structure)
   > Return Node!(execute)
  ]}
 VM_CategoryChanged {nodes [
   n1 = VM_CategoryChanged!() -> (then, Category:Season Pass Category VM Object Reference) {is=title K2Node_FunctionEntry}
   > n2 = Set SeasonPassCategoryVM(SeasonPassCategoryVM:Season Pass Category VM Object Reference=@n1) -> (Output_Get:Season Pass Category VM Object Reference)
   > ReinitializeSelectedPage() {is=title K2Node_CallFunction}
   > SetVisibility(
    self=[Get SecondaryLeft(), Get SecondaryRight(), Get Left(), Get Right(), Get CurrentPageText()]
    InVisibility=Get Page Text Visibility!(self, Category:Season Pass Category VM Object Reference=@n2) -> (Result:ESlateVisibility Enum=Visible) {is=title K2Node_CallFunction}
   )
   > SetVisibility(
    self=Get WBP_UIKit_TabGroup()
    InVisibility=Select(Option 0:ESlateVisibility Enum=Collapsed, Option 1:ESlateVisibility Enum=SelfHitTestInvisible, Index=GetShouldShowNavigationTabs(self=@n1)) -> (ReturnValue:ESlateVisibility Enum)
   )
   n3 = Get WBP_UIKit_TabGroup()
   > Fire Event Element Interacted(
    Element=GetTabButtonBaseByID!(
     self:Common Tab List Widget Base Object Reference=@n3
     TabNameID:Name=GetActiveTab!(self:Common Tab List Widget Base Object Reference=@n3) -> (ReturnValue:Name=None)
    ) -> (ReturnValue:Common Button Base Object Reference)
    AdditionalInfo=Get Analytics Info Pass Name(MakeFlowAnalyticsAdditionalInfo=Make Flow Analytics Additional Info(), SeasonPassSelectedVM=Get SeasonPassSelectionVM())
   )
  ]}
 VM_ItemVMsChanged {nodes [
   VM_ItemVMsChanged!() -> (then, ItemVMs:Array of Fort Item VM Object References) {is=title K2Node_FunctionEntry}
   > Reset(self=Get VariantCards(), bDeleteWidgets=false)
   n1 = Get ItemVMs!() -> (ItemVMs)
   > Branch(Condition="integer > integer"(A=Length(TargetArray:Array of Fort Item VM Object References=@n1), B=1))
   > n2 = For Each Loop(Array:Array of Fort Item VM Object References=Get ItemVMs()) -> (Array Element:Fort Item VM Object Reference)
   > CreateSeasonPassVariant!(execute, self, ItemVM:Fort Item VM Object Reference=@n2, ItemDef:Fort Item Definition Object Reference=Get ItemDefinition(self=@n2)) -> (then) {is=title K2Node_CallFunction}
  ]}
 VM_PageOwnedRewardsChanged {nodes [
   n1 = VM_PageOwnedRewardsChanged!() -> (then, OwnedFromPage:int) {is=title K2Node_FunctionEntry}
   n2 = GetRewardCount!(self:Season Pass Page VM Object Reference=Get SelectedPage(self=Get SeasonPassSelectionVM())) -> (ReturnValue:int=0)
   > Branch(Condition="Equal (Integer)"(A=@n1, B=int - int(A=@n2)))
   n3 = Get Pages(self=Get SeasonPassCategoryVM())
   n4 = Get SeasonPassSelectionVM()
   > Branch(Condition="Not Equal (Object)"(A:Object Reference="Get (a copy)"(Array=@n3), B:Object Reference=Get SelectedPage(self=@n4)) -> (ReturnValue:bool=~) {is=title K2Node_PromotableOperator})
  ]}
 VM_PreviewDataChanged {nodes [
   n1 = VM_PreviewDataChanged!() -> (then, PreviewItemVM:Fort Cosmetic Preview Data Structure) {is=title K2Node_FunctionEntry}
   > n2 = Branch(Condition=Get IsViewingReward())
   SetPreviewData(execute=@n2.else, self=Get SeasonPassCategoryItemPreviewWidget(), Data=@n1)
  ]}
 VM_PreviewItemChanged {nodes [
   n1 = VM_PreviewItemChanged!() -> (then, PreviewItemVM:Fort Item VM Object Reference) {is=title K2Node_FunctionEntry}
   > Is Valid(InputObject=@n1)
   n2 = Get PreviewItemVM() -> (PreviewItemVM:Fort Item VM Object Reference)
   > Set FortItemDetailsVM(FortItemDetailsVM=Get ItemDetails(self=@n2), self=Get WBP_UIKit_ItemDescription_Compact())
   n3 = Get ItemDefinition(self=@n2)
   > Set ItemName(ItemName="To String (Text)"(InText=Get ItemName(self=@n3)), self=Get WBP_UIKit_ItemDescription_Compact())
   > SetFortItemDetailsVM(self=Get WBP_CreatorBadge(), Viewmodel=Get ItemDetails(self=Get PreviewItemVM()))
   > SetFortItemVM(self:Reward Counter Object Reference=Get RewardCounter() -> (RewardCounter:Reward Counter Object Reference), Viewmodel=Get PreviewItemVM())
   > Branch(Condition=IsActivated())
   > UpdateItemPreviewFraming() {is=title K2Node_CallFunction}
  ]}
 VM_SeasonPassesUpdated {nodes [
   n1 = VM_SeasonPassesUpdated!() -> (then, SeasonPasses:Array of Season Pass VM Object References) {is=title K2Node_FunctionEntry}
   > SetupPassesTabs!(execute, self, SeasonPasses:Array of Season Pass VM Object References=@n1) -> (then) {is=title K2Node_CallFunction}
  ]}
 VM_SelectedPageChanged {nodes [
   n1 = VM_SelectedPageChanged!() -> (then, PassPageVM:Season Pass Page VM Object Reference) {is=title K2Node_FunctionEntry}
   > Initialize Selected Page And Reward(PassPageVM=@n1, RewardToSelect=Get SelectedReward(self=Get SeasonPassSelectionVM()))
  ]}
 VM_SelectedPassChanged {nodes [
   n1 = VM_SelectedPassChanged!() -> (then, SelectedPass:Season Pass VM Object Reference) {is=title K2Node_FunctionEntry}
   n2 = GetDisplayName(self=@n1)
   > Set NewTabId(NewTabId:Name=String To Name(InString="To String (Text)"(InText=@n2))) -> (Output_Get:Name=None)
   n3 = Get SelectedPass!() -> (SelectedPass)
   n4 = GetSelectedTabIndex!(self:UIKit Tab Group Object Reference=Get WBP_UIKit_TabGroup()) -> (ReturnValue:int=0)
   n5 = Get LastSelectedTabID() -> (LastSelectedTabID:int=0)
   > RequestLoadBackground!(
    execute
    self
    BackgroundAsset:Soft Object Reference=Break Season Pass Theme Data!(SeasonPassThemeData:"Season Pass Theme Data Structure (by ref)"=GetThemeData!(self:Season Pass VM Object Reference=@n3) -> (ReturnValue:"Season Pass Theme Data Structure (by ref)")) -> (
     TitleColor:Linear Color Structure
     TitleMaterial:Soft Object Reference
     TitleOutline:Font Outline Settings Structure
     SeasonColor:Linear Color Structure
     ProgressionColor:Linear Color Structure
     OverlayColor:Linear Color Structure
     VaultBackgroundMaterial:Soft Object Reference
     GiftboxVaultBackgroundMaterial:Soft Object Reference
     UIBackgroundMaterial:Soft Object Reference
     UpsellBackgroundMaterial:Soft Object Reference
     CrewCategoryBackgroundMaterial:Soft Object Reference
     CustomLogo:Soft Object Reference
     CustomSeasonTitle:Text
     CustomSeasonTitleColor:Linear Color Structure
     Upsell2BackgrounMaterial:Soft Object Reference
     ImageReferenceTable:Fort Image Reference Table Structure
     ColorReferenceTable:Fort Color Reference Table Structure
     TextReferenceTable:Fort Text Reference Table Structure
    ) {is=title K2Node_BreakStruct}.VaultBackgroundMaterial
    Direction:EVaultWorldTransitionDirection Enum=Select(
     Option 0:EVaultWorldTransitionDirection Enum=Forward
     Option 1:EVaultWorldTransitionDirection Enum=Backward
     Index="integer > integer"(A=@n5, B=@n4)
    ) -> (ReturnValue:EVaultWorldTransitionDirection Enum)
    FocalPointValue:double=0
   ) -> (then) {is=title K2Node_CallFunction}
   > Set LastSelectedTabID(LastSelectedTabID:int=@n4) -> (Output_Get:int=0)
   > SelectTabByID!(
    execute
    self:Common Tab List Widget Base Object Reference=Get WBP_UIKit_TabGroup()
    TabNameID:Name=Get NewTabId() -> (NewTabId:Name=None)
    bSuppressClickFeedback:bool=true
   ) -> (then, ReturnValue:bool)
  ]}
 VM_SelectedRewardUpdated {nodes [
   n1 = VM_SelectedRewardUpdated!() -> (then, SelectedReward:Season Pass Reward VM Object Reference) {is=title K2Node_FunctionEntry}
   > Is Valid(InputObject=@n1)
   n2 = Get SeasonPassSelectionVM()
   n3 = Find Item(TargetArray=GetRewardsInPage(Page=Get SelectedPage(self=@n2)), ItemToFind=@n1)
   > Branch(Condition="Not Equal (Integer)"(A=@n3, B=-1))
   > Set SelectedRewardIndex(SelectedRewardIndex=@n3) -> (Output_Get=0)
   RequestRefreshFocus()
  ]}
 ViewingRewardStateChanged {nodes [ViewingRewardStateChanged!() -> (then, IsViewingReward:bool=~) {is=title K2Node_FunctionEntry}]}
]
variables [
 ActiveRewardList {type=DynamicEntryBox, default=null}
 ActiveTrack {type=NamedSlot, default=null}
 BaseTileSizeX {type=number, default=128}
 CurrentItemFramingData {type=S_ItemPreviewFramingData, default {cameraMode=Configured, zoomLevel=0, offset=<0 0>}}
 EnableNavigationPageWrapping {type=boolean}
 FallbackFramingData_Landing {type=S_ItemPreviewFramingData, default {cameraMode=FrameCapsule, zoomLevel=0, offset=<0 0>}}
 FallbackFramingData_Landing_ViewReward {type=S_ItemPreviewFramingData, default {cameraMode=FrameCapsule, zoomLevel=0, offset=<0 0>}}
 IsClaimAllInProgress {type=boolean}
 IsInitializingFromLandingPage {type=boolean}
 IsModalVisible {type=boolean}
 IsViewingReward {type=boolean}
 ItemPreviewFramingOverrides {default {
   "(ItemType_2_#30=AthenaBackpack,ProductTag_5_#31=())" {landing Data {cameraMode=FrameCapsule, zoomLevel=0.75, offset=<0 25>}, viewReward Data {cameraMode=FrameCapsule, zoomLevel=1, offset=<0 0>}}
   "(ItemType_2_#30=AthenaItemWrap,ProductTag_5_#31=())" {
    landing Data {cameraMode=FrameCapsule, zoomLevel=0.25, offset=<-50 40>}
    viewReward Data {cameraMode=Configured, zoomLevel=0.4, offset=<0 0>}
   }
  }}
 LastScrolledDirection {type=EVaultWorldTransitionDirection, default=None}
 LastSelectedTabID {type=integer}
 NewVariantCardSelected {}
 PageIndex {type=integer}
 RequestedBackgroundSoftPath {type=SoftObjectPath, default=""}
 SelectedRewardIndex {type=integer}
 TargetFramingData {type=S_ItemPreviewFramingData, default {cameraMode=Configured, zoomLevel=0, offset=<0 0>}}
 ViewingRewardStateChanged {}
 ViewRewardAnimValue {type=number}
]
widgets {
 info {
  inheritedWidgetCount=0
  namedSlotCount=1
  parentClass=&CommonActivatableWidget
  rootWidgetClass=&Overlay
  widgetCount=66
 }
 widgets [
  Overlay_0 {widgetClassPath=&Overlay, widget=Overlay_0}
  WBP_CaptureForPostBufferUpdate {
   widgetClassPath=&WBP_CaptureForPostBufferUpdate
   bIsVariable=true
   parent=Overlay_0
   slot {class=OverlaySlot, properties {horizontalAlignment=HAlign_Fill, verticalAlignment=VAlign_Fill}}
  }
  SafeZone {
   widgetClassPath=&SafeZone
   bIsVariable=true
   parent=Overlay_0
   widget=SafeZone
   slot {class=OverlaySlot, properties {horizontalAlignment=HAlign_Fill, verticalAlignment=VAlign_Fill}}
  }
  SizeBox_0 {
   widgetClassPath=&SizeBox
   bIsVariable=true
   parent=SafeZone
   widget=SizeBox_0
   properties {bOverride_WidthOverride=true, widthOverride=1920}
   slot {class=SafeZoneSlot}
  }
  Overlay_Content {
   widgetClassPath=&Overlay
   bIsVariable=true
   parent=SizeBox_0
   widget=Overlay_Content
   slot {class=SizeBoxSlot, properties {padding=<ltrb 80 40 80 40>}}
  }
  GridPanel_0 {
   widgetClassPath=&GridPanel
   parent=Overlay_Content
   widget=GridPanel_0
   properties {bOverride_Cursor=true, columnFill [1], rowFill [0,0,1,0,0]}
   slot {class=OverlaySlot, properties {horizontalAlignment=HAlign_Fill, verticalAlignment=VAlign_Fill}}
  }
  SizeBox_1 {
   widgetClassPath=&SizeBox
   parent=GridPanel_0
   widget=SizeBox_1
   properties {bOverride_MaxAspectRatio=true, bOverride_MinAspectRatio=true, bOverride_WidthOverride=true, widthOverride=1024}
   slot {class=GridSlot, properties {horizontalAlignment=HAlign_Center, layer=-10, row=1, rowSpan=4}}
  }
  Overlay_1 {widgetClassPath=&Overlay, parent=SizeBox_1, widget=Overlay_1, slot {class=SizeBoxSlot}}
  SeasonPassCategoryItemPreviewWidget {
   widgetClassPath=/FortniteCoreUI/Frontend/ItemPreview/WBP_ItemPreview
   bIsVariable=true
   parent=Overlay_1
   widget=SeasonPassCategoryItemPreviewWidget
   properties {
    bAllowEquippedVariants=false
    bAllowPawnOverrideOnExclusiveItems=true
    bCanDisplayUserAccessories=false
    bRestrictZoomToWidget=true
    bSupportsActivationFocus=false
    infoWidget=&self:WidgetTree.WBP_ItemPreview_Info
    itemViewObject=&self:WidgetTree.SeasonPassCategoryItemPreviewWidget.ItemViewObject
    visibility=Visible
    defaultPreviewContextTags {gameplayTags [{tagName=UI.ItemPreview.PreviewType.OnPawn}]}
    infoPadding=<ltrb 0 0 0 0>
    previewActionButtons [
     &self:WidgetTree.Button_PreviewAction
     &self:WidgetTree.Button_PreviewAction1
     &self:WidgetTree.Button_PreviewAction2
    ]
    previewContextTagOverrides [{
      bOverrideSuccessiveRequests=true
      contexTags {gameplayTags [{tagName=UI.ItemPreview.PreviewType.OffPawn}]}
      query {
       autoDescription=" ALL( Cosmetics.Type.Mimosa )"
       tokenStreamVersion=0
       userDescription=""
       queryTokenStream [0,1,2,1,0]
       tagDictionary [{tagName=Cosmetics.Type.Mimosa}]
      }
     }]
    previewSceneChangerTag {tagName=UI.ItemPreview.SceneChanger.MPItemShop}
    screenTag {tagName=UI.ItemPreview.Screen.BattlePass.RewardTrack}
   }
   slot {class=OverlaySlot, properties {horizontalAlignment=HAlign_Fill, verticalAlignment=VAlign_Fill}}
  }
  WBP_ItemPreview_Info {
   widgetClassPath=/FortniteCoreUI/Frontend/ItemPreview/WBP_ItemPreview_Info
   bIsVariable=true
   parent=Overlay_1
   widget=WBP_ItemPreview_Info
   slot {class=OverlaySlot, properties {horizontalAlignment=HAlign_Right}}
  }
  WBP_UIKit_TabGroup {
   widgetClassPath=/FortUILibrary/UIKit/Tabs/WBP_UIKit_TabGroup
   bIsVariable=true
   parent=GridPanel_0
   widget=WBP_UIKit_TabGroup
   properties {
    backplateBrightness=Dark
    buttonClass=&/SeasonPass_Button_TopNav.SeasonPass_Button_TopNav_C
    promptBackplateBrightness=Dark
    tabSpacing=8
    usePromptBackplate=false
    initialTabsData [
     {
      bEnableTabId=false
      tabId=None
      tabViewModel=&self:WidgetTree.WBP_UIKit_TabGroup.UIKitSelectableButtonViewModel_0
     }
     {
      bEnableTabId=false
      tabId=None
      tabViewModel=&self:WidgetTree.WBP_UIKit_TabGroup.UIKitSelectableButtonViewModel_2
     }
     {
      bEnableTabId=false
      tabId=None
      tabViewModel=&self:WidgetTree.WBP_UIKit_TabGroup.UIKitSelectableButtonViewModel_1
     }
    ]
   }
   slot {class=GridSlot, properties {horizontalAlignment=HAlign_Center}}
  }
  Overlay_3 {widgetClassPath=&Overlay, parent=GridPanel_0, widget=Overlay_3, slot {class=GridSlot, properties {row=1}}}
  HorizontalBox_2 {
   widgetClassPath=&HorizontalBox
   parent=Overlay_3
   widget=HorizontalBox_2
   slot {class=OverlaySlot, properties {horizontalAlignment=HAlign_Right, verticalAlignment=VAlign_Center}}
  }
  WBP_CreatorBadge {
   widgetClassPath=/FortniteCoreUI/Items/WBP_CreatorBadge
   bIsVariable=true
   parent=HorizontalBox_2
   widget=WBP_CreatorBadge
   slot {class=HorizontalBoxSlot, properties {horizontalAlignment=HAlign_Left, verticalAlignment=VAlign_Top, padding=<ltrb 0 0 20 0>}}
  }
  SeasonPass_Level {
   widgetClassPath=/FortniteCoreUI/MultiProduct/SeasonPass_Level
   bIsVariable=true
   parent=HorizontalBox_2
   widget=SeasonPass_Level
   slot {class=HorizontalBoxSlot}
  }
  SeasonPassPageHeader {
   widgetClassPath=&/SeasonPassPageHeader
   bIsVariable=true
   parent=Overlay_3
   widget=SeasonPassPageHeader
   slot {class=OverlaySlot, properties {horizontalAlignment=HAlign_Fill, verticalAlignment=VAlign_Center}}
  }
  HorizontalBox_0 {
   widgetClassPath=&HorizontalBox
   bIsVariable=true
   parent=GridPanel_0
   widget=HorizontalBox_0
   slot {class=GridSlot, properties {row=2}}
  }
  SizeBox_LeftContent {
   widgetClassPath=&SizeBox
   bIsVariable=true
   parent=HorizontalBox_0
   widget=SizeBox_LeftContent
   properties {bOverride_MaxDesiredWidth=true, maxDesiredWidth=560}
   slot {class=HorizontalBoxSlot, properties {horizontalAlignment=HAlign_Left}}
  }
  VerticalBox_1 {widgetClassPath=&VerticalBox, parent=SizeBox_LeftContent, widget=VerticalBox_1, slot {class=SizeBoxSlot}}
  RewardCounter {
   widgetClassPath=&/RewardCounter
   bIsVariable=true
   parent=VerticalBox_1
   widget=RewardCounter
   slot {class=VerticalBoxSlot, properties {horizontalAlignment=HAlign_Left, verticalAlignment=VAlign_Top, padding=<ltrb 0 0 0 14>}}
  }
  WBP_UIKit_ItemDescription_Compact {
   widgetClassPath=/FortUILibrary/UIKit/Items/WBP_UIKit_ItemDescription_Compact
   bIsVariable=true
   parent=VerticalBox_1
   widget=WBP_UIKit_ItemDescription_Compact
   properties {
    customModalWidget=/SeasonPass/Widgets/Reboot/WBP_CrossProductCompatibilityTags.WBP_CrossProductCompatibilityTags_C
    hideRarityAndSeriesTag=true
    input Action Row {dataTable=&InputActionDataTable, rowName=ItemDescription_ViewCompatibility}
   }
   slot {class=VerticalBoxSlot}
  }
  BonusInfo_VariantCard_Switcher {
   widgetClassPath=&CommonVisibilitySwitcher
   bIsVariable=true
   parent=VerticalBox_1
   widget=BonusInfo_VariantCard_Switcher
   properties {visibility=Collapsed}
   slot {class=VerticalBoxSlot}
  }
  VariantCards {
   widgetClassPath=/Script/FortniteUI.FortDynamicEntryBox
   bIsVariable=true
   parent=BonusInfo_VariantCard_Switcher
   widget=VariantCards
   properties {
    entryWidgetClass=/FortUILibrary/UIKit/Items/WBP_UIKit_ItemCard_BattlePass.WBP_UIKit_ItemCard_BattlePass_C
    numDesignerPreviewEntries=2
    entrySpacing=<14 0>
   }
   slot {class=CommonVisibilitySwitcherSlot, properties {padding=<ltrb 0 20 0 0>}}
  }
  WBP_CustomItemDetailsContainer {
   widgetClassPath=/FortniteCoreUI/Customization/MPL/WBP_CustomItemDetailsContainer
   bIsVariable=true
   parent=BonusInfo_VariantCard_Switcher
   widget=WBP_CustomItemDetailsContainer
   slot {class=CommonVisibilitySwitcherSlot, properties {padding=<ltrb 0 12 12 12>}}
  }
  NamedSlot_CustomBanner_Left {
   widgetClassPath=&NamedSlot
   bIsVariable=true
   parent=VerticalBox_1
   widget=NamedSlot_CustomBanner_Left
   slot {class=VerticalBoxSlot, properties {padding=<ltrb 0 48 12 12>}}
  }
  FortItemPreviewBackground {
   widgetClassPath=/Script/FortniteUI.FortItemPreviewBackground
   bIsVariable=true
   parent=HorizontalBox_0
   widget=FortItemPreviewBackground
   properties {previewSceneChangerTag {tagName=UI.ItemPreview.SceneChanger.MPItemShop}}
   slot {class=HorizontalBoxSlot, properties {size {sizeRule=Fill, value=1}}}
  }
  VerticalBox_3 {
   widgetClassPath=&VerticalBox
   parent=HorizontalBox_0
   widget=VerticalBox_3
   slot {class=HorizontalBoxSlot, properties {horizontalAlignment=HAlign_Center}}
  }
  SizeBox_RightContent {
   widgetClassPath=&SizeBox
   bIsVariable=true
   parent=VerticalBox_3
   widget=SizeBox_RightContent
   properties {bOverride_WidthOverride=true, maxDesiredWidth=400, minDesiredHeight=400, minDesiredWidth=300, widthOverride=392}
   slot {class=VerticalBoxSlot, properties {horizontalAlignment=HAlign_Right, verticalAlignment=VAlign_Center}}
  }
  SeasonPass_ButtonsSwitcher {
   widgetClassPath=&/SeasonPassCategory_ButtonsSwitcher
   bIsVariable=true
   parent=SizeBox_RightContent
   widget=SeasonPass_ButtonsSwitcher
   slot {class=SizeBoxSlot, properties {padding=<ltrb 0 48 0 0>}}
  }
  NamedSlot_CustomBanner {
   widgetClassPath=&NamedSlot
   bIsVariable=true
   parent=VerticalBox_3
   widget=NamedSlot_CustomBanner
   properties {visibility=Collapsed}
   slot {
    class=VerticalBoxSlot
    properties {horizontalAlignment=HAlign_Right, verticalAlignment=VAlign_Bottom, padding=<ltrb 0 0 0 12>, size {sizeRule=Fill, value=1}}
   }
  }
  TrackAndStatusBarContainer {
   widgetClassPath=&VerticalBox
   bIsVariable=true
   parent=GridPanel_0
   widget=TrackAndStatusBarContainer
   slot {class=GridSlot, properties {row=3}}
  }
  StatusBarSizeBox {
   widgetClassPath=&SizeBox
   bIsVariable=true
   parent=TrackAndStatusBarContainer
   widget=StatusBarSizeBox
   slot {class=VerticalBoxSlot, properties {horizontalAlignment=HAlign_Center, padding=<ltrb 0 0 0 16>}}
  }
  SeasonPass_RewardTrack_StatusBar {
   widgetClassPath=&/SeasonPass_RewardTrack_StatusBar
   bIsVariable=true
   parent=StatusBarSizeBox
   widget=SeasonPass_RewardTrack_StatusBar
   slot {class=SizeBoxSlot}
  }
  TrackContainer {
   widgetClassPath=&GridPanel
   bIsVariable=true
   parent=TrackAndStatusBarContainer
   widget=TrackContainer
   properties {bOverride_Cursor=true}
   slot {class=VerticalBoxSlot, properties {horizontalAlignment=HAlign_Center, padding=<ltrb 0 0 0 32>}}
  }
  NamedSlot_Track_Current {
   widgetClassPath=&NamedSlot
   bIsVariable=true
   parent=TrackContainer
   widget=NamedSlot_Track_Current
   slot {class=GridSlot, properties {horizontalAlignment=HAlign_Center}}
  }
  RewardCards_A {
   widgetClassPath=&DynamicEntryBox
   bIsVariable=true
   parent=NamedSlot_Track_Current
   widget=RewardCards_A
   properties {
    entryHorizontalAlignment=HAlign_Left
    entryVerticalAlignment=VAlign_Center
    entryWidgetClass=&/SeasonPassRewardButton.SeasonPassRewardButton_C
    navigation=&self:WidgetTree.RewardCards_A.WidgetNavigation_1
    numDesignerPreviewEntries=8
    entrySpacing=<16 0>
   }
   slot {class=PanelSlot}
  }
  NamedSlot_Track_Next {
   widgetClassPath=&NamedSlot
   bIsVariable=true
   parent=TrackContainer
   widget=NamedSlot_Track_Next
   slot {class=GridSlot, properties {horizontalAlignment=HAlign_Center}}
  }
  RewardCards_B {
   widgetClassPath=&DynamicEntryBox
   bIsVariable=true
   parent=NamedSlot_Track_Next
   widget=RewardCards_B
   properties {
    entryHorizontalAlignment=HAlign_Left
    entryVerticalAlignment=VAlign_Center
    entryWidgetClass=&SeasonPassRewardButton_C
    navigation=&self:WidgetTree.RewardCards_B.WidgetNavigation_1
    numDesignerPreviewEntries=8
    entrySpacing=<16 0>
   }
   slot {class=PanelSlot}
  }
  SecondaryNavigationArrows {
   widgetClassPath=&Overlay
   bIsVariable=true
   parent=TrackContainer
   widget=SecondaryNavigationArrows
   properties {visibility=Collapsed}
   slot {class=GridSlot}
  }
  Spacer_RewardsListWidth {
   widgetClassPath=&Spacer
   bIsVariable=true
   parent=SecondaryNavigationArrows
   widget=Spacer_RewardsListWidth
   properties {size=<1200 1>}
   slot {class=OverlaySlot}
  }
  FortVisualAttachment_0 {
   widgetClassPath=&FortVisualAttachment
   parent=SecondaryNavigationArrows
   widget=FortVisualAttachment_0
   properties {contentAnchor=<1 0.5>}
   slot {class=OverlaySlot, properties {verticalAlignment=VAlign_Center}}
  }
  SecondaryLeft {
   widgetClassPath=&WBP_UIKit_Button_Generic
   bIsVariable=true
   parent=FortVisualAttachment_0
   widget=SecondaryLeft
   properties {
    bIsFocusable=false
    iconMaterial=/FortUILibrary/Foundation/Buttons/Materials/Arrows/MI_UIKit_Icon_Button_Previous
    minHeight=48
    mobileMinHeight=80
    mobileMinWidth=80
    shouldPromptDirectionFollowCulture=false
    text=""
    useTextBlock=false
    buttonPadding=<ltrb 8 0 8 0>
    iconSize=<32 32>
    promptNudge=<-48 0>
    triggeringInputAction {dataTable=&InputActionDataTable, rowName=GenericLeftTrigger}
   }
   slot {class=SizeBoxSlot}
  }
  FortVisualAttachment_1 {
   widgetClassPath=&FortVisualAttachment
   parent=SecondaryNavigationArrows
   widget=FortVisualAttachment_1
   properties {contentAnchor=<0 0.5>}
   slot {class=OverlaySlot, properties {horizontalAlignment=HAlign_Right, verticalAlignment=VAlign_Center}}
  }
  SecondaryRight {
   widgetClassPath=&WBP_UIKit_Button_Generic
   bIsVariable=true
   parent=FortVisualAttachment_1
   widget=SecondaryRight
   properties {
    bIsFocusable=false
    iconMaterial=/FortUILibrary/Foundation/Buttons/Materials/Arrows/MI_UIKit_Icon_Button_Next
    minHeight=48
    mobileMinHeight=80
    mobileMinWidth=80
    promptHorizontalAlignment=HAlign_Right
    shouldPromptDirectionFollowCulture=false
    text=""
    useTextBlock=false
    buttonPadding=<ltrb 8 0 8 0>
    iconSize=<32 32>
    promptNudge=<48 0>
    triggeringInputAction {dataTable=&InputActionDataTable, rowName=GenericRightTrigger}
   }
   slot {class=SizeBoxSlot}
  }
  FortVisualAttachment_79 {
   widgetClassPath=&FortVisualAttachment
   parent=TrackContainer
   widget=FortVisualAttachment_79
   properties {contentAnchor=<0 1>}
   slot {class=GridSlot, properties {layer=100}}
  }
  SeasonPass_GrantRewardByLevel_Coachmark {
   widgetClassPath=&/SeasonPass_GrantRewardByLevel_Coachmark
   bIsVariable=true
   parent=FortVisualAttachment_79
   widget=SeasonPass_GrantRewardByLevel_Coachmark
   slot {class=SizeBoxSlot}
  }
  LegalandActionBar {
   widgetClassPath=&Overlay
   bIsVariable=true
   parent=GridPanel_0
   widget=LegalandActionBar
   slot {class=GridSlot, properties {row=4}}
  }
  LegalText {
   widgetClassPath=&CommonTextBlock
   bIsVariable=true
   parent=LegalandActionBar
   widget=LegalText
   properties {
    autoWrapText=true
    bIsScrollingEnabled=false
    mobileFontSizeMultiplier=1.25
    text=Legal text goes here
    visibility=Collapsed
    wrapTextAt=720
    colorAndOpacity {colorUseRule=UseColor_Specified, specifiedColor=<rgba 1 1 1 0.6>}
    font {
     bForceMonospaced=false
     bMaterialIsStencil=false
     fontMaterial=null
     fontObject=&HeadingNow
     letterSpacing=0
     monospacedWidth=1
     size=24
     skewAmount=0
     typefaceFontName=86 Bold
     outlineSettings {
      bApplyOutlineToDropShadows=false
      bMiteredCorners=false
      bSeparateFillAlpha=false
      outlineMaterial=null
      outlineSize=0
      outlineColor=<rgba 0 0 0 1>
     }
    }
    shadowOffset=<0 0>
   }
   slot {class=OverlaySlot, properties {verticalAlignment=VAlign_Fill}}
  }
  Overlay_NavigationArrows {
   widgetClassPath=&Overlay
   bIsVariable=true
   parent=LegalandActionBar
   widget=Overlay_NavigationArrows
   slot {class=OverlaySlot, properties {horizontalAlignment=HAlign_Center}}
  }
  HorizontalBox_NavigationArrows {
   widgetClassPath=&HorizontalBox
   bIsVariable=true
   parent=Overlay_NavigationArrows
   widget=HorizontalBox_NavigationArrows
   slot {class=OverlaySlot}
  }
  Left {
   widgetClassPath=&WBP_UIKit_Button_Generic
   bIsVariable=true
   parent=HorizontalBox_NavigationArrows
   widget=Left
   properties {
    bIsFocusable=false
    bShouldUseFallbackDefaultInputAction=false
    iconMaterial=&MI_UIKit_Icon_Button_Previous
    minHeight=48
    mobileMinHeight=80
    mobileMinWidth=80
    shouldPromptDirectionFollowCulture=false
    text=""
    useTextBlock=false
    buttonPadding=<ltrb 8 0 8 0>
    iconSize=<32 32>
    promptNudge=<-48 0>
    triggeringInputAction {dataTable=&InputActionDataTable, rowName=GenericLeftTrigger}
   }
   slot {class=HorizontalBoxSlot}
  }
  CurrentPageText {
   widgetClassPath=&CommonTextBlock
   bIsVariable=true
   parent=HorizontalBox_NavigationArrows
   widget=CurrentPageText
   properties {
    text=Text Block
    colorAndOpacity {colorUseRule=UseColor_Specified, specifiedColor=<rgba 1 1 1 0.8>}
    font {
     bForceMonospaced=false
     bMaterialIsStencil=false
     fontMaterial=null
     fontObject=&HeadingNow
     letterSpacing=0
     monospacedWidth=1
     size=24
     skewAmount=0
     typefaceFontName=74 Regular
     outlineSettings {
      bApplyOutlineToDropShadows=false
      bMiteredCorners=false
      bSeparateFillAlpha=false
      outlineMaterial=null
      outlineSize=0
      outlineColor=<rgba 0 0 0 1>
     }
    }
    shadowOffset=<0 0>
   }
   slot {class=HorizontalBoxSlot, properties {horizontalAlignment=HAlign_Center, verticalAlignment=VAlign_Center, padding=<ltrb 32 0 32 0>}}
  }
  Right {
   widgetClassPath=&WBP_UIKit_Button_Generic
   bIsVariable=true
   parent=HorizontalBox_NavigationArrows
   widget=Right
   properties {
    bIsFocusable=false
    bShouldUseFallbackDefaultInputAction=false
    iconMaterial=&MI_UIKit_Icon_Button_Next
    minHeight=48
    mobileMinHeight=80
    mobileMinWidth=80
    promptHorizontalAlignment=HAlign_Right
    shouldPromptDirectionFollowCulture=false
    text=""
    useTextBlock=false
    buttonPadding=<ltrb 8 0 8 0>
    iconSize=<32 32>
    promptNudge=<48 0>
    triggeringInputAction {dataTable=&InputActionDataTable, rowName=GenericRightTrigger}
   }
   slot {class=HorizontalBoxSlot}
  }
  Spacer_151 {widgetClassPath=&Spacer, parent=LegalandActionBar, widget=Spacer_151, properties {size=<1 48>}, slot {class=OverlaySlot}}
  HorizontalBox_1 {
   widgetClassPath=&HorizontalBox
   parent=LegalandActionBar
   widget=HorizontalBox_1
   slot {class=OverlaySlot, properties {horizontalAlignment=HAlign_Right}}
  }
  SecondaryCurrentPageText {
   widgetClassPath=&CommonTextBlock
   bIsVariable=true
   parent=HorizontalBox_1
   widget=SecondaryCurrentPageText
   properties {
    text=Text Block
    visibility=Collapsed
    colorAndOpacity {colorUseRule=UseColor_Specified, specifiedColor=<rgba 1 1 1 0.8>}
    font {
     bForceMonospaced=false
     bMaterialIsStencil=false
     fontMaterial=null
     fontObject=&HeadingNow
     letterSpacing=0
     monospacedWidth=1
     size=24
     skewAmount=0
     typefaceFontName=74 Regular
     outlineSettings {
      bApplyOutlineToDropShadows=false
      bMiteredCorners=false
      bSeparateFillAlpha=false
      outlineMaterial=null
      outlineSize=0
      outlineColor=<rgba 0 0 0 1>
     }
    }
    shadowOffset=<0 0>
   }
   slot {class=HorizontalBoxSlot, properties {horizontalAlignment=HAlign_Center, verticalAlignment=VAlign_Center, padding=<ltrb 32 0 32 0>}}
  }
  WBP_ActionBarContainer {
   widgetClassPath=/Game/UI/Frontend/BottomBar/WBP_ActionBarContainer
   bIsVariable=true
   parent=HorizontalBox_1
   widget=WBP_ActionBarContainer
   slot {class=HorizontalBoxSlot}
  }
  Button_Bar {widgetClassPath=&HorizontalBox, namedSlotHost=WBP_ActionBarContainer, widget=Button_Bar}
  ActionButtonsWrapper {
   widgetClassPath=&HorizontalBox
   bIsVariable=true
   parent=Button_Bar
   widget=ActionButtonsWrapper
   properties {visibility=Collapsed}
   slot {class=HorizontalBoxSlot, properties {verticalAlignment=VAlign_Center}}
  }
  Button_PreviewAction {
   widgetClassPath=/FortUILibrary/UIKit/Buttons/WBP_UIKit_Button_Quiet
   parent=ActionButtonsWrapper
   widget=Button_PreviewAction
   properties {
    bAutoUpdateFromInputAction=true
    bDisplayInputActionWhenNotInteractable=false
    bIsFocusable=false
    bShouldUseFallbackDefaultInputAction=false
    minWidth=48
    text=""
   }
   slot {class=HorizontalBoxSlot, properties {padding=<ltrb 0 0 24 0>}}
  }
  Button_PreviewAction1 {
   widgetClassPath=&WBP_UIKit_Button_Quiet
   parent=ActionButtonsWrapper
   widget=Button_PreviewAction1
   properties {
    bAutoUpdateFromInputAction=true
    bDisplayInputActionWhenNotInteractable=false
    bIsFocusable=false
    bShouldUseFallbackDefaultInputAction=false
    minWidth=48
    text=""
   }
   slot {class=HorizontalBoxSlot, properties {padding=<ltrb 0 0 24 0>}}
  }
  Button_PreviewAction2 {
   widgetClassPath=&WBP_UIKit_Button_Quiet
   parent=ActionButtonsWrapper
   widget=Button_PreviewAction2
   properties {
    bAutoUpdateFromInputAction=true
    bDisplayInputActionWhenNotInteractable=false
    bIsFocusable=false
    bShouldUseFallbackDefaultInputAction=false
    minWidth=48
    text=""
   }
   slot {class=HorizontalBoxSlot, properties {padding=<ltrb 0 0 24 0>}}
  }
  ClaimAllButtonWrapper {
   widgetClassPath=&Overlay
   bIsVariable=true
   parent=Button_Bar
   widget=ClaimAllButtonWrapper
   slot {class=HorizontalBoxSlot, properties {verticalAlignment=VAlign_Center}}
  }
  GridPanel_1 {
   widgetClassPath=&GridPanel
   parent=ClaimAllButtonWrapper
   widget=GridPanel_1
   slot {class=OverlaySlot, properties {horizontalAlignment=HAlign_Right}}
  }
  WBP_AthenaBottomBar_WithNotifications {
   widgetClassPath=/Game/UI/Frontend/BottomBar/WBP_AthenaBottomBar_WithNotifications
   bIsVariable=true
   parent=Button_Bar
   widget=WBP_AthenaBottomBar_WithNotifications
   slot {class=HorizontalBoxSlot, properties {padding=<ltrb 0 0 24 0>}}
  }
  Button_Back {
   widgetClassPath=&WBP_UIKit_Button_Quiet
   bIsVariable=true
   parent=Button_Bar
   widget=Button_Back
   properties {
    bIsFocusable=false
    text=Back
    visibility=Visible
    triggeringInputAction {dataTable=&InputActionDataTable, rowName=GenericBack}
   }
   slot {class=HorizontalBoxSlot, properties {verticalAlignment=VAlign_Center}}
  }
 ]
}
